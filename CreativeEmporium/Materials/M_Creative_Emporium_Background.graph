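# Translucent UI material that draws a Creative Emporium background as a box filled with a two-color linear gradient.
# MF_UI_Box builds the shape from normalized UI UVs, with Size and Corner Roundness parameters and aspect ratio preserved.
# MF_UI_LinearGradient supplies a rotatable, offsettable gradient, and two color parameters, both dark gray by default, form the blend.
# Named reroutes organize the shape, color and opacity stages, with a Background Opacity parameter scaling the saturated opacity.
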Material /CreativeEmporium/Materials/M_Creative_Emporium_Background
graph {
 properties {blendMode=BLEND_Translucent, materialDomain=MD_UI, pixelDepthOffsetMode=PDOM_Legacy, usedShadingModels=Default Lit}
 nodes [
  NamedRerouteDeclaration(None=MaterialFunctionCall(
    UVs=MaterialFunctionCall() {materialFunction=&GetUserInterfaceUV}.Normalized UV
    Size=ScalarParameter() {defaultValue=1, group=Background Properties, parameterName=Size, sortPriority=10}
    CornerRoundness=ScalarParameter() {group=Background Properties, parameterName=Corner Roundness, sortPriority=11}
    KeepAspectRatio=StaticBool() {value=true}
   ) {materialFunction=&MF_UI_Box}.Fill) {name=Background, nodeColor=<rgba 0 0 0 1>}
  NamedRerouteDeclaration(None=MaterialFunctionCall(
    RotationAngle=ScalarParameter() {group=Background Properties, parameterName=Linear Gradient Rotation, sliderMax=1, sortPriority=12}
    GradientOffset=ScalarParameter() {group=Background Properties, parameterName=Linear Gradient Offset, sliderMax=1, sliderMin=-1, sortPriority=13}
   ) {materialFunction=&MF_UI_LinearGradient}.Result) {name=gradient, nodeColor=<rgba 0 0 0 1>}
  n1 = VectorParameter() {group=Background Properties, parameterName=Color 1, sortPriority=15, defaultValue=<rgba 0.0529 0.0529 0.0529 1>}
  n2 = VectorParameter() {group=Background Properties, parameterName=Color 2, sortPriority=16, defaultValue=<rgba 0.007 0.007 0.007 1>}
  n3 = NamedRerouteUsage()
  NamedRerouteDeclaration(None=Saturate(None=LinearInterpolate(A=@n1.RGB, B=@n2.RGB, Alpha=@n3))) {name=Colored Gradient, nodeColor=<rgba 0 0 0 1>}
  NamedRerouteDeclaration(None=Multiply(A=NamedRerouteUsage(), B=NamedRerouteUsage())) {name=Colors, nodeColor=<rgba 0 0 0 1>}
  n4 = NamedRerouteUsage()
  n5 = ScalarParameter() {defaultValue=1, group=Background Properties, parameterName=Background Opacity, sliderMax=1, sortPriority=14}
  NamedRerouteDeclaration(None=Saturate(None=Multiply(A=@n4, B=@n5))) {name=Opacity, nodeColor=<rgba 0 0 0 1>}
  NamedRerouteUsage()
  NamedRerouteUsage()
 ]
}
