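# Translucent UI material drawing a progress-masked circle, or a square bar variant, with a hover-switchable two-color gradient.
# The gradient follows normalized UI UVs and color-dodges with a panning, rotated triangle grid in aspect-corrected screen space, plus optional shine swipe.
# A static switch selects the mask: resolution-corrected radial circle cut by angled progress, or gradient with bottom bar cut by horizontal progress.
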Material /FortUILibrary/Foundation/Art_ToolKit/Tools/M_UI_Circle_UVspaceGradient
graph {
 properties {
  bUsedWithStaticMesh=false
  blendMode=BLEND_Translucent
  materialDomain=MD_UI
  pixelDepthOffsetMode=PDOM_Legacy
  usedShadingModels=Default Lit
 }
 nodes [
  n1 = MaterialFunctionCall(Float2=MaterialFunctionCall() {materialFunction=&ScreenResolution}.Visible Resolution) {materialFunction=&BreakOutFloat2Components}
  n2 = AppendVector(A=Divide(A=@n1.R, B=Max(A=@n1.G)), B=Constant() {r=1})
  n3 = MaterialFunctionCall() {materialFunction=&ScreenAlignedUVs}
  n4 = Panner(
   Coordinate=Rotator(Coordinate=Multiply(A=@n2, B=@n3."X 100%, Y 100%"), Time=ScalarParameter() {defaultValue=-0.0571, parameterName=Tri Angle}) {speed=6.2832}
   Speed=AppendVector(A=ScalarParameter() {defaultValue=0.0025, parameterName=PanSpeed}, B=Constant())
  )
  n5 = ScalarParameter() {defaultValue=4, parameterName=Tri Tile Amount}
  n6 = Multiply(
   A=TextureSample(UVs=Multiply(A=@n4, B=@n5)) {texture=/Game/Global/Textures/Default/Masks_and_Gradients/T-TriangleGrid}.R
   B=ScalarParameter() {defaultValue=0.2, parameterName=TriangleDensity, sliderMax=1}
  )
  n7 = MaterialFunctionCall() {materialFunction=&GetUserInterfaceUV}
  n8 = Saturate(None=Power(Base=ComponentMask(None=@n7.Normalized UV) {r=true}, Exp=ScalarParameter() {defaultValue=2, parameterName=Power}) {constExponent=3})
  n9 = LinearInterpolate(
   A=LinearInterpolate(
    A=VectorParameter() {group=Color, parameterName=Color_1, defaultValue=<rgba 0 0.1283 0.5 1>}.RGB
    B=VectorParameter() {group=Color, parameterName=Color_2, defaultValue=<rgba 0 0.5385 1 1>}.RGB
    Alpha=@n8
   )
   B=LinearInterpolate(
    A=VectorParameter() {group=Color, parameterName=Color_1_Hover, defaultValue=<rgba 0.5 0.5 0.5 1>}.RGB
    B=VectorParameter() {group=Color, parameterName=Color_2_Hover, defaultValue=<rgba 0.5 0.5 0.5 1>}.RGB
    Alpha=@n8
   )
   Alpha=ScalarParameter() {parameterName=Hover, sliderMax=1}
  )
  n10 = MaterialFunctionCall() {materialFunction=&GetUserInterfaceUV}
  LinearInterpolate(
   A=Saturate(None=MaterialFunctionCall!(Blend=@n6, Base=@n9) {materialFunction=/Engine/Functions/Engine_MaterialFunctions03/Blends/Blend_ColorDodge}.Result)
   Alpha=LinearInterpolate(
    B=MaterialFunctionCall!(In_Uv=@n10.Normalized UV) {materialFunction=/Game/UI/Foundation/Materials/M-Function-ShineSwipe}.Result
    Alpha=ScalarParameter() {parameterName=Use shine}
   )
  )
  n11 = ScalarParameter() {parameterName=BackGradient-Min}
  n12 = ScalarParameter() {parameterName=BackGradient-Max}
  n13 = TextureCoordinate()
  n14 = OneMinus(None=ComponentMask(None=@n13) {r=true})
  n15 = Clamp(None=@n14, Min=ScalarParameter() {parameterName=BottomBarGradient-Min}, Max=ScalarParameter() {parameterName=BottomBarGradien-Max})
  n16 = ComponentMask(None=Rotator(Coordinate=@n13, Time=ScalarParameter() {defaultValue=0.005, parameterName=BottomBarAngle}) {speed=6.2832}) {g=true}
  n17 = ScalarParameter() {defaultValue=0.8, parameterName=BottomBarHeight}
  n18 = Saturate(None=Multiply(A=Subtract(A=@n16, B=@n17), B=ScalarParameter() {defaultValue=32, parameterName=BottomBarEdge}))
  n19 = Power(
   Base=LinearInterpolate(A=Clamp(None=@n14, Min=@n11, Max=@n12), B=ScalarParameter() {defaultValue=1, parameterName=BarIntensity}, Alpha=Multiply(A=@n15, B=@n18))
   Exp=ScalarParameter() {defaultValue=1, parameterName=Shape-Power}
  )
  n20 = LinearInterpolate(Alpha=ScalarParameter() {defaultValue=1, parameterName=Progress, sliderMax=1}) {constA=1, constB=0}
  n21 = Saturate(None=Multiply(A=Subtract(A=@n14, B=@n20)) {constB=512})
  n22 = ScalarParameter() {defaultValue=1, parameterName=Hover Intensity}
  n23 = TextureCoordinate()
  n24 = AppendVector(
   A=ScalarParameter() {defaultValue=0.5, group=CirlceTreatment, parameterName=CenterPosition_U}
   B=ScalarParameter() {defaultValue=0.5, group=CirlceTreatment, parameterName=CenterPosition_V}
  )
  n25 = ScalarParameter() {defaultValue=0.5, group=CirlceTreatment, parameterName=Radius}
  n26 = Multiply(
   A=ScalarParameter() {defaultValue=2.333, parameterName=Density}
   B=MaterialFunctionCall!(Native_U_PixelSize=ScalarParameter() {defaultValue=32, group=ResolutionCorrection, parameterName=NativePixelSize}) {materialFunction=&/M_UI_Resolution_Sharpness_Correction}.Result
  )
  n27 = Multiply(
   A=Saturate(None=MaterialFunctionCall(UVs=@n23, CenterPosition=@n24, Radius=@n25, Density=@n26) {materialFunction=&RadialGradientExponential}.RadialGradientExponential)
   B=LinearInterpolate(
    A=ScalarParameter() {defaultValue=0.75, group=Alpha, parameterName=Alpha}
    B=ScalarParameter() {defaultValue=1, group=Alpha, parameterName=Alpha_Hover}
    Alpha=ScalarParameter() {parameterName=Hover, sliderMax=1}
   )
  )
  n28 = Add(
   A=MaterialFunctionCall() {materialFunction=&GetUserInterfaceUV}.Normalized UV
   B=LinearInterpolate(
    A=ScalarParameter() {group=progress, parameterName=Progress_Max}
    B=ScalarParameter() {defaultValue=-1, group=progress, parameterName=Progress_Min}
    Alpha=ScalarParameter() {defaultValue=1, parameterName=Progress}
   )
  ) {constB=-0.5}
  n29 = ScalarParameter() {defaultValue=-0.002, group=progress, parameterName=Angle}
  n30 = Saturate(None=ComponentMask(None=Rotator(Coordinate=@n28, Time=@n29) {speed=6.2832}) {r=true})
  n31 = ScalarParameter() {defaultValue=256, group=progress, parameterName=progress strenght}
  n32 = OneMinus(None=Saturate(None=Multiply(A=@n30, B=@n31)))
  StaticSwitchParameter(True=DotProduct(A=Multiply(A=@n21, B=@n22), B=Saturate(None=@n19)), False=Multiply(A=Saturate(None=@n27), B=Saturate(None=@n32))) {parameterName=bIsSquare}
 ]
}
