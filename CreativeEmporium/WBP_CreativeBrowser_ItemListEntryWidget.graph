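# Creative browser item tile widget (AthenaCreativeItemTileButton subclass) showing an item's card or icon, with label and details tooltip.
# On load it async-loads the preview texture, then switches between rarity card, plain device icon or rounded material icon by item type and tags.
# Hover, focus and selection drive border dynamic-material parameters and fade animations, gated by input type.
# Also supports mouse drag-and-drop into inventory and an overflow "+N" count state.
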WidgetBlueprint /CreativeEmporium/WBP_CreativeBrowser_ItemListEntryWidget
animations [
 pulsecardsize {
  binding_count=1
  display_label=pulse card size
  length_seconds=0.15
  object=&self:pulsecardsize
  sequence {bindings [WeaponCard {
     id=#1
     possessed_class=FortMultiSizeItemCard
     tracks [{
       type=MovieScene2DTransformTrack
       sections [{
         type=MovieScene2DTransformSection
         end_frame=3
         start_frame=0
         channels [
          Translation.X {type=MovieSceneScriptingFloatChannel}
          Translation.Y {type=MovieSceneScriptingFloatChannel}
          Angle {type=MovieSceneScriptingFloatChannel}
          Scale.X {type=MovieSceneScriptingFloatChannel, keys [[0,1],[2,1.1],[3,1]]}
          Scale.Y {type=MovieSceneScriptingFloatChannel, keys [[0,1],[2,1.1],[3,1]]}
          Shear.X {type=MovieSceneScriptingFloatChannel}
          Shear.Y {type=MovieSceneScriptingFloatChannel}
         ]
        }]
      }]
    }]}
 }
 FadeIn {
  binding_count=1
  length_seconds=0.2
  object=&self:FadeIn
  sequence {bindings [Overlay_Body {
     id=#2
     possessed_class=Overlay
     tracks [{
       type=MovieSceneFloatTrack
       display_name=Render Opacity
       sections [{
         type=MovieSceneFloatSection
         end_frame=4
         start_frame=0
         channels [None {type=MovieSceneScriptingFloatChannel, keys [[0,0],[1,0.8],[4.0007,1]]}]
        }]
      }]
    }]}
 }
 Hover {
  binding_count=1
  length_seconds=0.25
  object=&self:Hover
  sequence {bindings [HoverBorder_Image {
     id=#3
     possessed_class=Image
     tracks [{
       type=MovieSceneFloatTrack
       display_name=Render Opacity
       sections [{type=MovieSceneFloatSection, end_frame=5, start_frame=0, channels [None {type=MovieSceneScriptingFloatChannel, keys [[0,0],[5,1]]}]}]
      }]
    }]}
 }
]
graphs [
 EnterPurchasedState {nodes [
   n1 = EnterPurchasedState!() -> (then, Purchased:bool=~) {is=title K2Node_FunctionEntry}
   > n2 = Branch(Condition=@n1)
   n3 = Get PurchaseBorder() -> (PurchaseBorder:Common Border Object Reference)
   > SetBrushColor(self=@n3, InBrushColor=(R=1,G=1,B=1,A=1))
   SetBrushColor(execute=@n2.else, self=@n3, InBrushColor=(R=1,G=1,B=1,A=0))
  ]}
 EventGraph {nodes [
   "On Hovered (Tooltip_Button)"() {is=title K2Node_ComponentBoundEvent}
   n1 = Get Tooltip_MenuAnchor() -> (Tooltip_MenuAnchor:Menu Anchor Object Reference)
   > Open!(execute, self:Menu Anchor Object Reference=@n1, bFocusMenu:bool=false) -> (then)
   "On Unhovered (Tooltip_Button)"() {is=title K2Node_ComponentBoundEvent}
   > Close(self:Menu Anchor Object Reference=@n1)
   n2 = "On Menu Open Changed (Tooltip_MenuAnchor)"!() -> (OutputDelegate, then, bIsOpen:bool) {is=title K2Node_ComponentBoundEvent}
   > Branch(Condition=@n2.bIsOpen)
   > SetItemDetails!(
    execute
    self:Fort Emporium Item Details Panel Object Reference=Get DetailsTooltipWidget() -> (DetailsTooltipWidget:WBP Cr Browser Details Object Reference)
    FortItem:Fort Item Object Reference=CreateTemporaryItemInstance!(
     ItemDefinition:Item Definition Base Object Reference=GetItemDefinition!() -> (OutItemDefinition:Item Definition Base Object Reference)
     Count:int=1
     Level:int=-1
    ) -> (ReturnValue:Fort Item Object Reference)
    ParentDefinition:Item Definition Base Object Reference
    bUseLargeThumbnail:bool=false
    bAllowInteractiveTags:bool=false
   ) -> (then)
   Shine Item Widget() {is=title K2Node_CustomEvent}
   > PlayAnimation(InAnimation=Get pulsecardsize() -> (pulsecardsize:Widget Animation Object Reference))
   Highlight() {is=title K2Node_CustomEvent}
   > SetLabelVisibility!(execute, self, bShow:bool=true) -> (then)
   > ResetScrollState(self:Common Text Object Reference=Get Text_Label() -> (Text_Label:Common Text Object Reference))
   > n3 = Branch(Condition=IsAnimationPlaying(InAnimation=Get Hover() -> (Hover:Widget Animation Object Reference)))
   n4 = Get IsSelected() -> (IsSelected:bool)
   n5 = Branch(execute=@n3.else, Condition=@n4)
   > SetVectorParameterValue(
    self=Get OuterBorder_DM() -> (OuterBorder_DM:Material Instance Dynamic Object Reference)
    ParameterName="Color 2 (gradient on)"
    Value=Get SelectedColor() -> (SelectedColor:Linear Color Structure)
   )
   > PlayAnimationForward(InAnimation=Get Hover())
   PlayAnimationForward(execute=@n5.else, InAnimation=Get Hover())
   > SetScalarParameterValue(self=Get HoverBorder_DM() -> (HoverBorder_DM:Material Instance Dynamic Object Reference), ParameterName=Fill Transparency, Value=0.35)
   Unhighlight() {is=title K2Node_CustomEvent}
   > n6 = Branch(Condition=@n4)
   > SetVectorParameterValue(self=Get OuterBorder_DM(), ParameterName="Color 2 (gradient on)", Value=Get SelectedColor())
   > Check and Stop Animation!(execute:Exec, Animation:Widget Animation Object Reference=Get Hover()) -> (then:Exec) {is=title K2Node_MacroInstance}
   > n7 = SetRenderOpacity(self=Get HoverBorder_Image() -> (HoverBorder_Image:Image Object Reference), InOpacity=0)
   SetLabelVisibility(execute=[@n7.then, @n8.then], bShow=false)
   Check and Stop Animation(execute=@n6.else, Animation=Get Hover())
   > SetRenderOpacity(self=Get HoverBorder_Image(), InOpacity=0)
   > SetScalarParameterValue(self=Get HoverBorder_DM(), ParameterName=Fill Transparency, Value=0)
   > n8 = SetVectorParameterValue(self=Get OuterBorder_DM(), ParameterName="Color 2 (gradient on)", Value=Get NotSelectedColor() -> (NotSelectedColor:Linear Color Structure))
   n9 = ShowOverflowCount!() -> (OutputDelegate, then, Count:int) {is=title K2Node_CustomEvent}
   > "SetText (Text)"(
    self=Get Text_OverflowCount() -> (Text_OverflowCount:Common Text Object Reference)
    InText=Format Text!(Format="LOCTEXT(\"+{Count}\")", Count:int=@n9.Count) -> (Result)
   )
   > SetActiveWidget(
    self=Get Switcher_WeaponCard() -> (Switcher_WeaponCard:Common Visibility Switcher Object Reference)
    Widget=Select(
     Option 0:Widget Object Reference=Get GridPanel_WeaponCard() -> (GridPanel_WeaponCard:Grid Panel Object Reference)
     Option 1:Widget Object Reference=Get GridPanel_OverflowCount() -> (GridPanel_OverflowCount:Grid Panel Object Reference)
     Index="integer > integer"(A=@n9.Count)
    ) -> (ReturnValue:Widget Object Reference)
   )
   n10 = SetLabelOverride!() -> (OutputDelegate, then, LabelOverride:Text) {is=title K2Node_CustomEvent}
   > Set LabelOverride(LabelOverride:Text=@n10.LabelOverride) -> (Output_Get:Text)
   Event Construct()
   > n11 = GetDynamicMaterial(self=Get HoverBorder_Image())
   > Set HoverBorder_DM(HoverBorder_DM:Material Instance Dynamic Object Reference=@n11) -> (Output_Get:Material Instance Dynamic Object Reference)
   > n12 = GetDynamicMaterial(self=Get OuterBorder_Image() -> (OuterBorder_Image:Image Object Reference))
   > Set OuterBorder_DM(OuterBorder_DM:Material Instance Dynamic Object Reference=@n12) -> (Output_Get:Material Instance Dynamic Object Reference)
   Event On Hovered()
   > PlaySound2D(Sound=Get HoverSound() -> (HoverSound:Sound Base Object Reference))
   n13 = Get CommonInputSubsystem()
   > Branch(Condition="Equal (Enum)"(A=GetCurrentInputType(self=@n13), B=MouseAndKeyboard))
   > Highlight() {is=title K2Node_CallFunction}
   > UpdateTooltip()
   Event OnLoadedItem() {is=title K2Node_Event}
   > n14 = GetTempItemInstance!(execute, self, Count:int=GetItemTypeCount!() -> (ReturnValue:int=0)) -> (then, ReturnValue:Fort Item Object Reference)
   > n15 = Set TempItem(TempItem:Fort Item Object Reference=@n14) -> (Output_Get:Fort Item Object Reference)
   > n16 = Async Load Asset(Asset=GetSmallPreviewImage!(self:Fort Item Object Reference=@n15) -> (ReturnValue:Texture 2D Soft Object Reference))
   n17 = Cast To Texture2D(execute=@n16.Completed, Object=@n16)
   > Set TempTexture2D(TempTexture2D:Texture 2D Object Reference=@n17) -> (Output_Get:Texture 2D Object Reference)
   > Update()
   Event On Unhovered()
   > Unhighlight() {is=title K2Node_CallFunction}
   > SetVisibility(self=Get Tooltip_Button() -> (Tooltip_Button:Button Object Reference), InVisibility=Collapsed)
   Event On Selected()
   > Set IsSelected(IsSelected:bool=true) -> (Output_Get:bool)
   n18 = Get OuterBorder_DM()
   > SetVectorParameterValue(self=@n18, ParameterName="Color 2 (gradient on)", Value=Get SelectedColor())
   > SetScalarParameterValue(self=@n18, ParameterName=Stroke Transparency, Value=1)
   n19 = Get CommonInputSubsystem()
   > Branch(Condition="Equal (Enum)"(A=GetCurrentInputType(self=@n19), B=Gamepad))
   > UpdateTooltip()
   Event On Deselected()
   > Set IsSelected(IsSelected=false)
   n20 = Get OuterBorder_DM()
   > SetVectorParameterValue(self=@n20, ParameterName="Color 2 (gradient on)", Value=Get NotSelectedColor())
   > SetScalarParameterValue(self=@n20, ParameterName=Stroke Transparency, Value=0)
   > Unhighlight()
   > SetVisibility(self=Get Tooltip_Button(), InVisibility=Collapsed)
   Event On Focused() {is=title K2Node_Event}
   n21 = Get CommonInputSubsystem()
   > n22 = Branch(Condition="Equal (Enum)"(A=GetCurrentInputType(self=@n21), B=Gamepad))
   > Highlight()
   PlaySound2D(execute=@n22.else, Sound=Get Press Sound() -> (Press Sound:Sound Base Object Reference))
   Event On Unfocused() {is=title K2Node_Event}
   > Unhighlight()
   SetVisibility(self=Get Outline_Border() -> (Outline_Border:Image Object Reference), InVisibility=Hidden)
   SetVisibility(self=Get Outline_Border(), InVisibility=SelfHitTestInvisible)
   Branch(Condition=Get IsSelected())
  ]}
 GetItemDefinition {nodes [
   GetItemDefinition() {is=title K2Node_FunctionEntry}
   n1 = Get ItemDefinition() -> (ItemDefinition:Fort Item Definition Object Reference)
   > n2 = Is Valid(InputObject=@n1)
   Return Node!(
    execute=@n2.Is Not Valid
    OutItemDefinition:Item Definition Base Object Reference=Get!() -> (
     ItemEntry_ItemDefinition:Item Definition Base Object Reference
     ItemEntry_ItemGuid:Guid Structure
     ItemEntry_Count:int=0
     ItemEntry_Level:int=0
     ItemEntry_LoadedAmmo:int=0
     ItemEntry_PhantomReserveAmmo:int=0
    )
   )
   > Return Node(OutItemDefinition=@n1)
  ]}
 GetLabelText {nodes [
   n1 = GetLabelText!() -> (then, Item:Fort Item Object Reference) {is=title K2Node_FunctionEntry}
   n2 = Get LabelOverride() -> (LabelOverride:Text)
   > Return Node!(
    execute
    Label:Text=Select(Option 0:Text=@n2, Option 1:Text=GetDisplayName!(self:Fort Item Object Reference=@n1) -> (ReturnValue:Text), Index=TextIsEmpty(InText=@n2)) -> (ReturnValue:Text)
   )
  ]}
 GetTempItemInstance {nodes [
   n1 = GetTempItemInstance!() -> (then, Count:int) {is=title K2Node_FunctionEntry}
   n2 = Get Override Fort Item Def() -> (Override Fort Item Def:Item Definition Base Object Reference)
   > n3 = Is Valid(InputObject=@n2)
   Return Node!(
    execute=@n3.Is Not Valid
    ReturnValue:Fort Item Object Reference=CreateTemporaryItemInstance(ItemDefinition=GetItemDefinition() {is=title K2Node_CallFunction}, Count=@n1, Level=-1)
   )
   > Return Node(ReturnValue=CreateTemporaryItemInstance(ItemDefinition=@n2, Count=1, Level=-1))
  ]}
 OnDragDetected {nodes [
   OnDragDetected!() -> (then, MyGeometry:Geometry Structure, PointerEvent:"Pointer Event Structure (by ref)") {is=title K2Node_FunctionEntry}
   > n1 = Is Valid(InputObject=Get ItemToRepresent!(self:Fort Item Widget NUI Object Reference=Get WeaponCard() -> (WeaponCard:Fort Multi Size Item Card Object Reference)) -> (ItemToRepresent:Fort Item Weak object))
   UpdateCount!(execute=@n1.Is Not Valid, self, Count:int=0) -> (then)
   > n2 = Is Valid(InputObject=Get ItemToRepresent(self=Get WeaponCard()))
   n3 = Create Widget!(
    execute:Exec=[@n1.Is Valid, @n2.Is Valid]
    Class:User Widget Class Reference=/Game/Athena/HUD/Inventory/AthenaInventoryDragVisual.AthenaInventoryDragVisual_C
    OwningPlayer:Player Controller Object Reference=GetOwningPlayer()
    Item:Fort Item Object Reference=Get ItemToRepresent(self=Get WeaponCard())
   ) -> (then:Exec, ReturnValue:Athena Inventory Drag Visual Object Reference)
   > SetIsSelected!(execute, self:Common Button Base Object Reference, InSelected:bool=true, bGiveClickFeedback:bool=false) -> (then)
   > n4 = Construct Object from Class!(
    execute
    Class=/Game/Athena/HUD/Inventory/AthenaInventoryDragDrop.AthenaInventoryDragDrop_C
    self=Self-Reference()
    SplitTheStack:bool=False
    bIsFromCreativeInventory:bool=true
    InstigatorSlotIndex:int=0
    Tag:String=FromQuickbar
    Payload:Object Reference=Get ItemToRepresent(self=Get WeaponCard())
    DefaultDragVisual:Widget Object Reference=@n3
    Pivot:EDragPivot Enum=MouseDown
    Offset:Vector 2D Structure=(X=0,Y=0)
   ) -> (then, ReturnValue:Athena Inventory Drag Drop Object Reference)
   > n5 = Set bShowStackCountText!(
    execute
    bShowStackCountText:bool=false
    self:Athena Item Icon Loadout Object Reference=Get ItemIcon!(self:Athena Inventory Drag Visual Object Reference=@n3) -> (ItemIcon:Athena Item Icon Loadout Object Reference)
   ) -> (then, Output_Get:bool)
   Return Node!(execute=[@n5.then, @n2.Is Not Valid], Operation:Drag Drop Operation Object Reference=@n4)
  ]}
 OnPreviewMouseButtonDown {nodes [
   n1 = OnPreviewMouseButtonDown!() -> (then, MyGeometry:Geometry Structure, MouseEvent:"Pointer Event Structure (by ref)") {is=title K2Node_FunctionEntry}
   n2 = Get CommonInputSubsystem()
   > n3 = Branch(Condition="Equal (Enum)"(A=GetCurrentInputType(self=@n2), B=MouseAndKeyboard))
   > n4 = DetectDragIfPressed!(execute, PointerEvent:"Pointer Event Structure (by ref)"=@n1.MouseEvent, DragKey:Key Structure=LeftMouseButton) -> (then, ReturnValue:Event Reply Structure)
   > Return Node!(execute, ReturnValue:Event Reply Structure=@n4)
   Return Node(execute=@n3.else, ReturnValue=Unhandled())
  ]}
 On_Tooltip_MenuAnchor_GetUserMenuContent {nodes [
   On_Tooltip_MenuAnchor_GetUserMenuContent() {is=title K2Node_FunctionEntry}
   > n1 = Create Widget!(execute, Class=&/WBP_CrBrowser_Details.WBP_CrBrowser_Details_C, OwningPlayer=GetOwningPlayer()) -> (then, ReturnValue:WBP Cr Browser Details Object Reference)
   > n2 = Set DetailsTooltipWidget(DetailsTooltipWidget:WBP Cr Browser Details Object Reference=@n1) -> (Output_Get:WBP Cr Browser Details Object Reference)
   > SetVisibility(self=@n2, InVisibility=HitTestInvisible)
   > Return Node(ReturnValue:User Widget Object Reference=@n2)
  ]}
 SetActiveState {nodes [
   n1 = SetActiveState!() -> (then, IsItActive:bool=~) {is=title K2Node_FunctionEntry}
   > SetVisibility(
    self=Get SqueegeeInjectionSlot() -> (SqueegeeInjectionSlot:Squeegee Injection Slot Object Reference)
    InVisibility=Select(Option 0:ESlateVisibility Enum=Collapsed, Option 1:ESlateVisibility Enum=SelfHitTestInvisible, Index=@n1) -> (ReturnValue:ESlateVisibility Enum)
   )
   n2 = Get WeaponCard()
   > SetShouldCollapseBorderPadding!(execute, self:Fort Multi Size Item Card Object Reference=@n2, Value:bool=@n1) -> (then)
   > SetShouldSuppressItemTypeIcon!(execute, self:Fort Multi Size Item Card Object Reference=@n2, Value:bool=@n1) -> (then)
   > SetShouldCollapseQuantity!(execute, self:Fort Multi Size Item Card Object Reference=@n2, Value:bool=@n1) -> (then)
  ]}
 SetItemDirectly {nodes [
   n1 = SetItemDirectly!() -> (then, inFortItemDef:Fort Item Definition Object Reference) {is=title K2Node_FunctionEntry}
   > Set Override Fort Item Def(Override Fort Item Def:Item Definition Base Object Reference=@n1) -> (Output_Get:Item Definition Base Object Reference)
  ]}
 SetLabelVisibility {nodes [
   SetLabelVisibility!() -> (then, bShow:bool=~) {is=title K2Node_FunctionEntry}
   n1 = Get CommonInputSubsystem()
   > n2 = Branch(Condition="Equal (Enum)"(A=GetCurrentInputType(self=@n1), B=Touch))
   Check and Stop Animation(execute=[@n7.else, @n6.then, @n4.Is Not Valid, @n2.then], Animation=Get FadeIn() -> (FadeIn:Widget Animation Object Reference))
   > SetRenderOpacity(self=Get Overlay_Body() -> (Overlay_Body:Overlay Object Reference), InOpacity=0)
   n3 = GetTempItemInstance(execute=@n2.else, Count=0) {is=title K2Node_CallFunction}
   > n4 = Is Valid(InputObject=@n3)
   n5 = GetLabelText!(Item:Fort Item Object Reference=@n3) -> (Label:Text) {is=title K2Node_CallFunction}
   > n6 = Branch(Condition=TextIsEmpty(InText=@n5))
   "SetText (Text)"(execute=@n6.else, self=Get Text_Label(), InText=@n5)
   > n7 = Branch(Condition=Get bShow() -> (bShow:bool))
   > n8 = Branch(Condition=IsAnimationPlaying(InAnimation=Get FadeIn()))
   PlayAnimationForward(execute=@n8.else, InAnimation=Get FadeIn())
  ]}
 SetSubTabState {nodes [
   n1 = SetSubTabState!() -> (then, isSubTabOpened:bool=~) {is=title K2Node_FunctionEntry}
   > Set isGalleryView(isGalleryView:bool=@n1) -> (Output_Get:bool)
  ]}
 ShiftLabel {nodes [
   ShiftLabel!() -> (then, Alignment:EHorizontalAlignment Enum=HAlign_Fill) {is=title K2Node_FunctionEntry}
   n1 = SlotAsOverlaySlot(Widget=Get SizeBox_Label() -> (SizeBox_Label:Size Box Object Reference))
   > SetPadding!(execute, self:Overlay Slot Object Reference=@n1, InPadding:Margin Structure=Make Margin(Left=-36, Top=0, Right=-36, Bottom=0)) -> (then)
   > SetHorizontalAlignment(self=@n1, InHorizontalAlignment=HAlign_Center)
  ]}
 Update {nodes [
   n1 = GetTempItemInstance(Count=GetItemTypeCount() -> (ReturnValue=0))
   > SetItemToRepresent!(execute, self:Fort Item Widget NUI Object Reference=Get WeaponCard(), ItemToRepresent:Fort Item Object Reference=@n1) -> (then)
   > SetBrushFromSoftTexture!(
    execute
    self:Image Object Reference=Get Icon_Image() -> (Icon_Image:Image Object Reference)
    SoftTexture:Texture 2D Soft Object Reference=GetSmallPreviewImage(self=@n1)
    bMatchSize:bool=true
   ) -> (then)
   > n2 = Branch(Condition=Contains Item(
     TargetArray:Array of EFortItemType Enums=Get ShowRarityItemTypes() -> (ShowRarityItemTypes:Array of EFortItemType Enums=WorldItem)
     ItemToFind:"EFortItemType Enum (by ref)"=GetType!(self:Fort Item Object Reference=@n1) -> (ReturnValue:EFortItemType Enum=WorldItem)
    ))
   n3 = Get WeaponCard()
   > SetVisibility(self=@n3, InVisibility=SelfHitTestInvisible)
   n4 = Get Icon_Image()
   > SetVisibility(self=@n4, InVisibility=Hidden)
   SetVisibility(execute=@n2.else, self=@n3, InVisibility=Hidden)
   > SetVisibility(self=@n4, InVisibility=SelfHitTestInvisible)
   Update() {is=title K2Node_FunctionEntry}
   > Set DeviceFound(DeviceFound:bool=false) -> (Output_Get:bool)
   > n5 = For Each Loop with Break!(
    Exec:Exec
    Array:Array of Names=GetCreativeTagIDs!(ItemDefinition:Item Definition Base Object Reference=GetItemDefinition()) -> (ReturnValue:Array of Names)
    Break:Exec=@n6.then
   ) -> (LoopBody:Exec, Array Element:Name, Array Index:int, Completed:Exec)
   > Branch(Condition=Contains(SearchIn="To String (Name)"(InName=@n5), Substring=Device))
   > n6 = Set DeviceFound(DeviceFound=true)
   n7 = Get TempItem() -> (TempItem:Fort Item Object Reference)
   n8 = Branch(
    execute=@n5.Completed
    Condition=Contains Item(TargetArray=Get ShowRarityItemTypes() -> (ShowRarityItemTypes=WorldItem), ItemToFind=GetType(self=@n7) -> (ReturnValue=WorldItem))
   )
   > SetItemToRepresent(self=Get WeaponCard(), ItemToRepresent=Get TempItem())
   > SetActiveWidgetIndex(self=Get Switcher() -> (Switcher:Common Animated Switcher Object Reference), Index=0)
   n9 = Branch(execute=@n8.else, Condition=Get DeviceFound() -> (DeviceFound:bool))
   > SetBrushFromTexture!(
    execute
    self:Image Object Reference=Get Icon_Image()
    Texture:Texture 2D Object Reference=Get TempTexture2D() -> (TempTexture2D:Texture 2D Object Reference)
    bMatchSize:bool=true
   ) -> (then)
   > SetActiveWidgetIndex(self=Get Switcher(), Index=1)
   n10 = GetDynamicMaterial(execute=@n9.else, self=Get RoundedIcon_Image() -> (RoundedIcon_Image:Image Object Reference))
   > Set DM_Icon(DM_Icon:Material Instance Dynamic Object Reference=@n10) -> (Output_Get:Material Instance Dynamic Object Reference)
   > SetTextureParameterValueByInfo!(
    execute
    self:Material Instance Dynamic Object Reference=Get DM_Icon() -> (DM_Icon:Material Instance Dynamic Object Reference)
    ParameterInfo:"Material Parameter Info Structure (by ref)"=Make Material Parameter Info!(Name:Name=T_Texture, Association:EMaterialParameterAssociation Enum=GlobalParameter, Index:int=-1) -> (MaterialParameterInfo:Material Parameter Info Structure) {is=title K2Node_MakeStruct}
    Value:Texture Object Reference=Get TempTexture2D()
   ) -> (then)
   > SetActiveWidgetIndex(self=Get Switcher(), Index=2)
  ]}
 UpdateCount {nodes [
   n1 = UpdateCount!() -> (then, Count:int) {is=title K2Node_FunctionEntry}
   > n2 = GetTempItemInstance(Count=@n1)
   > SetItemToRepresent(self=Get WeaponCard(), ItemToRepresent=@n2)
  ]}
 UpdateSelected {nodes [
   n1 = UpdateSelected!() -> (then, IsSelected:bool=~) {is=title K2Node_FunctionEntry}
   > n2 = Set IsSelected(IsSelected=@n1)
   > n3 = Branch(Condition=@n2)
   > SetVectorParameterValue(self=Get OuterBorder_DM(), ParameterName="Color 2 (gradient on)", Value=Get SelectedColor())
   SetVectorParameterValue(execute=@n3.else, self=Get OuterBorder_DM(), ParameterName="Color 2 (gradient on)", Value=Get NotSelectedColor())
   > CancelDragDrop!(execute) -> (then)
  ]}
 UpdateTooltip {nodes [
   UpdateTooltip() {is=title K2Node_FunctionEntry}
   > n1 = Branch(Condition=Get isGalleryView() -> (isGalleryView:bool))
   n2 = Get Tooltip_Button()
   SetVisibility(execute=[@n5.then, @n4.Is Not Valid, @n1.then], self=@n2, InVisibility=Collapsed)
   n3 = GetTempItemInstance(execute=@n1.else, Count=0)
   > n4 = Is Valid(InputObject=@n3)
   > n5 = Branch(Condition=TextIsEmpty(InText=GetLabelText(Item=@n3)))
   SetVisibility(execute=@n5.else, self=@n2, InVisibility=SelfHitTestInvisible)
  ]}
]
variables [
 DetailsTooltipWidget {type=WBP_CrBrowser_Details_C, default=null}
 DM_Icon {type=MaterialInstanceDynamic}
 HoverBorder_DM {type=MaterialInstanceDynamic}
 HoveredBrush {
  type=SlateBrush
  default {
   bIsDynamicallyLoaded=false
   drawAs=Image
   imageType=NoImage
   mirroring=NoMirror
   resourceName=None
   resourceObject=/CreativeEmporium/Materials/MI_CreativeBrowser_HoverBorder
   tiling=NoTile
   imageSize=<32 32>
   margin=<ltrb 0 0 0 0>
   outlineSettings {
    bUseBrushTransparency=false
    roundingType=HalfHeightRadius
    width=0
    color {colorUseRule=UseColor_Specified, specifiedColor=<rgba 0 0 0 0>}
    cornerRadii=<0 0 0 1>
   }
   tintColor {colorUseRule=UseColor_Specified, specifiedColor=<rgba 1 1 1 0>}
   uVRegion {bIsValid=false, max=<0 0>, min=<0 0>}
  }
 }
 HoverSound {type=SoundBase, default=/Game/Sounds/Fort_UI_Sounds/FrontEnd/DefaultSounds/PressedButtons/MSS_UI_Button_Hover_Default}
 isGalleryView {type=boolean}
 IsSelected {type=boolean}
 LabelOverride {type=string, default=""}
 NotSelectedColor {type=LinearColor, default=<rgba 1 1 1 0.0588>}
 OuterBorder_DM {type=MaterialInstanceDynamic}
 Override Fort Item Def {type=ItemDefinitionBase, default=null}
 Press Sound {type=SoundBase, default=/Game/Sounds/Fort_UI_Sounds/FrontEnd/DefaultSounds/PressedButtons/MSS_UI_Button_Pressed_Default}
 SelectedBrush {
  type=SlateBrush
  default {
   bIsDynamicallyLoaded=false
   drawAs=Image
   imageType=NoImage
   mirroring=NoMirror
   resourceName=None
   resourceObject=&MI_CreativeBrowser_HoverBorder
   tiling=NoTile
   imageSize=<48 48>
   margin=<ltrb 0 0 0 0>
   outlineSettings {
    bUseBrushTransparency=false
    roundingType=HalfHeightRadius
    width=0
    color {colorUseRule=UseColor_Specified, specifiedColor=<rgba 0 0 0 0>}
    cornerRadii=<0 0 0 1>
   }
   tintColor {colorUseRule=UseColor_Specified, specifiedColor=<rgba 1 1 1 1>}
   uVRegion {bIsValid=false, max=<0 0>, min=<0 0>}
  }
 }
 SelectedColor {type=LinearColor, default=<rgba 1 1 1 0.8908>}
 ShowRarityItemTypes {type="array<EFortItemType>", default [Ammo, AthenaGadget, Consumable, Trap, Weapon, WeaponMelee, WeaponRanged, WorldItem, WorldResource]}
 TempItem {type=FortItem, default=null}
 TempTexture2D {type=Texture2D}
]
widgets {
 info {
  inheritedWidgetCount=1
  parentClass=/Script/FortniteUI.AthenaCreativeItemTileButton
  rootWidgetClass=/Script/FortniteUI.AthenaProfileStatBox
  widgetCount=33
 }
 widgets [
  ProfileStatBox {
   widgetClassPath=&AthenaProfileStatBox
   bIsVariable=true
   widget=ProfileStatBox
   properties {bOverride_HeightOverride=true, bOverride_MinDesiredWidth=true, heightOverride=144, minDesiredWidth=960, widthOverride=163}
  }
  Overlay_0 {widgetClassPath=&Overlay, parent=ProfileStatBox, widget=Overlay_0, slot {class=SizeBoxSlot}}
  OuterBorder_Image {
   widgetClassPath=&Image
   bIsVariable=true
   parent=Overlay_0
   widget=OuterBorder_Image
   properties {brush {
     bIsDynamicallyLoaded=false
     drawAs=Image
     imageType=NoImage
     mirroring=NoMirror
     resourceName=None
     resourceObject=/CreativeEmporium/Materials/MI_CreativeBrowser_OuterBorder
     tiling=NoTile
     imageSize=<32 32>
     margin=<ltrb 0 0 0 0>
     outlineSettings {
      bUseBrushTransparency=false
      roundingType=FixedRadius
      width=2
      color {colorUseRule=UseColor_Specified, specifiedColor=<rgba 1 1 1 0.3>}
      cornerRadii=<10 10 10 10>
     }
     tintColor {colorUseRule=UseColor_Specified, specifiedColor=<rgba 1 1 1 1>}
     uVRegion {bIsValid=false, max=<0 0>, min=<0 0>}
    }}
   slot {class=OverlaySlot, properties {horizontalAlignment=HAlign_Fill, verticalAlignment=VAlign_Fill}}
  }
  SelectedBorder_Image {
   widgetClassPath=&Image
   bIsVariable=true
   parent=Overlay_0
   widget=SelectedBorder_Image
   properties {
    renderOpacity=0
    visibility=Hidden
    brush {
     bIsDynamicallyLoaded=false
     drawAs=Image
     imageType=NoImage
     mirroring=NoMirror
     resourceName=None
     resourceObject=/CreativeEmporium/Materials/MI_CrBrowser_Tile_Selected
     tiling=NoTile
     imageSize=<32 32>
     margin=<ltrb 0 0 0 0>
     outlineSettings {
      bUseBrushTransparency=false
      roundingType=FixedRadius
      width=2
      color {colorUseRule=UseColor_Specified, specifiedColor=<rgba 1 1 1 0.3>}
      cornerRadii=<10 10 10 10>
     }
     tintColor {colorUseRule=UseColor_Specified, specifiedColor=<rgba 1 1 1 1>}
     uVRegion {bIsValid=false, max=<0 0>, min=<0 0>}
    }
   }
   slot {class=OverlaySlot, properties {horizontalAlignment=HAlign_Fill, verticalAlignment=VAlign_Fill}}
  }
  HoverBorder_Image {
   widgetClassPath=&Image
   bIsVariable=true
   parent=Overlay_0
   widget=HoverBorder_Image
   properties {
    renderOpacity=0
    brush {
     bIsDynamicallyLoaded=false
     drawAs=Image
     imageType=NoImage
     mirroring=NoMirror
     resourceName=None
     resourceObject=/CreativeEmporium/Materials/MI_CrBrowser_TileOutline
     tiling=NoTile
     imageSize=<32 32>
     margin=<ltrb 0 0 0 0>
     outlineSettings {
      bUseBrushTransparency=false
      roundingType=FixedRadius
      width=2
      color {colorUseRule=UseColor_Specified, specifiedColor=<rgba 1 1 1 0.3>}
      cornerRadii=<10 10 10 10>
     }
     tintColor {colorUseRule=UseColor_Specified, specifiedColor=<rgba 1 1 1 1>}
     uVRegion {bIsValid=false, max=<0 0>, min=<0 0>}
    }
    colorAndOpacity=<rgba 1 1 1 0.5>
   }
   slot {class=OverlaySlot, properties {horizontalAlignment=HAlign_Fill, verticalAlignment=VAlign_Fill, padding=<ltrb -8 -8 -8 -8>}}
  }
  Overlay_2 {
   widgetClassPath=&Overlay
   parent=Overlay_0
   widget=Overlay_2
   slot {class=OverlaySlot, properties {horizontalAlignment=HAlign_Fill, verticalAlignment=VAlign_Fill, padding=<ltrb 2 2 2 2>}}
  }
  Overlay_1 {
   widgetClassPath=&Overlay
   parent=Overlay_2
   widget=Overlay_1
   slot {class=OverlaySlot, properties {horizontalAlignment=HAlign_Fill, verticalAlignment=VAlign_Fill}}
  }
  Container_Background {
   widgetClassPath=&Image
   bIsVariable=true
   parent=Overlay_1
   widget=Container_Background
   properties {brush {
     bIsDynamicallyLoaded=false
     drawAs=RoundedBox
     imageType=NoImage
     mirroring=NoMirror
     resourceName=None
     resourceObject=null
     tiling=NoTile
     imageSize=<32 32>
     margin=<ltrb 0 0 0 0>
     outlineSettings {
      bUseBrushTransparency=false
      roundingType=FixedRadius
      width=2
      color {colorUseRule=UseColor_Specified, specifiedColor=<rgba 1 1 1 0.1>}
      cornerRadii=<16 16 16 16>
     }
     tintColor {colorUseRule=UseColor_Specified, specifiedColor=<rgba 1 1 1 0.05>}
     uVRegion {bIsValid=false, max=<0 0>, min=<0 0>}
    }}
   slot {class=OverlaySlot, properties {horizontalAlignment=HAlign_Fill, verticalAlignment=VAlign_Fill}}
  }
  PurchaseBorder {
   widgetClassPath=&CommonBorder
   bIsVariable=true
   parent=Overlay_1
   widget=PurchaseBorder
   properties {
    background {
     bIsDynamicallyLoaded=false
     drawAs=NoDrawType
     imageType=NoImage
     mirroring=NoMirror
     resourceName=None
     resourceObject=null
     tiling=NoTile
     imageSize=<32 32>
     margin=<ltrb 0.49 0.49 0.49 0.49>
     outlineSettings {
      bUseBrushTransparency=false
      roundingType=HalfHeightRadius
      width=0
      color {colorUseRule=UseColor_Specified, specifiedColor=<rgba 0 0 0 0>}
      cornerRadii=<0 0 0 0>
     }
     tintColor {colorUseRule=UseColor_Specified, specifiedColor=<rgba 1 1 1 1>}
     uVRegion {bIsValid=false, max=<0 0>, min=<0 0>}
    }
    padding=<ltrb 0 0 0 0>
   }
   slot {class=OverlaySlot, properties {horizontalAlignment=HAlign_Fill, verticalAlignment=VAlign_Fill}}
  }
  Switcher_WeaponCard {
   widgetClassPath=&CommonVisibilitySwitcher
   bIsVariable=true
   parent=PurchaseBorder
   widget=Switcher_WeaponCard
   slot {class=BorderSlot, properties {padding=<ltrb 0 0 0 0>}}
  }
  GridPanel_WeaponCard {
   widgetClassPath=&GridPanel
   bIsVariable=true
   parent=Switcher_WeaponCard
   widget=GridPanel_WeaponCard
   properties {columnFill [1], rowFill [1]}
   slot {class=CommonVisibilitySwitcherSlot}
  }
  Switcher {
   widgetClassPath=/Script/CommonUI.CommonAnimatedSwitcher
   bIsVariable=true
   parent=GridPanel_WeaponCard
   widget=Switcher
   slot {class=GridSlot}
  }
  WeaponCard {
   widgetClassPath=/Script/FortniteUI.FortMultiSizeItemCard
   bIsVariable=true
   parent=Switcher
   widget=WeaponCard
   properties {bPItemCardSize=M, designTimePreviewItemType=/Game/Athena/Items/Weapons/WID_Shotgun_SemiAuto_Athena_R_Ore_T03}
   slot {class=WidgetSwitcherSlot, properties {padding=<ltrb 8 8 8 8>}}
  }
  Icon_Image {
   widgetClassPath=&Image
   bIsVariable=true
   parent=Switcher
   widget=Icon_Image
   properties {visibility=HitTestInvisible}
   slot {class=WidgetSwitcherSlot, properties {padding=<ltrb 8 8 8 8>}}
  }
  RoundedIcon_Image {
   widgetClassPath=&Image
   bIsVariable=true
   parent=Switcher
   widget=RoundedIcon_Image
   properties {
    visibility=SelfHitTestInvisible
    brush {
     bIsDynamicallyLoaded=false
     drawAs=Image
     imageType=FullColor
     mirroring=NoMirror
     resourceName=None
     resourceObject=/CreativeEmporium/Materials/MI_UI_CrBrowserTileRounded
     tiling=NoTile
     imageSize=<32 32>
     margin=<ltrb 0 0 0 0>
     outlineSettings {
      bUseBrushTransparency=false
      roundingType=HalfHeightRadius
      width=0
      color {colorUseRule=UseColor_Specified, specifiedColor=<rgba 0 0 0 0>}
      cornerRadii=<0 0 0 0>
     }
     tintColor {colorUseRule=UseColor_Specified, specifiedColor=<rgba 1 1 1 1>}
     uVRegion {bIsValid=false, max=<0 0>, min=<0 0>}
    }
   }
   slot {class=WidgetSwitcherSlot}
  }
  SqueegeeInjectionSlot {
   widgetClassPath=/Game/UI/Plugin/Squeegee/SqueegeeInjectionSlot
   bIsVariable=true
   parent=GridPanel_WeaponCard
   widget=SqueegeeInjectionSlot
   properties {colorAndOpacity=<rgba -0.072 1 1 0.416>}
   slot {class=GridSlot, properties {padding=<ltrb 8 8 8 8>}}
  }
  IconMask_Image {
   widgetClassPath=&Image
   bIsVariable=true
   parent=GridPanel_WeaponCard
   widget=IconMask_Image
   properties {
    visibility=Collapsed
    brush {
     bIsDynamicallyLoaded=false
     drawAs=Image
     imageType=NoImage
     mirroring=NoMirror
     resourceName=None
     resourceObject=/CreativeEmporium/Materials/MI_CreativeBrowser_IconBorder
     tiling=NoTile
     imageSize=<32 32>
     margin=<ltrb 0 0 0 0>
     outlineSettings {
      bUseBrushTransparency=false
      roundingType=HalfHeightRadius
      width=0
      color {colorUseRule=UseColor_Specified, specifiedColor=<rgba 0 0 0 0>}
      cornerRadii=<0 0 0 0>
     }
     tintColor {colorUseRule=UseColor_Specified, specifiedColor=<rgba 1 1 1 1>}
     uVRegion {bIsValid=false, max=<0 0>, min=<0 0>}
    }
    colorAndOpacity=<rgba 1 1 1 1.6858>
   }
   slot {class=GridSlot, properties {layer=100, padding=<ltrb 24 24 24 24>}}
  }
  GridPanel_OverflowCount {
   widgetClassPath=&GridPanel
   bIsVariable=true
   parent=Switcher_WeaponCard
   widget=GridPanel_OverflowCount
   properties {columnFill [1], rowFill [1]}
   slot {class=CommonVisibilitySwitcherSlot}
  }
  Icon_OverflowCount {
   widgetClassPath=&Image
   bIsVariable=true
   parent=GridPanel_OverflowCount
   widget=Icon_OverflowCount
   properties {
    visibility=HitTestInvisible
    brush {
     bIsDynamicallyLoaded=false
     drawAs=Image
     imageType=NoImage
     mirroring=NoMirror
     resourceName=None
     resourceObject=null
     tiling=NoTile
     imageSize=<32 32>
     margin=<ltrb 0 0 0 0>
     outlineSettings {
      bUseBrushTransparency=false
      roundingType=HalfHeightRadius
      width=0
      color {colorUseRule=UseColor_Specified, specifiedColor=<rgba 0 0 0 0>}
      cornerRadii=<0 0 0 0>
     }
     tintColor {colorUseRule=UseColor_Specified, specifiedColor=<rgba 0.25 0.25 0.25 1>}
     uVRegion {bIsValid=false, max=<0 0>, min=<0 0>}
    }
   }
   slot {class=GridSlot, properties {padding=<ltrb 8 8 8 8>}}
  }
  Text_OverflowCount {
   widgetClassPath=&CommonTextBlock
   bIsVariable=true
   parent=GridPanel_OverflowCount
   widget=Text_OverflowCount
   properties {
    style=&TextStyle-Base_C
    text=+888
    font {
     bForceMonospaced=false
     bMaterialIsStencil=false
     fontMaterial=null
     fontObject=&NotoSans
     letterSpacing=0
     monospacedWidth=1
     size=24
     skewAmount=0
     typefaceFontName=NotoSans-Regular
     outlineSettings {
      bApplyOutlineToDropShadows=false
      bMiteredCorners=false
      bSeparateFillAlpha=false
      outlineMaterial=null
      outlineSize=0
      outlineColor=<rgba 0 0 0 1>
     }
    }
    shadowOffset=<0 0>
   }
   slot {class=GridSlot, properties {horizontalAlignment=HAlign_Center, verticalAlignment=VAlign_Center}}
  }
  Outline_Border {
   widgetClassPath=&Image
   bIsVariable=true
   parent=Overlay_2
   widget=Outline_Border
   properties {
    visibility=Hidden
    brush {
     bIsDynamicallyLoaded=false
     drawAs=RoundedBox
     imageType=NoImage
     mirroring=NoMirror
     resourceName=None
     resourceObject=null
     tiling=NoTile
     imageSize=<32 32>
     margin=<ltrb 0 0 0 0>
     outlineSettings {
      bUseBrushTransparency=false
      roundingType=FixedRadius
      width=2
      color {colorUseRule=UseColor_Specified, specifiedColor=<rgba 1 1 1 1>}
      cornerRadii=<16 16 16 16>
     }
     tintColor {colorUseRule=UseColor_Specified, specifiedColor=<rgba 1 1 1 0>}
     uVRegion {bIsValid=false, max=<0 0>, min=<0 0>}
    }
   }
   slot {class=OverlaySlot, properties {horizontalAlignment=HAlign_Fill, verticalAlignment=VAlign_Fill}}
  }
  SizeBox_Label {
   widgetClassPath=&SizeBox
   bIsVariable=true
   parent=Overlay_0
   widget=SizeBox_Label
   properties {
    bOverride_HeightOverride=true
    bOverride_WidthOverride=true
    heightOverride=28
    minDesiredHeight=28
    widthOverride=172
    renderTransform {angle=0, scale=<1 1>, shear=<0 0>, translation=<0 12>}
   }
   slot {class=OverlaySlot, properties {horizontalAlignment=HAlign_Center, verticalAlignment=VAlign_Bottom, padding=<ltrb -36 0 -36 -15>}}
  }
  Overlay_Body {
   widgetClassPath=&Overlay
   bIsVariable=true
   parent=SizeBox_Label
   widget=Overlay_Body
   properties {renderOpacity=0}
   slot {class=SizeBoxSlot}
  }
  Image_BG {
   widgetClassPath=&Image
   parent=Overlay_Body
   widget=Image_BG
   properties {brush {
     bIsDynamicallyLoaded=false
     drawAs=RoundedBox
     imageType=NoImage
     mirroring=NoMirror
     resourceName=None
     resourceObject=null
     tiling=NoTile
     imageSize=<32 32>
     margin=<ltrb 0.1 0.1 0.1 0.1>
     outlineSettings {
      bUseBrushTransparency=true
      roundingType=FixedRadius
      width=1
      color {colorUseRule=UseColor_Specified, specifiedColor=<rgba 1 1 1 1>}
      cornerRadii=<8 8 8 8>
     }
     tintColor {colorUseRule=UseColor_Specified, specifiedColor=<rgba 1 1 1 1>}
     uVRegion {bIsValid=false, max=<0 0>, min=<0 0>}
    }}
   slot {class=OverlaySlot, properties {horizontalAlignment=HAlign_Fill, verticalAlignment=VAlign_Fill}}
  }
  Text_Label {
   widgetClassPath=&CommonTextBlock
   bIsVariable=true
   parent=Overlay_Body
   widget=Text_Label
   properties {
    justification=Center
    scrollStyle=&TextScrollStyle-Base_C
    text="[Item Name]"
    colorAndOpacity {colorUseRule=UseColor_Specified, specifiedColor=<rgba 0 0 0 1>}
    font {
     bForceMonospaced=false
     bMaterialIsStencil=false
     fontMaterial=null
     fontObject=&HeadingNow
     letterSpacing=0
     monospacedWidth=1
     size=15
     skewAmount=0
     typefaceFontName=75 Medium
     outlineSettings {
      bApplyOutlineToDropShadows=false
      bMiteredCorners=false
      bSeparateFillAlpha=false
      outlineMaterial=null
      outlineSize=0
      outlineColor=<rgba 0 0 0 1>
     }
    }
    shadowOffset=<0 0>
   }
   slot {class=OverlaySlot, properties {horizontalAlignment=HAlign_Center, verticalAlignment=VAlign_Center, padding=<ltrb 6 0 6 0>}}
  }
  Tooltip_MenuAnchor {
   widgetClassPath=/Script/UMG.MenuAnchor
   bIsVariable=true
   parent=Overlay_0
   widget=Tooltip_MenuAnchor
   properties {menuClass=&WBP_CrBrowser_Details_C}
   slot {class=OverlaySlot, properties {horizontalAlignment=HAlign_Right}}
  }
  TooltipTrigger_Overlay {
   widgetClassPath=&Overlay
   bIsVariable=true
   parent=Overlay_0
   widget=TooltipTrigger_Overlay
   slot {class=OverlaySlot, properties {horizontalAlignment=HAlign_Right, padding=<ltrb 0 2.5 -8 0>}}
  }
  Tootip_HardwareVisibility {
   widgetClassPath=&CommonHardwareVisibilityBorder
   bIsVariable=true
   parent=TooltipTrigger_Overlay
   widget=Tootip_HardwareVisibility
   properties {visibilityQuery {
     autoDescription=" ANY( Input.MouseAndKeyboard )"
     tokenStreamVersion=0
     userDescription=""
     queryTokenStream [0,1,1,1,0]
     tagDictionary [{tagName=Input.MouseAndKeyboard}]
    }}
   slot {class=OverlaySlot}
  }
  Tooltip_Button {
   widgetClassPath=&Button
   parent=Tootip_HardwareVisibility
   widget=Tooltip_Button
   properties {
    visibility=Collapsed
    widgetStyle {
     clickedSlateSound {resourceObject=null}
     disabled {
      bIsDynamicallyLoaded=false
      drawAs=NoDrawType
      imageType=NoImage
      mirroring=NoMirror
      resourceName=None
      resourceObject=null
      tiling=NoTile
      imageSize=<0 0>
      margin=<ltrb 0 0 0 0>
      outlineSettings {
       bUseBrushTransparency=false
       roundingType=HalfHeightRadius
       width=0
       color {colorUseRule=UseColor_Specified, specifiedColor=<rgba 0 0 0 0>}
       cornerRadii=<0 0 0 0>
      }
      tintColor {colorUseRule=UseColor_Specified, specifiedColor=<rgba 0.4955 0.4955 0.4955 0>}
      uVRegion {bIsValid=false, max=<0 0>, min=<0 0>}
     }
     disabledForeground {colorUseRule=UseColor_Specified, specifiedColor=<rgba 0.5271 0.5271 0.5271 1>}
     hovered {
      bIsDynamicallyLoaded=false
      drawAs=RoundedBox
      imageType=NoImage
      mirroring=NoMirror
      resourceName=None
      resourceObject=null
      tiling=NoTile
      imageSize=<32 32>
      margin=<ltrb 0 0 0 0>
      outlineSettings {
       bUseBrushTransparency=true
       roundingType=FixedRadius
       width=1
       color {colorUseRule=UseColor_Specified, specifiedColor=<rgba 0.7243 0.7243 0.7243 1>}
       cornerRadii=<4 4 4 4>
      }
      tintColor {colorUseRule=UseColor_Specified, specifiedColor=<rgba 0.4955 0.4955 0.4955 0>}
      uVRegion {bIsValid=false, max=<0 0>, min=<0 0>}
     }
     hoveredForeground {colorUseRule=UseColor_Specified, specifiedColor=<rgba 0 1 1 1>}
     hoveredSlateSound {resourceObject=null}
     normal {
      bIsDynamicallyLoaded=false
      drawAs=RoundedBox
      imageType=NoImage
      mirroring=NoMirror
      resourceName=None
      resourceObject=null
      tiling=NoTile
      imageSize=<32 32>
      margin=<ltrb 0 0 0 0>
      outlineSettings {
       bUseBrushTransparency=true
       roundingType=FixedRadius
       width=1
       color {colorUseRule=UseColor_Specified, specifiedColor=<rgba 0.6951 0.6951 0.6951 1>}
       cornerRadii=<4 4 4 4>
      }
      tintColor {colorUseRule=UseColor_Specified, specifiedColor=<rgba 0.4955 0.4955 0.4955 0>}
      uVRegion {bIsValid=false, max=<0 0>, min=<0 0>}
     }
     normalForeground {colorUseRule=UseColor_Specified, specifiedColor=<rgba 0 1 1 1>}
     normalPadding=<ltrb 12 1.5 12 1.5>
     pressed {
      bIsDynamicallyLoaded=false
      drawAs=RoundedBox
      imageType=NoImage
      mirroring=NoMirror
      resourceName=None
      resourceObject=null
      tiling=NoTile
      imageSize=<32 32>
      margin=<ltrb 0 0 0 0>
      outlineSettings {
       bUseBrushTransparency=true
       roundingType=FixedRadius
       width=1
       color {colorUseRule=UseColor_Specified, specifiedColor=<rgba 0.7243 0.7243 0.7243 1>}
       cornerRadii=<4 4 4 4>
      }
      tintColor {colorUseRule=UseColor_Specified, specifiedColor=<rgba 0.4955 0.4955 0.4955 0>}
      uVRegion {bIsValid=false, max=<0 0>, min=<0 0>}
     }
     pressedForeground {colorUseRule=UseColor_Specified, specifiedColor=<rgba 0 1 1 1>}
     pressedPadding=<ltrb 12 1.5 12 1.5>
     pressedSlateSound {resourceObject=null}
    }
   }
   slot {class=BorderSlot, properties {padding=<ltrb 0 0 0 0>}}
  }
  Overlay_3 {widgetClassPath=&Overlay, parent=Tooltip_Button, widget=Overlay_3, slot {class=ButtonSlot, properties {padding=<ltrb 0 0 0 0>}}}
  Image_64 {
   widgetClassPath=&Image
   parent=Overlay_3
   widget=Image_64
   properties {brush {
     bIsDynamicallyLoaded=false
     drawAs=Image
     imageType=FullColor
     mirroring=NoMirror
     resourceName=None
     resourceObject=/FortUILibrary/Foundation/Art_ToolKit/Tools/M_UI_Circle_UVspaceGradient
     tiling=NoTile
     imageSize=<32 32>
     margin=<ltrb 0 0 0 0>
     outlineSettings {
      bUseBrushTransparency=false
      roundingType=HalfHeightRadius
      width=0
      color {colorUseRule=UseColor_Specified, specifiedColor=<rgba 0 0 0 0>}
      cornerRadii=<0 0 0 0>
     }
     tintColor {colorUseRule=UseColor_Specified, specifiedColor=<rgba 0 0 0 0.75>}
     uVRegion {bIsValid=false, max=<0 0>, min=<0 0>}
    }}
   slot {class=OverlaySlot}
  }
  Tooltip_Image {
   widgetClassPath=&Image
   parent=Overlay_3
   widget=Tooltip_Image
   properties {brush {
     bIsDynamicallyLoaded=false
     drawAs=Image
     imageType=FullColor
     mirroring=NoMirror
     resourceName=None
     resourceObject=/FortUILibrary/Textures/MaterialInstances/MI_UIKit_Icon_Information
     tiling=NoTile
     imageSize=<32 32>
     margin=<ltrb 0 0 0 0>
     outlineSettings {
      bUseBrushTransparency=false
      roundingType=HalfHeightRadius
      width=0
      color {colorUseRule=UseColor_Specified, specifiedColor=<rgba 0 0 0 0>}
      cornerRadii=<0 0 0 0>
     }
     tintColor {colorUseRule=UseColor_Specified, specifiedColor=<rgba 1 1 1 1>}
     uVRegion {bIsValid=false, max=<0 0>, min=<0 0>}
    }}
   slot {class=OverlaySlot}
  }
  InputActionWidget {widgetClassPath=&CommonActionWidget, bInherited=true, widget=null}
 ]
}
