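# Item list tab panel for the Creative Emporium browser, showing items in a tile view with a sort button and item count.
# Generated entries get sub-tab state, grid-column label alignment, count, and category-based card size.
# Selection is tracked by item data and restored as tile widgets are generated and released.
# Selecting an item sends a temporary item instance to the HUD context, but only if the player can create in the volume.
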WidgetBlueprint /CreativeEmporium/WBP_Creative_Emporium_Browser_ListView
graphs [
 BP_GetDesiredFocusTarget {nodes [
   Get Desired Focus Target()
   n1 = Get CreativeTileView_ItemOptions() -> (CreativeTileView_ItemOptions:Athena Creative Item Tile View Object Reference)
   > n2 = Branch(Condition="integer > integer"(A=GetNumItems!(self:List View Object Reference=@n1) -> (ReturnValue:int=0)))
   > Return Node!(execute, ReturnValue:Widget Object Reference=@n1)
   Return Node(
    execute=@n2.else
    ReturnValue=GetEmptyTabFocusWidget!(self:Fort Emporium Item List Tab Panel Object Reference) -> (ReturnValue:Widget Object Reference)
   )
  ]}
 EventGraph {nodes [
   n1 = "On Entry Generated (CreativeTileView_ItemOptions)"!() -> (OutputDelegate, then, Widget:User Widget Object Reference) {is=title K2Node_ComponentBoundEvent}
   > n2 = Cast To WBP_Creative_Emporium_ItemEntry!(execute, Object=@n1.Widget) -> (then, CastFailed, AsWBP Creative Emporium Item Entry:WBP Creative Emporium Item Entry Object Reference) {is=title K2Node_DynamicCast}
   > SetSubTabState!(
    execute
    self:WBP Creative Browser Item List Entry Widget Object Reference=@n2
    isSubTabOpened:bool=IsSubTabOpened!(self:Fort Emporium Item List Tab Panel Object Reference) -> (ReturnValue:bool)
   ) -> (then)
   > Update Label Offset!(execute, self, Entry:WBP Creative Emporium Item Entry Object Reference=@n2) -> (then)
   > n3 = Update Item Count!(execute, self, In Entry Widget:WBP Creative Emporium Item Entry Object Reference=@n2) -> (then, Out Entry Widget:WBP Creative Emporium Item Entry Object Reference)
   n4 = Get WeaponCard!(self:WBP Creative Browser Item List Entry Widget Object Reference=@n3) -> (WeaponCard:Fort Multi Size Item Card Object Reference)
   > SetCardSize!(
    execute
    self:Fort Multi Size Item Card Object Reference=@n4
    CardSize:EFortItemCardSize Enum=GetCardSizeForCategory!(self:Fort Emporium Item List Tab Panel Object Reference) -> (ReturnValue:EFortItemCardSize Enum=XXS)
   ) -> (then)
   > SetShouldCollapseBorderPadding!(
    execute
    self:Fort Multi Size Item Card Object Reference=@n4
    Value:bool=GetCollapseBorderPadFlagForCategory!(self:Fort Emporium Item List Tab Panel Object Reference) -> (ReturnValue:bool)
   ) -> (then)
   > Branch(Condition=CompareToItemEntry!(
     self:Athena Creative Item Tile Button Object Reference=@n2
     InItemEntry:"Fort Item Entry Structure (by ref)"=Get CurrentItemEntry() -> (CurrentItemEntry:Fort Item Entry Structure)
    ) -> (ReturnValue:bool))
   > UpdateSelected!(execute, self:WBP Creative Browser Item List Entry Widget Object Reference=@n2, IsSelected:bool=true) -> (then)
   > Set CurrentEntryWidget(CurrentEntryWidget:WBP Creative Emporium Item Entry Object Reference=@n2) -> (Output_Get:WBP Creative Emporium Item Entry Object Reference)
   n5 = "On Entry Released (CreativeTileView_ItemOptions)"!() -> (OutputDelegate, then, Widget:User Widget Object Reference) {is=title K2Node_ComponentBoundEvent}
   > n6 = Cast To WBP_Creative_Emporium_ItemEntry(Object=@n5.Widget)
   > EnterPurchasedState!(execute, self:WBP Creative Browser Item List Entry Widget Object Reference=@n6, Purchased:bool=false) -> (then)
   > Branch(Condition=CompareToItemEntry(self=@n6, InItemEntry=Get CurrentItemEntry()))
   > UpdateSelected(self=@n6, IsSelected=false)
   > Set CurrentEntryWidget()
   n7 = "On Item Selection Changed (CreativeTileView_ItemOptions)"!() -> (OutputDelegate, then, Item:Object Reference, bIsSelected:bool) {is=title K2Node_ComponentBoundEvent}
   > Branch(Condition=@n7.bIsSelected)
   > n8 = Cast To AthenaCreativeItemEntryObjectWrapper!(execute, Object=@n7.Item) -> (then, CastFailed, AsAthena Creative Item Entry Object Wrapper:Athena Creative Item Entry Object Wrapper Object Reference) {is=title K2Node_DynamicCast}
   > SwitchSelection!(
    execute
    self
    ItemEntry:Fort Item Entry Structure=Get ItemEntry!(self:Athena Creative Item Entry Object Wrapper Object Reference=@n8) -> (ItemEntry:Fort Item Entry Structure)
   ) -> (then)
   Shine All Items() {is=title K2Node_CustomEvent}
   > n9 = For Each Loop with Break!(
    Exec:Exec
    Array:Array of User Widget Object References=GetDisplayedEntryWidgets!(self:List View Base Object Reference=Get CreativeTileView_ItemOptions()) -> (ReturnValue:Array of User Widget Object References)
    Break:Exec
   ) -> (LoopBody:Exec, Array Element:User Widget Object Reference, Array Index:int, Completed:Exec)
   > n10 = Cast To WBP_Creative_Emporium_ItemEntry(Object=@n9)
   > Shine Item Widget(self:WBP Creative Browser Item List Entry Widget Object Reference=@n10)
   n11 = SwitchSelection!() -> (OutputDelegate, then, ItemEntry:Fort Item Entry Structure) {is=title K2Node_CustomEvent}
   > UpdateSelected(self=Get CurrentEntryWidget() -> (CurrentEntryWidget:WBP Creative Emporium Item Entry Object Reference), IsSelected=false)
   > n12 = Set CurrentItemEntry(CurrentItemEntry:Fort Item Entry Structure=@n11.ItemEntry) -> (Output_Get:Fort Item Entry Structure)
   > n13 = GetEntryWidgetFromCurrentItem!(execute, self) -> (then, Entry:WBP Creative Emporium Item Entry Object Reference)
   > n14 = Set CurrentEntryWidget(CurrentEntryWidget=@n13)
   > UpdateSelected(self=@n14, IsSelected=true)
   n15 = BreakFortItemEntry!(InItemEntry:"Fort Item Entry Structure (by ref)"=@n12) -> (
    ItemDefinition:Item Definition Base Object Reference
    ItemGuid:Guid Structure
    Count:int=0
    Level:int=0
    LoadedAmmo:int=0
    PhantomReserveAmmo:int=0
   )
   > SetSelectedInventoryItem!(
    execute
    self:Athena HUDContext Object Reference=Get AthenaHUDContext!() -> (ReturnValue:Athena HUDContext Object Reference) {is=title K2Node_GetBlueprintContextEx}
    Item:Fort Item Object Reference=Select(
     Option 0:Fort Item Object Reference
     Option 1:Fort Item Object Reference=CreateTemporaryItemInstance!(ItemDefinition:Item Definition Base Object Reference=@n15, Count:int=@n15.Count, Level:int=-1) -> (ReturnValue:Fort Item Object Reference)
     Index=CanPlayerCreateInVolume!(self:Fort Emporium Item List Tab Panel Object Reference) -> (ReturnValue:bool)
    ) -> (ReturnValue:Fort Item Object Reference)
   ) -> (then)
   > UpdateTabHeader(self:Fort Emporium Item List Tab Panel Object Reference)
   Shine Selected Item() {is=title K2Node_CustomEvent}
   > Shine Item Widget(self=Get CurrentEntryWidget())
   Event FocusCurrentSelection() {is=title K2Node_Event}
   n16 = Get CurrentEntryWidget()
   > Is Valid(InputObject=@n16)
   > SetFocus(self=@n16)
   Event UpdateTabHeader() {is=title K2Node_Event}
   n17 = Get Items() -> (Items:Array of Fort Item Entry Structures)
   > "SetText (Text)"(
    self=[Get Filter_Count() -> (Filter_Count:Common Text Object Reference), Get Filter_Count_V2() -> (Filter_Count_V2:Common Text Object Reference)]
    InText=Format Text!(Format="LOCTEXT(\"{Count} {Count}|plural(one=ITEM,other=ITEMS)\")", Count:int=Length(TargetArray:Array of Fort Item Entry Structures=@n17)) -> (Result)
   )
   n18 = Event UpdateSortButton!() -> (OutputDelegate, then, SortType:"Text (by ref)") {is=title K2Node_Event}
   > n19 = Cast To WBP_UIKit_Button_Regular!(execute, Object=Get Button_Sort() -> (Button_Sort:Common Button Base Object Reference)) -> (then, CastFailed, AsWBP UIKit Button Regular:WBP UIKit Button Regular Object Reference) {is=title K2Node_DynamicCast}
   > SetText!(
    execute
    self:Fort CTAButton Object Reference=@n19
    ButtonText:"Text (by ref)"=Format Text!(Format="LOCTEXT(\"SORT: {SortType}\")", SortType:"Text (by ref)"=@n18.SortType) -> (Result)
   ) -> (then)
   Event ItemEquipped!() -> (OutputDelegate, then, Item:"Fort Item Entry Structure (by ref)") {is=title K2Node_Event}
   > OnItemEquipped()
   n20 = Event ShowTopButtonBar!() -> (OutputDelegate, then, bShow:bool) {is=title K2Node_Event}
   > SetVisibility(
    self=Get Box_TopButtonBar() -> (Box_TopButtonBar:Vertical Box Object Reference)
    InVisibility=Select(Option 0:ESlateVisibility Enum=Collapsed, Option 1:ESlateVisibility Enum=SelfHitTestInvisible, Index=@n20.bShow) -> (ReturnValue:ESlateVisibility Enum)
   )
   n21 = Event ShowLegacyFilterCounter!() -> (OutputDelegate, then, bShow:bool) {is=title K2Node_Event}
   > SetVisibility(
    self=Get Box_FilterCount() -> (Box_FilterCount:Vertical Box Object Reference)
    InVisibility=Select(Option 0=Collapsed, Option 1=SelfHitTestInvisible, Index=@n21.bShow)
   )
  ]}
 FormatFilterSelection {nodes [
   FormatFilterSelection() {is=title K2Node_FunctionEntry}
   > n1 = Branch(Condition=IsSubTabOpened())
   n2 = GetSubTabSelectedItemDefinition!(self:Fort Emporium Item List Tab Panel Object Reference) -> (ReturnValue:Item Definition Base Object Reference)
   > n3 = Is Valid(InputObject=@n2)
   Return Node!(execute=@n3.Is Not Valid, FilterText:Text)
   > Return Node(FilterText=GetDisplayName!(self:Item Definition Base Object Reference=@n2) -> (ReturnValue:Text))
   n4 = Branch(execute=@n1.else, Condition=Is Empty!(TargetArray:Array of Names=Get LastFilters() -> (LastFilters:Array of Names)) -> (ReturnValue:bool))
   > Return Node(FilterText="LOCTEXT(\"All\")")
   n5 = Get LastFilters()
   n6 = For Loop(execute=@n4.else, LastIndex=int - int(A=Length(TargetArray:Array of Names=@n5)))
   Return Node(execute=@n6.Completed, FilterText=Get FilterList() -> (FilterList:Text))
   > n7 = Branch(Condition="Equal (Integer)"(A=@n6))
   n8 = GetCategoryDisplayName!(CategoryID:"Name (by ref)"="Get (a ref)"(Array:Array of Names=@n5, Dimension 1=@n6) -> (Output:"Name (by ref)")) -> (ReturnValue:"Text (by ref)")
   > Set FilterList(FilterList:Text=@n8) -> (Output_Get:Text)
   Set FilterList(
    execute=@n7.else
    FilterList=Format Text!(Format="LOCTEXT(\"{FilterList}, {FilterText}\")", FilterList:Text=Get FilterList(), FilterText:"Text (by ref)"=@n8) -> (Result)
   )
  ]}
 GetEntryWidgetFromCurrentItem {nodes [
   GetEntryWidgetFromCurrentItem() {is=title K2Node_FunctionEntry}
   > n1 = For Each Loop(Array:Array of User Widget Object References=GetDisplayedEntryWidgets(self=Get CreativeTileView_ItemOptions())) -> (Array Element:User Widget Object Reference)
   Return Node!(execute=@n1.Completed, Entry:WBP Creative Emporium Item Entry Object Reference)
   > n2 = Cast To WBP_Creative_Emporium_ItemEntry(Object=@n1)
   > Branch(Condition=CompareToItemEntry(self=@n2, InItemEntry=Get CurrentItemEntry()))
   > Return Node(Entry=@n2)
  ]}
 GetSelectedItem {nodes [
   GetSelectedItem() {is=title K2Node_FunctionEntry}
   > n1 = Cast To AthenaCreativeItemEntryObjectWrapper(Object=Get Selected Item!(self:List View Object Reference=Get CreativeTileView_ItemOptions()) -> (ReturnValue:Object Reference) {is=title K2Node_CallFunction})
   n2 = BreakFortItemEntry(InItemEntry=Get ItemEntry(self=@n1)) -> (Count=0, Level=0, LoadedAmmo=0, PhantomReserveAmmo=0)
   > Return Node!(execute, ReturnValue:Fort Item Object Reference=CreateTemporaryItemInstance(ItemDefinition=@n2, Count=@n2.Count, Level=-1))
  ]}
 IsShowingModal {nodes [
   IsShowingModal() {is=title K2Node_FunctionEntry}
   > Return Node!(execute, Showing:bool=false)
  ]}
 NavigateToTopButtonBar {nodes [
   n1 = NavigateToTopButtonBar!() -> (then, Navigation:EUINavigation Enum) {is=title K2Node_FunctionEntry}
   > Branch(Condition="Equal (Enum)"(A:EUINavigation Enum=@n1, B:EUINavigation Enum=Up))
   n2 = Get Box_TopButtonBar()
   > Branch(Condition="Not Equal (Enum)"!(
     A:ESlateVisibility Enum=GetVisibility!(self:Widget Object Reference=@n2) -> (ReturnValue:ESlateVisibility Enum=Visible)
     B:ESlateVisibility Enum=Collapsed
    ) -> (ReturnValue:bool))
   > Return Node!(execute, ReturnValue:Widget Object Reference=Get Button_Sort())
  ]}
 OnItemEquipped {nodes [
   OnItemEquipped() {is=title K2Node_FunctionEntry}
   > EnterPurchasedState(self=Get CurrentEntryWidget(), Purchased=false)
   > Shine Selected Item() {is=title K2Node_CallFunction}
  ]}
 RegenerateTileView {nodes [
   RegenerateTileView() {is=title K2Node_FunctionEntry}
   > RegenerateAllEntries(self:List View Base Object Reference=Get CreativeTileView_ItemOptions())
  ]}
 Show Playset Item in Detail Panel {nodes [Show Playset Item in Detail Panel!() -> (then, Fort Item:Fort Item Object Reference) {is=title K2Node_FunctionEntry}]}
 Update Item Count {nodes [
   n1 = Update Item Count!() -> (then, In Entry Widget:WBP Creative Emporium Item Entry Object Reference) {is=title K2Node_FunctionEntry}
   > UpdateCount!(execute, self:WBP Creative Browser Item List Entry Widget Object Reference=@n1, Count:int=-1) -> (then)
   > Return Node!(execute, Out Entry Widget:WBP Creative Emporium Item Entry Object Reference=@n1)
   n2 = BreakFortItemEntry(InItemEntry=Get ItemEntry(self:Athena Creative Item Tile Button Object Reference=@n1)) -> (Count=0, Level=0, LoadedAmmo=0, PhantomReserveAmmo=0)
   Select(Option 0:int=@n2.Count, Option 1:int=-1, Index="Equal (Integer)"(A=@n2.Count, B=1) {is=title K2Node_CallFunction}) -> (ReturnValue:int)
  ]}
 Update Label Offset {nodes [
   n1 = Update Label Offset!() -> (then, Entry:WBP Creative Emporium Item Entry Object Reference) {is=title K2Node_FunctionEntry}
   n2 = GetListItemObject!(UserObjectListEntry:User Object List Entry Interface=@n1) -> (ReturnValue:Object Reference)
   > Is Valid(InputObject=@n2)
   n3 = Get GridWidth() -> (GridWidth:int=0)
   n4 = "% (Integer)"(
    A=GetIndexForItem!(self:List View Object Reference=Get CreativeTileView_ItemOptions(), Item:Object Reference=@n2) -> (ReturnValue:int=0)
    B=@n3
   ) -> (ReturnValue=0)
   > n5 = Branch(Condition="Equal (Integer)"(A=@n4, B=0))
   > ShiftLabel!(execute, self:WBP Creative Browser Item List Entry Widget Object Reference=@n1, Alignment:EHorizontalAlignment Enum=HAlign_Right) -> (then)
   n6 = Branch(execute=@n5.else, Condition="Equal (Integer)"(A=@n4, B=int - int(A=@n3)))
   > ShiftLabel(self=@n1, Alignment=HAlign_Left)
   ShiftLabel(execute=@n6.else, self=@n1, Alignment=HAlign_Center)
  ]}
]
variables [
 CurrentEntryWidget {type=WBP_Creative_Emporium_ItemEntry_C, default=null}
 GridWidth {type=integer, default=6}
 Show Playset Item in Detail Panel {}
]
widgets {
 info {
  inheritedWidgetCount=2
  parentClass=/Script/CreativeEmporiumUI.FortEmporiumItemListTabPanel
  rootWidgetClass=&CommonBorder
  widgetCount=15
 }
 widgets [
  CommonBorder_0 {widgetClassPath=&CommonBorder, widget=CommonBorder_0, properties {padding=<ltrb 0 0 0 0>}}
  VerticalBox_0 {
   widgetClassPath=&VerticalBox
   parent=CommonBorder_0
   widget=VerticalBox_0
   slot {class=BorderSlot, properties {padding=<ltrb 0 0 0 0>}}
  }
  Box_FilterCount {widgetClassPath=&VerticalBox, bIsVariable=true, parent=VerticalBox_0, widget=Box_FilterCount, slot {class=VerticalBoxSlot}}
  Filter_Count {
   widgetClassPath=&CommonTextBlock
   bIsVariable=true
   parent=Box_FilterCount
   widget=Filter_Count
   properties {
    mobileFontSizeMultiplier=1.34
    scrollStyle=&TextScrollStyle-Base_C
    text="{Count}"
    textTransformPolicy=ToUpper
    colorAndOpacity {colorUseRule=UseColor_Specified, specifiedColor=<rgba 1 1 1 0.6>}
    font {
     bForceMonospaced=false
     bMaterialIsStencil=false
     fontMaterial=null
     fontObject=&HeadingNow
     letterSpacing=0
     monospacedWidth=1
     size=18
     skewAmount=0
     typefaceFontName=75 Medium
     outlineSettings {
      bApplyOutlineToDropShadows=false
      bMiteredCorners=false
      bSeparateFillAlpha=false
      outlineMaterial=null
      outlineSize=0
      outlineColor=<rgba 0 0 0 1>
     }
    }
   }
   slot {class=VerticalBoxSlot, properties {verticalAlignment=VAlign_Center, padding=<ltrb 10 0 0 0>}}
  }
  Spacer {widgetClassPath=&Spacer, parent=Box_FilterCount, widget=Spacer, properties {size=<1 6>}, slot {class=VerticalBoxSlot}}
  Box_TopButtonBar {
   widgetClassPath=&VerticalBox
   bIsVariable=true
   parent=VerticalBox_0
   widget=Box_TopButtonBar
   slot {class=VerticalBoxSlot, properties {padding=<ltrb 0 10 4 0>}}
  }
  HorizontalBox_139 {
   widgetClassPath=&HorizontalBox
   parent=Box_TopButtonBar
   widget=HorizontalBox_139
   slot {class=VerticalBoxSlot, properties {size {sizeRule=Fill, value=1}}}
  }
  Button_Sort {
   widgetClassPath=&WBP_UIKit_Button_Regular
   bInherited=true
   bIsVariable=true
   parent=HorizontalBox_139
   widget=Button_Sort
   properties {
    buttonSize=Small
    minHeight=56
    minWidth=142
    text=Sort
    triggeringInputAction {dataTable=/FortUILibrary/Input/CreativeInputActionDataTable, rowName=InventoryV2_Sort}
   }
   slot {class=HorizontalBoxSlot, properties {horizontalAlignment=HAlign_Left, verticalAlignment=VAlign_Center, padding=<ltrb 12 0 8 0>}}
  }
  Filter_Count_V2 {
   widgetClassPath=&CommonTextBlock
   bIsVariable=true
   parent=HorizontalBox_139
   widget=Filter_Count_V2
   properties {
    mobileFontSizeMultiplier=1.34
    scrollStyle=&TextScrollStyle-Base_C
    text="{Count}"
    textTransformPolicy=ToUpper
    colorAndOpacity {colorUseRule=UseColor_Specified, specifiedColor=<rgba 1 1 1 0.6>}
    font {
     bForceMonospaced=false
     bMaterialIsStencil=false
     fontMaterial=null
     fontObject=&HeadingNow
     letterSpacing=0
     monospacedWidth=1
     size=18
     skewAmount=0
     typefaceFontName=75 Medium
     outlineSettings {
      bApplyOutlineToDropShadows=false
      bMiteredCorners=false
      bSeparateFillAlpha=false
      outlineMaterial=null
      outlineSize=0
      outlineColor=<rgba 0 0 0 1>
     }
    }
   }
   slot {class=HorizontalBoxSlot, properties {horizontalAlignment=HAlign_Left, verticalAlignment=VAlign_Center, padding=<ltrb 10 0 0 0>}}
  }
  Spacer_225 {widgetClassPath=&Spacer, parent=Box_TopButtonBar, widget=Spacer_225, properties {size=<1 14>}, slot {class=VerticalBoxSlot}}
  MenuAnchor_0 {
   widgetClassPath=/Script/UMG.MenuAnchor
   bIsVariable=true
   parent=VerticalBox_0
   widget=MenuAnchor_0
   properties {clipping=ClipToBoundsWithoutIntersecting, placement=MenuPlacement_Center}
   slot {class=VerticalBoxSlot, properties {padding=<ltrb -24 0 -32 -32>, size {sizeRule=Fill, value=1}}}
  }
  SizeBox_1 {
   widgetClassPath=&SizeBox
   parent=MenuAnchor_0
   widget=SizeBox_1
   properties {bOverride_MinDesiredHeight=true, minDesiredHeight=250}
   slot {class=PanelSlot}
  }
  WidgetSwitcherNoItemsinChest {
   widgetClassPath=&WidgetSwitcher
   bIsVariable=true
   parent=SizeBox_1
   widget=WidgetSwitcherNoItemsinChest
   slot {class=SizeBoxSlot}
  }
  CreativeTileView_ItemOptions {
   widgetClassPath=/Script/FortniteUI.AthenaCreativeItemTileView
   bInherited=true
   bIsVariable=true
   parent=WidgetSwitcherNoItemsinChest
   widget=CreativeTileView_ItemOptions
   properties {
    bEnableScrollAnimation=false
    bReturnFocusToSelection=true
    clipping=Inherit
    entryWidgetClass=&/WBP_Creative_Emporium_ItemEntry.WBP_Creative_Emporium_ItemEntry_C
    horizontalEntrySpacing=12
    navigation=&self:WidgetTree.CreativeTileView_ItemOptions.WidgetNavigation_0
    numDesignerPreviewEntries=20
    verticalEntrySpacing=12
   }
   slot {class=WidgetSwitcherSlot, properties {padding=<ltrb 22 0 0 20>}}
  }
  QuantityTextBlock {
   widgetClassPath=&CommonTextBlock
   bIsVariable=true
   parent=WidgetSwitcherNoItemsinChest
   widget=QuantityTextBlock
   properties {
    autoWrapText=true
    justification=Center
    navigation=&self:WidgetTree.QuantityTextBlock.WidgetNavigation_0
    style=/FortUILibrary/Styles/Typography/Foundation/TextStyle-Base-L.TextStyle-Base-L_C
    text=Add items to the chest and then create a chest or llama.
    font {
     bForceMonospaced=false
     bMaterialIsStencil=false
     fontMaterial=null
     fontObject=&NotoSans
     letterSpacing=0
     monospacedWidth=1
     size=32
     skewAmount=0
     typefaceFontName=NotoSans-Regular
     outlineSettings {
      bApplyOutlineToDropShadows=false
      bMiteredCorners=false
      bSeparateFillAlpha=false
      outlineMaterial=null
      outlineSize=0
      outlineColor=<rgba 0 0 0 1>
     }
    }
    shadowOffset=<0 0>
   }
   slot {class=WidgetSwitcherSlot, properties {padding=<ltrb 0 40 0 0>}}
  }
 ]
}
