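# Filter entry row for the Creative Emporium browser, a checkbox toggle button labeled with its tag's category name, plus an item count when above zero.
# Parent-class events update the label and checkmark, and selection refreshes resync the toggle to bFilterActive.
# UpdateSelection shows an outline on hover or click, adding looping pulse arrows for gamepad input.
# PreConstruct adjusts padding on mobile, while several inherited events and helper functions are empty stubs.
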WidgetBlueprint /CreativeEmporium/WBP_CrBrowser_FilterEntryWidget
animations [PulseArrows {
  binding_count=2
  length_seconds=1
  object=&self:PulseArrows
  sequence {bindings [
    TopArrow_Image {
     id=#1
     possessed_class=Image
     tracks [{
       type=MovieScene2DTransformTrack
       sections [{
         type=MovieScene2DTransformSection
         end_frame=20
         start_frame=0
         channels [
          Translation.X {type=MovieSceneScriptingFloatChannel, const [0,1]}
          Translation.Y {type=MovieSceneScriptingFloatChannel, keys [[0,8],[20,0],[20.0003,8]]}
          Angle {type=MovieSceneScriptingFloatChannel}
          Scale.X {type=MovieSceneScriptingFloatChannel}
          Scale.Y {type=MovieSceneScriptingFloatChannel}
          Shear.X {type=MovieSceneScriptingFloatChannel}
          Shear.Y {type=MovieSceneScriptingFloatChannel}
         ]
        }]
      }]
    }
    BottomArrow_Image {
     id=#2
     possessed_class=Image
     tracks [{
       type=MovieScene2DTransformTrack
       sections [{
         type=MovieScene2DTransformSection
         end_frame=20
         start_frame=0
         channels [
          Translation.X {type=MovieSceneScriptingFloatChannel, const [0,1]}
          Translation.Y {type=MovieSceneScriptingFloatChannel, keys [[0,-8],[20,0]]}
          Angle {type=MovieSceneScriptingFloatChannel}
          Scale.X {type=MovieSceneScriptingFloatChannel}
          Scale.Y {type=MovieSceneScriptingFloatChannel}
          Shear.X {type=MovieSceneScriptingFloatChannel}
          Shear.Y {type=MovieSceneScriptingFloatChannel}
         ]
        }]
      }]
    }
   ]}
 }]
graphs [
 EventGraph {nodes [
   "On Clicked (ToggleButton)"!() -> (OutputDelegate, then, Button:Common Button Base Object Reference) {is=title K2Node_ComponentBoundEvent}
   n1 = Get CommonInputSubsystem()
   > UpdateSelection!(execute, self, ShowSelected:bool=true, ShowArrows:bool="Equal (Enum)"(A=GetCurrentInputType(self=@n1), B=Gamepad)) -> (then)
   n2 = Event UpdateCheckMarkState!() -> (OutputDelegate, then, bIsChecked:bool) {is=title K2Node_Event}
   > SetIsSelected!(
    execute
    self:Common Button Base Object Reference=Get ToggleButton() -> (ToggleButton:WBP UIKit Button With Toggle Object Reference)
    InSelected:bool=@n2.bIsChecked
    bGiveClickFeedback:bool=false
   ) -> (then)
   n3 = Event UpdateItemCount!() -> (OutputDelegate, then, ItemCount:int=0) {is=title K2Node_Event}
   > UpdateButtonDisplayText!(execute, self, DisplayText:Text=Get FilterName() -> (FilterName:Text), Count:int=@n3.ItemCount) -> (then)
   n4 = Event SetTabLabelInfo!() -> (OutputDelegate, then, TabLabelInfo:"Fort Tab Button Label Info Structure (by ref)") {is=title K2Node_Event}
   > UpdateButtonDisplayText(
    DisplayText=Break Fort Tab Button Label Info!(FortTabButtonLabelInfo:"Fort Tab Button Label Info Structure (by ref)"=@n4.TabLabelInfo) -> (DisplayName:Text, IconBrush:Slate Brush Structure, IconResource:Soft Object Reference) {is=title K2Node_BreakStruct}
    Count=0
   )
   Event OnFocusLost!() -> (OutputDelegate, then, InFocusEvent:Focus Event Structure) {is=title K2Node_Event}
   > UpdateSelection(ShowSelected=false, ShowArrows=false)
   Event Construct()
   n5 = Get Tag() -> (Tag:Emporium Browser Tag Object Reference)
   > n6 = Set FilterName(FilterName:Text=GetCategoryDisplayName!(CategoryID:"Name (by ref)"=Get TagID!(self:Emporium Browser Tag Object Reference=@n5) -> (TagID:Name=None)) -> (ReturnValue:"Text (by ref)")) -> (Output_Get:Text)
   > UpdateButtonDisplayText(DisplayText=@n6, Count=0)
   > UpdateSelection(ShowSelected=false, ShowArrows=false)
   Event On Hovered()
   n7 = Get CommonInputSubsystem()
   > UpdateSelection(ShowSelected=true, ShowArrows="Equal (Enum)"(A=GetCurrentInputType(self=@n7), B=Gamepad))
   Event PreConstruct()
   > SetMobileStyle() {is=title K2Node_CallFunction}
   Event On Unhovered()
   > UpdateSelection(ShowSelected=false, ShowArrows=false)
   > SetIsSelected(self=Get ToggleButton(), InSelected=Get bFilterActive() -> (bFilterActive:bool), bGiveClickFeedback=false)
   "On Input Method Changed (InputActionWidget)"!() -> (OutputDelegate, then, bUsingGamepad:bool) {is=title K2Node_ComponentBoundEvent}
   "On Selected Changed (ToggleButton)"!() -> (OutputDelegate, then, Button:Common Button Base Object Reference, Selected:bool) {is=title K2Node_ComponentBoundEvent}
   SetCustomPaddingOnCheck!() -> (OutputDelegate, then, NewPadding:Margin Structure) {is=title K2Node_CustomEvent}
   Event Destruct()
   Event OnTriggeredInputActionChanged!() -> (OutputDelegate, then, NewTriggeredAction:"Data Table Row Handle Structure (by ref)") {is=title K2Node_Event}
   Event OnActionProgress!() -> (OutputDelegate, then, HeldPercent:float) {is=title K2Node_Event}
   Event OnActionComplete() {is=title K2Node_Event}
  ]}
 InitializeButton {nodes [InitializeButton() {is=title K2Node_FunctionEntry}]}
 Set Icon {nodes [Set Icon!() -> (then, IconBrush:Slate Brush Structure) {is=title K2Node_FunctionEntry}]}
 Set Text {nodes [
   n1 = Set Text!() -> (then, ButtonText:Text) {is=title K2Node_FunctionEntry}
   > n2 = Set OverrideButtonText(OverrideButtonText:Text=@n1) -> (Output_Get:Text)
   > SetTextInternal!(execute, self, InButtonText:Text=@n2) -> (then)
  ]}
 SetContentAlignment {nodes [
   n1 = SetContentAlignment!() -> (then, ContentAlignment:EHorizontalAlignment Enum=HAlign_Fill) {is=title K2Node_FunctionEntry}
   > Set ContentAlignment(ContentAlignment:EHorizontalAlignment Enum=@n1) -> (Output_Get:EHorizontalAlignment Enum=HAlign_Fill)
   > UpdateContentAlignment()
  ]}
 SetControllerStyle {nodes [
   n1 = SetControllerStyle!() -> (then, ControllerInputStyle:Common Button Style Class Reference) {is=title K2Node_FunctionEntry}
   > Set ControllerInputStyle(ControllerInputStyle:Common Button Style Class Reference=@n1) -> (Output_Get:Common Button Style Class Reference)
   n2 = Get CommonInputSubsystem!() -> (ReturnValue:Common Input Object Reference) {is=title K2Node_GetBlueprintContextEx}
   > UpdateStyle!(execute, self, UsingGamepad:bool="Equal (Enum)"(A=GetCurrentInputType(self=@n2), B=Gamepad)) -> (then)
  ]}
 SetInitialMouseKeyboardStyle {nodes [
   SetInitialMouseKeyboardStyle() {is=title K2Node_FunctionEntry}
   > n1 = Branch(Condition=Get bMouseKeyboardStyleSet() -> (bMouseKeyboardStyleSet:bool))
   Set bMouseKeyboardStyleSet(execute=@n1.else, bMouseKeyboardStyleSet:bool=true) -> (Output_Get:bool)
   > Set MouseKeyboardStyle(MouseKeyboardStyle:Common Button Style Class Reference=Get Style() -> (Style:Common Button Style Class Reference)) -> (Output_Get:Common Button Style Class Reference)
  ]}
 SetMobileStyle {nodes [
   SetMobileStyle() {is=title K2Node_FunctionEntry}
   > Branch(Condition=IsMobileGame())
   > SetPadding!(
    execute
    self:User Widget Object Reference
    InPadding_Left:float=0
    InPadding_Top:float=24
    InPadding_Right:float=0
    InPadding_Bottom:float=0
   ) -> (then)
   > n1 = Cast To OverlaySlot!(
    execute
    Object=Get Slot!(self:Widget Object Reference=Get Image_Backplate_hover() -> (Image_Backplate_hover:Image Object Reference)) -> (Slot:Panel Slot Object Reference)
   ) -> (then, CastFailed, AsOverlay Slot:Overlay Slot Object Reference) {is=title K2Node_DynamicCast}
   > Set Padding!(execute, Padding:Margin Structure=Get MobilePadding() -> (MobilePadding:Margin Structure), self:Overlay Slot Object Reference=@n1) -> (then, Output_Get:Margin Structure)
   > n2 = Cast To OverlaySlot(Object=Get Slot(self=Get Arrows_Overlay() -> (Arrows_Overlay:Overlay Object Reference)))
   > Set!(
    execute
    Padding_Left:float=0
    Padding_Top:float=-48
    Padding_Right:float=0
    Padding_Bottom:float=-48
    self:Overlay Slot Object Reference=@n2
   ) -> (then, Output_Get:Margin Structure)
  ]}
 SetMouseKeyboardStyle {nodes [
   n1 = SetMouseKeyboardStyle!() -> (then, ControllerInputStyle:Common Button Style Class Reference) {is=title K2Node_FunctionEntry}
   > Set MouseKeyboardStyle(MouseKeyboardStyle=@n1)
   n2 = Get CommonInputSubsystem()
   > UpdateStyle(UsingGamepad="Equal (Enum)"(A=GetCurrentInputType(self=@n2), B=Gamepad))
  ]}
 SetTextInternal {nodes [
   SetTextInternal!() -> (then, InButtonText:Text) {is=title K2Node_FunctionEntry}
   > UpdateButtonText(self:Emporium Browser Filter Entry Object Reference)
  ]}
 ShowIcon {nodes [ShowIcon!() -> (then, bShouldShow:bool=true) {is=title K2Node_FunctionEntry}]}
 UpdateButtonDisplayText {nodes [
   n1 = UpdateButtonDisplayText!() -> (then, DisplayText:Text, Count:int) {is=title K2Node_FunctionEntry}
   > SetText!(
    execute
    self:WBP UIKit Button Generic Base Object Reference=Get ToggleButton()
    Text:Text=SelectText(
     A=Format Text!(Format="LOCTEXT(\"{main} ({count})\")", main:Text=@n1, count:int=@n1.Count) -> (Result)
     B=Get DisplayText() -> (DisplayText:Text)
     bPickA="integer > integer"(A=@n1.Count)
    )
   ) -> (then)
  ]}
 UpdateCheckmark {nodes [UpdateCheckmark!() -> (then, Checked:bool=~) {is=title K2Node_FunctionEntry}]}
 UpdateContentAlignment {nodes [UpdateContentAlignment() {is=title K2Node_FunctionEntry}]}
 UpdateSelection {nodes [
   UpdateSelection!() -> (then, ShowSelected:bool=~, ShowArrows:bool=~) {is=title K2Node_FunctionEntry}
   > n1 = Sequence!(execute) -> (then_0, then_1, then_2:Exec)
   > SetIsSelected(self=Get ToggleButton(), InSelected=Get bFilterActive(), bGiveClickFeedback=false)
   n2 = Branch(execute=@n1.then_1, Condition=Get ShowSelected() -> (ShowSelected:bool))
   n3 = Get Image_Backplate_hover()
   SetVisibility(execute=[@n2.then, @n4.then], self=@n3, InVisibility=HitTestInvisible)
   n4 = Branch(execute=@n2.else, Condition=IsHovered())
   SetVisibility(execute=@n4.else, self=@n3, InVisibility=Hidden)
   n5 = Branch(execute=@n1.then_2, Condition=Get ShowArrows() -> (ShowArrows:bool))
   n6 = Get Arrows_Overlay()
   > SetVisibility(self=@n6, InVisibility=HitTestInvisible)
   > PlayAnimation(InAnimation=Get PulseArrows() -> (PulseArrows:Widget Animation Object Reference), NumLoopsToPlay=0)
   SetVisibility(execute=@n5.else, self=@n6, InVisibility=Hidden)
  ]}
 UpdateStyle {nodes [
   n1 = UpdateStyle!() -> (then, UsingGamepad:bool=~) {is=title K2Node_FunctionEntry}
   > n2 = Branch(Condition=@n1)
   > SetStyle!(
    execute
    self:Common Button Base Object Reference
    InStyle:Common Button Style Class Reference=Get ControllerInputStyle() -> (ControllerInputStyle:Common Button Style Class Reference)
   ) -> (then)
   SetStyle(execute=@n2.else, InStyle=Get MouseKeyboardStyle() -> (MouseKeyboardStyle:Common Button Style Class Reference))
  ]}
 UpdateTextAndStyle {nodes [
   n1 = UpdateTextAndStyle!() -> (then, bUsingGamepad:bool=~) {is=title K2Node_FunctionEntry}
   > UpdateStyle(UsingGamepad=@n1) {is=title K2Node_CallFunction}
   > UpdateTextStyle()
   > UpdateButtonText()
  ]}
 UpdateTextStyle {nodes [
   SetStyle(
    self:Common Text Object Reference=Get TextBlock_FilterName() -> (TextBlock_FilterName:Common Text Object Reference)
    InStyle:Common Text Style Class Reference=GetCurrentTextStyleClass!(self:Common Button Base Object Reference) -> (ReturnValue:Common Text Style Class Reference)
   )
   UpdateTextStyle() {is=title K2Node_FunctionEntry}
  ]}
]
variables [
 bMouseKeyboardStyleSet {type=boolean}
 CheckBrush {type=SlateColor, default {colorUseRule=UseColor_Specified, specifiedColor=<rgba 0 0 0 1>}}
 ContentAlignment {type=EHorizontalAlignment, default=HAlign_Fill}
 ControllerInputStyle {type=Class, default=null}
 CustomCheckPadding {type=Margin, default=<ltrb 0 0 6 0>}
 FilterName {type=string, default=""}
 IconBrush {
  type=SlateBrush
  default {
   bIsDynamicallyLoaded=false
   drawAs=NoDrawType
   imageType=NoImage
   mirroring=NoMirror
   resourceName=None
   resourceObject=null
   tiling=NoTile
   imageSize=<0 0>
   margin=<ltrb 0 0 0 0>
   outlineSettings {
    bUseBrushTransparency=false
    roundingType=HalfHeightRadius
    width=0
    color {colorUseRule=UseColor_Specified, specifiedColor=<rgba 0 0 0 0>}
    cornerRadii=<0 0 0 1>
   }
   tintColor {colorUseRule=UseColor_Specified, specifiedColor=<rgba 1 1 1 1>}
   uVRegion {bIsValid=false, max=<0 0>, min=<0 0>}
  }
 }
 MobilePadding {type=Margin, default=<ltrb -16 -22 -16 -22>}
 MouseKeyboardStyle {type=Class, default=null}
 OverrideButtonText {type=string, default=""}
 PreviewCheck {type=boolean}
 RightClickAction {type=boolean}
 SelectOnly {type=boolean}
 UncheckBrush {type=SlateColor, default {colorUseRule=UseColor_Specified, specifiedColor=<rgba 1 1 1 0.8>}}
]
widgets {
 info {
  inheritedWidgetCount=2
  parentClass=/Script/CreativeEmporiumUI.EmporiumBrowserFilterEntry
  rootWidgetClass=&Overlay
  widgetCount=9
 }
 widgets [
  Overlay {widgetClassPath=&Overlay, widget=Overlay}
  Overlay_0 {
   widgetClassPath=&Overlay
   parent=Overlay
   widget=Overlay_0
   slot {class=OverlaySlot, properties {horizontalAlignment=HAlign_Fill, verticalAlignment=VAlign_Center}}
  }
  ToggleButton {
   widgetClassPath=/FortUILibrary/UIKit/Buttons/WBP_UIKit_ButtonWithToggle
   bIsVariable=true
   parent=Overlay_0
   widget=ToggleButton
   properties {
    bInteractableWhenSelected=false
    clickMethod=MouseDown
    text=""
    textJustification=Left
    textTransformPolicy=None
    toggleType=CheckBox
    visibility=HitTestInvisible
    font {
     bForceMonospaced=false
     bMaterialIsStencil=false
     fontMaterial=/FortUILibrary/UIKit/Buttons/MaterialInstances/GenericButtons/MI_UIKit_ButtonWithToggle_Text
     fontObject=&HeadingNow
     letterSpacing=48
     monospacedWidth=1
     size=18
     skewAmount=0
     typefaceFontName=75 Medium
     outlineSettings {
      bApplyOutlineToDropShadows=false
      bMiteredCorners=false
      bSeparateFillAlpha=false
      outlineMaterial=null
      outlineSize=0
      outlineColor=<rgba 0 0 0 1>
     }
    }
    textPadding=<ltrb 14 12 8 0>
   }
   slot {class=OverlaySlot, properties {padding=<ltrb 6 0 0 0>}}
  }
  Image_Backplate_hover {
   widgetClassPath=&Image
   bIsVariable=true
   parent=Overlay_0
   widget=Image_Backplate_hover
   properties {
    visibility=Hidden
    brush {
     bIsDynamicallyLoaded=false
     drawAs=Image
     imageType=NoImage
     mirroring=NoMirror
     resourceName=None
     resourceObject=/CreativeEmporium/Materials/MI_CrBrowser_TabOutline
     tiling=NoTile
     imageSize=<156 64>
     margin=<ltrb 0 0 0 0>
     outlineSettings {
      bUseBrushTransparency=false
      roundingType=HalfHeightRadius
      width=0
      color {colorUseRule=UseColor_Specified, specifiedColor=<rgba 0 0 0 0>}
      cornerRadii=<0 0 0 0>
     }
     tintColor {colorUseRule=UseColor_Specified, specifiedColor=<rgba 1 1 1 0.5>}
     uVRegion {bIsValid=false, max=<0 0>, min=<0 0>}
    }
   }
   slot {class=OverlaySlot, properties {horizontalAlignment=HAlign_Fill, verticalAlignment=VAlign_Fill, padding=<ltrb -10 -10 -10 -10>}}
  }
  Arrows_Overlay {
   widgetClassPath=&Overlay
   bIsVariable=true
   parent=Overlay
   widget=Arrows_Overlay
   properties {visibility=Collapsed}
   slot {class=OverlaySlot, properties {horizontalAlignment=HAlign_Fill, verticalAlignment=VAlign_Fill, padding=<ltrb 0 -24 0 -24>}}
  }
  TopArrow_Image {
   widgetClassPath=&Image
   bIsVariable=true
   parent=Arrows_Overlay
   widget=TopArrow_Image
   properties {
    brush {
     bIsDynamicallyLoaded=false
     drawAs=Image
     imageType=FullColor
     mirroring=NoMirror
     resourceName=None
     resourceObject=/CreativeEmporium/Materials/T_UI_CrBrowser_NavArrow
     tiling=NoTile
     imageSize=<32 32>
     margin=<ltrb 0 0 0 0>
     outlineSettings {
      bUseBrushTransparency=false
      roundingType=HalfHeightRadius
      width=0
      color {colorUseRule=UseColor_Specified, specifiedColor=<rgba 0 0 0 0>}
      cornerRadii=<0 0 0 0>
     }
     tintColor {colorUseRule=UseColor_Specified, specifiedColor=<rgba 1 1 1 1>}
     uVRegion {bIsValid=false, max=<0 0>, min=<0 0>}
    }
    colorAndOpacity=<rgba 1 1 1 0.75>
    renderTransform {angle=180, scale=<1 -1>, shear=<0 0>, translation=<0 0>}
   }
   slot {class=OverlaySlot, properties {horizontalAlignment=HAlign_Center}}
  }
  BottomArrow_Image {
   widgetClassPath=&Image
   bIsVariable=true
   parent=Arrows_Overlay
   widget=BottomArrow_Image
   properties {
    brush {
     bIsDynamicallyLoaded=false
     drawAs=Image
     imageType=FullColor
     mirroring=NoMirror
     resourceName=None
     resourceObject=&T_UI_CrBrowser_NavArrow
     tiling=NoTile
     imageSize=<32 32>
     margin=<ltrb 0 0 0 0>
     outlineSettings {
      bUseBrushTransparency=false
      roundingType=HalfHeightRadius
      width=0
      color {colorUseRule=UseColor_Specified, specifiedColor=<rgba 0 0 0 0>}
      cornerRadii=<0 0 0 0>
     }
     tintColor {colorUseRule=UseColor_Specified, specifiedColor=<rgba 1 1 1 1>}
     uVRegion {bIsValid=false, max=<0 0>, min=<0 0>}
    }
    colorAndOpacity=<rgba 1 1 1 0.75>
    renderTransform {angle=0, scale=<1 -1>, shear=<0 0>, translation=<0 0>}
   }
   slot {class=OverlaySlot, properties {horizontalAlignment=HAlign_Center, verticalAlignment=VAlign_Bottom}}
  }
  Textblock_Filtername {
   widgetClassPath=&CommonTextBlock
   bInherited=true
   parent=Overlay
   widget=Textblock_Filtername
   properties {
    style=&TextStyle-Base_C
    text=Text Block
    visibility=Collapsed
    font {
     bForceMonospaced=false
     bMaterialIsStencil=false
     fontMaterial=null
     fontObject=&NotoSans
     letterSpacing=0
     monospacedWidth=1
     size=24
     skewAmount=0
     typefaceFontName=NotoSans-Regular
     outlineSettings {
      bApplyOutlineToDropShadows=false
      bMiteredCorners=false
      bSeparateFillAlpha=false
      outlineMaterial=null
      outlineSize=0
      outlineColor=<rgba 0 0 0 1>
     }
    }
    shadowOffset=<0 0>
   }
   slot {class=OverlaySlot}
  }
  InputActionWidget {widgetClassPath=&CommonActionWidget, bInherited=true, widget=null}
 ]
}
