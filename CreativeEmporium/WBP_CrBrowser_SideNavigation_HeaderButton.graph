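# Creative browser side-navigation header button, derived from the UIKit generic button, with collapsed and expandable states.
# Collapsing fades and shrinks content into a pip, with an animated value meant to interpolate cached and collapsed button sizes.
# Expansion changes rotate a tick; selection animates the pip material.
# Hover or selection shows pulsing arrows only for gamepad input while the side panel is focused, tracked through a global gameplay event.
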WidgetBlueprint /CreativeEmporium/WBP_CrBrowser_SideNavigation_HeaderButton
animations [
 OnCollapsed {
  binding_count=4
  length_seconds=0.25
  object=&self:OnCollapsed
  sequence {
   bindings [
    Slot_Background {
     id=#1
     possessed_class=NamedSlot
     tracks [{
       type=MovieSceneByteTrack
       display_name=Visibility
       sections [{type=MovieSceneByteSection, end_frame=25, start_frame=0, channels [None {type=MovieSceneScriptingByteChannel, keys [[0,4],[10,1],[25,1]]}]}]
      }]
    }
    Main {
     id=#2
     possessed_class=GridPanel
     tracks [
      {
       type=MovieSceneFloatTrack
       display_name=Render Opacity
       sections [{
         type=MovieSceneFloatSection
         end_frame=25
         start_frame=0
         channels [None {type=MovieSceneScriptingFloatChannel, thinned_from=4, keys [[0,1],[5,0],[25,0]]}]
        }]
      }
      {
       type=MovieScene2DTransformTrack
       sections [{
         type=MovieScene2DTransformSection
         end_frame=25
         start_frame=0
         channels [
          Scale.X {type=MovieSceneScriptingFloatChannel, keys [[0,1],[25,0]]}
          Scale.Y {type=MovieSceneScriptingFloatChannel, keys [[0,1],[25,0]]}
         ]
        }]
      }
     ]
    }
    Image_Pip {
     id=#3
     possessed_class=Image
     tracks [{
       type=MovieSceneColorTrack
       display_name=Color and Opacity
       sections [{
         type=MovieSceneColorSection
         end_frame=25
         start_frame=0
         channels [
          Color.R {type=MovieSceneScriptingFloatChannel, const [1,1]}
          Color.G {type=MovieSceneScriptingFloatChannel, const [1,1]}
          Color.B {type=MovieSceneScriptingFloatChannel, const [1,1]}
          Color.A {type=MovieSceneScriptingFloatChannel, keys [[0,0],[25,1]]}
         ]
        }]
      }]
    }
    WBP_UIKit_SideNavigation_Button_C_23 {
     id=#4
     possessed_class=WBP_UIKit_SideNavigation_Button_C
     tracks [
      {
       type=MovieSceneDoubleTrack
       display_name=Collapsed Anim Value
       sections [{type=MovieSceneDoubleSection, end_frame=25, start_frame=0, channels [None {type=MovieSceneScriptingDoubleChannel, keys [[0,0],[25,1]]}]}]
      }
      {
       type=MovieSceneByteTrack
       display_name=Visibility
       sections [{type=MovieSceneByteSection, end_frame=25, start_frame=0, channels [None {type=MovieSceneScriptingByteChannel, keys [[0,4],[10,3],[25,3]]}]}]
      }
     ]
    }
   ]
   display_rate {numerator=100}
  }
 }
 OnSelected {
  binding_count=1
  length_seconds=0.15
  object=&self:OnSelected
  sequence {bindings [Image_Pip {
     id=#5
     possessed_class=Image
     tracks [{
       type=MovieSceneWidgetMaterialTrack
       sections [{type=MovieSceneParameterSection, end_frame=3, start_frame=0, channels [Selected {type=MovieSceneScriptingFloatChannel, keys [[0,0],[3,1]]}]}]
      }]
    }]}
 }
 PulseArrows {
  binding_count=2
  length_seconds=1
  object=&self:PulseArrows
  sequence {bindings [
    TopArrow_Image {
     id=#6
     possessed_class=Image
     tracks [{
       type=MovieScene2DTransformTrack
       sections [{
         type=MovieScene2DTransformSection
         end_frame=20
         start_frame=0
         channels [
          Translation.X {type=MovieSceneScriptingFloatChannel, const [0,1]}
          Translation.Y {type=MovieSceneScriptingFloatChannel, keys [[0,8],[20,0],[20.0003,8]]}
          Angle {type=MovieSceneScriptingFloatChannel}
          Scale.X {type=MovieSceneScriptingFloatChannel}
          Scale.Y {type=MovieSceneScriptingFloatChannel}
          Shear.X {type=MovieSceneScriptingFloatChannel}
          Shear.Y {type=MovieSceneScriptingFloatChannel}
         ]
        }]
      }]
    }
    BottomArrow_Image {
     id=#7
     possessed_class=Image
     tracks [{
       type=MovieScene2DTransformTrack
       sections [{
         type=MovieScene2DTransformSection
         end_frame=20
         start_frame=0
         channels [
          Translation.X {type=MovieSceneScriptingFloatChannel, const [0,1]}
          Translation.Y {type=MovieSceneScriptingFloatChannel, keys [[0,-8],[20,0]]}
          Angle {type=MovieSceneScriptingFloatChannel}
          Scale.X {type=MovieSceneScriptingFloatChannel}
          Scale.Y {type=MovieSceneScriptingFloatChannel}
          Shear.X {type=MovieSceneScriptingFloatChannel}
          Shear.Y {type=MovieSceneScriptingFloatChannel}
         ]
        }]
      }]
    }
   ]}
 }
 Expand {
  binding_count=1
  length_seconds=0.25
  object=&self:Expand
  sequence {bindings [ExpandDirectionTick {
     id=#8
     possessed_class=Image
     tracks [{
       type=MovieScene2DTransformTrack
       sections [{
         type=MovieScene2DTransformSection
         channels [
          Angle {type=MovieSceneScriptingFloatChannel, keys [[-1,-90],[0,0],[4,90]]}
          Scale.X {type=MovieSceneScriptingFloatChannel, keys [[0,1],[3,1.5],[4,1]]}
          Scale.Y {type=MovieSceneScriptingFloatChannel, keys [[0,1],[3,1.5],[4,1]]}
          Shear.X {type=MovieSceneScriptingFloatChannel}
          Shear.Y {type=MovieSceneScriptingFloatChannel}
         ]
        }]
      }]
    }]}
 }
]
graphs [
 EventGraph {nodes [
   n1 = On Panel Focused!() -> (OutputDelegate, then, Focused:bool=~) {is=title K2Node_CustomEvent}
   > Set Focused(Focused:bool=@n1.Focused) -> (Output_Get:bool)
   > ShowArrows!(execute, self, ShowArrows:bool=Get ArrowsShown() -> (ArrowsShown:bool)) -> (then)
   n2 = OnButtonCollapsed!() -> (OutputDelegate, then, SkipAnimation:bool=~) {is=title K2Node_CustomEvent}
   > n3 = Branch(Condition=@n2.SkipAnimation)
   n4 = Get OnCollapsed() -> (OnCollapsed:Widget Animation Object Reference)
   > PlayAnimation(InAnimation=@n4, StartAtTime=GetEndTime(self=@n4))
   PlayAnimationForward(execute=@n3.else, InAnimation=Get OnCollapsed())
   n5 = Event PreConstruct()
   > Parent: PreConstruct(IsDesignTime=@n5.IsDesignTime)
   > AddChild(
    self=Get VisualAttachment() -> (VisualAttachment:Common Visual Attachment Object Reference)
    Content=Get Main() -> (Main:Grid Panel Object Reference)
   )
   > SetButtonPadding!(
    execute
    self:WBP UIKit Button Generic Base Object Reference
    Padding:Margin Structure=Get ButtonPadding() -> (ButtonPadding:Margin Structure)
   ) -> (then)
   > ResetAnimations()
   > InitializeText()
   > RefreshButtonSize()
   > n6 = StartListeningToEvent!(
    execute:Exec
    Target:Gameplay Event Router Component Object Reference=GetGlobalEventRouter()
    EventType:Script Struct Object Reference=&/GameplayEvent_SidePanelFocused
    Context:Object Reference
    EventBubblingRule:EEventBubblingRule Enum
    None:%=GetOwningPlayer()
   ) -> (
    then:Exec
    AsyncTaskProxy:Async Action Start Listening To Event Object Reference
    OnEventReceived:Exec
    Payload_Panel_1_#9:bool=~
    Payload_FocusedState_3_#10:bool=~
   )
   Branch(execute=@n6.OnEventReceived, Condition=NOT Boolean(A=@n6.Payload_Panel_1_#9))
   > On Panel Focused!(execute, self, Focused:bool=@n6.Payload_FocusedState_3_#10) -> (then) {is=title K2Node_CallFunction}
   Event On Hovered()
   > n7 = Parent: On Hovered() {is=title K2Node_CallParentFunction}
   ShowArrows(execute=[@n17.then, @n7.then], ShowArrows=true)
   n8 = Event SetIsCollapsed!() -> (OutputDelegate, then, IsCollapsed:bool, SkipAnimation:bool) {is=title K2Node_Event}
   > n9 = Branch(Condition=@n8.IsCollapsed)
   > OnButtonCollapsed!(execute, self, SkipAnimation:bool=@n8.SkipAnimation) -> (then) {is=title K2Node_CallFunction}
   OnButtonExpanded!(execute=@n9.else, self, SkipAnimation:bool=@n8.SkipAnimation) -> (then)
   K2Node_Event_2 = OnExpansionChanged!() -> (OutputDelegate, then, bIsExpanded:bool=~) {is=title K2Node_CustomEvent}
   > n10 = Branch(Condition=@K2Node_Event_2.bIsExpanded)
   n11 = Get ExpansionAnimationSpeed() -> (ExpansionAnimationSpeed:double)
   n12 = Get Expand() -> (Expand:Widget Animation Object Reference)
   > PlayAnimationForward(InAnimation=@n12, PlaybackSpeed=@n11)
   PlayAnimationReverse(execute=@n10.else, InAnimation=@n12, PlaybackSpeed=@n11)
   K2Node_Event_3 = OnButtonExpanded!() -> (OutputDelegate, then, SkipAnimation:bool=~) {is=title K2Node_CustomEvent}
   > n13 = Branch(Condition=@K2Node_Event_3.SkipAnimation)
   n14 = Get OnCollapsed()
   > PlayAnimation(InAnimation=@n14, StartAtTime=GetEndTime(self=@n14), PlayMode=Reverse)
   PlayAnimationReverse(execute=@n13.else, InAnimation=Get OnCollapsed())
   Event On Widget Styled() {is=title K2Node_Event}
   > Parent: On Widget Styled() {is=title K2Node_CallParentFunction}
   > UpdateText(self:WBP UIKit Button Generic Base Object Reference)
   Event On Unhovered()
   > n15 = Parent: On Unhovered() {is=title K2Node_CallParentFunction}
   ShowArrows(execute=[@n18.then, @n15.then], ShowArrows=false)
   Event On Selected()
   > Parent: On Selected() {is=title K2Node_CallParentFunction}
   n16 = Get OnSelected() -> (OnSelected:Widget Animation Object Reference)
   > n17 = PlayAnimationForward(InAnimation=@n16)
   Event On Deselected()
   > Parent: On Deselected() {is=title K2Node_CallParentFunction}
   > n18 = PlayAnimationReverse(InAnimation=@n16)
   PlaySound2D(Sound=/Game/Sounds/Fort_UI_Sounds/Gameplay/UI_QuickMenu_Open_Cue)
   PlaySound2D(Sound=/Game/Sounds/Fort_UI_Sounds/Gameplay/UI_QuickMenu_Close_Cue)
  ]}
 GetRemainingAnimTime {nodes [
   GetRemainingAnimTime() {is=title K2Node_FunctionEntry}
   > Return Node!(execute, FunctionIsImplemented:bool=false, RemainingTime:double=0, TotalTime:double=0)
  ]}
 InitializeText {nodes [
   InitializeText() {is=title K2Node_FunctionEntry}
   n1 = Get TextBlock!(self:WBP UIKit Block Text Object Reference=Get Block_Text() -> (Block_Text:WBP UIKit Block Text Object Reference)) -> (TextBlock:Common Text Object Reference)
   > SetWrapTextWidth!(execute, self:Common Text Object Reference=@n1, InWrapTextAt:int=Get WrapTextAt() -> (WrapTextAt:int=0)) -> (then)
   > Set Justification!(execute, Justification:ETextJustify Enum=Left, self:Text Layout Widget Object Reference=@n1) -> (then, Output_Get:ETextJustify Enum=Left)
   > Set TextTransformPolicy!(execute, TextTransformPolicy:ETextTransformPolicy Enum=None, self:Text Object Reference=@n1) -> (then, Output_Get:ETextTransformPolicy Enum=None)
  ]}
 RefreshButtonSize {nodes [
   RefreshButtonSize() {is=title K2Node_FunctionEntry}
   n1 = Get Main()
   > ForceLayoutPrepass(self:Widget Object Reference=@n1)
   n2 = IsMobileGame()
   n3 = SelectInt(A=Get MobileMinWidth() -> (MobileMinWidth:int=0), B=Get MinWidth() -> (MinWidth:int=0), bPickA=@n2)
   n4 = GetDesiredSize!(self:Widget Object Reference=@n1) -> (ReturnValue_X:double, ReturnValue_Y:double)
   n5 = Select!(
    Option 0:Margin Structure=Get ButtonPadding()
    Option 1:Margin Structure=Get MobileButtonPadding() -> (MobileButtonPadding:Margin Structure)
    Index=@n2
   ) -> (ReturnValue_Left:float=~, ReturnValue_Top:float=~, ReturnValue_Right:float=~, ReturnValue_Bottom:float=~)
   > n6 = Set!(
    execute
    CachedContentSize_X:double="Max (Float)"(A=float + float!(A=@n4, B=@n5, C:float=@n5.ReturnValue_Right) -> (ReturnValue), B="To Float (Integer)"(InInt=@n3))
    CachedContentSize_Y:double=float + float(A=@n4.ReturnValue_Y, B=@n5.ReturnValue_Top, C=@n5.ReturnValue_Bottom)
   ) -> (then, Output_Get:Vector 2D Structure)
   n7 = BreakVector2D(InVec=@n6)
   > SetMinDimensions!(
    execute
    self:Common Button Base Object Reference
    InMinWidth:int=Round(A=@n7) {is=title K2Node_CallFunction}
    InMinHeight:int=Round(A=@n7.Y)
   ) -> (then)
  ]}
 ResetAnimations {nodes [
   ResetAnimations() {is=title K2Node_FunctionEntry}
   n1 = Get OnSelected()
   > PlayAnimation(InAnimation=@n1, StartAtTime=GetEndTime(self=@n1), PlayMode=Reverse)
   n2 = Get OnCollapsed()
   > PlayAnimation(InAnimation=@n2, StartAtTime=GetEndTime(self=@n2), PlayMode=Reverse)
  ]}
 SetButtonPadding {nodes [
   n1 = SetButtonPadding!() -> (then, Padding:Margin Structure) {is=title K2Node_FunctionEntry}
   > n2 = Set ButtonPadding(ButtonPadding:Margin Structure=@n1) -> (Output_Get:Margin Structure)
   > SetGridSlotPadding!(execute, Widget:Widget Object Reference=Get VisualAttachment(), Padding:Margin Structure=@n2) -> (then)
  ]}
 SetCollapsedAnimValue {nodes [
   n1 = SetCollapsedAnimValue!() -> (then, Value:double=~) {is=title K2Node_FunctionEntry}
   n2 = Get!() -> (CollapsedSize_X:double, CollapsedSize_Y:double)
   n3 = Get!() -> (CachedContentSize_X:double, CachedContentSize_Y:double)
   > SetMinDimensions(InMinWidth=Round(A=Lerp(A=@n3, B=@n2, Alpha=@n1)), InMinHeight=Round(A=Lerp(A=@n3.CachedContentSize_Y, B=@n2.CollapsedSize_Y, Alpha=@n1)))
   > n4 = Get!(execute:Exec) -> (then:Exec, else:Exec, Block_Background:WBP UIKit Block Background Base Object Reference)
   > Set RenderOpacity!(execute, RenderOpacity:float=float - float(A=1, B=Get Value() -> (Value:double)), self:Widget Object Reference=@n4) -> (then, Output_Get:float)
  ]}
 SetText {nodes [
   n1 = SetText!() -> (then, Text:Text) {is=title K2Node_FunctionEntry}
   > Parent: SetText!(execute, Text:Text=@n1) -> (then) {is=title K2Node_CallParentFunction}
   > RefreshButtonSize() {is=title K2Node_CallFunction}
  ]}
 ShowArrows {nodes [
   n1 = ShowArrows!() -> (then, ShowArrows:bool=~) {is=title K2Node_FunctionEntry}
   > n2 = Set ArrowsShown(ArrowsShown:bool=@n1) -> (Output_Get:bool)
   n3 = GetCurrentInputType(self=Get CommonInputSubsystem())
   > n4 = Branch(Condition=AND Boolean!(A=@n2, B="Equal (Enum)"(A=@n3, B=Gamepad), C:bool=Get Focused() -> (Focused:bool)) -> (ReturnValue))
   > PlayAnimation(InAnimation=Get PulseArrows() -> (PulseArrows:Widget Animation Object Reference), NumLoopsToPlay=0)
   > SetVisibility(self=Get Arrows_Overlay() -> (Arrows_Overlay:Overlay Object Reference), InVisibility=SelfHitTestInvisible)
   SetVisibility(execute=@n4.else, self=Get Arrows_Overlay(), InVisibility=Collapsed)
  ]}
]
variables [
 ArrowsShown {type=boolean}
 CachedContentSize {type=Vector2D, default=<0 0>}
 CollapsedAnimValue {type=number}
 CollapsedSize {type=Vector2D, default=<28 28>}
 ExpansionAnimationSpeed {type=number, default=1.4}
 Focused {type=boolean}
 WrapTextAt {type=integer, default=232}
]
widgets {
 info {
  inheritedWidgetCount=0
  namedSlotCount=1
  parentClass=/FortUILibrary/Systems/ModularButton/WBP_UIKit_ButtonGeneric_Base.WBP_UIKit_ButtonGeneric_Base_C
  rootWidgetClass=null
  widgetCount=7
 }
 widgets [
  Overlay_40 {widgetClassPath=&Overlay, widget=Overlay_40}
  Image_Pip {
   widgetClassPath=&Image
   bIsVariable=true
   parent=Overlay_40
   widget=Image_Pip
   properties {brush {
     bIsDynamicallyLoaded=false
     drawAs=Image
     imageType=NoImage
     mirroring=NoMirror
     resourceName=None
     resourceObject=/FortUILibrary/UIKit/Buttons/Materials/M_UIKit_NavigationPip
     tiling=NoTile
     imageSize=<18 18>
     margin=<ltrb 0 0 0 0>
     outlineSettings {
      bUseBrushTransparency=false
      roundingType=FixedRadius
      width=0
      color {colorUseRule=UseColor_Specified, specifiedColor=<rgba 0 0 0 0>}
      cornerRadii=<4 4 4 4>
     }
     tintColor {colorUseRule=UseColor_Specified, specifiedColor=<rgba 1 1 1 1>}
     uVRegion {bIsValid=false, max=<0 0>, min=<0 0>}
    }}
   slot {class=OverlaySlot, properties {horizontalAlignment=HAlign_Center, verticalAlignment=VAlign_Center}}
  }
  Arrows_Overlay {
   widgetClassPath=&Overlay
   bIsVariable=true
   parent=Overlay_40
   widget=Arrows_Overlay
   properties {clipping=ClipToBoundsWithoutIntersecting, visibility=Collapsed}
   slot {class=OverlaySlot, properties {horizontalAlignment=HAlign_Fill, verticalAlignment=VAlign_Fill, padding=<ltrb 0 -24 0 -24>}}
  }
  TopArrow_Image {
   widgetClassPath=&Image
   bIsVariable=true
   parent=Arrows_Overlay
   widget=TopArrow_Image
   properties {
    brush {
     bIsDynamicallyLoaded=false
     drawAs=Image
     imageType=FullColor
     mirroring=NoMirror
     resourceName=None
     resourceObject=/CreativeEmporium/Materials/T_UI_CrBrowser_NavArrow
     tiling=NoTile
     imageSize=<32 32>
     margin=<ltrb 0 0 0 0>
     outlineSettings {
      bUseBrushTransparency=false
      roundingType=HalfHeightRadius
      width=0
      color {colorUseRule=UseColor_Specified, specifiedColor=<rgba 0 0 0 0>}
      cornerRadii=<0 0 0 0>
     }
     tintColor {colorUseRule=UseColor_Specified, specifiedColor=<rgba 1 1 1 1>}
     uVRegion {bIsValid=false, max=<0 0>, min=<0 0>}
    }
    colorAndOpacity=<rgba 1 1 1 0.75>
    renderTransform {angle=180, scale=<1 -1>, shear=<0 0>, translation=<0 0>}
   }
   slot {class=OverlaySlot, properties {horizontalAlignment=HAlign_Center}}
  }
  BottomArrow_Image {
   widgetClassPath=&Image
   bIsVariable=true
   parent=Arrows_Overlay
   widget=BottomArrow_Image
   properties {
    brush {
     bIsDynamicallyLoaded=false
     drawAs=Image
     imageType=FullColor
     mirroring=NoMirror
     resourceName=None
     resourceObject=&T_UI_CrBrowser_NavArrow
     tiling=NoTile
     imageSize=<32 32>
     margin=<ltrb 0 0 0 0>
     outlineSettings {
      bUseBrushTransparency=false
      roundingType=HalfHeightRadius
      width=0
      color {colorUseRule=UseColor_Specified, specifiedColor=<rgba 0 0 0 0>}
      cornerRadii=<0 0 0 0>
     }
     tintColor {colorUseRule=UseColor_Specified, specifiedColor=<rgba 1 1 1 1>}
     uVRegion {bIsValid=false, max=<0 0>, min=<0 0>}
    }
    colorAndOpacity=<rgba 1 1 1 0.75>
    renderTransform {angle=0, scale=<1 -1>, shear=<0 0>, translation=<0 0>}
   }
   slot {class=OverlaySlot, properties {horizontalAlignment=HAlign_Center, verticalAlignment=VAlign_Bottom}}
  }
  ArrowSpacingSB {
   widgetClassPath=&SizeBox
   parent=Overlay_40
   widget=ArrowSpacingSB
   properties {bOverride_MinDesiredHeight=true, heightOverride=72, minDesiredHeight=16, widthOverride=72}
   slot {class=OverlaySlot, properties {horizontalAlignment=HAlign_Right, verticalAlignment=VAlign_Center, padding=<ltrb 10 0 10 0>}}
  }
  ExpandDirectionTick {
   widgetClassPath=/Script/FortniteUI.FortMobileImage
   bIsVariable=true
   parent=ArrowSpacingSB
   widget=ExpandDirectionTick
   properties {
    mobileSizeMultiplier=1.2
    brush {
     bIsDynamicallyLoaded=false
     drawAs=Image
     imageType=NoImage
     mirroring=NoMirror
     resourceName=None
     resourceObject=/Game/UI/SocialPanel/Art/Textures/T_UI_Icon_TreeArrow_64
     tiling=NoTile
     imageSize=<16 16>
     margin=<ltrb 0 0 0 0>
     outlineSettings {
      bUseBrushTransparency=false
      roundingType=HalfHeightRadius
      width=0
      color {colorUseRule=UseColor_Specified, specifiedColor=<rgba 0 0 0 0>}
      cornerRadii=<0 0 0 0>
     }
     tintColor {colorUseRule=UseColor_Specified, specifiedColor=<rgba 1 1 1 1>}
     uVRegion {bIsValid=false, max=<0 0>, min=<0 0>}
    }
    renderTransform {angle=-90, scale=<1 1>, shear=<0 0>, translation=<0 0>}
   }
   slot {class=SizeBoxSlot, properties {horizontalAlignment=HAlign_Left, verticalAlignment=VAlign_Center}}
  }
 ]
}
