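# Creative Emporium search bar widget, extending the UIKit input field with a clear button and a gamepad action icon.
# Typed text is sanitized and written back; a non-empty field swaps the search icon for backspace and shows the clear button outside gamepad input.
# The clear button only fires the OnClearRequested dispatcher.
# Gamepad input pads the content and reveals the action icon, whose hold progress drives a dynamic circular-progress material.
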WidgetBlueprint /CreativeEmporium/WBP_Creative_Emporium_SearchBar
graphs [
 EventGraph {nodes [
   "On Clicked (Button_Clear)"!() -> (OutputDelegate, then, Button:Common Button Base Object Reference) {is=title K2Node_ComponentBoundEvent}
   > Call On Clear Requested!(execute:Exec, self:WBP Creative Emporium Search Bar Object Reference) -> (then:Exec) {is=title K2Node_CallDelegate}
   n1 = SetSearchText!() -> (OutputDelegate, then, Text:Text) {is=title K2Node_CustomEvent}
   > SetText!(execute, self:WBP UIKit Input Field Base Object Reference, Text:Text=@n1.Text) -> (then)
   > Update Text Icon()
   Event Construct()
   > n2 = GetDynamicMaterial(self=Get ActionHoldProgressCircle() -> (ActionHoldProgressCircle:Image Object Reference))
   > Set MID_ProgressCircle(MID_ProgressCircle:Material Instance Dynamic Object Reference=@n2) -> (Output_Get:Material Instance Dynamic Object Reference)
   > Update Text Icon()
   > HandleInputModeChanged!(execute, self, InputType:ECommonInputType Enum=GetCurrentInputType(self=Get CommonInputSubsystem())) -> (then)
   n3 = Event On Input Method Changed!() -> (OutputDelegate, then, CurrentInputType:ECommonInputType Enum) {is=title K2Node_Event}
   > HandleInputModeChanged(InputType=@n3.CurrentInputType)
   n4 = Event OnTextChanged_Event!() -> (OutputDelegate, then, Text:"Text (by ref)") {is=title K2Node_Event}
   n5 = "To Text (String)"(InString=Immediate Sanitize Text!(InText:String="To String (Text)"(InText=@n4.Text)) -> (ReturnValue:String))
   > SetText(
    self:WBP UIKit Block Editable Text Object Reference=Get Block_EditableText() -> (Block_EditableText:WBP UIKit Block Editable Text Object Reference)
    Text=@n5
   )
   > Parent: OnTextChanged_Event!(execute, Text:"Text (by ref)"=@n5) -> (then) {is=title K2Node_CallParentFunction}
   > Update Text Icon()
  ]}
 HandleInputModeChanged {nodes [
   n1 = HandleInputModeChanged!() -> (then, InputType:ECommonInputType Enum) {is=title K2Node_FunctionEntry}
   n2 = "Equal (Enum)"(A=@n1, B=Gamepad)
   > SetGridPadding!(
    execute
    self:WBP UIKit Input Field Base Object Reference
    WIdget:Widget Object Reference=Get Content() -> (Content:Grid Panel Object Reference)
    Padding:Margin Structure=Make Margin(Left=Select(Option 0:float=0, Option 1:float=24, Index=@n2) -> (ReturnValue:float), Top=0, Right=0, Bottom=0)
   ) -> (then)
   > SetVisibility(
    self=Get Overlay_ActionIcon() -> (Overlay_ActionIcon:Overlay Object Reference)
    InVisibility=Select(Option 0:ESlateVisibility Enum=Collapsed, Option 1:ESlateVisibility Enum=SelfHitTestInvisible, Index=@n2) -> (ReturnValue:ESlateVisibility Enum)
   )
   > Update Text Icon()
  ]}
 OnClearRequested {nodes [OnClearRequested() {is=title K2Node_FunctionEntry}]}
 Set Action Hold Progress {nodes [
   n1 = Set Action Hold Progress!() -> (then, Delta:double=~) {is=title K2Node_FunctionEntry}
   > SetScalarParameterValue(self=Get MID_ProgressCircle() -> (MID_ProgressCircle:Material Instance Dynamic Object Reference), ParameterName=Percentage, Value=@n1)
  ]}
 Set Action Icon {nodes [
   n1 = Set Action Icon!() -> (then, Icon:Slate Brush Structure) {is=title K2Node_FunctionEntry}
   n2 = Get ActionIcon() -> (ActionIcon:Image Object Reference)
   n3 = Break Slate Brush!(SlateBrush:"Slate Brush Structure (by ref)"=Get Brush!(self:Image Object Reference=@n2) -> (Brush:Slate Brush Structure)) -> (
    TintColor:Slate Color Structure
    DrawAs:ESlateBrushDrawType Enum
    Tiling:ESlateBrushTileType Enum
    Mirroring:ESlateBrushMirrorType Enum
    ImageSize:Vector 2D Structure
    Margin:Margin Structure
    ResourceObject:Object Reference
    OutlineSettings:Slate Brush Outline Settings Structure
   ) {is=title K2Node_BreakStruct}
   > SetBrush(
    self=@n2
    InBrush=Make Slate Brush!(
     TintColor:Slate Color Structure=@n3
     DrawAs:ESlateBrushDrawType Enum=@n3.DrawAs
     Tiling:ESlateBrushTileType Enum=@n3.Tiling
     Mirroring:ESlateBrushMirrorType Enum=@n3.Mirroring
     ImageSize:Vector 2D Structure=@n3.ImageSize
     Margin:Margin Structure=@n3.Margin
     ResourceObject:Object Reference=Break Slate Brush(SlateBrush=@n1).ResourceObject
     OutlineSettings:Slate Brush Outline Settings Structure=@n3.OutlineSettings
    ) -> (SlateBrush:Slate Brush Structure) {is=title K2Node_MakeStruct}
   )
  ]}
 Update Text Icon {nodes [
   Update Text Icon() {is=title K2Node_FunctionEntry}
   n1 = TextIsEmpty(InText="GetText (Editable Text)"!(self:Editable Text Object Reference=GetEditableText!() -> (NewParam:UIKit Editable Text Object Reference)) -> (ReturnValue:Text))
   n2 = Select(
    Option 0:Texture 2D Object Reference=/FortUILibrary/Textures/Icons/T_UI_Icon_UIKit_Backspace
    Option 1:Texture 2D Object Reference=/FortUILibrary/Textures/Icons/T_UI_Icon_UIKit_Search
    Index=@n1
   ) -> (ReturnValue:Texture 2D Object Reference)
   > SetIcon!(execute, self:WBP UIKit Input Field Base Object Reference, Icon:Texture 2D Object Reference=@n2) -> (then)
   > SetIcon(
    self:WBP UIKit Button Generic Base Object Reference=Get Button_Clear() -> (Button_Clear:WBP UIKit Button Generic Object Reference)
    Icon:Object Reference=@n2
   )
   n3 = NOT Boolean(A=@n1)
   n4 = Get CommonInputSubsystem()
   n5 = NOT Boolean(A="Equal (Enum)"(A=GetCurrentInputType(self=@n4), B=Gamepad))
   > SetVisibility(
    self=Get Button_Clear()
    InVisibility=Select(Option 0:ESlateVisibility Enum=Hidden, Option 1:ESlateVisibility Enum=Visible, Index=AND Boolean(A=@n3, B=@n5)) -> (ReturnValue:ESlateVisibility Enum)
   )
  ]}
]
variables [
 MID_ProgressCircle {type=MaterialInstanceDynamic}
 OnClearRequested {}
]
widgets {
 info {
  inheritedWidgetCount=0
  namedSlotCount=1
  parentClass=/FortUILibrary/UIKit/InputField/WBP_UIKit_InputField.WBP_UIKit_InputField_C
  rootWidgetClass=null
  widgetCount=5
 }
 widgets [
  Overlay_Additional {widgetClassPath=&Overlay, bIsVariable=true, widget=Overlay_Additional}
  Overlay_ActionIcon {
   widgetClassPath=&Overlay
   bIsVariable=true
   parent=Overlay_Additional
   widget=Overlay_ActionIcon
   properties {visibility=Collapsed}
   slot {class=OverlaySlot, properties {horizontalAlignment=HAlign_Fill, verticalAlignment=VAlign_Fill}}
  }
  ActionIcon {
   widgetClassPath=&Image
   bIsVariable=true
   parent=Overlay_ActionIcon
   widget=ActionIcon
   properties {
    visibility=SelfHitTestInvisible
    brush {
     bIsDynamicallyLoaded=false
     drawAs=Image
     imageType=FullColor
     mirroring=NoMirror
     resourceName=None
     resourceObject=null
     tiling=NoTile
     imageSize=<32 32>
     margin=<ltrb 0 0 0 0>
     outlineSettings {
      bUseBrushTransparency=false
      roundingType=HalfHeightRadius
      width=0
      color {colorUseRule=UseColor_Specified, specifiedColor=<rgba 0 0 0 0>}
      cornerRadii=<0 0 0 0>
     }
     tintColor {colorUseRule=UseColor_Specified, specifiedColor=<rgba 1 1 1 1>}
     uVRegion {bIsValid=false, max=<0 0>, min=<0 0>}
    }
   }
   slot {class=OverlaySlot, properties {verticalAlignment=VAlign_Center, padding=<ltrb 8 0 0 0>}}
  }
  ActionHoldProgressCircle {
   widgetClassPath=&Image
   bIsVariable=true
   parent=Overlay_ActionIcon
   widget=ActionHoldProgressCircle
   properties {
    visibility=SelfHitTestInvisible
    brush {
     bIsDynamicallyLoaded=false
     drawAs=Image
     imageType=FullColor
     mirroring=NoMirror
     resourceName=None
     resourceObject=/Game/UI/Materials/HollowCircularProgressBar
     tiling=NoTile
     imageSize=<48 48>
     margin=<ltrb 0 0 0 0>
     outlineSettings {
      bUseBrushTransparency=false
      roundingType=HalfHeightRadius
      width=0
      color {colorUseRule=UseColor_Specified, specifiedColor=<rgba 0 0 0 0>}
      cornerRadii=<0 0 0 0>
     }
     tintColor {colorUseRule=UseColor_Specified, specifiedColor=<rgba 1 0 0 1>}
     uVRegion {bIsValid=false, max=<0 0>, min=<0 0>}
    }
   }
   slot {class=OverlaySlot, properties {verticalAlignment=VAlign_Center}}
  }
  Button_Clear {
   widgetClassPath=&WBP_UIKit_Button_Generic
   bIsVariable=true
   parent=Overlay_Additional
   widget=Button_Clear
   properties {
    bShouldUseFallbackDefaultInputAction=false
    icon=&T_UI_Icon_UIKit_Search
    iconMaterial=/CreativeEmporium/Materials/MI_UI_Creative_Emporium_SearchBar_Icon
    minHeight=32
    minWidth=32
    text=""
    visibility=Visible
    buttonPadding=<ltrb 0 0 0 0>
   }
   slot {class=OverlaySlot, properties {horizontalAlignment=HAlign_Right, verticalAlignment=VAlign_Center, padding=<ltrb 0 0 32 0>}}
  }
 ]
}
