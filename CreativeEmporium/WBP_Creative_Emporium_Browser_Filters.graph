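# Creative Emporium browser filters panel widget with a search bar, Sort & Filter button and category side navigation, built on EmporiumBrowserFiltersPanel.
# PopulateCategories builds header tabs and child buttons from the category hierarchy; native events update item counts and focus the active category.
# Search text, clear and category clicks route to parent-class handlers.
# An optional parent search bar overrides the built-in one, and the search action icon follows text state and input method.
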WidgetBlueprint /CreativeEmporium/WBP_Creative_Emporium_Browser_Filters
graphs [
 ClearSearchProgress {nodes [
   n1 = ClearSearchProgress!() -> (then, Delta:double=~) {is=title K2Node_FunctionEntry}
   > Set Action Hold Progress!(
    execute
    self:WBP Creative Emporium Search Bar Object Reference=Get SearchPanel() -> (SearchPanel:WBP Creative Emporium Search Bar Object Reference)
    Delta:double=@n1
   ) -> (then) {is=title K2Node_CallFunction}
   > n2 = Get!(execute:Exec) -> (then:Exec, else:Exec, ParentSearchPanel:WBP Creative Emporium Search Bar Object Reference)
   > Set Action Hold Progress(self=@n2, Delta=@n1)
  ]}
 EventGraph {nodes [
   n1 = "On Clicked (SearchPanel)"!() -> (OutputDelegate, then, Button:Common Button Base Object Reference) {is=title K2Node_ComponentBoundEvent}
   SetSearchFocus(execute=[@n1.then, @n13.then])
   n2 = OnSideNavSectionSelectedChange!() -> (OutputDelegate, then, Button:Common Button Base Object Reference, Selected:bool=~) {is=title K2Node_CustomEvent}
   > Branch(Condition=@n2.Selected)
   > Cast To WBP_CrBrowser_SideNavigation_HeaderButton!(execute, Object=@n2.Button) -> (then, CastFailed, AsWBP Cr Browser Side Navigation Header Button:WBP Cr Browser Side Navigation Header Button Object Reference) {is=title K2Node_DynamicCast}
   n3 = Get UIKit_SideNavigation() -> (UIKit_SideNavigation:WBP UIKit Side Navigation Object Reference)
   n4 = Find(
    TargetMap:Map of Names to WBP UIKit Tab Container No Layout Object References=Get TabContainers!(self:WBP UIKit Side Navigation Base Object Reference=@n3) -> (TabContainers:Map of Names to WBP UIKit Tab Container No Layout Object References)
    Key:"Name (by ref)"=GetSelectedTabId!(self:Common Tab List Widget Base Object Reference=@n3) -> (ReturnValue:Name=None)
   ) -> (Value:WBP UIKit Tab Container No Layout Object Reference)
   > Branch(Condition=@n4.ReturnValue)
   > ClearInnerTabsSelection(self:WBP UIKit Tab Container No Layout Object Reference=@n4)
   n5 = OnCategoryButtonClicked!() -> (OutputDelegate, then, TabID:Name, InnerTabID:Name, TabViewModel:MVVM Base Viewmodel Object Reference) {is=title K2Node_CustomEvent}
   > HandleCategoryButtonClicked!(execute, self:Emporium Browser Filters Panel Object Reference, CategoryID:Name=@n5.InnerTabID) -> (then)
   n6 = Override Search Bar!() -> (OutputDelegate, then, ParentSearchBar:WBP Creative Emporium Search Bar Object Reference) {is=title K2Node_CustomEvent}
   > n7 = Set ParentSearchPanel(ParentSearchPanel:WBP Creative Emporium Search Bar Object Reference=@n6.ParentSearchBar) -> (Output_Get:WBP Creative Emporium Search Bar Object Reference)
   > Bind Event to On Text Changed(self:WBP UIKit Input Field Base Object Reference=@n7, Delegate=Create Event()) {is=title K2Node_AddDelegate}
   > Bind Event to On Clear Requested(self:WBP Creative Emporium Search Bar Object Reference=@n7, Delegate=Create Event()) {is=title K2Node_AddDelegate}
   > SetVisibility(self=[Get HeaderSpacer() -> (HeaderSpacer:Overlay Object Reference), Get SearchPanel()], InVisibility=Collapsed)
   OnInputMethodChanged!() -> (OutputDelegate, then, bNewInputType:ECommonInputType Enum) {is=title K2Node_CustomEvent}
   > UpdateSearchActionIcon()
   n8 = OnSearchTextChanged!() -> (OutputDelegate, then, Text:Text) {is=title K2Node_CustomEvent}
   > HandleSearchTextChanged!(execute, self:Emporium Browser Filters Panel Object Reference, Text:"Text (by ref)"=@n8.Text) -> (then)
   > UpdateSearchActionIcon()
   OnClearRequested() {is=title K2Node_CustomEvent}
   > ResetSearch(self:Emporium Browser Filters Panel Object Reference)
   n9 = OnSectionExpanded!() -> (OutputDelegate, then, TabID:Name, IsExpanded:bool=~) {is=title K2Node_CustomEvent}
   > n10 = Cast To WBP_CrBrowser_SideNavigation_HeaderButton(Object=GetTabButtonBaseByID!(self:Common Tab List Widget Base Object Reference=Get UIKit_SideNavigation(), TabNameID:Name=@n9.TabID) -> (ReturnValue:Common Button Base Object Reference))
   > OnExpansionChanged!(execute, self:WBP Cr Browser Side Navigation Header Button Object Reference=@n10, bIsExpanded:bool=@n9.IsExpanded) -> (then)
   n11 = OnSideNavTabSelected!() -> (OutputDelegate, then, TabID:Name, InnerTabID:Name, TabViewModel:MVVM Base Viewmodel Object Reference) {is=title K2Node_CustomEvent}
   > HandleCategoryButtonFocused!(execute, self:Emporium Browser Filters Panel Object Reference, CategoryID:Name=@n11.InnerTabID) -> (then)
   n12 = Event OnSetSearchVisible!() -> (OutputDelegate, then, bVisible:bool) {is=title K2Node_Event}
   > SetSearchVisibility!(execute, self, Visibility:bool=@n12.bVisible) -> (then)
   n13 = Event OnSetSearchFocused() {is=title K2Node_Event}
   n14 = Event OnUpdateActiveCategory!() -> (OutputDelegate, then, CategoryID:"Name (by ref)"=None) {is=title K2Node_Event}
   > Branch(Condition=Get Category Hierarchy Entry!(CategoryID:"Name (by ref)"=@n14.CategoryID) -> (OutFoundCategory:Fort Emporium Filter Category Structure, ReturnValue:bool) {is=title K2Node_CallFunction}.ReturnValue)
   > n15 = Values!(execute, TargetMap:Map of Names to WBP UIKit Tab Container No Layout Object References=Get TabContainers(self=Get UIKit_SideNavigation())) -> (then, Values:Array of WBP UIKit Tab Container No Layout Object References)
   > n16 = For Each Loop(Array:Array of WBP UIKit Tab Container No Layout Object References=@n15) -> (Array Element:WBP UIKit Tab Container No Layout Object Reference)
   n17 = GetInnerTabGroupWidget!(self:WBP UIKit Tab Container No Layout Object Reference=@n16) -> (InnerTabGroupWidget:WBP UIKit Tab Group Compact Base Object Reference)
   > n18 = For Each Loop(Array:Array of Widget Object References=GetAllChildren(self=Get TabsContainer!(self:WBP UIKit Tab Group Compact Base Object Reference=@n17) -> (TabsContainer:Grid Panel Object Reference))) -> (Array Element:Widget Object Reference)
   > n19 = Cast To WBP_CrBrowser_SideNavigation_Button!(execute, Object=@n18) -> (then, CastFailed, AsWBP Cr Browser Side Navigation Button:WBP Cr Browser Side Navigation Button Object Reference) {is=title K2Node_DynamicCast}
   > Branch(Condition="Equal (Name)"(A=Get CategoryID!(self:WBP Cr Browser Side Navigation Button Object Reference=@n19) -> (CategoryID:Name=None), B=@n14.CategoryID))
   > n20 = Branch(Condition=HasAnyUserFocus!(self:Widget Object Reference=@n19) -> (ReturnValue:bool))
   SetFocus(execute=@n20.else, self=@n19)
   n21 = Event OnSetSearchText!() -> (OutputDelegate, then, NewSearchText:"Text (by ref)") {is=title K2Node_Event}
   > SetSearchText!(execute, self, Text:Text=@n21.NewSearchText) -> (then)
   Event OnInitialized()
   n22 = Get SearchPanel()
   > Bind Event to On Text Changed(self=@n22, Delegate=Create Event())
   > Bind Event to On Clear Requested(self=@n22, Delegate=Create Event())
   > UpdateSearchActionIcon()
   > Bind Event to On Input Method Changed(self=Get CommonInputSubsystem(), Delegate=Create Event())
   n23 = Get UIKit_SideNavigation()
   > Bind Event to On Side Navigation Inner Tab Clicked(self:WBP UIKit Side Navigation Base Object Reference=@n23, Delegate=Create Event()) {is=title K2Node_AddDelegate}
   > Bind Event to On Side Navigation Selection Changed(self:WBP UIKit Side Navigation Base Object Reference=@n23, Delegate=Create Event()) {is=title K2Node_AddDelegate}
   n24 = Event OnClearSearchProgress!() -> (OutputDelegate, then, Progress:float) {is=title K2Node_Event}
   > ClearSearchProgress!(execute, self, Delta:double=@n24.Progress) -> (then) {is=title K2Node_CallFunction}
   n25 = Event OnPopulateCategories!() -> (OutputDelegate, then, CategoryIDs:Array of Names) {is=title K2Node_Event}
   > PopulateCategories!(execute, self, CategoryIDs:Array of Names=@n25.CategoryIDs) -> (then)
   n26 = Event OnUpdateCategoryCountDisplay!() -> (OutputDelegate, then, CategoryID:"Name (by ref)"=None, NumItems:int=0, bHideIfEmpty:bool) {is=title K2Node_Event}
   n27 = Get Category Hierarchy Entry(CategoryID=@n26.CategoryID)
   > Branch(Condition=@n27.ReturnValue)
   n28 = Break Fort Emporium Filter Category!(FortEmporiumFilterCategory:"Fort Emporium Filter Category Structure (by ref)"=@n27) -> (ID:Name, ParentID:Name, bIsTab:bool=~, bShowCategoryModal:bool=~, SubcategoryIDList:Array of Names) {is=title K2Node_BreakStruct}
   > n29 = Cast To WBP_CrBrowser_SideNavigation_Button(Object=GetInnerTabButton!(self:WBP UIKit Side Navigation Base Object Reference=Get UIKit_SideNavigation(), TabID:Name=@n28.ParentID, InnerTabID:Name=@n28) -> (TabButton:Common Button Base Object Reference))
   > "SetText (Text)"(
    self=Get Text_NumEntries!(self:WBP Cr Browser Side Navigation Button Object Reference=@n29) -> (Text_NumEntries:Common Text Object Reference)
    InText=Select(
     Option 0:Text=GetEmptyText!() -> (ReturnValue:Text)
     Option 1:Text="To Text (Integer)"(Value=@n26.NumItems)
     Index="integer > integer"(A=@n26.NumItems)
    ) -> (ReturnValue:Text)
   )
  ]}
 FilterSelectionChanged {nodes [FilterSelectionChanged!() -> (then, FilterText:String, GameplayTag:Gameplay Tag Structure) {is=title K2Node_FunctionEntry}]}
 NavigateSideBarWrap {nodes [
   n1 = NavigateSideBarWrap!() -> (then, Navigation:EUINavigation Enum) {is=title K2Node_FunctionEntry}
   > n2 = Branch(Condition="Equal (Enum)"(A:EUINavigation Enum=@n1, B:EUINavigation Enum=Up))
   > n3 = Values(TargetMap=Get TabContainers(self=Get UIKit_SideNavigation()))
   > n4 = Branch(Condition=Is Not Empty!(TargetArray:Array of WBP UIKit Tab Container No Layout Object References=@n3) -> (ReturnValue:bool))
   n5 = "Get (a copy)"(
    Array:Array of WBP UIKit Tab Container No Layout Object References=@n3
    Dimension 1=int - int(A=Length(TargetArray:Array of WBP UIKit Tab Container No Layout Object References=@n3))
   ) -> (Output:WBP UIKit Tab Container No Layout Object Reference)
   n6 = Get ContainerTabID!(self:WBP UIKit Tab Container No Layout Object Reference=@n5) -> (ContainerTabID:Name=None)
   n7 = GetInnerTabsCount!(self:WBP UIKit Side Navigation Base Object Reference=Get UIKit_SideNavigation(), TabID:Name=@n6) -> (Count:int=0)
   > n8 = Branch(Condition="integer > integer"(A=@n7))
   n9 = int - int(A=@n7)
   n10 = Get UIKit_SideNavigation()
   > Return Node!(
    execute
    ReturnValue:Widget Object Reference=GetInnerTabButton(
     self=@n10
     TabID=@n6
     InnerTabID=GetInnerTabIdAtIndex!(self:WBP UIKit Side Navigation Base Object Reference=@n10, TabID:"Name (by ref)"=@n6, Index:int=@n9) -> (InnerTabID:Name=None)
    )
   )
   Return Node(
    execute=@n8.else
    ReturnValue=GetTabButton!(self:WBP UIKit Tab Container No Layout Object Reference=@n5) -> (TabButton:Common Button Base Object Reference)
   )
   Return Node(execute=@n4.else)
   n11 = Values(execute=@n2.else, TargetMap=Get TabContainers(self=Get UIKit_SideNavigation()))
   > n12 = Branch(Condition=Is Not Empty(TargetArray=@n11))
   > Return Node(ReturnValue=GetTabButton(self="Get (a copy)"(Array=@n11)))
   Return Node(execute=@n12.else)
  ]}
 PopulateCategories {nodes [
   n1 = PopulateCategories!() -> (then, CategoryIDs:Array of Names) {is=title K2Node_FunctionEntry}
   > ClearSideNavigationTabs(self:WBP UIKit Side Navigation Base Object Reference=Get UIKit_SideNavigation())
   > n2 = For Each Loop(Array:Array of Names=@n1) -> (Array Element:Name)
   n3 = For Each Loop(Exec=@n2.Completed, Array=@n1)
   n4 = Get Category Hierarchy Entry(CategoryID=@n3)
   > Branch(Condition=@n4.ReturnValue)
   n5 = Break Fort Emporium Filter Category(FortEmporiumFilterCategory=@n4) -> (bIsTab=~, bShowCategoryModal=~)
   > n6 = Branch(Condition="Equal (Name)"(A=@n5.ParentID))
   n7 = Get UIKit_SideNavigation()
   > n8 = Cast To WBP_CrBrowser_SideNavigation_HeaderButton(Object=GetTabButtonBaseByID(self=@n7, TabNameID=@n5))
   > SetText!(
    execute
    self:WBP UIKit Button Generic Base Object Reference=@n8
    Text:Text=GetCategoryDisplayName!(CategoryID:"Name (by ref)"=@n5) -> (ReturnValue:"Text (by ref)")
   ) -> (then)
   > Bind Event to On Selected Changed(self:Common Button Base Object Reference=@n8, Delegate=Create Event()) {is=title K2Node_AddDelegate}
   n9 = Find(TargetMap=Get TabContainers(self=@n7), Key=@n5)
   > Branch(Condition=@n9.ReturnValue)
   > Bind Event to On Container Expanded State Changed(self:WBP UIKit Tab Container No Layout Object Reference=@n9, Delegate=Create Event()) {is=title K2Node_AddDelegate}
   > SetIsExpanded!(execute, self:WBP UIKit Tab Container No Layout Object Reference=@n9, InIsExpanded:bool=true, SkipAnim:bool=true) -> (then)
   n10 = Cast To WBP_CrBrowser_SideNavigation_Button(execute=@n6.else, Object=GetInnerTabButton(self=Get UIKit_SideNavigation(), TabID=@n5.ParentID, InnerTabID=@n5))
   > SetText(self=@n10, Text=GetCategoryDisplayName(CategoryID=@n5))
   > Set CategoryID!(execute, CategoryID:Name=@n5, self:WBP Cr Browser Side Navigation Button Object Reference=@n10) -> (then, Output_Get:Name=None)
   n11 = Get Category Hierarchy Entry(CategoryID=@n2)
   Branch(execute=@n2.LoopBody, Condition=@n11.ReturnValue)
   n12 = Break Fort Emporium Filter Category(FortEmporiumFilterCategory=@n11) -> (bIsTab=~, bShowCategoryModal=~)
   > n13 = Branch(Condition="Equal (Name)"(A=@n12.ParentID))
   > AddSideNavigationTab!(
    execute
    self:WBP UIKit Side Navigation Base Object Reference=Get UIKit_SideNavigation()
    TabViewModel:MVVM Base Viewmodel Object Reference
    TabID:Name=@n12
    Index:int=-1
   ) -> (then)
   AddSideNavigationInnerTab!(
    execute=@n13.else
    self:WBP UIKit Side Navigation Base Object Reference=Get UIKit_SideNavigation()
    TabViewModel:MVVM Base Viewmodel Object Reference
    TabID:Name=@n12.ParentID
    InnerTabID:Name=@n12
    Index:int=-1
   ) -> (then)
  ]}
 RefreshSideNavigationFocus {nodes [
   RefreshSideNavigationFocus() {is=title K2Node_FunctionEntry}
   > Set FocusCategoryID(FocusCategoryID:Name=GetHomeSectionID!(self:Emporium Browser Filters Panel Object Reference) -> (ReturnValue:Name=None)) -> (Output_Get:Name=None)
   > n1 = Sequence()
   n2 = Get LastSelectedTab() -> (LastSelectedTab:Fort Creative Content Browser Tab Panel Base Weak object)
   > Is Valid(InputObject=@n2)
   > Set FocusCategoryID(FocusCategoryID=Get TabName!(self:Fort Creative Content Browser Tab Panel Base Object Reference=@n2) -> (TabName:Name=None)) -> (Output_Get=None)
   n3 = Get FocusCategoryID() -> (FocusCategoryID:Name=None)
   n4 = GetInnerTabButton(
    self=Get UIKit_SideNavigation()
    TabID="Get (a ref)"(Array:Array of Names=SplitTagID!(TagID:"Name (by ref)"=@n3) -> (ReturnValue:Array of Names), Dimension 1=0) -> (Output:"Name (by ref)")
    InnerTabID=@n3
   )
   n5 = Cast To WBP_CrBrowser_SideNavigation_Button(execute=@n1.then_1, Object=@n4)
   > SetIsSelected!(execute, self:Common Button Base Object Reference=@n5, InSelected:bool=true, bGiveClickFeedback:bool=false) -> (then)
   > Return Node!(execute, OutFocusCategory:Name=@n3, ReturnValue=@n4)
  ]}
 SetSearchFocus {nodes [
   SetSearchFocus() {is=title K2Node_FunctionEntry}
   > n1 = Get()
   > SetFocus(self=[
     @n1
     Get EditableText!(self:WBP UIKit Block Editable Text Object Reference=Get Block_EditableText!(self:WBP UIKit Input Field Base Object Reference=@n1) -> (Block_EditableText:WBP UIKit Block Editable Text Object Reference)) -> (EditableText:UIKit Editable Text Object Reference)
    ])
   n2 = Get SearchPanel()
   SetFocus(execute=@n1.else, self=Get EditableText(self=Get Block_EditableText(self=@n2)))
  ]}
 SetSearchText {nodes [
   n1 = SetSearchText!() -> (then, Text:Text) {is=title K2Node_FunctionEntry}
   > n2 = Get()
   > SetSearchText(self:WBP Creative Emporium Search Bar Object Reference=@n2, Text=@n1) {is=title K2Node_CallFunction}
   SetSearchText(execute=@n2.else, self=Get SearchPanel(), Text=@n1)
  ]}
 SetSearchVisibility {nodes [
   n1 = SetSearchVisibility!() -> (then, Visibility:bool=~) {is=title K2Node_FunctionEntry}
   > n2 = Get()
   n3 = Select(Option 0:ESlateVisibility Enum=Hidden, Option 1:ESlateVisibility Enum=SelfHitTestInvisible, Index=@n1) -> (ReturnValue:ESlateVisibility Enum)
   > SetVisibility(self=@n2, InVisibility=@n3)
   > SetVisibility(self=Get SearchPanel(), InVisibility=Collapsed)
   SetVisibility(execute=@n2.else, self=Get SearchPanel(), InVisibility=@n3)
  ]}
 UpdateSearchActionIcon {nodes [
   UpdateSearchActionIcon() {is=title K2Node_FunctionEntry}
   > n1 = Get()
   > n2 = Set UpdateWidget(UpdateWidget:WBP Creative Emporium Search Bar Object Reference=@n1) -> (Output_Get:WBP Creative Emporium Search Bar Object Reference)
   n3 = Set UpdateWidget(execute=@n1.else, UpdateWidget=Get SearchPanel())
   n4 = Get Text(self:WBP UIKit Input Field Base Object Reference=@n3)
   n5 = GetIconForInputAction!(
    execute=[@n3.then, @n2.then]
    self:Emporium Browser Filters Panel Object Reference
    InInputActionRow:"Data Table Row Handle Structure (by ref)"=Select(
     Option 0:"Data Table Row Handle Structure (by ref)"=GetClearSearchInputAction!() -> (ReturnValue:Data Table Row Handle Structure)
     Option 1:"Data Table Row Handle Structure (by ref)"=GetFocusSearchInputAction!() -> (ReturnValue:Data Table Row Handle Structure)
     Index=TextIsEmpty(InText=@n4)
    ) -> (ReturnValue:"Data Table Row Handle Structure (by ref)")
   ) -> (then, bOutFoundIcon:bool, ReturnValue:Slate Brush Structure)
   > Branch(Condition=@n5)
   > Set Action Icon!(
    execute
    self:WBP Creative Emporium Search Bar Object Reference=Get UpdateWidget() -> (UpdateWidget:WBP Creative Emporium Search Bar Object Reference)
    Icon:Slate Brush Structure=@n5.ReturnValue
   ) -> (then) {is=title K2Node_CallFunction}
  ]}
]
variables [
 FilterSelectionChanged {}
 ParentSearchPanel {type=WBP_Creative_Emporium_SearchBar_C, default=null}
]
widgets {
 info {
  inheritedWidgetCount=1
  parentClass=/Script/CreativeEmporiumUI.EmporiumBrowserFiltersPanel
  rootWidgetClass=&VerticalBox
  widgetCount=9
 }
 widgets [
  VerticalBox_0 {widgetClassPath=&VerticalBox, widget=VerticalBox_0}
  SizeBox_0 {
   widgetClassPath=&SizeBox
   parent=VerticalBox_0
   widget=SizeBox_0
   properties {bOverride_HeightOverride=true, bOverride_MinDesiredWidth=true, heightOverride=72, widthOverride=350}
   slot {class=VerticalBoxSlot, properties {verticalAlignment=VAlign_Center}}
  }
  SearchPanel {
   widgetClassPath=&/WBP_Creative_Emporium_SearchBar
   bIsVariable=true
   parent=SizeBox_0
   widget=SearchPanel
   properties {
    bHideInputAction=false
    hint Text=Search...
    holdData=&CommonUI_HoldData_Base_C
    minHeight=64
    visibility=Collapsed
   }
   slot {class=SizeBoxSlot}
  }
  Button_SortAndFilter {
   widgetClassPath=&WBP_UIKit_Button_Regular
   bInherited=true
   bIsVariable=true
   parent=VerticalBox_0
   widget=Button_SortAndFilter
   properties {
    buttonSize=Small
    buttonType=Quiet
    text="Sort && Filter"
    triggeringInputAction {dataTable=&InputActionDataTable, rowName=GenericRightTrigger}
   }
   slot {class=VerticalBoxSlot, properties {padding=<ltrb 0 6 0 0>}}
  }
  HeaderSpacer {widgetClassPath=&Overlay, bIsVariable=true, parent=VerticalBox_0, widget=HeaderSpacer, slot {class=VerticalBoxSlot}}
  Spacer {
   widgetClassPath=&Spacer
   parent=HeaderSpacer
   widget=Spacer
   properties {size=<1 24>}
   slot {class=OverlaySlot, properties {horizontalAlignment=HAlign_Fill}}
  }
  HitTestBlocker {
   widgetClassPath=&Image
   parent=HeaderSpacer
   widget=HitTestBlocker
   properties {colorAndOpacity=<rgba 1 1 1 0>}
   slot {class=OverlaySlot, properties {horizontalAlignment=HAlign_Fill, verticalAlignment=VAlign_Fill}}
  }
  Overlay_0 {
   widgetClassPath=&Overlay
   parent=VerticalBox_0
   widget=Overlay_0
   slot {class=VerticalBoxSlot, properties {size {sizeRule=Fill, value=1}}}
  }
  UIKit_SideNavigation {
   widgetClassPath=/FortUILibrary/UIKit/Tabs/WBP_UIKit_SideNavigation
   bIsVariable=true
   parent=Overlay_0
   widget=UIKit_SideNavigation
   properties {
    allowContainerInteractionOnGamepad=true
    allowMultipleExpandedContainers=true
    buttonClass=&/WBP_CrBrowser_SideNavigation_HeaderButton.WBP_CrBrowser_SideNavigation_HeaderButton_C
    innerTabButtonClassOverride=&/WBP_CrBrowser_SideNavigation_Button.WBP_CrBrowser_SideNavigation_Button_C
    navigation=&self:WidgetTree.UIKit_SideNavigation.WidgetNavigation_0
    scrollBarVisibility=Visible
    tabContainerClass=/CreativeEmporium/Widgets/WBP_Creative_Emporium_TabContainer_SideNav.WBP_Creative_Emporium_TabContainer_SideNav_C
    tabGroupHorizontalAlignment=HAlign_Fill
    useBackplate=false
    initialSideNavigationTabs []
   }
   slot {class=OverlaySlot, properties {horizontalAlignment=HAlign_Fill}}
  }
 ]
}
