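# Selectable tag button for the Creative Emporium, built on the UIKit generic button, that represents one TagID.
# SetTag resolves the label from the tag, falling back to the category display name when empty, and tints beta tags differently.
# Selection and deselection broadcast OnTagSelected with the TagID and selected state.
# Hover plays a bounce animation, press and release run a squash animation forward and reverse, and disabling dims opacity.
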WidgetBlueprint /CreativeEmporium/WBP_CreativeTag
animations [
 Anim_Bounce {
  binding_count=1
  length_seconds=0.2961
  object=&self:Anim_Bounce
  sequence {
   bindings [Main {
     id=#1
     possessed_class=GridPanel
     tracks [{
       type=MovieScene2DTransformTrack
       sections [{
         type=MovieScene2DTransformSection
         channels [
          Translation.X {type=MovieSceneScriptingFloatChannel}
          Translation.Y {type=MovieSceneScriptingFloatChannel, keys [[0,0],[4.423,-3],[13.268,0],[17.765,0]]}
         ]
        }]
      }]
    }]
   display_rate {numerator=60}
  }
 }
 Anim_Pressed {
  binding_count=2
  length_seconds=0.1
  object=&self:Anim_Pressed
  sequence {
   bindings [
    Root {
     id=#2
     possessed_class=GridPanel
     tracks [{
       type=MovieScene2DTransformTrack
       sections [{
         type=MovieScene2DTransformSection
         end_frame=6
         start_frame=0
         channels [
          Scale.X {type=MovieSceneScriptingFloatChannel, keys [[0,1],[6,0.975]]}
          Scale.Y {type=MovieSceneScriptingFloatChannel, keys [[0,1],[6,0.975]]}
         ]
        }]
      }]
    }
    Main {
     id=#3
     possessed_class=GridPanel
     tracks [{
       type=MovieScene2DTransformTrack
       sections [{
         type=MovieScene2DTransformSection
         channels [
          Scale.X {type=MovieSceneScriptingFloatChannel, keys [[0,1],[6,0.95]]}
          Scale.Y {type=MovieSceneScriptingFloatChannel, keys [[0,1],[6,0.975]]}
         ]
        }]
      }]
    }
   ]
   display_rate {numerator=60}
  }
 }
]
graphs [
 EventGraph {nodes [
   n1 = Event PreConstruct()
   > Parent: PreConstruct(IsDesignTime=@n1.IsDesignTime)
   Event Construct()
   > Parent: Construct()
   Event On Hovered()
   > Parent: On Hovered() {is=title K2Node_CallParentFunction}
   > PlayAnimationForward(InAnimation=Get Anim_Bounce() -> (Anim_Bounce:Widget Animation Object Reference))
   Event On Selected()
   > Parent: On Selected() {is=title K2Node_CallParentFunction}
   > Call On Tag Selected!(execute:Exec, self:WBP Creative Tag Object Reference, TagID:Name=Get TagID() -> (TagID:Name=None), Selected:bool=true) -> (then:Exec) {is=title K2Node_CallDelegate}
   Event On Pressed() {is=title K2Node_Event}
   > Parent: On Pressed() {is=title K2Node_CallParentFunction}
   n2 = Get Anim_Pressed() -> (Anim_Pressed:Widget Animation Object Reference)
   > PlayAnimationForward(InAnimation=@n2)
   Event On Released() {is=title K2Node_Event}
   > Parent: On Released() {is=title K2Node_CallParentFunction}
   > PlayAnimationReverse(InAnimation=@n2)
   Event On Deselected()
   > Parent: On Deselected() {is=title K2Node_CallParentFunction}
   > Call On Tag Selected(TagID=Get TagID() -> (TagID=None))
   Event On Disabled() {is=title K2Node_Event}
   > SetRenderOpacity(InOpacity=0.65)
   Event On Enabled() {is=title K2Node_Event}
   > SetRenderOpacity(InOpacity=1)
  ]}
 OnTagSelected {nodes [OnTagSelected!() -> (then, TagID:Name, Selected:bool=~) {is=title K2Node_FunctionEntry}]}
 SetTag {nodes [
   n1 = SetTag!() -> (then, TagID:Name) {is=title K2Node_FunctionEntry}
   > Set TagID(TagID:Name=@n1) -> (Output_Get:Name=None)
   > n2 = Sequence()
   n3 = GetCreativeTagLabel!(TagID:"Name (by ref)"=Get TagID() -> (TagID=None)) -> (bOutIsBeta:bool, ReturnValue:Text)
   > n4 = Branch(Condition=TextIsEmpty(InText=@n3.ReturnValue))
   > Set Label(Label:Text=GetCategoryDisplayName!(CategoryID:"Name (by ref)"=Get TagID() -> (TagID=None)) -> (ReturnValue:"Text (by ref)")) -> (Output_Get:Text)
   > Set bIsBeta(bIsBeta:bool=false) -> (Output_Get:bool)
   Set Label(execute=@n4.else, Label=@n3.ReturnValue)
   > Set bIsBeta(bIsBeta=@n3)
   SetText!(execute=@n2.then_1, self:WBP UIKit Button Generic Base Object Reference, Text:Text=Get Label() -> (Label:Text)) -> (then)
   > SetColorAndOpacity(
    self:User Widget Object Reference
    InColorAndOpacity:Linear Color Structure=Select(
     Option 0:Linear Color Structure=Get Tint_Normal() -> (Tint_Normal:Linear Color Structure)
     Option 1:Linear Color Structure=Get Tint_CreativeBeta() -> (Tint_CreativeBeta:Linear Color Structure)
     Index=Get bIsBeta() -> (bIsBeta:bool)
    ) -> (ReturnValue:Linear Color Structure)
   )
  ]}
]
variables [
 OnTagSelected {}
 TagID {type=string, default=None}
 Tint_CreativeBeta {type=LinearColor, default=<rgba 1 1 0 1>}
 Tint_Normal {type=LinearColor, default=<rgba 1 1 1 1>}
]
widgets {
 info {
  inheritedWidgetCount=0
  parentClass=/FortUILibrary/Systems/ModularButton/WBP_UIKit_ButtonGeneric_Base.WBP_UIKit_ButtonGeneric_Base_C
  rootWidgetClass=null
  widgetCount=0
 }
 widgets []
}
