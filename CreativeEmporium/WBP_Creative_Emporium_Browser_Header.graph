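# Creative Emporium browser header widget that shows section and category labels, active tag filter chips, and a match any/all toggle button.
# SetActiveItemDetailTags dedupes tags through TagMap, keying category tags by last segment so they override matching creative tags, then removes stale chips and adds missing ones.
# OnTagClicked removes the matching chip and broadcasts ItemDetailTagSelected.
# The mode button broadcasts OnTagFilterModeToggle, collapses without tags, and changes icon through UpdateTagFilterMode.
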WidgetBlueprint /CreativeEmporium/WBP_Creative_Emporium_Browser_Header
graphs [
 AddTag {nodes [
   n1 = AddTag!() -> (then, Tag ID:Name) {is=title K2Node_FunctionEntry}
   > n2 = Create Widget!(
    execute:Exec
    Class:User Widget Class Reference=&/WBP_CreativeTag.WBP_CreativeTag_C
    OwningPlayer:Player Controller Object Reference=GetOwningPlayer()
    Style:Common Button Style Class Reference=/FortUILibrary/UIKit/Buttons/Blocks/ButtonStyle_UIKit_Empty.ButtonStyle_UIKit_Empty_C
    bApplyAlphaOnDisable:bool=False
    bLocked:bool=False
    bSelectable:bool=True
    bShouldSelectUponReceivingFocus:bool=False
    bInteractableWhenSelected:bool=True
    bToggleable:bool=True
    bTriggerClickedAfterSelection:bool=True
    bDisplayInputActionWhenNotInteractable:bool=False
    bHideInputActionWithKeyboard:bool=False
    bShouldUseFallbackDefaultInputAction:bool=True
    bRequiresHold:bool=False
    HoldData:Common UIHold Data Class Reference=&CommonUI_HoldData_Base_C
    ClickMethod:EButtonClickMethod Enum=DownAndUp
    InputPriority:int=0
    TriggeringInputAction:Data Table Row Handle Structure
   ) -> (then:Exec, ReturnValue:WBP Creative Tag Object Reference)
   > SetTag!(execute, self:WBP Creative Tag Object Reference=@n2, TagID:Name=@n1) -> (then)
   > Set TextNudge!(execute, TextNudge:Vector 2D Structure=(X=8,Y=0), self:WBP UIKit Button Generic Base Object Reference=@n2) -> (then, Output_Get:Vector 2D Structure)
   > Set ButtonPadding!(
    execute
    ButtonPadding:Margin Structure=Make Margin(Left=16, Top=0, Right=24, Bottom=0)
    self:WBP UIKit Button Generic Base Object Reference=@n2
   ) -> (then, Output_Get:Margin Structure)
   > n3 = AddChildToHorizontalBox!(
    execute
    self:Horizontal Box Object Reference=Get ItemDetailTagContainer() -> (ItemDetailTagContainer:Horizontal Box Object Reference)
    Content:Widget Object Reference=@n2
   ) -> (then, ReturnValue:Horizontal Box Slot Object Reference)
   > SetPadding!(execute, self:Horizontal Box Slot Object Reference=@n3, InPadding:Margin Structure=Make Margin(Left=4, Top=0, Right=4, Bottom=0)) -> (then)
   > Bind Event to On Tag Selected(self:WBP Creative Tag Object Reference=@n2, Delegate=Create Event()) {is=title K2Node_AddDelegate}
  ]}
 EventGraph {nodes [
   "On Clicked (Button_Mode)"!() -> (OutputDelegate, then, Button:Common Button Base Object Reference) {is=title K2Node_ComponentBoundEvent}
   > Call On Tag Filter Mode Toggle!(execute:Exec, self:WBP Creative Emporium Browser Header Object Reference) -> (then:Exec) {is=title K2Node_CallDelegate}
   n1 = SetActiveItemDetailTags!() -> (OutputDelegate, then, Category Tags:Array of Names, Creative Tags:Array of Names) {is=title K2Node_CustomEvent}
   > n2 = Sequence!(execute) -> (then_0, then_1, then_2:Exec, then_3:Exec, then_4:Exec)
   n3 = For Each Loop(Exec=@n2.then_1, Array:Array of Names=@n1.Category Tags) -> (Array Element:Name)
   n4 = SplitTagID!(TagID:"Name (by ref)"=@n3) -> (ReturnValue:Array of Names)
   > Branch(Condition=Is Not Empty!(TargetArray:Array of Names=@n4) -> (ReturnValue:bool))
   > Add!(
    execute
    TargetMap:Map of Names to Names=Get TagMap() -> (TagMap:Map of Names to Names)
    Key:"Name (by ref)"="Get (a ref)"(Array:Array of Names=@n4, Dimension 1=Last Index!(TargetArray:Array of Names=@n4) -> (ReturnValue=0)) -> (Output:"Name (by ref)")
    Value:"Name (by ref)"=@n3
   ) -> (then) {is=title K2Node_CallFunction}
   n5 = For Each Loop(Exec=@n2.then_2, Array=@n1.Creative Tags)
   n6 = Get TagMap()
   > n7 = Branch(Condition=Contains!(TargetMap:Map of Names to Names=@n6, Key:"Name (by ref)"=@n5) -> (ReturnValue))
   Add(execute=@n7.else, TargetMap=@n6, Key=@n5, Value=@n5)
   n8 = Values!(execute=@n2.then_3, TargetMap:Map of Names to Names=Get TagMap()) -> (then, Values:Array of Names)
   > RemoveUnusedTags!(execute, self, Tags:Array of Names=@n8) -> (then)
   > n9 = For Each Loop(Array=@n8)
   > n10 = HasTag!(execute, self, Tag:Name=@n9) -> (then, HasTag:bool)
   > n11 = Branch(Condition=@n10)
   AddTag!(execute=@n11.else, self, Tag ID:Name=@n9) -> (then) {is=title K2Node_CallFunction}
   n12 = Get TagMap()
   SetVisibility(
    execute=@n2.then_4
    self=Get Button_Mode() -> (Button_Mode:WBP UIKit Button Generic Object Reference)
    InVisibility=Select(
     Option 0=Visible
     Option 1=Collapsed
     Index=Is Empty!(TargetMap:Map of Names to Names=@n12) -> (ReturnValue:bool) {is=title K2Node_CallFunction}
    )
   )
   Clear!(execute=@n2.then_0, TargetMap:Map of Names to Names=Get TagMap()) -> (then) {is=title K2Node_CallFunction}
   n13 = OnTagClicked!() -> (OutputDelegate, then, TagID:Name, Selected:bool=~) {is=title K2Node_CustomEvent}
   > n14 = Sequence!(execute) -> (then_0, then_1)
   > n15 = For Each Loop with Break!(Exec:Exec, Array:Array of Widget Object References=GetAllChildren(self=Get ItemDetailTagContainer()), Break:Exec=@n17.then) -> (LoopBody:Exec, Array Element:Widget Object Reference, Array Index:int, Completed:Exec)
   > n16 = Cast To WBP_CreativeTag!(execute, Object=@n15) -> (then, CastFailed, AsWBP Creative Tag:WBP Creative Tag Object Reference) {is=title K2Node_DynamicCast}
   > Branch(Condition="Equal (Name)"(A=Get TagID!(self:WBP Creative Tag Object Reference=@n16) -> (TagID:Name=None), B=@n13.TagID))
   > n17 = RemoveChild(self=Get ItemDetailTagContainer(), Content=@n16)
   Call Item Detail Tag Selected!(execute:Exec=@n14.then_1, self:WBP Creative Emporium Browser Header Object Reference, Tag ID:Name=@n13.TagID) -> (then:Exec) {is=title K2Node_CallDelegate}
   n18 = UpdateTagFilterMode!() -> (OutputDelegate, then, NewParam:EFortEmporiumItemTagFilterMode Enum=MatchAny) {is=title K2Node_CustomEvent}
   > SetIcon!(
    execute
    self:WBP UIKit Button Generic Base Object Reference=Get Button_Mode()
    Icon:Object Reference=Select!(
     MatchAny:Object Reference=/CreativeEmporium/Materials/T_UI_Icon_FilterUnion
     MatchAll:Object Reference=/CreativeEmporium/Materials/T_UI_Icon_FilterIntersect
     Index:EFortEmporiumItemTagFilterMode Enum=@n18.NewParam
    ) -> (ReturnValue:Object Reference)
   ) -> (then)
   K2Node_Event_1 = UpdateCategoryLabel!() -> (OutputDelegate, then, Section Label:Text, Category Label:Text) {is=title K2Node_CustomEvent}
   > n19 = Sequence()
   n20 = TextIsEmpty(InText=@K2Node_Event_1.Section Label)
   n21 = TextIsEmpty(InText=@K2Node_Event_1.Category Label)
   > SetVisibility(
    self=Get Text_Spacer() -> (Text_Spacer:Common Text Object Reference)
    InVisibility=Select!(Option 0:ESlateVisibility Enum=HitTestInvisible, Option 1:ESlateVisibility Enum=Collapsed, Index:bool=OR Boolean(A=@n20, B=@n21)) -> (ReturnValue:ESlateVisibility Enum)
   )
   "SetText (Text)"(execute=@n19.then_1, self=Get Text_Label1() -> (Text_Label1:Common Text Object Reference), InText=@K2Node_Event_1.Section Label)
   > "SetText (Text)"(self=Get Text_Label2() -> (Text_Label2:Common Text Object Reference), InText=@K2Node_Event_1.Category Label)
   Event PreConstruct()
   Event Tick()
   Event Construct()
   Format Text!(
    Format="LOCTEXT(\"MATCH: {Mode}\")"
    Mode:"Text (by ref)"=Select!(
     MatchAny:"Text (by ref)"=MakeLiteralText(Value="LOCTEXT(\"ANY\")")
     MatchAll:"Text (by ref)"=MakeLiteralText(Value="LOCTEXT(\"ALL\")")
     Index:EFortEmporiumItemTagFilterMode Enum=@n18.NewParam
    ) -> (ReturnValue:"Text (by ref)")
   ) -> (Result)
  ]}
 HasTag {nodes [
   n1 = HasTag!() -> (then, Tag:Name) {is=title K2Node_FunctionEntry}
   > n2 = For Each Loop(Array:Array of Widget Object References=GetAllChildren(self=Get ItemDetailTagContainer())) -> (Array Element:Widget Object Reference)
   Return Node!(execute=@n2.Completed, HasTag:bool=false)
   > n3 = Cast To WBP_CreativeTag(Object=@n2)
   > Branch(Condition="Equal (Name)"(A=Get TagID(self=@n3) -> (TagID=None), B=@n1))
   > Return Node(HasTag=true)
  ]}
 ItemDetailTagSelected {nodes [ItemDetailTagSelected!() -> (then, Tag ID:Name) {is=title K2Node_FunctionEntry}]}
 OnTagFilterModeToggle {nodes [OnTagFilterModeToggle() {is=title K2Node_FunctionEntry}]}
 RemoveUnusedTags {nodes [
   n1 = RemoveUnusedTags!() -> (then, Tags:Array of Names) {is=title K2Node_FunctionEntry}
   > n2 = For Each Loop(Array=GetAllChildren(self=Get ItemDetailTagContainer()))
   n3 = For Each Loop(Exec=@n2.Completed, Array=Get TagsToRemove() -> (TagsToRemove:Array of Widget Object References))
   > n4 = Cast To WBP_CreativeTag(Object=@n3)
   > Unbind all Events from On Tag Selected!(execute:Exec, self:WBP Creative Tag Object Reference=@n4) -> (then:Exec) {is=title K2Node_ClearDelegate}
   > RemoveChild(self=Get ItemDetailTagContainer(), Content=@n4)
   n5 = Cast To WBP_CreativeTag(execute=@n2.LoopBody, Object=@n2)
   > n6 = Branch(Condition=Contains Item(TargetArray=@n1, ItemToFind=Get TagID(self=@n5) -> (TagID=None)))
   Add(execute=@n6.else, TargetArray:Array of Widget Object References=Get TagsToRemove(), NewItem:"Widget Object Reference (by ref)"=@n5) {is=title K2Node_CallArrayFunction}
  ]}
]
variables [
 ItemDetailTagSelected {}
 OnTagFilterModeToggle {}
 TagMap {type="map<string>", default {}}
]
widgets {
 info {inheritedWidgetCount=0, parentClass=&CommonActivatableWidget, rootWidgetClass=&CommonBorder, widgetCount=14}
 widgets [
  CommonBorder_0 {widgetClassPath=&CommonBorder, widget=CommonBorder_0, properties {padding=<ltrb 0 0 0 0>}}
  VerticalBox_0 {
   widgetClassPath=&VerticalBox
   parent=CommonBorder_0
   widget=VerticalBox_0
   slot {class=BorderSlot, properties {padding=<ltrb 0 0 0 0>}}
  }
  HorizontalBox_1 {widgetClassPath=&HorizontalBox, parent=VerticalBox_0, widget=HorizontalBox_1, slot {class=VerticalBoxSlot}}
  SizeBox_67 {
   widgetClassPath=&SizeBox
   parent=HorizontalBox_1
   widget=SizeBox_67
   properties {bOverride_MaxDesiredWidth=true, maxDesiredWidth=1400}
   slot {class=HorizontalBoxSlot}
  }
  HorizontalBox_0 {
   widgetClassPath=&HorizontalBox
   parent=SizeBox_67
   widget=HorizontalBox_0
   slot {class=SizeBoxSlot, properties {verticalAlignment=VAlign_Center, padding=<ltrb 0 6 0 0>}}
  }
  Box_Label1 {
   widgetClassPath=&SizeBox
   bIsVariable=true
   parent=HorizontalBox_0
   widget=Box_Label1
   properties {maxDesiredWidth=450}
   slot {class=HorizontalBoxSlot, properties {verticalAlignment=VAlign_Bottom}}
  }
  Text_Label1 {
   widgetClassPath=&CommonTextBlock
   bIsVariable=true
   parent=Box_Label1
   widget=Text_Label1
   properties {
    bIsScrollingEnabled=false
    mobileFontSizeMultiplier=1.34
    scrollStyle=&TextScrollStyle-Base_C
    text="{VERY LONG CATEGORY NAME}"
    textTransformPolicy=ToUpper
    font {
     bForceMonospaced=false
     bMaterialIsStencil=false
     fontMaterial=null
     fontObject=&HeadingNow
     letterSpacing=0
     monospacedWidth=1
     size=32.25
     skewAmount=0
     typefaceFontName=86 Bold Italic
     outlineSettings {
      bApplyOutlineToDropShadows=false
      bMiteredCorners=false
      bSeparateFillAlpha=false
      outlineMaterial=null
      outlineSize=0
      outlineColor=<rgba 0 0 0 1>
     }
    }
   }
   slot {class=SizeBoxSlot}
  }
  Text_Spacer {
   widgetClassPath=&CommonTextBlock
   bIsVariable=true
   parent=HorizontalBox_0
   widget=Text_Spacer
   properties {
    mobileFontSizeMultiplier=1.34
    scrollStyle=&TextScrollStyle-Base_C
    text=" | "
    textTransformPolicy=ToUpper
    colorAndOpacity {colorUseRule=UseColor_Specified, specifiedColor=<rgba 1 1 1 0.7>}
    font {
     bForceMonospaced=false
     bMaterialIsStencil=false
     fontMaterial=null
     fontObject=&HeadingNow
     letterSpacing=0
     monospacedWidth=1
     size=32.25
     skewAmount=0
     typefaceFontName=86 Bold Italic
     outlineSettings {
      bApplyOutlineToDropShadows=false
      bMiteredCorners=false
      bSeparateFillAlpha=false
      outlineMaterial=null
      outlineSize=0
      outlineColor=<rgba 0 0 0 1>
     }
    }
   }
   slot {class=HorizontalBoxSlot}
  }
  Box_Label2 {
   widgetClassPath=&SizeBox
   bIsVariable=true
   parent=HorizontalBox_0
   widget=Box_Label2
   properties {maxDesiredWidth=450}
   slot {class=HorizontalBoxSlot, properties {verticalAlignment=VAlign_Center, size {sizeRule=Fill, value=1}}}
  }
  Text_Label2 {
   widgetClassPath=&CommonTextBlock
   bIsVariable=true
   parent=Box_Label2
   widget=Text_Label2
   properties {
    mobileFontSizeMultiplier=1.34
    scrollStyle=&TextScrollStyle-Base_C
    text="{VERY LONG CATEGORY NAME}"
    textTransformPolicy=ToUpper
    colorAndOpacity {colorUseRule=UseColor_Specified, specifiedColor=<rgba 1 1 1 0.7>}
    font {
     bForceMonospaced=false
     bMaterialIsStencil=false
     fontMaterial=null
     fontObject=&HeadingNow
     letterSpacing=0
     monospacedWidth=1
     size=32.25
     skewAmount=0
     typefaceFontName=86 Bold Italic
     outlineSettings {
      bApplyOutlineToDropShadows=false
      bMiteredCorners=false
      bSeparateFillAlpha=false
      outlineMaterial=null
      outlineSize=0
      outlineColor=<rgba 0 0 0 1>
     }
    }
   }
   slot {class=SizeBoxSlot}
  }
  HorizontalBox_139 {
   widgetClassPath=&HorizontalBox
   parent=VerticalBox_0
   widget=HorizontalBox_139
   slot {class=VerticalBoxSlot, properties {padding=<ltrb 0 6 24 6>}}
  }
  Button_Mode {
   widgetClassPath=&WBP_UIKit_Button_Generic
   bIsVariable=true
   parent=HorizontalBox_139
   widget=Button_Mode
   properties {
    backgroundMaterial=/CreativeEmporium/Materials/MI_CrBrowser_FilterMode_Button_Material
    icon=&T_UI_Icon_FilterUnion
    iconMaterial=/CreativeEmporium/Materials/MI_CrBrowser_FilterModeIcon_Color
    minHeight=48
    minWidth=48
    useTextBlock=false
    visibility=HitTestInvisible
    triggeringInputAction {dataTable=&InputActionDataTable, rowName=None}
   }
   slot {class=HorizontalBoxSlot, properties {padding=<ltrb 0 0 8 0>}}
  }
  Overlay_0 {
   widgetClassPath=&Overlay
   parent=HorizontalBox_139
   widget=Overlay_0
   slot {class=HorizontalBoxSlot, properties {verticalAlignment=VAlign_Center, padding=<ltrb 24 0 0 0>, size {sizeRule=Fill, value=1}}}
  }
  ItemDetailTagContainer {
   widgetClassPath=&HorizontalBox
   bIsVariable=true
   parent=Overlay_0
   widget=ItemDetailTagContainer
   slot {class=OverlaySlot, properties {horizontalAlignment=HAlign_Fill}}
  }
 ]
}
