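# Creative Emporium item info panel showing name, author, price, license, file size, description and tags for a selected item.
# UpdateWithFortItem resets the text, reads Emporium item data minus hidden tags, or falls back to creative tag IDs for other items.
# A parent playset alters name and description and reveals the Go To Item button.
# PopulateTags rebuilds optionally interactive WBP_CreativeTag buttons; dispatchers relay tag selection and focus requests.
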WidgetBlueprint /CreativeEmporium/WBP_Creative_Emporium_ItemInfo
graphs [
 EventGraph {nodes [
   "On Clicked (Button_ItemDetailFocus)"!() -> (OutputDelegate, then, Button:Common Button Base Object Reference) {is=title K2Node_ComponentBoundEvent}
   > Call On Request Item Focus!(execute:Exec, self:WBP Creative Emporium Item Info Object Reference) -> (then:Exec) {is=title K2Node_CallDelegate}
   n1 = OnTagClicked!() -> (OutputDelegate, then, TagID:Name, Selected:bool=~) {is=title K2Node_CustomEvent}
   > Call On Tag Selected!(execute:Exec, self:WBP Creative Emporium Item Info Object Reference, TagID:Name=@n1.TagID, Selected:bool=@n1.Selected) -> (then:Exec) {is=title K2Node_CallDelegate}
   n2 = Update SelectedTags!() -> (OutputDelegate, then, Selected Tags:Array of Names) {is=title K2Node_CustomEvent}
   > n3 = For Each Loop(Array:Array of Widget Object References=GetAllChildren(self=Get Wbox_Tags() -> (Wbox_Tags:Wrap Box Object Reference))) -> (Array Element:Widget Object Reference)
   > n4 = Cast To WBP_CreativeTag!(execute, Object=@n3) -> (then, CastFailed, AsWBP Creative Tag:WBP Creative Tag Object Reference) {is=title K2Node_DynamicCast}
   > SetIsSelected!(
    execute
    self:Common Button Base Object Reference=@n4
    InSelected:bool=Contains Item(TargetArray=@n2.Selected Tags, ItemToFind=Get TagID!(self:WBP Creative Tag Object Reference=@n4) -> (TagID:Name=None))
    bGiveClickFeedback:bool=true
   ) -> (then)
   n5 = Event SetSelectItemTagsIcon!() -> (OutputDelegate, then, SelectItemTagsIcon:Slate Brush Structure) {is=title K2Node_Event}
   > SetBrush(self=Get Image_ItemTagActionPrompt() -> (Image_ItemTagActionPrompt:Image Object Reference), InBrush=@n5.SelectItemTagsIcon)
   n6 = Event ShowSelectItemTagsPrompt!() -> (OutputDelegate, then, bShowPrompt:bool) {is=title K2Node_Event}
   > Set bShowItemTagsPrompt(bShowItemTagsPrompt:bool=@n6.bShowPrompt) -> (Output_Get:bool)
   > UpdateSelectItemTagsPrompt()
   Event SetTagFocus() {is=title K2Node_Event}
   n7 = Get Wbox_Tags()
   > Branch(Condition=HasAnyChildren!(self:Panel Widget Object Reference=@n7) -> (ReturnValue:bool))
   > SetFocus(self=GetChildAt!(self:Panel Widget Object Reference=@n7, Index:int=0) -> (ReturnValue:Widget Object Reference))
   n8 = Event UpdateWithFortItem!() -> (OutputDelegate, then, Item:Fort Item Object Reference, ParentDefinition:Item Definition Base Object Reference, bAllowInteractiveTags:bool) {is=title K2Node_Event}
   > n9 = Set Item(Item:Fort Item Object Reference=@n8.Item) -> (Output_Get:Fort Item Object Reference)
   > Is Valid(InputObject=@n9)
   > n10 = Sequence!(execute) -> (then_0, then_1, then_2:Exec, then_3:Exec, then_4:Exec)
   "SetText (Text)"(
    execute=@n10.then_1
    self=Get ItemDisplayName() -> (ItemDisplayName:Common Text Object Reference)
    InText=GenerateItemName!() -> (DisplayText:Text)
   )
   n11 = GetItemDefinition!(self:Fort Item Object Reference=Get Item() -> (Item:Fort Item Object Reference)) -> (ReturnValue:Fort Item Definition Object Reference)
   Is Valid(exec=@n10.then_3, InputObject=@n11)
   > n12 = Cast To FortEmporiumItemDefinition!(execute, Object=@n11) -> (then, CastFailed, AsFort Emporium Item Definition:Fort Emporium Item Definition Object Reference) {is=title K2Node_DynamicCast}
   > Set ItemData(ItemData:Fort Emporium Item Data Structure=GetEmporiumItemData!(self:Fort Emporium Item Definition Object Reference=@n12) -> (ReturnValue:"Fort Emporium Item Data Structure (by ref)")) -> (Output_Get:Fort Emporium Item Data Structure)
   > n13 = Sequence()
   n14 = Get ItemData() -> (ItemData:Fort Emporium Item Data Structure)
   > "SetText (Text)"(
    self=Get ItemAuthorName() -> (ItemAuthorName:Common Text Object Reference)
    InText="To Text (String)"(InString=Break Fort Emporium Item Data!(FortEmporiumItemData:"Fort Emporium Item Data Structure (by ref)"=@n14) -> (
      Id:String
      Title:String
      Namespace:String
      EntityType:String
      ThumbnailURL:String
      Price:float=~
      BaseCurrency:String
      SketchfabUID:String
      TagList:Array of Names
      LicenseType:EFortEmporiumItemLicense Enum
      bRequiresEntitlement:bool=~
      VersePath:String
      AssetIds:Array of Strings
      FileSize:int
      Description:String
      Seller:String
      MaterialCount:int
      PolygonCount:int
      VertexCount:int
      ChildAssetCount:int
     ) {is=title K2Node_BreakStruct}.Seller)
   )
   PopulateItemPrice(execute=@n13.then_1)
   n15 = Get ItemData()
   "SetText (Text)"(
    execute=@n13.then_2
    self=Get ItemLicenseType() -> (ItemLicenseType:Common Text Object Reference)
    InText=GetTextFromEmporiumLicense!(License:EFortEmporiumItemLicense Enum=Break Fort Emporium Item Data(FortEmporiumItemData=@n15) -> (Price=~, bRequiresEntitlement=~).LicenseType) -> (ReturnValue:"Text (by ref)")
   )
   PopulateFileSize(execute=@n13.then_3)
   n16 = For Each Loop(
    Exec=@n13.then_4
    Array:Array of Names=Break Fort Emporium Item Data(FortEmporiumItemData=Get ItemData()) -> (Price=~, bRequiresEntitlement=~).TagList
   ) -> (Array Element:Name)
   PopulateTags!(
    execute=@n16.Completed
    self
    inTagArray:Array of Names=Get ItemVisibleTags() -> (ItemVisibleTags:Array of Names)
    InAllowInteractiveTags:bool=@n8.bAllowInteractiveTags
   ) -> (then)
   n17 = Branch(execute=@n16.LoopBody, Condition=Contains Item(TargetArray=Get HiddenTags() -> (HiddenTags:Array of Names), ItemToFind=@n16))
   Add(execute=@n17.else, TargetArray:Array of Names=Get ItemVisibleTags(), NewItem:"Name (by ref)"=@n16) {is=title K2Node_CallArrayFunction}
   PopulateTags(
    execute=@n12.CastFailed
    inTagArray=GetCreativeTagIDs!(ItemDefinition:Item Definition Base Object Reference=@n11) -> (ReturnValue:Array of Names)
    InAllowInteractiveTags=@n8.bAllowInteractiveTags
   )
   n18 = Get ParentPlaysetDefinition() -> (ParentPlaysetDefinition:Fort Playset Item Definition Object Reference)
   n19 = Branch(execute=@n10.then_4, Condition=IsValid(Object=@n18))
   n20 = Get Text_Description() -> (Text_Description:Common Text Object Reference)
   > "SetText (Text)"(
    self=@n20
    InText=Format Text!(
     Format="LOCTEXT(\"From {Collection}\")"
     Collection:Text=GetDisplayName!(self:Item Definition Base Object Reference=@n18) -> (ReturnValue:Text)
    ) -> (Result)
   )
   "SetText (Text)"(execute=@n19.else, self=@n20, InText=GetDescription!(self:Fort Item Object Reference=Get Item()) -> (ReturnValue:Text))
   n21 = Get TypeTextBlock() -> (TypeTextBlock:Common Text Object Reference)
   "SetText (Text)"(
    execute=@n10.then_2
    self=@n21
    InText=GetItemTypeName!(self:Fort Item Object Reference=Get Item(), bUsePlural:bool=false) -> (ReturnValue:Text)
   )
   > SetVisibility(self=@n21, InVisibility=Collapsed)
   ResetText(execute=@n10.then_0)
   > n22 = Cast To FortPlaysetItemDefinition!(execute, Object=@n8.ParentDefinition) -> (then, CastFailed, AsFort Playset Item Definition:Fort Playset Item Definition Object Reference) {is=title K2Node_DynamicCast}
   > Set ParentPlaysetDefinition(ParentPlaysetDefinition:Fort Playset Item Definition Object Reference=@n22) -> (Output_Get:Fort Playset Item Definition Object Reference)
   n23 = Get Button_ItemDetailFocus() -> (Button_ItemDetailFocus:WBP UIKit Button Quiet Object Reference)
   > SetVisibility(self=@n23, InVisibility=Visible)
   Set ParentPlaysetDefinition(execute=@n22.CastFailed)
   > SetVisibility(self=@n23, InVisibility=Collapsed)
   n24 = Event SetSelectItemTagsPrompt!() -> (OutputDelegate, then, Prompt:"Text (by ref)") {is=title K2Node_Event}
   > "SetText (Text)"(self=Get Text_ItemTagActionPrompt() -> (Text_ItemTagActionPrompt:Common Text Object Reference), InText=@n24.Prompt)
  ]}
 GenerateItemName {nodes [
   GenerateItemName() {is=title K2Node_FunctionEntry}
   > n1 = Branch(Condition=IsValid(Object=Get ParentPlaysetDefinition()))
   n2 = Get Item()
   > n3 = Cast To FortPlaysetPropItemDefinition!(execute, Object=GetItemDefinition(self=@n2)) -> (then, CastFailed, AsFort Playset Prop Item Definition:Fort Playset Prop Item Definition Object Reference) {is=title K2Node_DynamicCast}
   > Return Node!(
    execute
    DisplayText:Text=GenerateDisplayNameForPlaysetPropDefinition!(PlaysetDefinition:Fort Playset Prop Item Definition Object Reference=@n3) -> (ReturnValue:Text)
   )
   Return Node(execute=[@n3.CastFailed, @n1.else], DisplayText=GetDisplayName(self:Fort Item Object Reference=@n2))
  ]}
 HasFocusableTags {nodes [
   HasFocusableTags() {is=title K2Node_FunctionEntry}
   n1 = Get Wbox_Tags()
   > Return Node!(execute, ReturnValue:bool=AND Boolean(A=HasAnyChildren(self=@n1), B=Get bAllowInteractiveTags() -> (bAllowInteractiveTags:bool)))
  ]}
 OnRequestItemFocus {nodes [OnRequestItemFocus() {is=title K2Node_FunctionEntry}]}
 OnTagSelected {nodes [OnTagSelected!() -> (then, TagID:Name, Selected:bool=~) {is=title K2Node_FunctionEntry}]}
 PopulateFileSize {nodes [
   PopulateFileSize() {is=title K2Node_FunctionEntry}
   n1 = Break Fort Emporium Item Data(FortEmporiumItemData=Get ItemData()) -> (Price=~, bRequiresEntitlement=~)
   > n2 = Set bShowSizeInMB(bShowSizeInMB:bool="integer > integer"(A=@n1.FileSize, B=1000000)) -> (Output_Get:bool)
   n3 = float / float(
    A="To Float (Integer)"(InInt=@n1.FileSize)
    B=Select(Option 0:double=1000, Option 1:double=1000000, Index=Get bShowSizeInMB() -> (bShowSizeInMB:bool)) -> (ReturnValue:double)
   )
   > "SetText (Text)"(
    self=Get ItemFileSize() -> (ItemFileSize:Common Text Object Reference)
    InText=Format Text!(
     Format=Select(Option 0:Text="LOCTEXT(\"{0} KB\")", Option 1:Text="LOCTEXT(\"{0} MB\")", Index=@n2) -> (ReturnValue:Text)
     0:Text="To Text (Float)"(Value=@n3, bUseGrouping=false, MinimumIntegralDigits=1, MaximumIntegralDigits=324, MinimumFractionalDigits=2, MaximumFractionalDigits=2)
    ) -> (Result)
   )
  ]}
 PopulateItemPrice {nodes [
   PopulateItemPrice() {is=title K2Node_FunctionEntry}
   n1 = Break Fort Emporium Item Data(FortEmporiumItemData=Get ItemData()) -> (Price=~, bRequiresEntitlement=~)
   > "SetText (Text)"(
    self=Get ItemPrice() -> (ItemPrice:Common Text Object Reference)
    InText=Select(
     Option 0="To Text (Float)"(
      Value=@n1.Price
      bUseGrouping=false
      MinimumIntegralDigits=1
      MaximumIntegralDigits=324
      MinimumFractionalDigits=2
      MaximumFractionalDigits=2
     )
     Option 1="LOCTEXT(\"Free\")"
     Index="Equal (Float)"(A=@n1.Price)
    )
   )
  ]}
 PopulateTags {nodes [
   n1 = PopulateTags!() -> (then, inTagArray:Array of Names, InAllowInteractiveTags:bool=~) {is=title K2Node_FunctionEntry}
   > Set bAllowInteractiveTags(bAllowInteractiveTags:bool=@n1.InAllowInteractiveTags) -> (Output_Get:bool)
   > n2 = Sequence!(execute) -> (then_0, then_1)
   > n3 = For Each Loop(Array:Array of Widget Object References=GetAllChildren(self=Get Wbox_Tags())) -> (Array Element:Widget Object Reference)
   ClearChildren(execute=@n3.Completed, self=Get Wbox_Tags())
   > Branch(Condition="integer > integer"(A=Length(TargetArray:Array of Names=@n1), B=0) {is=title K2Node_CallFunction})
   > Set OwningPlayer(OwningPlayer:Player Controller Object Reference=GetOwningPlayer()) -> (Output_Get:Player Controller Object Reference)
   > n4 = For Each Loop(Array:Array of Names=@n1) -> (Array Element:Name)
   > n5 = Create Widget!(
    execute:Exec
    Class:User Widget Class Reference=&/WBP_CreativeTag.WBP_CreativeTag_C
    OwningPlayer:Player Controller Object Reference=Get OwningPlayer() -> (OwningPlayer:Player Controller Object Reference)
    Style:Common Button Style Class Reference=/FortUILibrary/UIKit/Buttons/Blocks/ButtonStyle_UIKit_Empty.ButtonStyle_UIKit_Empty_C
    bApplyAlphaOnDisable:bool=False
    bLocked:bool=False
    bSelectable:bool=True
    bShouldSelectUponReceivingFocus:bool=False
    bInteractableWhenSelected:bool=True
    bToggleable:bool=True
    bTriggerClickedAfterSelection:bool=True
    bDisplayInputActionWhenNotInteractable:bool=False
    bHideInputActionWithKeyboard:bool=False
    bShouldUseFallbackDefaultInputAction:bool=True
    bRequiresHold:bool=False
    HoldData:Common UIHold Data Class Reference=&CommonUI_HoldData_Base_C
    ClickMethod:EButtonClickMethod Enum=DownAndUp
    InputPriority:int=0
    TriggeringInputAction:Data Table Row Handle Structure
   ) -> (then:Exec, ReturnValue:WBP Creative Tag Object Reference)
   > SetTag!(execute, self:WBP Creative Tag Object Reference=@n5, TagID:Name=@n4) -> (then)
   > SetIsEnabled(self=@n5, bInIsEnabled=Get bAllowInteractiveTags())
   > SetIcon!(execute, self:WBP UIKit Button Generic Base Object Reference=@n5, Icon:Object Reference) -> (then)
   > SetIconMaterial!(execute, self:WBP UIKit Button Generic Base Object Reference=@n5, Material Instance:Material Instance Object Reference) -> (then)
   > n6 = AddChildToWrapBox!(execute, self:Wrap Box Object Reference=Get Wbox_Tags(), Content:Widget Object Reference=@n5) -> (then, ReturnValue:Wrap Box Slot Object Reference)
   > SetPadding!(execute, self:Wrap Box Slot Object Reference=@n6, InPadding:Margin Structure=Make Margin(Left=3, Top=3, Right=3, Bottom=3)) -> (then)
   > Bind Event to On Tag Selected(self:WBP Creative Tag Object Reference=@n5, Delegate=Create Event()) {is=title K2Node_AddDelegate}
   n7 = Cast To WBP_CreativeTag(execute=@n3.LoopBody, Object=@n3)
   > Unbind all Events from On Tag Selected!(execute:Exec, self:WBP Creative Tag Object Reference=@n7) -> (then:Exec) {is=title K2Node_ClearDelegate}
   UpdateSelectItemTagsPrompt(execute=@n2.then_1)
  ]}
 ResetText {nodes [
   ResetText() {is=title K2Node_FunctionEntry}
   > "SetText (Text)"(
    self=[Get TypeTextBlock(), Get ItemAuthorName(), Get ItemPrice(), Get ItemLicenseType(), Get ItemFileSize(), Get ItemDisplayName()]
    InText=GetEmptyText!() -> (ReturnValue:Text)
   )
   n1 = Get ItemVisibleTags()
   > Clear(TargetArray:Array of Names=@n1)
   > PopulateTags(inTagArray=@n1, InAllowInteractiveTags=false) {is=title K2Node_CallFunction}
  ]}
 UpdateSelectItemTagsPrompt {nodes [
   UpdateSelectItemTagsPrompt() {is=title K2Node_FunctionEntry}
   n1 = HasFocusableTags!(self:Fort Emporium Item Info Object Reference) -> (ReturnValue:bool) {is=title K2Node_CallFunction}
   n2 = Get bShowItemTagsPrompt() -> (bShowItemTagsPrompt:bool)
   > SetVisibility(
    self=Get ItemTagActionPrompt() -> (ItemTagActionPrompt:Horizontal Box Object Reference)
    InVisibility=Select(Option 0:ESlateVisibility Enum=Collapsed, Option 1:ESlateVisibility Enum=HitTestInvisible, Index=AND Boolean(A=@n2, B=@n1)) -> (ReturnValue:ESlateVisibility Enum)
   )
  ]}
]
variables [
 bAllowInteractiveTags {type=boolean}
 bShowItemTagsPrompt {type=boolean}
 Item {type=FortItem, default=null}
 ItemData {
  type=FortEmporiumItemData
  default {
   id=""
   bRequiresEntitlement=false
   baseCurrency=""
   childAssetCount=0
   description=""
   entityType=""
   fileSize=0
   licenseType=Invalid
   materialCount=0
   namespace=""
   polygonCount=0
   price=0
   seller=""
   sketchfabUId=""
   thumbnailURL=""
   title=""
   versePath=""
   vertexCount=0
   assetIds []
   tagList []
  }
 }
 ItemVisibleTags {type="array<string>", default []}
 OnRequestItemFocus {}
 OnTagSelected {}
 ParentPlaysetDefinition {type=FortPlaysetItemDefinition, default=null}
]
widgets {
 info {inheritedWidgetCount=0, parentClass=/Script/CreativeEmporiumUI.FortEmporiumItemInfo, rootWidgetClass=&SizeBox, widgetCount=20}
 widgets [
  SizeBox_Body {
   widgetClassPath=&SizeBox
   bIsVariable=true
   widget=SizeBox_Body
   properties {bOverride_MinDesiredHeight=true, heightOverride=168, minDesiredHeight=104, minDesiredWidth=400, widthOverride=660}
  }
  Overlay_0 {widgetClassPath=&Overlay, parent=SizeBox_Body, widget=Overlay_0, slot {class=SizeBoxSlot}}
  VerticalBox_0 {
   widgetClassPath=&VerticalBox
   parent=Overlay_0
   widget=VerticalBox_0
   slot {class=OverlaySlot, properties {horizontalAlignment=HAlign_Fill, verticalAlignment=VAlign_Fill}}
  }
  Overlay_1 {widgetClassPath=&Overlay, parent=VerticalBox_0, widget=Overlay_1, slot {class=VerticalBoxSlot}}
  Image_Backplate {
   widgetClassPath=&Image
   bIsVariable=true
   parent=Overlay_1
   widget=Image_Backplate
   properties {
    visibility=Collapsed
    brush {
     bIsDynamicallyLoaded=false
     drawAs=Image
     imageType=NoImage
     mirroring=NoMirror
     resourceName=None
     resourceObject=/Game/UI/Inventory/Art/M_UI_Creative_ContentBrowser_ItemInfoBackground
     tiling=NoTile
     imageSize=<32 32>
     margin=<ltrb 0 0 0 0>
     outlineSettings {
      bUseBrushTransparency=false
      roundingType=HalfHeightRadius
      width=0
      color {colorUseRule=UseColor_Specified, specifiedColor=<rgba 0 0 0 0>}
      cornerRadii=<0 0 0 0>
     }
     tintColor {colorUseRule=UseColor_Specified, specifiedColor=<rgba 1 1 1 1>}
     uVRegion {bIsValid=false, max=<0 0>, min=<0 0>}
    }
   }
   slot {class=OverlaySlot, properties {horizontalAlignment=HAlign_Fill, verticalAlignment=VAlign_Fill}}
  }
  Image_Underline {
   widgetClassPath=&Image
   bIsVariable=true
   parent=Overlay_1
   widget=Image_Underline
   properties {
    visibility=Collapsed
    brush {
     bIsDynamicallyLoaded=false
     drawAs=Image
     imageType=NoImage
     mirroring=NoMirror
     resourceName=None
     resourceObject=null
     tiling=NoTile
     imageSize=<32 3>
     margin=<ltrb 0 0 0 0>
     outlineSettings {
      bUseBrushTransparency=false
      roundingType=HalfHeightRadius
      width=0
      color {colorUseRule=UseColor_Specified, specifiedColor=<rgba 0 0 0 0>}
      cornerRadii=<0 0 0 0>
     }
     tintColor {colorUseRule=UseColor_Specified, specifiedColor=<rgba 1 1 1 1>}
     uVRegion {bIsValid=false, max=<0 0>, min=<0 0>}
    }
    colorAndOpacity=<rgba 0.1329 0.6038 1 1>
   }
   slot {class=OverlaySlot, properties {horizontalAlignment=HAlign_Fill, verticalAlignment=VAlign_Bottom, padding=<ltrb 5 0 0 0>}}
  }
  VerticalBox_2 {
   widgetClassPath=&VerticalBox
   parent=Overlay_1
   widget=VerticalBox_2
   slot {class=OverlaySlot, properties {horizontalAlignment=HAlign_Fill, verticalAlignment=VAlign_Fill, padding=<ltrb 24 6 16 16>}}
  }
  TypeTextBlock {
   widgetClassPath=&CommonTextBlock
   bIsVariable=true
   parent=VerticalBox_2
   widget=TypeTextBlock
   properties {
    autoWrapText=true
    bAutoCollapseWithEmptyText=true
    scrollStyle=&TextScrollStyle-Base_C
    text="{TypeName}"
    colorAndOpacity {colorUseRule=UseColor_Specified, specifiedColor=<rgba 1 1 1 0.6>}
    font {
     bForceMonospaced=false
     bMaterialIsStencil=false
     fontMaterial=null
     fontObject=&HeadingNow
     letterSpacing=0
     monospacedWidth=1
     size=15.75
     skewAmount=0
     typefaceFontName=75 Medium
     outlineSettings {
      bApplyOutlineToDropShadows=false
      bMiteredCorners=false
      bSeparateFillAlpha=false
      outlineMaterial=null
      outlineSize=0
      outlineColor=<rgba 0 0 0 1>
     }
    }
    shadowOffset=<0 0>
   }
   slot {class=VerticalBoxSlot}
  }
  ItemDisplayName {
   widgetClassPath=&CommonTextBlock
   bIsVariable=true
   parent=VerticalBox_2
   widget=ItemDisplayName
   properties {
    autoWrapText=true
    bAutoCollapseWithEmptyText=true
    bIsScrollingEnabled=false
    text="{ItemName}"
    colorAndOpacity {colorUseRule=UseColor_Specified, specifiedColor=<rgba 1 1 1 0.9>}
    font {
     bForceMonospaced=false
     bMaterialIsStencil=false
     fontMaterial=null
     fontObject=&HeadingNow
     letterSpacing=0
     monospacedWidth=1
     size=22.5
     skewAmount=0
     typefaceFontName=75 Medium Italic
     outlineSettings {
      bApplyOutlineToDropShadows=false
      bMiteredCorners=false
      bSeparateFillAlpha=false
      outlineMaterial=null
      outlineSize=0
      outlineColor=<rgba 0 0 0 1>
     }
    }
    shadowOffset=<0 0>
   }
   slot {class=VerticalBoxSlot, properties {padding=<ltrb 0 0 0 24>}}
  }
  ItemAuthorName {
   widgetClassPath=&CommonTextBlock
   bIsVariable=true
   parent=VerticalBox_2
   widget=ItemAuthorName
   properties {
    bAutoCollapseWithEmptyText=true
    scrollStyle=&TextScrollStyle-Base_C
    text="{ItemAuthor}"
    colorAndOpacity {colorUseRule=UseColor_Specified, specifiedColor=<rgba 1 1 1 0.7>}
    font {
     bForceMonospaced=false
     bMaterialIsStencil=false
     fontMaterial=null
     fontObject=&HeadingNow
     letterSpacing=0
     monospacedWidth=1
     size=15.75
     skewAmount=0
     typefaceFontName=74 Regular
     outlineSettings {
      bApplyOutlineToDropShadows=false
      bMiteredCorners=false
      bSeparateFillAlpha=false
      outlineMaterial=null
      outlineSize=0
      outlineColor=<rgba 0 0 0 1>
     }
    }
    shadowOffset=<0 0>
   }
   slot {class=VerticalBoxSlot, properties {padding=<ltrb 0 0 0 12>}}
  }
  ItemPrice {
   widgetClassPath=&CommonTextBlock
   bIsVariable=true
   parent=VerticalBox_2
   widget=ItemPrice
   properties {
    bAutoCollapseWithEmptyText=true
    scrollStyle=&TextScrollStyle-Base_C
    text="{ItemPrice}"
    colorAndOpacity {colorUseRule=UseColor_Specified, specifiedColor=<rgba 1 1 1 0.7>}
    font {
     bForceMonospaced=false
     bMaterialIsStencil=false
     fontMaterial=null
     fontObject=&HeadingNow
     letterSpacing=0
     monospacedWidth=1
     size=15.75
     skewAmount=0
     typefaceFontName=75 Medium
     outlineSettings {
      bApplyOutlineToDropShadows=false
      bMiteredCorners=false
      bSeparateFillAlpha=false
      outlineMaterial=null
      outlineSize=0
      outlineColor=<rgba 0 0 0 1>
     }
    }
    shadowOffset=<0 0>
   }
   slot {class=VerticalBoxSlot, properties {padding=<ltrb 0 0 0 12>}}
  }
  ItemLicenseType {
   widgetClassPath=&CommonTextBlock
   bIsVariable=true
   parent=VerticalBox_2
   widget=ItemLicenseType
   properties {
    bAutoCollapseWithEmptyText=true
    scrollStyle=&TextScrollStyle-Base_C
    text="{ItemLicenseType}"
    colorAndOpacity {colorUseRule=UseColor_Specified, specifiedColor=<rgba 1 1 1 0.7>}
    font {
     bForceMonospaced=false
     bMaterialIsStencil=false
     fontMaterial=null
     fontObject=&HeadingNow
     letterSpacing=0
     monospacedWidth=1
     size=15.75
     skewAmount=0
     typefaceFontName=74 Regular
     outlineSettings {
      bApplyOutlineToDropShadows=false
      bMiteredCorners=false
      bSeparateFillAlpha=false
      outlineMaterial=null
      outlineSize=0
      outlineColor=<rgba 0 0 0 1>
     }
    }
    shadowOffset=<0 0>
   }
   slot {class=VerticalBoxSlot, properties {padding=<ltrb 0 0 0 12>}}
  }
  ItemFileSize {
   widgetClassPath=&CommonTextBlock
   bIsVariable=true
   parent=VerticalBox_2
   widget=ItemFileSize
   properties {
    bAutoCollapseWithEmptyText=true
    scrollStyle=&TextScrollStyle-Base_C
    text="{ItemFileSize}"
    colorAndOpacity {colorUseRule=UseColor_Specified, specifiedColor=<rgba 1 1 1 0.7>}
    font {
     bForceMonospaced=false
     bMaterialIsStencil=false
     fontMaterial=null
     fontObject=&HeadingNow
     letterSpacing=0
     monospacedWidth=1
     size=15.75
     skewAmount=0
     typefaceFontName=74 Regular
     outlineSettings {
      bApplyOutlineToDropShadows=false
      bMiteredCorners=false
      bSeparateFillAlpha=false
      outlineMaterial=null
      outlineSize=0
      outlineColor=<rgba 0 0 0 1>
     }
    }
    shadowOffset=<0 0>
   }
   slot {class=VerticalBoxSlot, properties {padding=<ltrb 0 0 0 12>}}
  }
  SizeBox_Description {
   widgetClassPath=&SizeBox
   bIsVariable=true
   parent=VerticalBox_2
   widget=SizeBox_Description
   properties {bOverride_MaxDesiredHeight=true, clipping=ClipToBoundsAlways, maxDesiredHeight=300}
   slot {class=VerticalBoxSlot}
  }
  Text_Description {
   widgetClassPath=&CommonTextBlock
   bIsVariable=true
   parent=SizeBox_Description
   widget=Text_Description
   properties {
    autoWrapText=true
    bAutoCollapseWithEmptyText=true
    text="{Description}"
    textOverflowPolicy=Ellipsis
    visibility=HitTestInvisible
    colorAndOpacity {colorUseRule=UseColor_Specified, specifiedColor=<rgba 1 1 1 0.8>}
    font {
     bForceMonospaced=false
     bMaterialIsStencil=false
     fontMaterial=null
     fontObject=&HeadingNow
     letterSpacing=0
     monospacedWidth=1
     size=15.75
     skewAmount=0
     typefaceFontName=74 Regular
     outlineSettings {
      bApplyOutlineToDropShadows=false
      bMiteredCorners=false
      bSeparateFillAlpha=false
      outlineMaterial=null
      outlineSize=0
      outlineColor=<rgba 0 0 0 1>
     }
    }
    shadowOffset=<0 0>
   }
   slot {class=SizeBoxSlot}
  }
  Button_ItemDetailFocus {
   widgetClassPath=/FortUILibrary/UIKit/Buttons/WBP_UIKit_Button_Quiet
   bIsVariable=true
   parent=VerticalBox_0
   widget=Button_ItemDetailFocus
   properties {text=Go To Item, triggeringInputAction {dataTable=/FortUILibrary/Input/CreativeInputActionDataTable, rowName=ItemDetails}}
   slot {class=VerticalBoxSlot, properties {horizontalAlignment=HAlign_Left, verticalAlignment=VAlign_Center, padding=<ltrb 24 0 0 0>}}
  }
  ItemTagActionPrompt {
   widgetClassPath=&HorizontalBox
   bIsVariable=true
   parent=VerticalBox_0
   widget=ItemTagActionPrompt
   slot {class=VerticalBoxSlot, properties {horizontalAlignment=HAlign_Left, padding=<ltrb 24 0 0 8>}}
  }
  Image_ItemTagActionPrompt {
   widgetClassPath=&Image
   bIsVariable=true
   parent=ItemTagActionPrompt
   widget=Image_ItemTagActionPrompt
   slot {class=HorizontalBoxSlot, properties {horizontalAlignment=HAlign_Left}}
  }
  Text_ItemTagActionPrompt {
   widgetClassPath=&CommonTextBlock
   bIsVariable=true
   parent=ItemTagActionPrompt
   widget=Text_ItemTagActionPrompt
   properties {
    bAutoCollapseWithEmptyText=true
    text=Select Tags
    textTransformPolicy=ToUpper
    colorAndOpacity {colorUseRule=UseColor_Specified, specifiedColor=<rgba 1 1 1 0.7>}
    font {
     bForceMonospaced=false
     bMaterialIsStencil=false
     fontMaterial=null
     fontObject=&HeadingNow
     letterSpacing=0
     monospacedWidth=1
     size=13.5
     skewAmount=0
     typefaceFontName=75 Medium
     outlineSettings {
      bApplyOutlineToDropShadows=false
      bMiteredCorners=false
      bSeparateFillAlpha=false
      outlineMaterial=null
      outlineSize=0
      outlineColor=<rgba 0 0 0 1>
     }
    }
    shadowOffset=<0 0>
   }
   slot {class=HorizontalBoxSlot, properties {horizontalAlignment=HAlign_Center, verticalAlignment=VAlign_Center, padding=<ltrb 6 0 0 0>}}
  }
  Wbox_Tags {
   widgetClassPath=&WrapBox
   bIsVariable=true
   parent=VerticalBox_0
   widget=Wbox_Tags
   properties {navigation=&self:WidgetTree.Wbox_Tags.WidgetNavigation_0, innerSlotPadding=<2 2>}
   slot {class=VerticalBoxSlot, properties {verticalAlignment=VAlign_Top, padding=<ltrb 12 0 0 4>, size {sizeRule=Fill, value=1}}}
  }
 ]
}
