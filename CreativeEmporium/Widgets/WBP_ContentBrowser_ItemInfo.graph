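# Creative content browser item info panel, built on AthenaInventoryItemInfo, with a header for name, rarity and category above a description and stats panel.
# Parent events switch empty and filled states, toggle category visibility, and hide rarity for resources, ammo, augments or invalid items.
# Rarity data tints text, glow and dynamic materials on the header and icon.
# Extension widgets are routed by slot into the header or expanded-panel horizontal box.
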WidgetBlueprint /CreativeEmporium/Widgets/WBP_ContentBrowser_ItemInfo
graphs [
 AddExtensionWidget {nodes [
   n1 = AddExtensionWidget!() -> (then, InExtensionSlot:EUIExtensionSlot Enum=Invalid, InExtensionWidget:User Widget Object Reference) {is=title K2Node_FunctionEntry}
   > Set ExtensionWidgetToAdd(ExtensionWidgetToAdd:User Widget Object Reference=@n1.InExtensionWidget) -> (Output_Get:User Widget Object Reference)
   > n2 = Branch(Condition="Equal (Enum)"(A:EUIExtensionSlot Enum=@n1, B:EUIExtensionSlot Enum=InventoryItemInfoInspectPreview))
   > n3 = AddChildToHorizontalBox!(
    execute
    self:Horizontal Box Object Reference=Get HBoxExtensionWidgets_InspectPreview() -> (HBoxExtensionWidgets_InspectPreview:Horizontal Box Object Reference)
    Content:Widget Object Reference=Get ExtensionWidgetToAdd() -> (ExtensionWidgetToAdd:User Widget Object Reference)
   ) -> (then, ReturnValue:Horizontal Box Slot Object Reference)
   > SetHorizontalAlignment(self:Horizontal Box Slot Object Reference=@n3, InHorizontalAlignment=HAlign_Fill)
   > SetVerticalAlignment(self:Horizontal Box Slot Object Reference=@n3, InVerticalAlignment=VAlign_Fill)
   Branch(execute=@n2.else, Condition="Equal (Enum)"(A=@n1, B=InventoryItemInfoContents))
   > n4 = AddChildToHorizontalBox(
    self=Get HBoxExtensionWidgets_ExtendedInspect() -> (HBoxExtensionWidgets_ExtendedInspect:Horizontal Box Object Reference)
    Content=Get ExtensionWidgetToAdd()
   )
   > SetHorizontalAlignment(self=@n4, InHorizontalAlignment=HAlign_Fill)
   > SetVerticalAlignment(self=@n4, InVerticalAlignment=VAlign_Fill)
  ]}
 EventGraph {nodes [
   Event PreConstruct()
   > n1 = GetDynamicMaterial(self=Get Image_ItemIcon!(self:Athena Item Icon Object Reference=Get ItemIcon() -> (ItemIcon:Athena Item Icon Object Reference)) -> (Image_ItemIcon:Image Object Reference))
   > SetScalarParameterValue(self=@n1, ParameterName=IsInfoPanelPresentation, Value=1)
   n2 = Event OnItemRarityUpdated!() -> (OutputDelegate, then, ItemRarity:EFortRarity Enum=Common, RarityItemData:Fort Rarity Item Data Structure) {is=title K2Node_Event}
   > SetRarityColors!(
    execute
    self
    InItemRarity:EFortRarity Enum=@n2.ItemRarity
    InFortRarityItemData:"Fort Rarity Item Data Structure (by ref)"=@n2.RarityItemData
   ) -> (then)
   n3 = Event OnPostItemToRepresentChanged!() -> (OutputDelegate, then, ItemToRepresent:Fort Item Object Reference) {is=title K2Node_Event}
   > SetActiveWidget(
    self=Get SwitcherInfoContent() -> (SwitcherInfoContent:Common Visibility Switcher Object Reference)
    Widget=Select(
     Option 0:Widget Object Reference=Get Overlay_EmptyState() -> (Overlay_EmptyState:Overlay Object Reference)
     Option 1:Widget Object Reference=Get Overlay_FilledState() -> (Overlay_FilledState:Overlay Object Reference)
     Index=IsValid(Object=@n3.ItemToRepresent)
    ) -> (ReturnValue:Widget Object Reference)
   )
   > UpdateRarityInfoVisibility!(execute, self, Item:Fort Item Object Reference=@n3.ItemToRepresent) -> (then)
   > UpdateRechargeIcon!(execute, self:Athena Inventory Item Info Object Reference, ItemToRepresent:Fort Item Object Reference=@n3.ItemToRepresent) -> (then)
   n4 = Event OnItemCategoryIconUpdated!() -> (OutputDelegate, then, bHasCategoryIcon:bool) {is=title K2Node_Event}
   > SetVisibility(
    self=Select(
     Option 0=Get Image_ItemCategory() -> (Image_ItemCategory:Image Object Reference)
     Option 1=Get Border_ItemType() -> (Border_ItemType:Border Object Reference)
     Index=Get bHideEntireCategoryWhenIconMissing() -> (bHideEntireCategoryWhenIconMissing:bool)
    )
    InVisibility=Select(Option 0:ESlateVisibility Enum=Collapsed, Option 1:ESlateVisibility Enum=SelfHitTestInvisible, Index=@n4.bHasCategoryIcon) -> (ReturnValue:ESlateVisibility Enum)
   )
   n5 = Event OnExtensionWidgetAdded!() -> (OutputDelegate, then, ExtensionSlot:EUIExtensionSlot Enum=Invalid, ExtensionWidget:User Widget Object Reference) {is=title K2Node_Event}
   > AddExtensionWidget!(execute, self, InExtensionSlot:EUIExtensionSlot Enum=@n5.ExtensionSlot, InExtensionWidget:User Widget Object Reference=@n5.ExtensionWidget) -> (then) {is=title K2Node_CallFunction}
  ]}
 SetRarityColors {nodes [
   n1 = SetRarityColors!() -> (then, InItemRarity:EFortRarity Enum=Common, InFortRarityItemData:"Fort Rarity Item Data Structure (by ref)") {is=title K2Node_FunctionEntry}
   > n2 = Set FortRarityItemData(FortRarityItemData:Fort Rarity Item Data Structure=@n1.InFortRarityItemData) -> (Output_Get:Fort Rarity Item Data Structure)
   n3 = Break Fort Rarity Item Data!(FortRarityItemData:"Fort Rarity Item Data Structure (by ref)"=@n2) -> (
    Name:Text
    Color1:Linear Color Structure
    Color2:Linear Color Structure
    Color3:Linear Color Structure
    Color4:Linear Color Structure
    Color5:Linear Color Structure
    Radius:float=~
    Falloff:float=~
    Brightness:float=~
    Roughness:float=~
    Glow:float=~
   ) {is=title K2Node_BreakStruct}
   > SetRarityGlowColors!(execute, self, Specified Color:Linear Color Structure=@n3.Color3, Index:EFortRarity Enum=@n1) -> (then)
   > SetColorAndOpacity(self=Get Text_ItemRarity() -> (Text_ItemRarity:Common Text Object Reference), InColorAndOpacity=Make Slate Color(SpecifiedColor=@n3.Color1))
   > n4 = GetDynamicMaterial(self=Get Image_HeaderBlackBacking() -> (Image_HeaderBlackBacking:Image Object Reference))
   > SetVectorParameterValue(self=@n4, ParameterName=Color_Stop, Value=@n3.Color3)
   > n5 = GetDynamicMaterial(self=Get Image_ItemIcon(self=Get ItemIcon()))
   n6 = Break Fort Rarity Item Data(FortRarityItemData=Get FortRarityItemData() -> (FortRarityItemData:Fort Rarity Item Data Structure)) -> (Radius=~, Falloff=~, Brightness=~, Roughness=~, Glow=~)
   > SetVectorParameterValue(self=@n5, ParameterName=InfoPanel_Color_A, Value=@n6.Color2)
   > SetVectorParameterValue(self=@n5, ParameterName=InfoPanel_Color_B, Value=@n6.Color1)
  ]}
 SetRarityGlowColors {nodes [
   n1 = SetRarityGlowColors!() -> (then, Specified Color:Linear Color Structure, Index:EFortRarity Enum=Common) {is=title K2Node_FunctionEntry}
   > SetBrushTintColor!(
    execute
    self:Image Object Reference=Get Image_RarityGlow() -> (Image_RarityGlow:Image Object Reference)
    TintColor:Slate Color Structure=Make Slate Color(SpecifiedColor=@n1)
   ) -> (then)
   > SetOpacity!(
    execute
    self:Image Object Reference=Get Image_RarityGlow()
    InOpacity:float=Select!(
     Common:double=0
     Uncommon:double=0.125
     Rare:double=0.25
     Epic:double=0.333
     Legendary:double=0.666
     Mythic:double=0.666
     Transcendent:double=0.5
     Unattainable:double=0.5
     Index:EFortRarity Enum=@n1.Index
    ) -> (ReturnValue:double)
   ) -> (then)
  ]}
 UpdateRarityInfoVisibility {nodes [
   n1 = UpdateRarityInfoVisibility!() -> (then, Item:Fort Item Object Reference) {is=title K2Node_FunctionEntry}
   > n2 = Is Valid(InputObject=@n1)
   n3 = GetItemDefinition!(self:Fort Item Object Reference=@n1) -> (ReturnValue:Fort Item Definition Object Reference)
   > n4 = Is Valid(InputObject=@n3)
   n5 = GetItemType!(self:Fort Item Definition Object Reference=@n3) -> (ReturnValue:EFortItemType Enum=WorldItem)
   n6 = "Equal (Enum)"(A:EFortItemType Enum=@n5, B:EFortItemType Enum=WorldResource)
   n7 = "Equal (Enum)"(A=@n5, B=PlayerAugment)
   n8 = "Equal (Enum)"(A=@n5, B=Ammo)
   > SetVisibility(
    self=Get HBox_RarityInfo() -> (HBox_RarityInfo:Horizontal Box Object Reference)
    InVisibility=Select!(
     Option 0:ESlateVisibility Enum=SelfHitTestInvisible
     Option 1:ESlateVisibility Enum=Collapsed
     Index:bool=OR Boolean!(A=@n6, B=@n8, C:bool=@n7) -> (ReturnValue)
    ) -> (ReturnValue:ESlateVisibility Enum)
   )
   SetVisibility(execute=[@n4.Is Not Valid, @n2.Is Not Valid], self=Get HBox_RarityInfo(), InVisibility=Collapsed)
  ]}
]
variables [bHideEntireCategoryWhenIconMissing {type=boolean}]
widgets {
 info {inheritedWidgetCount=13, parentClass=/Script/FortniteUI.AthenaInventoryItemInfo, rootWidgetClass=&VerticalBox, widgetCount=46}
 widgets [
  VerticalBox_0 {widgetClassPath=&VerticalBox, widget=VerticalBox_0}
  SizeBox_Header {
   widgetClassPath=&SizeBox
   bIsVariable=true
   parent=VerticalBox_0
   widget=SizeBox_Header
   properties {bOverride_MinDesiredHeight=true, heightOverride=88, minDesiredHeight=88, widthOverride=688, renderTransformPivot=<0.5 1>}
   slot {class=VerticalBoxSlot, properties {verticalAlignment=VAlign_Bottom}}
  }
  Overlay_InspectPreview {widgetClassPath=&Overlay, parent=SizeBox_Header, widget=Overlay_InspectPreview, slot {class=SizeBoxSlot}}
  BackgroundBlur_100 {
   widgetClassPath=&BackgroundBlur
   parent=Overlay_InspectPreview
   widget=BackgroundBlur_100
   properties {
    blurStrength=2
    lowQualityFallbackBrush {
     bIsDynamicallyLoaded=false
     drawAs=Image
     imageType=NoImage
     mirroring=NoMirror
     resourceName=None
     resourceObject=/Game/Athena/HUD/Inventory/Art/M_UI_InventoryPanel_Backing
     tiling=NoTile
     imageSize=<10 10>
     margin=<ltrb 0 0 0 0>
     outlineSettings {
      bUseBrushTransparency=false
      roundingType=HalfHeightRadius
      width=0
      color {colorUseRule=UseColor_Specified, specifiedColor=<rgba 0 0 0 0>}
      cornerRadii=<0 0 0 0>
     }
     tintColor {colorUseRule=UseColor_Specified, specifiedColor=<rgba 1 1 1 1>}
     uVRegion {bIsValid=false, max=<0 0>, min=<0 0>}
    }
   }
   slot {class=OverlaySlot, properties {horizontalAlignment=HAlign_Fill, verticalAlignment=VAlign_Fill}}
  }
  Image_HeaderBlackBacking {
   widgetClassPath=&Image
   bIsVariable=true
   parent=Overlay_InspectPreview
   widget=Image_HeaderBlackBacking
   properties {
    visibility=SelfHitTestInvisible
    brush {
     bIsDynamicallyLoaded=false
     drawAs=Box
     imageType=NoImage
     mirroring=NoMirror
     resourceName=None
     resourceObject=/Game/Athena/HUD/Inventory/Art/M_UI_HUD_InventoryTooltip_Backing
     tiling=NoTile
     imageSize=<16 16>
     margin=<ltrb 0.5 0.5 0.5 0.5>
     outlineSettings {
      bUseBrushTransparency=false
      roundingType=HalfHeightRadius
      width=0
      color {colorUseRule=UseColor_Specified, specifiedColor=<rgba 0 0 0 0>}
      cornerRadii=<0 0 0 0>
     }
     tintColor {colorUseRule=UseColor_Specified, specifiedColor=<rgba 1 1 1 1>}
     uVRegion {bIsValid=false, max=<0 0>, min=<0 0>}
    }
   }
   slot {class=OverlaySlot, properties {horizontalAlignment=HAlign_Fill, verticalAlignment=VAlign_Fill}}
  }
  SwitcherInfoContent {
   widgetClassPath=&CommonVisibilitySwitcher
   bIsVariable=true
   parent=Overlay_InspectPreview
   widget=SwitcherInfoContent
   slot {class=OverlaySlot, properties {horizontalAlignment=HAlign_Fill, verticalAlignment=VAlign_Fill}}
  }
  Overlay_FilledState {
   widgetClassPath=&Overlay
   bIsVariable=true
   parent=SwitcherInfoContent
   widget=Overlay_FilledState
   slot {class=CommonVisibilitySwitcherSlot}
  }
  SizeBox_ItemIcon {
   widgetClassPath=&SizeBox
   parent=Overlay_FilledState
   widget=SizeBox_ItemIcon
   properties {
    bOverride_HeightOverride=true
    bOverride_WidthOverride=true
    flowDirectionPreference=Culture
    heightOverride=106
    renderOpacity=0
    widthOverride=136
   }
   slot {class=OverlaySlot}
  }
  ItemIcon {
   widgetClassPath=/Game/Athena/HUD/Inventory/AthenaItemIconLoadout
   bInherited=true
   parent=SizeBox_ItemIcon
   widget=ItemIcon
   properties {bShowAmmoIcon=false, bShowItemTier=false, bShowStackCountText=false}
   slot {class=SizeBoxSlot}
  }
  VerticalBox_2 {
   widgetClassPath=&VerticalBox
   parent=Overlay_FilledState
   widget=VerticalBox_2
   slot {class=OverlaySlot, properties {horizontalAlignment=HAlign_Fill, verticalAlignment=VAlign_Center, padding=<ltrb 16 8 12 8>}}
  }
  HorizontalBox_139 {widgetClassPath=&HorizontalBox, parent=VerticalBox_2, widget=HorizontalBox_139, slot {class=VerticalBoxSlot}}
  IconAreaScale {
   widgetClassPath=&Border
   parent=HorizontalBox_139
   widget=IconAreaScale
   properties {
    visibility=Collapsed
    background {
     bIsDynamicallyLoaded=false
     drawAs=NoDrawType
     imageType=NoImage
     mirroring=NoMirror
     resourceName=None
     resourceObject=null
     tiling=NoTile
     imageSize=<32 32>
     margin=<ltrb 0 0 0 0>
     outlineSettings {
      bUseBrushTransparency=false
      roundingType=HalfHeightRadius
      width=0
      color {colorUseRule=UseColor_Specified, specifiedColor=<rgba 0 0 0 0>}
      cornerRadii=<0 0 0 0>
     }
     tintColor {colorUseRule=UseColor_Specified, specifiedColor=<rgba 1 1 1 1>}
     uVRegion {bIsValid=false, max=<0 0>, min=<0 0>}
    }
    desiredSizeScale=<1 1.875>
    padding=<ltrb 0 0 0 0>
    renderTransformPivot=<0 0.5>
   }
   slot {class=HorizontalBoxSlot}
  }
  InfoPaddingAmount {
   widgetClassPath=&SizeBox
   parent=IconAreaScale
   widget=InfoPaddingAmount
   properties {bOverride_WidthOverride=true, widthOverride=88}
   slot {class=BorderSlot, properties {padding=<ltrb 0 0 0 0>}}
  }
  VerticalBox_3 {
   widgetClassPath=&VerticalBox
   parent=HorizontalBox_139
   widget=VerticalBox_3
   slot {class=HorizontalBoxSlot, properties {size {sizeRule=Fill, value=1}}}
  }
  VerticalBox_98 {widgetClassPath=&VerticalBox, parent=VerticalBox_3, widget=VerticalBox_98, slot {class=VerticalBoxSlot}}
  Text_ItemName {
   widgetClassPath=&CommonTextBlock
   bInherited=true
   parent=VerticalBox_98
   widget=Text_ItemName
   properties {
    autoWrapText=true
    lineHeightPercentage=0.95
    mobileFontSizeMultiplier=1.2
    scrollStyle=/FortUILibrary/Styles/Typography/Foundation/TextScrollStyle-NoFade.TextScrollStyle-NoFade_C
    text="{Item Name}\r\nItem name on a second line"
    textTransformPolicy=ToUpper
    font {
     bForceMonospaced=false
     bMaterialIsStencil=false
     fontMaterial=null
     fontObject=&Burbank
     letterSpacing=0
     monospacedWidth=1
     size=18
     skewAmount=0.26
     typefaceFontName=BurbankRegular-Black
     outlineSettings {
      bApplyOutlineToDropShadows=false
      bMiteredCorners=false
      bSeparateFillAlpha=false
      outlineMaterial=null
      outlineSize=0
      outlineColor=<rgba 0 0 0 0>
     }
    }
    margin=<ltrb 0 -2 0 0>
    shadowOffset=<0 0>
   }
   slot {class=VerticalBoxSlot, properties {verticalAlignment=VAlign_Center}}
  }
  HBox_RarityInfo {
   widgetClassPath=&HorizontalBox
   bIsVariable=true
   parent=VerticalBox_98
   widget=HBox_RarityInfo
   slot {class=VerticalBoxSlot, properties {horizontalAlignment=HAlign_Left, verticalAlignment=VAlign_Bottom, size {sizeRule=Fill, value=1}}}
  }
  RarityBadge {
   widgetClassPath=&ScaleBox
   parent=HBox_RarityInfo
   widget=RarityBadge
   properties {stretch=UserSpecified, userSpecifiedScale=1.25}
   slot {class=HorizontalBoxSlot, properties {verticalAlignment=VAlign_Center}}
  }
  ItemTierWidget {
   widgetClassPath=/Game/Athena/HUD/Inventory/AthenaItemTierWidget
   bInherited=true
   parent=RarityBadge
   widget=ItemTierWidget
   slot {class=ScaleBoxSlot}
  }
  Overlay_0 {
   widgetClassPath=&Overlay
   parent=HBox_RarityInfo
   widget=Overlay_0
   slot {class=HorizontalBoxSlot, properties {verticalAlignment=VAlign_Center, padding=<ltrb 8 0 0 1>}}
  }
  Text_ItemRarity {
   widgetClassPath=&CommonTextBlock
   bInherited=true
   parent=Overlay_0
   widget=Text_ItemRarity
   properties {
    mobileFontSizeMultiplier=1.2
    scrollStyle=&TextScrollStyle-NoFade_C
    text="{(Rarity)}"
    textTransformPolicy=ToUpper
    colorAndOpacity {colorUseRule=UseColor_Specified, specifiedColor=<rgba 0 1 0.9647 1>}
    font {
     bForceMonospaced=false
     bMaterialIsStencil=false
     fontMaterial=null
     fontObject=&Burbank
     letterSpacing=0
     monospacedWidth=1
     size=16
     skewAmount=0
     typefaceFontName=Burbank-Black
     outlineSettings {
      bApplyOutlineToDropShadows=false
      bMiteredCorners=false
      bSeparateFillAlpha=false
      outlineMaterial=null
      outlineSize=0
      outlineColor=<rgba 0 0 0 1>
     }
    }
    shadowOffset=<0 0>
   }
   slot {class=OverlaySlot, properties {horizontalAlignment=HAlign_Fill, verticalAlignment=VAlign_Fill}}
  }
  Image_RarityGlow {
   widgetClassPath=&Image
   bIsVariable=true
   parent=Overlay_0
   widget=Image_RarityGlow
   properties {
    visibility=SelfHitTestInvisible
    brush {
     bIsDynamicallyLoaded=false
     drawAs=Image
     imageType=NoImage
     mirroring=NoMirror
     resourceName=None
     resourceObject=/Game/Athena/HUD/LTM/Wax/Materials/M_UI_TextGlow_Add
     tiling=NoTile
     imageSize=<32 18>
     margin=<ltrb 0 0 0 0>
     outlineSettings {
      bUseBrushTransparency=false
      roundingType=HalfHeightRadius
      width=0
      color {colorUseRule=UseColor_Specified, specifiedColor=<rgba 0 0 0 0>}
      cornerRadii=<0 0 0 0>
     }
     tintColor {colorUseRule=UseColor_Specified, specifiedColor=<rgba 0 0.184 0.521 1>}
     uVRegion {bIsValid=false, max=<0 0>, min=<0 0>}
    }
    renderTransform {angle=0, scale=<1.125 1.75>, shear=<0 0>, translation=<0 0>}
   }
   slot {class=OverlaySlot, properties {horizontalAlignment=HAlign_Fill, verticalAlignment=VAlign_Center}}
  }
  SizeBox_0 {
   widgetClassPath=&SizeBox
   parent=VerticalBox_3
   widget=SizeBox_0
   properties {bOverride_MaxDesiredWidth=true, maxDesiredWidth=300}
   slot {class=VerticalBoxSlot}
  }
  VerticalBox_217 {widgetClassPath=&VerticalBox, parent=SizeBox_0, widget=VerticalBox_217, slot {class=SizeBoxSlot}}
  Border_ItemType {
   widgetClassPath=&Border
   bInherited=true
   bIsVariable=true
   parent=VerticalBox_217
   widget=Border_ItemType
   properties {
    background {
     bIsDynamicallyLoaded=false
     drawAs=Box
     imageType=NoImage
     mirroring=NoMirror
     resourceName=None
     resourceObject=/Game/Athena/HUD/Inventory/Art/M_UI_InventoryInfo_ItemType_Backing
     tiling=NoTile
     imageSize=<8 8>
     margin=<ltrb 0.5 0.5 0.5 0.5>
     outlineSettings {
      bUseBrushTransparency=false
      roundingType=HalfHeightRadius
      width=0
      color {colorUseRule=UseColor_Specified, specifiedColor=<rgba 0 0 0 0>}
      cornerRadii=<0 0 0 0>
     }
     tintColor {colorUseRule=UseColor_Specified, specifiedColor=<rgba 1 1 1 1>}
     uVRegion {bIsValid=false, max=<0 0>, min=<0 0>}
    }
    padding=<ltrb 8 4 4 4>
   }
   slot {class=VerticalBoxSlot, properties {horizontalAlignment=HAlign_Left, verticalAlignment=VAlign_Top}}
  }
  PlaceholderItemType {
   widgetClassPath=&HorizontalBox
   parent=Border_ItemType
   widget=PlaceholderItemType
   slot {class=BorderSlot, properties {padding=<ltrb 8 4 4 4>}}
  }
  Text_ItemCategory {
   widgetClassPath=&CommonTextBlock
   bInherited=true
   parent=PlaceholderItemType
   widget=Text_ItemCategory
   properties {
    bAutoCollapseWithEmptyText=true
    mobileFontSizeMultiplier=1.2
    renderOpacity=0.75
    scrollStyle=&TextScrollStyle-NoFade_C
    text="{Item Type}"
    font {
     bForceMonospaced=false
     bMaterialIsStencil=false
     fontMaterial=null
     fontObject=&BurbankSmall
     letterSpacing=0
     monospacedWidth=1
     size=13
     skewAmount=0
     typefaceFontName=Bold
     outlineSettings {
      bApplyOutlineToDropShadows=false
      bMiteredCorners=false
      bSeparateFillAlpha=false
      outlineMaterial=null
      outlineSize=0
      outlineColor=<rgba 0 0 0 1>
     }
    }
    shadowOffset=<0 0>
   }
   slot {class=HorizontalBoxSlot}
  }
  Image_ItemCategory {
   widgetClassPath=&Image
   bInherited=true
   bIsVariable=true
   parent=PlaceholderItemType
   widget=Image_ItemCategory
   properties {
    renderOpacity=0.75
    visibility=SelfHitTestInvisible
    brush {
     bIsDynamicallyLoaded=false
     drawAs=Image
     imageType=NoImage
     mirroring=NoMirror
     resourceName=None
     resourceObject=/Game/Athena/HUD/Inventory/Art/M_UI_Inventory_Loadout_AmmoIcon
     tiling=NoTile
     imageSize=<22 22>
     margin=<ltrb 0 0 0 0>
     outlineSettings {
      bUseBrushTransparency=false
      roundingType=HalfHeightRadius
      width=0
      color {colorUseRule=UseColor_Specified, specifiedColor=<rgba 0 0 0 0>}
      cornerRadii=<0 0 0 0>
     }
     tintColor {colorUseRule=UseColor_Specified, specifiedColor=<rgba 1 1 1 1>}
     uVRegion {bIsValid=false, max=<0 0>, min=<0 0>}
    }
   }
   slot {class=HorizontalBoxSlot, properties {horizontalAlignment=HAlign_Left, verticalAlignment=VAlign_Center, padding=<ltrb 2 0 2 0>}}
  }
  HBoxExtensionWidgets_BelowItemType {
   widgetClassPath=&HorizontalBox
   bIsVariable=true
   parent=VerticalBox_217
   widget=HBoxExtensionWidgets_BelowItemType
   slot {class=VerticalBoxSlot, properties {verticalAlignment=VAlign_Bottom, padding=<ltrb 0 8 2 0>, size {sizeRule=Fill, value=1}}}
  }
  Image_RechargeableIcon {
   widgetClassPath=&FortLazyImage
   bInherited=true
   bIsVariable=true
   parent=VerticalBox_217
   widget=Image_RechargeableIcon
   properties {
    bIsVolatile=true
    visibility=Collapsed
    brush {
     bIsDynamicallyLoaded=false
     drawAs=Image
     imageType=NoImage
     mirroring=NoMirror
     resourceName=None
     resourceObject=/Game/UI/Foundation/Textures/Icons/Ammo/T-Icon-RechargeAmmo-64
     tiling=NoTile
     imageSize=<22 22>
     margin=<ltrb 0 0 0 0>
     outlineSettings {
      bUseBrushTransparency=false
      roundingType=HalfHeightRadius
      width=0
      color {colorUseRule=UseColor_Specified, specifiedColor=<rgba 0 0 0 0>}
      cornerRadii=<0 0 0 0>
     }
     tintColor {colorUseRule=UseColor_Specified, specifiedColor=<rgba 1 1 1 1>}
     uVRegion {bIsValid=false, max=<0 0>, min=<0 0>}
    }
    loadingBackgroundBrush {
     bIsDynamicallyLoaded=false
     drawAs=NoDrawType
     imageType=NoImage
     mirroring=NoMirror
     resourceName=None
     resourceObject=null
     tiling=NoTile
     imageSize=<48 48>
     margin=<ltrb 0 0 0 0>
     outlineSettings {
      bUseBrushTransparency=false
      roundingType=HalfHeightRadius
      width=0
      color {colorUseRule=UseColor_Specified, specifiedColor=<rgba 0 0 0 0>}
      cornerRadii=<0 0 0 0>
     }
     tintColor {colorUseRule=UseColor_Specified, specifiedColor=<rgba 1 1 1 1>}
     uVRegion {bIsValid=false, max=<0 0>, min=<0 0>}
    }
   }
   slot {class=VerticalBoxSlot, properties {horizontalAlignment=HAlign_Left, verticalAlignment=VAlign_Center, padding=<ltrb 4 0 4 0>}}
  }
  HBoxExtensionWidgets_InspectPreview {
   widgetClassPath=&HorizontalBox
   bIsVariable=true
   parent=VerticalBox_2
   widget=HBoxExtensionWidgets_InspectPreview
   slot {class=VerticalBoxSlot, properties {padding=<ltrb 0 8 2 0>}}
  }
  Overlay_EmptyState {
   widgetClassPath=&Overlay
   bIsVariable=true
   parent=SwitcherInfoContent
   widget=Overlay_EmptyState
   slot {class=CommonVisibilitySwitcherSlot, properties {horizontalAlignment=HAlign_Center, verticalAlignment=VAlign_Center}}
  }
  Text_Empty {
   widgetClassPath=&CommonTextBlock
   parent=Overlay_EmptyState
   widget=Text_Empty
   properties {
    justification=Center
    text=Empty
    textTransformPolicy=ToUpper
    font {
     bForceMonospaced=false
     bMaterialIsStencil=false
     fontMaterial=/Game/Athena/HUD/Inventory/Art/M_UI_Inventory_ItemInfo_Description_FontColor
     fontObject=&BurbankSmall
     letterSpacing=0
     monospacedWidth=1
     size=18
     skewAmount=0
     typefaceFontName=Black
     outlineSettings {
      bApplyOutlineToDropShadows=false
      bMiteredCorners=false
      bSeparateFillAlpha=false
      outlineMaterial=null
      outlineSize=0
      outlineColor=<rgba 0 0 0 1>
     }
    }
    renderTransform {angle=0, scale=<1 1>, shear=<-12 0>, translation=<0 0>}
    shadowOffset=<0 0>
   }
   slot {class=OverlaySlot, properties {horizontalAlignment=HAlign_Fill, verticalAlignment=VAlign_Fill}}
  }
  PanelScale {
   widgetClassPath=&Border
   parent=VerticalBox_0
   widget=PanelScale
   properties {
    visibility=SelfHitTestInvisible
    background {
     bIsDynamicallyLoaded=false
     drawAs=NoDrawType
     imageType=NoImage
     mirroring=NoMirror
     resourceName=None
     resourceObject=null
     tiling=NoTile
     imageSize=<32 32>
     margin=<ltrb 0 0 0 0>
     outlineSettings {
      bUseBrushTransparency=false
      roundingType=HalfHeightRadius
      width=0
      color {colorUseRule=UseColor_Specified, specifiedColor=<rgba 0 0 0 0>}
      cornerRadii=<0 0 0 0>
     }
     tintColor {colorUseRule=UseColor_Specified, specifiedColor=<rgba 1 1 1 1>}
     uVRegion {bIsValid=false, max=<0 0>, min=<0 0>}
    }
    padding=<ltrb 0 0 0 0>
    renderTransformPivot=<0.5 1>
   }
   slot {class=VerticalBoxSlot, properties {verticalAlignment=VAlign_Top}}
  }
  ExpandedPanel {
   widgetClassPath=&Overlay
   parent=PanelScale
   widget=ExpandedPanel
   properties {clipping=ClipToBounds}
   slot {class=BorderSlot, properties {padding=<ltrb 0 0 0 0>}}
  }
  BackgroundBlur_244 {
   widgetClassPath=&BackgroundBlur
   parent=ExpandedPanel
   widget=BackgroundBlur_244
   properties {
    blurStrength=8
    lowQualityFallbackBrush {
     bIsDynamicallyLoaded=false
     drawAs=Image
     imageType=NoImage
     mirroring=NoMirror
     resourceName=None
     resourceObject=null
     tiling=NoTile
     imageSize=<0 0>
     margin=<ltrb 0 0 0 0>
     outlineSettings {
      bUseBrushTransparency=false
      roundingType=HalfHeightRadius
      width=0
      color {colorUseRule=UseColor_Specified, specifiedColor=<rgba 0 0 0 0>}
      cornerRadii=<0 0 0 0>
     }
     tintColor {colorUseRule=UseColor_Specified, specifiedColor=<rgba 0 0 0 0.9>}
     uVRegion {bIsValid=false, max=<0 0>, min=<0 0>}
    }
   }
   slot {class=OverlaySlot, properties {horizontalAlignment=HAlign_Fill, verticalAlignment=VAlign_Fill}}
  }
  PanelBlackBacking {
   widgetClassPath=&Image
   parent=ExpandedPanel
   widget=PanelBlackBacking
   properties {brush {
     bIsDynamicallyLoaded=false
     drawAs=Image
     imageType=NoImage
     mirroring=NoMirror
     resourceName=None
     resourceObject=/Game/Athena/HUD/Inventory/Art/M_UI_Inventory_ItemInfo_Inspect_Backing
     tiling=NoTile
     imageSize=<32 32>
     margin=<ltrb 0 0 0 0>
     outlineSettings {
      bUseBrushTransparency=false
      roundingType=HalfHeightRadius
      width=0
      color {colorUseRule=UseColor_Specified, specifiedColor=<rgba 0 0 0 0>}
      cornerRadii=<0 0 0 0>
     }
     tintColor {colorUseRule=UseColor_Specified, specifiedColor=<rgba 1 1 1 0>}
     uVRegion {bIsValid=false, max=<0 0>, min=<0 0>}
    }}
   slot {class=OverlaySlot, properties {horizontalAlignment=HAlign_Fill, verticalAlignment=VAlign_Fill}}
  }
  VerticalBox_1 {
   widgetClassPath=&VerticalBox
   parent=ExpandedPanel
   widget=VerticalBox_1
   slot {class=OverlaySlot, properties {horizontalAlignment=HAlign_Fill, verticalAlignment=VAlign_Fill, padding=<ltrb 16 16 16 16>}}
  }
  Text_ItemDescription {
   widgetClassPath=&CommonTextBlock
   bInherited=true
   parent=VerticalBox_1
   widget=Text_ItemDescription
   properties {
    autoWrapText=true
    mobileFontSizeMultiplier=1.35
    text="{Placeholder description text. Item Description would go here.}"
    font {
     bForceMonospaced=false
     bMaterialIsStencil=false
     fontMaterial=/CreativeEmporium/Materials/M_UI_Inventory_ItemInfo_Description_Creative_FontColor
     fontObject=&HeadingNow
     letterSpacing=0
     monospacedWidth=1
     size=15.75
     skewAmount=0
     typefaceFontName=74 Regular
     outlineSettings {
      bApplyOutlineToDropShadows=false
      bMiteredCorners=false
      bSeparateFillAlpha=false
      outlineMaterial=null
      outlineSize=0
      outlineColor=<rgba 0 0 0 1>
     }
    }
    shadowOffset=<0 0>
   }
   slot {class=VerticalBoxSlot, properties {padding=<ltrb 0 0 0 15>}}
  }
  HBoxModsLocalization_Handler {
   widgetClassPath=&HorizontalBox
   parent=VerticalBox_1
   widget=HBoxModsLocalization_Handler
   properties {flowDirectionPreference=Culture}
   slot {class=VerticalBoxSlot, properties {verticalAlignment=VAlign_Top, padding=<ltrb 0 0 2 15>}}
  }
  HBoxExtensionWidgets_ExtendedInspect {
   widgetClassPath=&HorizontalBox
   bIsVariable=true
   parent=HBoxModsLocalization_Handler
   widget=HBoxExtensionWidgets_ExtendedInspect
   slot {class=HorizontalBoxSlot, properties {size {sizeRule=Fill, value=1}}}
  }
  Spacer_131 {
   widgetClassPath=&Spacer
   parent=HBoxModsLocalization_Handler
   widget=Spacer_131
   slot {class=HorizontalBoxSlot, properties {size {sizeRule=Fill, value=1}}}
  }
  ItemStatsWidget {
   widgetClassPath=/Game/Athena/HUD/Inventory/AthenaInventoryItemStats
   bInherited=true
   bIsVariable=true
   parent=VerticalBox_1
   widget=ItemStatsWidget
   slot {class=VerticalBoxSlot, properties {verticalAlignment=VAlign_Top}}
  }
  Border_SecondaryItemType {widgetClassPath=&Border, bInherited=true, widget=null, properties=~}
  Text_SecondaryItemCategory {widgetClassPath=&CommonTextBlock, bInherited=true, widget=null}
  Image_SecondaryItemCategory {widgetClassPath=&Image, bInherited=true, widget=null}
 ]
}
