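# Inventory UI widget that shows a colored, skewed tag with icon and uppercase label marking a special item, defaulting to "GUARD WEAPON".
# PreConstruct and Construct apply default variables; Initialize Special Item Tag sets visibility from ItemFound, then applies supplied text, color, soft icon texture and transforms.
# Set Tag Visibility shows or collapses the widget, using a Gate so the fade-and-slide FadeIn plays once until hiding stops it and reopens the gate.
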WidgetBlueprint /Game/UI/Inventory/Cr_SpecialItemTag
animations [FadeIn {
  binding_count=1
  length_seconds=0.4
  object=&self:FadeIn
  sequence {bindings [Tag_SpecialItem {
     id=#1
     possessed_class=Border
     tracks [
      {
       type=MovieSceneFloatTrack
       display_name=Render Opacity
       sections [{type=MovieSceneFloatSection, end_frame=5, start_frame=-3, channels [None {type=MovieSceneScriptingFloatChannel, keys [[0,0],[5,1]]}]}]
      }
      {
       type=MovieScene2DTransformTrack
       sections [{
         type=MovieScene2DTransformSection
         end_frame=5
         start_frame=-3
         channels [
          Translation.X {type=MovieSceneScriptingFloatChannel, const [0,2]}
          Translation.Y {type=MovieSceneScriptingFloatChannel, keys [[0,30],[5,0]]}
         ]
        }]
      }
     ]
    }]}
 }]
graphs [EventGraph {nodes [
   n1 = Set Tag Visibility!() -> (OutputDelegate, then, Show:bool=~) {is=title K2Node_CustomEvent}
   > n2 = Branch(Condition=@n1.Show)
   > SetVisibility(InVisibility=SelfHitTestInvisible)
   n3 = Get FadeIn() -> (FadeIn:Widget Animation Object Reference)
   > Branch(Condition=NOT Boolean(A=IsAnimationPlaying(InAnimation=@n3)))
   > Gate!(Enter:Exec, Open:Exec=@n5.then, Close:Exec=@n4.then, Toggle:Exec, bStartClosed:bool=false) -> (Exit:Exec) {is=title K2Node_MacroInstance}
   > n4 = PlayAnimation(InAnimation=Get FadeIn())
   SetVisibility(execute=@n2.else, InVisibility=Collapsed)
   > n5 = StopAnimation(InAnimation=Get FadeIn())
   n6 = Initialize Special Item Tag!() -> (
    OutputDelegate
    then
    Special Item Details Tag_TagText_2_#2:Text
    Special Item Details Tag_TagColor_9_#3:Linear Color Structure
    Special Item Details Tag_TagIconTexture_11_#4:Texture 2D Soft Object Reference
    Special Item Details Tag_TagIconTransform_10_#5:Widget Transform Structure
    Special Item Details Tag_TagWidgetTransform_15_#6:Widget Transform Structure
    ItemFound:bool=~
   ) {is=title K2Node_CustomEvent}
   > Set Tag Visibility!(execute, self, Show:bool=@n6.ItemFound) -> (then) {is=title K2Node_CallFunction}
   > "SetText (Text)"(
    self=Get ItemDesc() -> (ItemDesc:Common Text Object Reference)
    InText=@n6.Special Item Details Tag_TagText_2_#2
   )
   > SetBrushColor(
    self=Get Tag_SpecialItem() -> (Tag_SpecialItem:Border Object Reference)
    InBrushColor=@n6.Special Item Details Tag_TagColor_9_#3
   )
   > SetBrushFromSoftTexture!(
    execute
    self:Image Object Reference=Get ItemIcon() -> (ItemIcon:Image Object Reference)
    SoftTexture:Texture 2D Soft Object Reference=@n6.Special Item Details Tag_TagIconTexture_11_#4
    bMatchSize:bool=false
   ) -> (then)
   > SetRenderTransform!(
    execute
    self:Widget Object Reference=Get ItemIcon()
    InTransform:Widget Transform Structure=@n6.Special Item Details Tag_TagIconTransform_10_#5
   ) -> (then)
   > SetRenderTransform(InTransform=@n6.Special Item Details Tag_TagWidgetTransform_15_#6)
   n7 = Event Construct()
   "SetText (Text)"(execute=[@n7.then, @n8.then], self=Get ItemDesc(), InText=Get ItemDescText() -> (ItemDescText:Text))
   > SetBrushFromSoftTexture(self=Get ItemIcon(), SoftTexture=Get IconSoftTexture() -> (IconSoftTexture:Texture 2D Soft Object Reference), bMatchSize=false)
   > SetBrushColor(self=Get Tag_SpecialItem(), InBrushColor=Get ItemBrushColor() -> (ItemBrushColor:Linear Color Structure))
   > SetRenderTransform(self=Get ItemIcon(), InTransform=Get IconTransformation() -> (IconTransformation:Widget Transform Structure))
   n8 = Event PreConstruct()
  ]}]
variables [
 IconSoftTexture {type=Texture2D, default=/Game/UI/Inventory/Art/T_UI_WeaponType_Guard_Icon}
 IconTransformation {type=WidgetTransform, default {angle=0, scale=<1 1>, shear=<0 0>, translation=<0 0>}}
 ItemBrushColor {type=LinearColor, default=<rgba 1 0.0252 0.1878 1>}
 ItemDescText {type=string, default=GUARD WEAPON}
]
widgets {
 info {inheritedWidgetCount=0, parentClass=&UserWidget, rootWidgetClass=&Border, widgetCount=6}
 widgets [
  Tag_SpecialItem {
   widgetClassPath=&Border
   bIsVariable=true
   widget=Tag_SpecialItem
   properties {
    horizontalAlignment=HAlign_Left
    verticalAlignment=VAlign_Top
    visibility=HitTestInvisible
    background {
     bIsDynamicallyLoaded=false
     drawAs=Box
     imageType=NoImage
     mirroring=NoMirror
     resourceName=None
     resourceObject=/Game/Athena/UI/ReplayBrowser/M_Ui_Skewed_NoV1
     tiling=NoTile
     imageSize=<32 32>
     margin=<ltrb 0.25 0 0.25 0>
     outlineSettings {
      bUseBrushTransparency=false
      roundingType=HalfHeightRadius
      width=0
      color {colorUseRule=UseColor_Specified, specifiedColor=<rgba 0 0 0 0>}
      cornerRadii=<0 0 0 0>
     }
     tintColor {colorUseRule=UseColor_Specified, specifiedColor=<rgba 1 1 1 1>}
     uVRegion {bIsValid=false, max=<0 0>, min=<0 0>}
    }
    brushColor=<rgba 1 0.0252 0.1878 1>
    padding=<ltrb 0 0 0 0>
   }
  }
  Overlay_122 {
   widgetClassPath=&Overlay
   parent=Tag_SpecialItem
   widget=Overlay_122
   slot {class=BorderSlot, properties {horizontalAlignment=HAlign_Left, verticalAlignment=VAlign_Top, padding=<ltrb 0 0 0 0>}}
  }
  HorizontalBox_50 {widgetClassPath=&HorizontalBox, parent=Overlay_122, widget=HorizontalBox_50, slot {class=OverlaySlot}}
  ItemIcon {
   widgetClassPath=&Image
   bIsVariable=true
   parent=HorizontalBox_50
   widget=ItemIcon
   properties {
    visibility=SelfHitTestInvisible
    brush {
     bIsDynamicallyLoaded=false
     drawAs=Image
     imageType=NoImage
     mirroring=NoMirror
     resourceName=None
     resourceObject=&T_UI_WeaponType_Guard_Icon
     tiling=NoTile
     imageSize=<24 24>
     margin=<ltrb 0 0 0 0>
     outlineSettings {
      bUseBrushTransparency=false
      roundingType=HalfHeightRadius
      width=0
      color {colorUseRule=UseColor_Specified, specifiedColor=<rgba 0 0 0 0>}
      cornerRadii=<0 0 0 0>
     }
     tintColor {colorUseRule=UseColor_Specified, specifiedColor=<rgba 1 1 1 1>}
     uVRegion {bIsValid=false, max=<0 0>, min=<0 0>}
    }
    renderTransform {angle=0, scale=<1.2 1.2>, shear=<0 0>, translation=<0 0>}
   }
   slot {class=HorizontalBoxSlot, properties {horizontalAlignment=HAlign_Center, verticalAlignment=VAlign_Center, padding=<ltrb 5 0 0 0>}}
  }
  SizeBox_39 {
   widgetClassPath=&SizeBox
   parent=HorizontalBox_50
   widget=SizeBox_39
   properties {bOverride_WidthOverride=true, widthOverride=165}
   slot {class=HorizontalBoxSlot}
  }
  ItemDesc {
   widgetClassPath=&CommonTextBlock
   bIsVariable=true
   parent=SizeBox_39
   widget=ItemDesc
   properties {
    scrollStyle=&TextScrollStyle-Base_C
    text=GUARD WEAPON
    textTransformPolicy=ToUpper
    font {
     bForceMonospaced=false
     bMaterialIsStencil=false
     fontMaterial=null
     fontObject=&BurbankSmall
     letterSpacing=0
     monospacedWidth=1
     size=14
     skewAmount=0
     typefaceFontName=Black
     outlineSettings {
      bApplyOutlineToDropShadows=false
      bMiteredCorners=false
      bSeparateFillAlpha=false
      outlineMaterial=null
      outlineSize=0
      outlineColor=<rgba 0 0 0 1>
     }
    }
    renderTransform {angle=0, scale=<1 1>, shear=<-8 0>, translation=<0 0>}
    shadowOffset=<0 0>
   }
   slot {class=SizeBoxSlot, properties {horizontalAlignment=HAlign_Center, verticalAlignment=VAlign_Center, padding=<ltrb 8 2 10 2>}}
  }
 ]
}
