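# Creative Emporium details panel widget showing the selected item's preview image, optional parent-item thumbnail, special tag and info, or a "No Item Selected" state.
# SetItemDetails resets scrolling, feeds the ItemInfo child and sets both images from soft textures.
# Special tag badges (guard, open beta, not ready for ship) are looked up by name in a map variable and passed to Cr_SpecialItemTag.
# SwitchItemDetailsPanel's legacy mode hides the preview and changes the background width.
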WidgetBlueprint /CreativeEmporium/WBP_Creative_Emporium_DetailsPanel
graphs [
 BP_GetSpecialItemTags {nodes [
   Get Special Item Tags() {is=title K2Node_FunctionEntry}
   > n1 = Keys!(
    execute
    TargetMap:Map of Names to Cr Item Details Special Tag Struct Structures=Get Special Item Definitions() -> (Special Item Definitions:Map of Names to Cr Item Details Special Tag Struct Structures)
   ) -> (then, Keys:Array of Names)
   > Return Node!(execute, ReturnValue:Array of Names=@n1)
  ]}
 EventGraph {nodes [
   n1 = Event SetItemDetails!() -> (
    OutputDelegate
    then
    FortItem:Fort Item Object Reference
    ParentDefinition:Item Definition Base Object Reference
    bUseLargeThumbnail:bool
    bAllowInteractiveTags:bool
   ) {is=title K2Node_Event}
   > n2 = Is Valid(InputObject=@n1.FortItem)
   PrintString!(execute=@n2.Is Not Valid, InString=notvalid, bPrintToScreen, bPrintToLog, TextColor, Duration, Key) -> (then)
   SetScrollOffset!(
    execute=@n2.Is Valid
    self:Scroll Box Object Reference=Get ScrollBox_Details() -> (ScrollBox_Details:Scroll Box Object Reference)
    NewScrollOffset:float=0
   ) -> (then)
   > UpdateWithFortItem!(
    execute
    self:Fort Emporium Item Info Object Reference=Get ItemDetails() -> (ItemDetails:Fort Emporium Item Info Object Reference)
    Item:Fort Item Object Reference=@n1.FortItem
    ParentDefinition:Item Definition Base Object Reference=@n1.ParentDefinition
    bAllowInteractiveTags:bool=@n1.bAllowInteractiveTags
   ) -> (then)
   n3 = Get Image_SelectedItem() -> (Image_SelectedItem:Image Object Reference)
   > SetBrushFromSoftTexture!(
    execute
    self:Image Object Reference=@n3
    SoftTexture:Texture 2D Soft Object Reference=Select(
     Option 0:Texture 2D Soft Object Reference=GetSmallPreviewImage!(self:Fort Item Object Reference=@n1.FortItem) -> (ReturnValue:Texture 2D Soft Object Reference)
     Option 1:Texture 2D Soft Object Reference=GetLargePreviewImage!(self:Fort Item Object Reference=@n1.FortItem) -> (ReturnValue:Texture 2D Soft Object Reference)
     Index=@n1.bUseLargeThumbnail
    ) -> (ReturnValue:Texture 2D Soft Object Reference)
    bMatchSize:bool=false
   ) -> (then)
   > InvalidateLayoutAndVolatility(self:Widget Object Reference=@n3)
   > n4 = Is Valid(InputObject=@n1.ParentDefinition)
   n5 = Get Image_ParentItem() -> (Image_ParentItem:Image Object Reference)
   SetVisibility(execute=@n4.Is Not Valid, self=@n5, InVisibility=Collapsed)
   SetVisibility(execute=@n4.Is Valid, self=@n5, InVisibility=HitTestInvisible)
   > SetBrushFromSoftTexture(
    self=@n5
    SoftTexture=GetSmallPreviewImage!(ItemDefinition:Item Definition Base Object Reference=@n1.ParentDefinition) -> (ReturnValue)
    bMatchSize=false
   )
   > InvalidateLayoutAndVolatility(self=@n5)
   Event Construct()
   > Parent: Construct()
   > n6 = Cast To WBP_Creative_Emporium_ItemInfo!(execute, Object=Get ItemDetails()) -> (then, CastFailed, AsWBP Creative Emporium Item Info:WBP Creative Emporium Item Info Object Reference) {is=title K2Node_DynamicCast}
   > Bind Event to On Tag Selected(self:WBP Creative Emporium Item Info Object Reference=@n6, Delegate=Create Event()) {is=title K2Node_AddDelegate}
   > Bind Event to On Request Item Focus(self:WBP Creative Emporium Item Info Object Reference=@n6, Delegate=Create Event()) {is=title K2Node_AddDelegate}
   n7 = Event SetSpecialItemTagVisibility!() -> (OutputDelegate, then, SpecialItemTag:"Name (by ref)"=None) {is=title K2Node_Event}
   n8 = Find(
    TargetMap:Map of Names to Cr Item Details Special Tag Struct Structures=Get Special Item Definitions()
    Key:"Name (by ref)"=@n7.SpecialItemTag
   ) -> (Value:Cr Item Details Special Tag Struct Structure)
   > Initialize Special Item Tag!(
    execute
    self:Cr Special Item Tag Object Reference=Get Cr_SpecialItemTag() -> (Cr_SpecialItemTag:Cr Special Item Tag Object Reference)
    Special Item Details Tag:Cr Item Details Special Tag Struct Structure=@n8
    ItemFound:bool=@n8.ReturnValue
   ) -> (then)
   n9 = Event UpdateItemDetailsDisplay!() -> (OutputDelegate, then, bShowItemDetails:bool) {is=title K2Node_Event}
   > SetActiveWidget(
    self:Widget Switcher Object Reference=Get Switcher_ItemDetails() -> (Switcher_ItemDetails:Common Animated Switcher Object Reference)
    Widget=Select(
     Option 0:Widget Object Reference=Get Overlay_NoItem() -> (Overlay_NoItem:Overlay Object Reference)
     Option 1:Widget Object Reference=Get Overlay_ItemDetails() -> (Overlay_ItemDetails:Overlay Object Reference)
     Index=@n9.bShowItemDetails
    ) -> (ReturnValue:Widget Object Reference)
   )
   n10 = Event SwitchItemDetailsPanel!() -> (OutputDelegate, then, bUseLegacy:bool) {is=title K2Node_Event}
   > Parent: SwitchItemDetailsPanel!(execute, bUseLegacy:bool=@n10.bUseLegacy) -> (then) {is=title K2Node_CallParentFunction}
   > UpdateSizeBoxes!(execute, self, Legacy:bool=@n10.bUseLegacy) -> (then)
   > SetVisibility(
    self=Get Preview_Overlay() -> (Preview_Overlay:Overlay Object Reference)
    InVisibility=Select(Option 0:ESlateVisibility Enum=SelfHitTestInvisible, Option 1:ESlateVisibility Enum=Collapsed, Index=@n10.bUseLegacy) -> (ReturnValue:ESlateVisibility Enum)
   )
   n11 = Event SetActiveItemDetailTags!() -> (OutputDelegate, then, ActiveTags:Array of Names) {is=title K2Node_Event}
   > n12 = Cast To WBP_Creative_Emporium_ItemInfo(Object=Get ItemDetails())
   > Update SelectedTags!(execute, self:WBP Creative Emporium Item Info Object Reference=@n12, Selected Tags:Array of Names=@n11.ActiveTags) -> (then)
  ]}
 UpdateSizeBoxes {nodes [
   n1 = UpdateSizeBoxes!() -> (then, Legacy:bool=~) {is=title K2Node_FunctionEntry}
   n2 = Get!() -> (PanelSizes_X:double, PanelSizes_Y:double)
   > SetWidthOverride(
    self=Get SizeBox_Background() -> (SizeBox_Background:Size Box Object Reference)
    InWidthOverride=Select(Option 0:double=@n2, Option 1:double=@n2.PanelSizes_Y, Index=@n1) -> (ReturnValue:double)
   )
  ]}
]
variables [
 PanelSizes {type=Vector2D, default=<286 390>}
 Special Item Definitions {
  type="map<CrItemDetailsSpecialTagStruct>"
  default {
   guard {
    tagIconTexture=/Game/UI/Inventory/Art/T_UI_WeaponType_Guard_Icon
    tagText=GUARD WEAPON
    tagColor=<rgba 1 0.0252 0.1878 1>
    tagIconTransform {angle=0, scale=<1 1>, shear=<0 0>, translation=<0 0>}
    tagWidgetTransform {angle=0, scale=<1 1>, shear=<0 0>, translation=<0 0>}
   }
   notReadyForShip {
    tagIconTexture=/Game/Athena/UI/InGame/Creative/PropertyEditor/Icon_Picker/Textures/64x64/T_UI_IconLibrary_enemy_64
    tagText="!!! Not Ready For Ship !!!"
    tagColor=<rgba 1 0.0292 0 1>
    tagIconTransform {angle=0, scale=<1 1>, shear=<0 0>, translation=<0 0>}
    tagWidgetTransform {angle=0, scale=<1.5 1.5>, shear=<0 0>, translation=<-128 -128>}
   }
   openBeta {
    tagIconTexture=/Game/Athena/UI/InGame/Creative/PropertyEditor/Icon_Picker/Textures_V2/64x64/T_UI_IconLibrary_Caution_64
    tagText=OPEN BETA
    tagColor=<rgba 1 0.2502 0 1>
    tagIconTransform {angle=0, scale=<1 1>, shear=<0 0>, translation=<5 0>}
    tagWidgetTransform {angle=0, scale=<1 1>, shear=<0 0>, translation=<0 0>}
   }
  }
 }
]
widgets {
 info {
  inheritedWidgetCount=3
  parentClass=/Script/CreativeEmporiumUI.FortEmporiumItemDetailsPanel
  rootWidgetClass=/Script/CommonUI.CommonAnimatedSwitcher
  widgetCount=19
 }
 widgets [
  Switcher_ItemDetails {widgetClassPath=&CommonAnimatedSwitcher, bIsVariable=true, widget=Switcher_ItemDetails, properties {transitionDuration=0}}
  Overlay_NoItem {widgetClassPath=&Overlay, bIsVariable=true, parent=Switcher_ItemDetails, widget=Overlay_NoItem, slot {class=WidgetSwitcherSlot}}
  ItemDisplayName {
   widgetClassPath=&CommonTextBlock
   bIsVariable=true
   parent=Overlay_NoItem
   widget=ItemDisplayName
   properties {
    autoWrapText=true
    bAutoCollapseWithEmptyText=true
    bIsScrollingEnabled=false
    text=No Item Selected
    colorAndOpacity {colorUseRule=UseColor_Specified, specifiedColor=<rgba 1 1 1 0.9>}
    font {
     bForceMonospaced=false
     bMaterialIsStencil=false
     fontMaterial=null
     fontObject=&HeadingNow
     letterSpacing=0
     monospacedWidth=1
     size=22.5
     skewAmount=0
     typefaceFontName=75 Medium Italic
     outlineSettings {
      bApplyOutlineToDropShadows=false
      bMiteredCorners=false
      bSeparateFillAlpha=false
      outlineMaterial=null
      outlineSize=0
      outlineColor=<rgba 0 0 0 1>
     }
    }
    shadowOffset=<0 0>
   }
   slot {class=OverlaySlot, properties {horizontalAlignment=HAlign_Center, verticalAlignment=VAlign_Center, padding=<ltrb 0 0 0 24>}}
  }
  Overlay_ItemDetails {
   widgetClassPath=&Overlay
   bIsVariable=true
   parent=Switcher_ItemDetails
   widget=Overlay_ItemDetails
   slot {class=WidgetSwitcherSlot, properties {horizontalAlignment=HAlign_Left, padding=<ltrb 0 24 0 24>}}
  }
  ScrollBox_Details {
   widgetClassPath=&ScrollBox
   bIsVariable=true
   parent=Overlay_ItemDetails
   widget=ScrollBox_Details
   slot {class=OverlaySlot, properties {horizontalAlignment=HAlign_Fill, verticalAlignment=VAlign_Fill}}
  }
  VerticalBox_61 {widgetClassPath=&VerticalBox, parent=ScrollBox_Details, widget=VerticalBox_61, slot {class=ScrollBoxSlot}}
  Preview_Overlay {
   widgetClassPath=&Overlay
   bIsVariable=true
   parent=VerticalBox_61
   widget=Preview_Overlay
   slot {class=VerticalBoxSlot, properties {horizontalAlignment=HAlign_Left, verticalAlignment=VAlign_Top, padding=<ltrb 24 0 24 12>}}
  }
  SizeBox_Background {
   widgetClassPath=&SizeBox
   bIsVariable=true
   parent=Preview_Overlay
   widget=SizeBox_Background
   properties {bOverride_HeightOverride=true, bOverride_WidthOverride=true, heightOverride=288, widthOverride=288}
   slot {class=OverlaySlot, properties {verticalAlignment=VAlign_Center}}
  }
  PreviewImageBorder {
   widgetClassPath=&Image
   parent=SizeBox_Background
   widget=PreviewImageBorder
   properties {brush {
     bIsDynamicallyLoaded=false
     drawAs=RoundedBox
     imageType=NoImage
     mirroring=NoMirror
     resourceName=None
     resourceObject=null
     tiling=NoTile
     imageSize=<32 32>
     margin=<ltrb 0 0 0 0>
     outlineSettings {
      bUseBrushTransparency=false
      roundingType=FixedRadius
      width=0
      color {colorUseRule=UseColor_Specified, specifiedColor=<rgba 0 0 0 0>}
      cornerRadii=<10 10 10 10>
     }
     tintColor {colorUseRule=UseColor_Specified, specifiedColor=<rgba 1 1 1 0.1>}
     uVRegion {bIsValid=false, max=<0 0>, min=<0 0>}
    }}
   slot {class=SizeBoxSlot}
  }
  SizeBox_Icon {
   widgetClassPath=&SizeBox
   parent=Preview_Overlay
   widget=SizeBox_Icon
   properties {bOverride_HeightOverride=true, bOverride_WidthOverride=true, heightOverride=256, widthOverride=256}
   slot {class=OverlaySlot, properties {horizontalAlignment=HAlign_Center, verticalAlignment=VAlign_Center, padding=<ltrb 16 16 16 16>}}
  }
  Image_SelectedItem {
   widgetClassPath=&Image
   bIsVariable=true
   parent=SizeBox_Icon
   widget=Image_SelectedItem
   properties {brush {
     bIsDynamicallyLoaded=false
     drawAs=Image
     imageType=NoImage
     mirroring=NoMirror
     resourceName=None
     resourceObject=null
     tiling=NoTile
     imageSize=<48 48>
     margin=<ltrb 0 0 0 0>
     outlineSettings {
      bUseBrushTransparency=false
      roundingType=HalfHeightRadius
      width=0
      color {colorUseRule=UseColor_Specified, specifiedColor=<rgba 0 0 0 0>}
      cornerRadii=<0 0 0 0>
     }
     tintColor {colorUseRule=UseColor_Specified, specifiedColor=<rgba 1 1 1 1>}
     uVRegion {bIsValid=false, max=<0 0>, min=<0 0>}
    }}
   slot {class=SizeBoxSlot}
  }
  SizeBox_Parent {
   widgetClassPath=&SizeBox
   parent=Preview_Overlay
   widget=SizeBox_Parent
   slot {class=OverlaySlot, properties {verticalAlignment=VAlign_Bottom, padding=<ltrb 16 16 16 16>}}
  }
  Image_ParentItem {
   widgetClassPath=&Image
   bIsVariable=true
   parent=SizeBox_Parent
   widget=Image_ParentItem
   properties {
    visibility=Collapsed
    brush {
     bIsDynamicallyLoaded=false
     drawAs=Image
     imageType=NoImage
     mirroring=NoMirror
     resourceName=None
     resourceObject=null
     tiling=NoTile
     imageSize=<96 96>
     margin=<ltrb 0 0 0 0>
     outlineSettings {
      bUseBrushTransparency=false
      roundingType=HalfHeightRadius
      width=0
      color {colorUseRule=UseColor_Specified, specifiedColor=<rgba 0 0 0 0>}
      cornerRadii=<0 0 0 0>
     }
     tintColor {colorUseRule=UseColor_Specified, specifiedColor=<rgba 1 1 1 1>}
     uVRegion {bIsValid=false, max=<0 0>, min=<0 0>}
    }
   }
   slot {class=SizeBoxSlot}
  }
  SpecialItemTagSizeBox {
   widgetClassPath=&SizeBox
   parent=Preview_Overlay
   widget=SpecialItemTagSizeBox
   slot {class=OverlaySlot, properties {horizontalAlignment=HAlign_Right, verticalAlignment=VAlign_Bottom}}
  }
  Cr_SpecialItemTag {
   widgetClassPath=/Game/UI/Inventory/Cr_SpecialItemTag
   bIsVariable=true
   parent=SpecialItemTagSizeBox
   widget=Cr_SpecialItemTag
   slot {class=SizeBoxSlot, properties {horizontalAlignment=HAlign_Right, verticalAlignment=VAlign_Bottom}}
  }
  SizeBox_Details {
   widgetClassPath=&SizeBox
   parent=VerticalBox_61
   widget=SizeBox_Details
   properties {bOverride_WidthOverride=true, widthOverride=360}
   slot {class=VerticalBoxSlot}
  }
  Switcher_Details {
   widgetClassPath=&CommonVisibilitySwitcher
   bInherited=true
   parent=SizeBox_Details
   widget=Switcher_Details
   slot {class=SizeBoxSlot}
  }
  ItemDetails {
   widgetClassPath=&/WBP_Creative_Emporium_ItemInfo
   bInherited=true
   bIsVariable=true
   parent=Switcher_Details
   widget=ItemDetails
   slot {class=CommonVisibilitySwitcherSlot}
  }
  LegacyItemDetails {
   widgetClassPath=/CreativeEmporium/Widgets/WBP_ContentBrowser_ItemInfo
   bInherited=true
   parent=Switcher_Details
   widget=LegacyItemDetails
   slot {class=CommonVisibilitySwitcherSlot}
  }
 ]
}
