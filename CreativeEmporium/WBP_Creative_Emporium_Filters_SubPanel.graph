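# Creative Emporium filter sidebar widget with a match-any/match-all toggle above a scrolling list of tag filter entries.
# The toggle fires OnCycleTagFilterMode, while ConfigureFilters updates its text, icon and indicator bars.
# Parent-class events add entries to a TagID-keyed map, update counts, hide empty ones and sync active states behind a re-entrancy guard.
# Selection scrolls to and focuses the entry unless DoNotScrollIntoView suppresses it, as Show All syncing does.
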WidgetBlueprint /CreativeEmporium/WBP_Creative_Emporium_Filters_SubPanel
graphs [
 ConfigureFilterButton {nodes [
   n1 = ConfigureFilterButton!() -> (then, InputMode:ECommonInputType Enum) {is=title K2Node_FunctionEntry}
   > SetRequiresHold!(
    execute
    self:Common Button Base Object Reference=Get FilterToggle_Button() -> (FilterToggle_Button:WBP UIKit Button Generic Object Reference)
    bInRequiresHold:bool=Select!(MouseAndKeyboard:bool=false, Gamepad:bool=false, Touch:bool=false, Count:bool=false, Index:ECommonInputType Enum=@n1) -> (ReturnValue:bool)
   ) -> (then)
  ]}
 ConfigureFilters {nodes [
   n1 = ConfigureFilters!() -> (then, TagFilterMode:EFortEmporiumItemTagFilterMode Enum=MatchAny) {is=title K2Node_FunctionEntry}
   > SetText!(
    execute
    self:WBP UIKit Button Generic Base Object Reference=Get FilterToggle_Button()
    Text:Text=Select!(
     MatchAny:Text=MakeLiteralText(Value="LOCTEXT(\"Union\")")
     MatchAll:Text=MakeLiteralText(Value="LOCTEXT(\"Intersect\")")
     Index:EFortEmporiumItemTagFilterMode Enum=@n1
    ) -> (ReturnValue:Text)
   ) -> (then)
   > SetIcon!(
    execute
    self:WBP UIKit Button Generic Base Object Reference=Get FilterToggle_Button()
    Icon:Object Reference=Select(
     MatchAny:Object Reference=/CreativeEmporium/Materials/T_UI_Icon_FilterUnion
     MatchAll:Object Reference=/CreativeEmporium/Materials/T_UI_Icon_FilterIntersect
     Index=@n1
    ) -> (ReturnValue:Object Reference)
   ) -> (then)
   > SetRenderOpacity(
    self=Get FilterOption1_Border() -> (FilterOption1_Border:Common Border Object Reference)
    InOpacity=Select(MatchAny:float=1, MatchAll:float=0.35, Index=@n1) -> (ReturnValue:float)
   )
   > SetRenderOpacity(
    self=Get FilterOption2_Border() -> (FilterOption2_Border:Common Border Object Reference)
    InOpacity=Select(MatchAny=0.35, MatchAll=1, Index=@n1)
   )
  ]}
 EventGraph {nodes [
   "On Clicked (FilterToggle_Button)"!() -> (OutputDelegate, then, Button:Common Button Base Object Reference) {is=title K2Node_ComponentBoundEvent}
   > Call On Cycle Tag Filter Mode!(execute:Exec, self:WBP Creative Emporium Filters Sub Panel Object Reference) -> (then:Exec) {is=title K2Node_CallDelegate}
   n1 = HandleInputModeChanged!() -> (OutputDelegate, then, bNewInputType:ECommonInputType Enum) {is=title K2Node_CustomEvent}
   > ConfigureFilterButton!(execute, self, InputMode:ECommonInputType Enum=@n1.bNewInputType) -> (then) {is=title K2Node_CallFunction}
   Event Construct()
   n2 = Get CommonInputSubsystem!() -> (ReturnValue:Common Input Object Reference) {is=title K2Node_GetBlueprintContextEx}
   > Bind Event to On Input Method Changed(self=@n2, Delegate=@n1)
   > ConfigureFilterButton(InputMode=GetCurrentInputType(self=@n2))
   n3 = Event OnUpdateFilterEntries!() -> (OutputDelegate, then, IncludedTags:Array of Names) {is=title K2Node_Event}
   > n4 = Branch(Condition=Get IsUpdatingFilterEntries() -> (IsUpdatingFilterEntries:bool))
   Set IsUpdatingFilterEntries(execute=@n4.else, IsUpdatingFilterEntries:bool=true) -> (Output_Get:bool)
   > n5 = For Each Loop(Array:Array of Widget Object References=GetAllChildren(self=Get ScrollBox_FilterScroll() -> (ScrollBox_FilterScroll:Athena Scroll Box Object Reference))) -> (Array Element:Widget Object Reference)
   Set IsUpdatingFilterEntries(execute=@n5.Completed, IsUpdatingFilterEntries=false)
   n6 = Cast To WBP_CrBrowser_FilterEntryWidget!(execute=@n5.LoopBody, Object=@n5) -> (then, CastFailed, AsWBP Cr Browser Filter Entry Widget:WBP Cr Browser Filter Entry Widget Object Reference) {is=title K2Node_DynamicCast}
   n7 = Get TagID!(self:Emporium Browser Tag Object Reference=Get Tag!(self:Emporium Browser Filter Entry Object Reference=@n6) -> (Tag:Emporium Browser Tag Object Reference)) -> (TagID:Name=None)
   > n8 = Branch(Condition="Equal (Name)"(A=GetShowAllButtonID!() -> (ReturnValue:Name=None), B=@n7))
   > SetIsFilterActive!(
    execute
    self:Emporium Browser Filter Entry Object Reference=@n6
    bInFilterActive:bool=Is Empty!(TargetArray:Array of Names=@n3.IncludedTags) -> (ReturnValue:bool)
   ) -> (then)
   > Set DoNotScrollIntoView(DoNotScrollIntoView:bool=true) -> (Output_Get:bool)
   SetIsFilterActive(execute=@n8.else, self=@n6, bInFilterActive=Contains Item(TargetArray=@n3.IncludedTags, ItemToFind=@n7))
   Event ClearFilterWidgets() {is=title K2Node_Event}
   > ClearChildren(self=Get ScrollBox_FilterScroll())
   > Clear!(
    execute
    TargetMap:Map of Names to Emporium Browser Filter Entry Object References=Get Map_FilterEntries() -> (Map_FilterEntries:Map of Names to Emporium Browser Filter Entry Object References)
   ) -> (then) {is=title K2Node_CallFunction}
   n9 = Event OnAddSubcategoryWidget!() -> (OutputDelegate, then, SubcategoryWidget:Emporium Browser Filter Entry Object Reference) {is=title K2Node_Event}
   n10 = Get ScrollBox_FilterScroll()
   > n11 = Branch(Condition=HasChild!(self:Panel Widget Object Reference=@n10, Content:Widget Object Reference=@n9.SubcategoryWidget) -> (ReturnValue:bool))
   AddChild(execute=@n11.else, self=@n10, Content=@n9.SubcategoryWidget)
   > Add!(
    execute
    TargetMap:Map of Names to Emporium Browser Filter Entry Object References=Get Map_FilterEntries()
    Key:"Name (by ref)"=Get TagID(self=Get Tag(self=@n9.SubcategoryWidget)) -> (TagID=None)
    Value:"Emporium Browser Filter Entry Object Reference (by ref)"=@n9.SubcategoryWidget
   ) -> (then) {is=title K2Node_CallFunction}
   n12 = Event OnFilterSelected!() -> (OutputDelegate, then, SubcategoryWidget:Emporium Browser Filter Entry Object Reference) {is=title K2Node_Event}
   > Is Valid(InputObject=@n12.SubcategoryWidget)
   > n13 = Branch(Condition=Get DoNotScrollIntoView() -> (DoNotScrollIntoView:bool))
   Set DoNotScrollIntoView(execute=[@n15.then, @n13.then], DoNotScrollIntoView=false)
   ScrollWidgetIntoView!(
    execute=@n13.else
    self:Scroll Box Object Reference=Get ScrollBox_FilterScroll()
    WidgetToFind:Widget Object Reference=@n12.SubcategoryWidget
    AnimateScroll:bool=false
    ScrollDestination:EDescendantScrollDestination Enum=IntoView
    Padding:float=0
   ) -> (then)
   > n14 = Branch(Condition=HasUserFocus!(self:Widget Object Reference=@n12.SubcategoryWidget, PlayerController:Player Controller Object Reference=GetOwningPlayer()) -> (ReturnValue:bool))
   n15 = SetFocus(execute=@n14.else, self=@n12.SubcategoryWidget)
   n16 = Event OnUpdateSubcategoryCount!() -> (OutputDelegate, then, CategoryID:"Name (by ref)"=None, Count:int=0, bShowOnEmpty:bool) {is=title K2Node_Event}
   n17 = Find(TargetMap:Map of Names to Emporium Browser Filter Entry Object References=Get Map_FilterEntries(), Key:"Name (by ref)"=@n16.CategoryID) -> (Value:Emporium Browser Filter Entry Object Reference)
   > Branch(Condition=@n17.ReturnValue)
   > UpdateItemCount!(execute, self:Emporium Browser Filter Entry Object Reference=@n17, ItemCount:int=@n16.Count) -> (then)
   n18 = "integer > integer"(A=@n16.Count)
   > SetVisibility(
    self=@n17
    InVisibility=Select!(
     Option 0:ESlateVisibility Enum=Collapsed
     Option 1:ESlateVisibility Enum=SelfHitTestInvisible
     Index:bool=OR Boolean(A=@n18, B=@n16.bShowOnEmpty)
    ) -> (ReturnValue:ESlateVisibility Enum)
   )
   Event OnMouseEnter()
   > SetScrollBarVisibility!(execute, self:Scroll Box Object Reference=Get ScrollBox_FilterScroll(), NewScrollBarVisibility:ESlateVisibility Enum=Visible) -> (then)
   Event OnMouseLeave!() -> (OutputDelegate, then, MouseEvent:"Pointer Event Structure (by ref)") {is=title K2Node_Event}
   > SetScrollBarVisibility(self=Get ScrollBox_FilterScroll(), NewScrollBarVisibility=Hidden)
  ]}
 OnCycleTagFilterMode {nodes [OnCycleTagFilterMode() {is=title K2Node_FunctionEntry}]}
 SelectShowAll {nodes [
   n1 = SelectShowAll!() -> (then, Tag:Emporium Browser Tag Object Reference, ScrollIntoView:bool=~) {is=title K2Node_FunctionEntry}
   n2 = Find(TargetMap=Get Map_FilterEntries(), Key=GetShowAllButtonID() -> (ReturnValue=None))
   > Branch(Condition=@n2.ReturnValue)
   > Set DoNotScrollIntoView(DoNotScrollIntoView=NOT Boolean(A=@n1.ScrollIntoView))
   > SetIsFilterActive(self=@n2, bInFilterActive=true)
  ]}
]
variables [
 DoNotScrollIntoView {type=boolean}
 IsUpdatingFilterEntries {type=boolean}
 Map_FilterEntries {type="map<EmporiumBrowserFilterEntry>", default {}}
 OnCycleTagFilterMode {}
]
widgets {
 info {
  inheritedWidgetCount=0
  namedSlotCount=1
  parentClass=/Script/CreativeEmporiumUI.FortEmporiumFiltersSubPanel
  rootWidgetClass=&Overlay
  widgetCount=11
 }
 widgets [
  Overlay_0 {widgetClassPath=&Overlay, widget=Overlay_0}
  VerticalBox_0 {
   widgetClassPath=&VerticalBox
   parent=Overlay_0
   widget=VerticalBox_0
   slot {class=OverlaySlot, properties {horizontalAlignment=HAlign_Fill, verticalAlignment=VAlign_Fill}}
  }
  Hrule_Image {
   widgetClassPath=&Image
   parent=VerticalBox_0
   widget=Hrule_Image
   properties {
    visibility=Hidden
    brush {
     bIsDynamicallyLoaded=false
     drawAs=Image
     imageType=NoImage
     mirroring=NoMirror
     resourceName=None
     resourceObject=null
     tiling=NoTile
     imageSize=<32 1>
     margin=<ltrb 0 0 0 0>
     outlineSettings {
      bUseBrushTransparency=false
      roundingType=HalfHeightRadius
      width=0
      color {colorUseRule=UseColor_Specified, specifiedColor=<rgba 0 0 0 0>}
      cornerRadii=<0 0 0 0>
     }
     tintColor {colorUseRule=UseColor_Specified, specifiedColor=<rgba 1 1 1 0.3>}
     uVRegion {bIsValid=false, max=<0 0>, min=<0 0>}
    }
   }
   slot {class=VerticalBoxSlot, properties {verticalAlignment=VAlign_Center, padding=<ltrb 12 0 12 6>}}
  }
  FilterToggle_Button {
   widgetClassPath=&WBP_UIKit_Button_Generic
   bIsVariable=true
   parent=VerticalBox_0
   widget=FilterToggle_Button
   properties {
    bHideInputAction=true
    holdData=/FortUILibrary/Input/ShortInteraction_HoldData.ShortInteraction_HoldData_C
    icon=&T_UI_Icon_FilterUnion
    iconMaterial=/CreativeEmporium/Materials/MI_CrBrowser_IconWhite
    minHeight=64
    mobileMinHeight=76
    secondText=Inclusive
    text=Filter Mode
    textTransformPolicy=None
    font {
     bForceMonospaced=false
     bMaterialIsStencil=false
     fontMaterial=/CreativeEmporium/Materials/MI_CrBrowser_WhiteText
     fontObject=&HeadingNow
     letterSpacing=48
     monospacedWidth=1
     size=16.5
     skewAmount=0
     typefaceFontName=75 Medium
     outlineSettings {
      bApplyOutlineToDropShadows=false
      bMiteredCorners=false
      bSeparateFillAlpha=false
      outlineMaterial=null
      outlineSize=0
      outlineColor=<rgba 0 0 0 1>
     }
    }
    iconPadding=<ltrb 0 0 4 0>
    iconSize=<42 42>
    promptNudge=<-12 0>
    secondFont {
     bForceMonospaced=false
     bMaterialIsStencil=false
     fontMaterial=/FortUILibrary/Systems/ModularButton/MaterialInstances/MI_UIKit_ColorText
     fontObject=&Burbank
     letterSpacing=48
     monospacedWidth=1
     size=18
     skewAmount=0
     typefaceFontName=BurbankRegular-Black
     outlineSettings {
      bApplyOutlineToDropShadows=false
      bMiteredCorners=false
      bSeparateFillAlpha=false
      outlineMaterial=null
      outlineSize=0
      outlineColor=<rgba 0 0 0 1>
     }
    }
    secondTextPadding=<ltrb 0 0 0 0>
   }
   slot {class=VerticalBoxSlot, properties {verticalAlignment=VAlign_Top, padding=<ltrb 12 0 12 0>}}
  }
  Overlay_Filter {widgetClassPath=&Overlay, namedSlotHost=FilterToggle_Button, widget=Overlay_Filter}
  SizeBox_0 {
   widgetClassPath=&SizeBox
   parent=Overlay_Filter
   widget=SizeBox_0
   properties {bOverride_HeightOverride=true, heightOverride=5, widthOverride=200}
   slot {class=OverlaySlot, properties {horizontalAlignment=HAlign_Fill, verticalAlignment=VAlign_Bottom, padding=<ltrb 24 0 24 4>}}
  }
  HorizontalBox_88 {widgetClassPath=&HorizontalBox, parent=SizeBox_0, widget=HorizontalBox_88, slot {class=SizeBoxSlot}}
  FilterOption1_Border {
   widgetClassPath=&CommonBorder
   bIsVariable=true
   parent=HorizontalBox_88
   widget=FilterOption1_Border
   properties {
    background {
     bIsDynamicallyLoaded=false
     drawAs=RoundedBox
     imageType=NoImage
     mirroring=NoMirror
     resourceName=None
     resourceObject=null
     tiling=NoTile
     imageSize=<32 32>
     margin=<ltrb 0 0 0 0>
     outlineSettings {
      bUseBrushTransparency=false
      roundingType=FixedRadius
      width=0
      color {colorUseRule=UseColor_Specified, specifiedColor=<rgba 0 0 0 0>}
      cornerRadii=<2 2 2 2>
     }
     tintColor {colorUseRule=UseColor_Specified, specifiedColor=<rgba 1 1 1 1>}
     uVRegion {bIsValid=false, max=<0 0>, min=<0 0>}
    }
    padding=<ltrb 0 0 0 0>
   }
   slot {class=HorizontalBoxSlot, properties {padding=<ltrb 0 0 2 0>, size {sizeRule=Fill, value=1}}}
  }
  FilterOption2_Border {
   widgetClassPath=&CommonBorder
   bIsVariable=true
   parent=HorizontalBox_88
   widget=FilterOption2_Border
   properties {
    background {
     bIsDynamicallyLoaded=false
     drawAs=RoundedBox
     imageType=NoImage
     mirroring=NoMirror
     resourceName=None
     resourceObject=null
     tiling=NoTile
     imageSize=<32 32>
     margin=<ltrb 0 0 0 0>
     outlineSettings {
      bUseBrushTransparency=false
      roundingType=FixedRadius
      width=0
      color {colorUseRule=UseColor_Specified, specifiedColor=<rgba 0 0 0 0>}
      cornerRadii=<2 2 2 2>
     }
     tintColor {colorUseRule=UseColor_Specified, specifiedColor=<rgba 1 1 1 1>}
     uVRegion {bIsValid=false, max=<0 0>, min=<0 0>}
    }
    padding=<ltrb 0 0 0 0>
   }
   slot {class=HorizontalBoxSlot, properties {padding=<ltrb 2 0 0 0>, size {sizeRule=Fill, value=1}}}
  }
  Overlay_2 {
   widgetClassPath=&Overlay
   parent=VerticalBox_0
   widget=Overlay_2
   properties {clipping=ClipToBounds}
   slot {class=VerticalBoxSlot, properties {padding=<ltrb 0 12 0 0>, size {sizeRule=Fill, value=1}}}
  }
  ScrollBox_FilterScroll {
   widgetClassPath=/Script/FortniteUI.AthenaScrollBox
   bIsVariable=true
   parent=Overlay_2
   widget=ScrollBox_FilterScroll
   properties {
    bAnimateWheelScrolling=false
    bDisableInertialScrolling=true
    bOverride_Cursor=true
    clipping=Inherit
    scrollBarVisibility=Hidden
    scrollbarPadding=<ltrb 2 2 2 24>
    widgetStyle {
     barThickness=9
     bottomShadowBrush {
      bIsDynamicallyLoaded=false
      drawAs=Box
      imageType=FullColor
      mirroring=NoMirror
      resourceName=../../../Engine/Content/Slate/Common/ScrollBoxShadowBottom.png
      resourceObject=null
      tiling=NoTile
      imageSize=<16 8>
      margin=<ltrb 0.5 0 0.5 1>
      outlineSettings {
       bUseBrushTransparency=false
       roundingType=HalfHeightRadius
       width=0
       color {colorUseRule=UseColor_Specified, specifiedColor=<rgba 0 0 0 0>}
       cornerRadii=<0 0 0 0>
      }
      tintColor {colorUseRule=UseColor_Specified, specifiedColor=<rgba 1 1 1 0>}
      uVRegion {bIsValid=false, max=<0 0>, min=<0 0>}
     }
     horizontalScrolledContentPadding=<ltrb 0 0 1 0>
     leftShadowBrush {
      bIsDynamicallyLoaded=false
      drawAs=Box
      imageType=FullColor
      mirroring=NoMirror
      resourceName=../../../Engine/Content/Slate/Common/ScrollBoxShadowLeft.png
      resourceObject=null
      tiling=NoTile
      imageSize=<8 16>
      margin=<ltrb 1 0.5 0 0.5>
      outlineSettings {
       bUseBrushTransparency=false
       roundingType=HalfHeightRadius
       width=0
       color {colorUseRule=UseColor_Specified, specifiedColor=<rgba 0 0 0 0>}
       cornerRadii=<0 0 0 0>
      }
      tintColor {colorUseRule=UseColor_Specified, specifiedColor=<rgba 1 1 1 1>}
      uVRegion {bIsValid=false, max=<0 0>, min=<0 0>}
     }
     rightShadowBrush {
      bIsDynamicallyLoaded=false
      drawAs=Box
      imageType=FullColor
      mirroring=NoMirror
      resourceName=../../../Engine/Content/Slate/Common/ScrollBoxShadowRight.png
      resourceObject=null
      tiling=NoTile
      imageSize=<8 16>
      margin=<ltrb 0 0.5 1 0.5>
      outlineSettings {
       bUseBrushTransparency=false
       roundingType=HalfHeightRadius
       width=0
       color {colorUseRule=UseColor_Specified, specifiedColor=<rgba 0 0 0 0>}
       cornerRadii=<0 0 0 0>
      }
      tintColor {colorUseRule=UseColor_Specified, specifiedColor=<rgba 1 1 1 1>}
      uVRegion {bIsValid=false, max=<0 0>, min=<0 0>}
     }
     topShadowBrush {
      bIsDynamicallyLoaded=false
      drawAs=Box
      imageType=FullColor
      mirroring=NoMirror
      resourceName=../../../Engine/Content/Slate/Common/ScrollBoxShadowTop.png
      resourceObject=null
      tiling=NoTile
      imageSize=<16 8>
      margin=<ltrb 0.5 1 0.5 0>
      outlineSettings {
       bUseBrushTransparency=false
       roundingType=HalfHeightRadius
       width=0
       color {colorUseRule=UseColor_Specified, specifiedColor=<rgba 0 0 0 0>}
       cornerRadii=<0 0 0 0>
      }
      tintColor {colorUseRule=UseColor_Specified, specifiedColor=<rgba 1 1 1 0>}
      uVRegion {bIsValid=false, max=<0 0>, min=<0 0>}
     }
     verticalScrolledContentPadding=<ltrb 0 0 0 24>
    }
   }
   slot {class=OverlaySlot, properties {horizontalAlignment=HAlign_Fill, verticalAlignment=VAlign_Fill, padding=<ltrb 14 20 14 18>}}
  }
 ]
}
