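# Details panel widget for Creative Emporium playset items, showing a preview image, item info, special tag and Back button.
# SetItemDetails feeds valid items to ItemDetails and soft-loads a small or large thumbnail.
# SetSpecialItemTagVisibility looks up a name-keyed map (guard, notReadyForShip, openBeta) to initialize the tag, replacing stored transforms with identity.
# Back, shown only for mouse/keyboard or touch, broadcasts OnExitSubtab.
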WidgetBlueprint /CreativeEmporium/WBP_Creative_Emporium_PlaysetDetailsPanel
graphs [
 EventGraph {nodes [
   "On Clicked (Button_Back)"!() -> (OutputDelegate, then, Button:Common Button Base Object Reference) {is=title K2Node_ComponentBoundEvent}
   > Call On Exit Subtab!(execute:Exec, self:WBP Creative Emporium Playset Details Panel Object Reference) -> (then:Exec) {is=title K2Node_CallDelegate}
   n1 = Event SetItemDetails!() -> (
    OutputDelegate
    then
    FortItem:Fort Item Object Reference
    ParentDefinition:Item Definition Base Object Reference
    bUseLargeThumbnail:bool
    bAllowInteractiveTags:bool
   ) {is=title K2Node_Event}
   > Is Valid(InputObject=@n1.FortItem)
   > UpdateWithFortItem!(
    execute
    self:Fort Emporium Item Info Object Reference=Get ItemDetails() -> (ItemDetails:Fort Emporium Item Info Object Reference)
    Item:Fort Item Object Reference=@n1.FortItem
    ParentDefinition:Item Definition Base Object Reference
    bAllowInteractiveTags:bool=@n1.bAllowInteractiveTags
   ) -> (then)
   n2 = Get Image_SelectedItem() -> (Image_SelectedItem:Image Object Reference)
   > SetBrushFromSoftTexture!(
    execute
    self:Image Object Reference=@n2
    SoftTexture:Texture 2D Soft Object Reference=Select(
     Option 0:Texture 2D Soft Object Reference=GetSmallPreviewImage!(self:Fort Item Object Reference=@n1.FortItem) -> (ReturnValue:Texture 2D Soft Object Reference)
     Option 1:Texture 2D Soft Object Reference=GetLargePreviewImage!(self:Fort Item Object Reference=@n1.FortItem) -> (ReturnValue:Texture 2D Soft Object Reference)
     Index=@n1.bUseLargeThumbnail
    ) -> (ReturnValue:Texture 2D Soft Object Reference)
    bMatchSize:bool=false
   ) -> (then)
   > InvalidateLayoutAndVolatility(self:Widget Object Reference=@n2)
   n3 = Event SetSpecialItemTagVisibility!() -> (OutputDelegate, then, SpecialItemTag:"Name (by ref)"=None) {is=title K2Node_Event}
   n4 = Find(
    TargetMap:Map of Names to Cr Item Details Special Tag Struct Structures=Get Special Item Definitions() -> (Special Item Definitions:Map of Names to Cr Item Details Special Tag Struct Structures)
    Key:"Name (by ref)"=@n3.SpecialItemTag
   ) -> (Value:Cr Item Details Special Tag Struct Structure)
   n5 = Break Cr Item Details Special Tag Struct!(CrItemDetailsSpecialTagStruct:"Cr Item Details Special Tag Struct Structure (by ref)"=@n4) -> (
    TagText_2_#1:Text
    TagColor_9_#2:Linear Color Structure
    TagIconTexture_11_#3:Texture 2D Soft Object Reference
    TagIconTransform_10_#4:Widget Transform Structure
    TagWidgetTransform_15_#5:Widget Transform Structure
   ) {is=title K2Node_BreakStruct}
   > Initialize Special Item Tag!(
    execute
    self:Cr Special Item Tag Object Reference=Get Cr_SpecialItemTag() -> (Cr_SpecialItemTag:Cr Special Item Tag Object Reference)
    Special Item Details Tag:Cr Item Details Special Tag Struct Structure=Make Cr Item Details Special Tag Struct!(
     TagText_2_#1:Text=@n5
     TagColor_9_#2:Linear Color Structure=@n5.TagColor_9_#2
     TagIconTexture_11_#3:Texture 2D Soft Object Reference=@n5.TagIconTexture_11_#3
     TagIconTransform_10_#4:Widget Transform Structure=(Translation=(X=0,Y=0),Scale=(X=1,Y=1),Shear=(X=0,Y=0),Angle=0)
     TagWidgetTransform_15_#5:Widget Transform Structure=(Translation=(X=0,Y=0),Scale=(X=1,Y=1),Shear=(X=0,Y=0),Angle=0)
    ) -> (CrItemDetailsSpecialTagStruct:Cr Item Details Special Tag Struct Structure) {is=title K2Node_MakeStruct}
    ItemFound:bool=@n4.ReturnValue
   ) -> (then)
  ]}
 OnExitSubtab {nodes [OnExitSubtab() {is=title K2Node_FunctionEntry}]}
]
variables [
 OnExitSubtab {}
 Special Item Definitions {
  type="map<CrItemDetailsSpecialTagStruct>"
  default {
   guard {
    tagIconTexture=/Game/UI/Inventory/Art/T_UI_WeaponType_Guard_Icon
    tagText=GUARD WEAPON
    tagColor=<rgba 1 0.0252 0.1878 1>
    tagIconTransform {angle=0, scale=<1 1>, shear=<0 0>, translation=<0 0>}
    tagWidgetTransform {angle=0, scale=<1 1>, shear=<0 0>, translation=<0 0>}
   }
   notReadyForShip {
    tagIconTexture=/Game/Athena/UI/InGame/Creative/PropertyEditor/Icon_Picker/Textures/64x64/T_UI_IconLibrary_enemy_64
    tagText="!!! Not Ready For Ship !!!"
    tagColor=<rgba 1 0.0292 0 1>
    tagIconTransform {angle=0, scale=<1 1>, shear=<0 0>, translation=<0 0>}
    tagWidgetTransform {angle=0, scale=<1.5 1.5>, shear=<0 0>, translation=<-128 -120>}
   }
   openBeta {
    tagIconTexture=/Game/Athena/UI/InGame/Creative/PropertyEditor/Icon_Picker/Textures_V2/64x64/T_UI_IconLibrary_Caution_64
    tagText=OPEN BETA
    tagColor=<rgba 1 0.2502 0 1>
    tagIconTransform {angle=0, scale=<1 1>, shear=<0 0>, translation=<5 0>}
    tagWidgetTransform {angle=0, scale=<1 1>, shear=<0 0>, translation=<0 0>}
   }
  }
 }
]
widgets {
 info {
  inheritedWidgetCount=3
  parentClass=/Script/CreativeEmporiumUI.FortEmporiumItemDetailsPanel
  rootWidgetClass=&Overlay
  widgetCount=15
 }
 widgets [
  Overlay {widgetClassPath=&Overlay, widget=Overlay}
  VerticalBox_61 {
   widgetClassPath=&VerticalBox
   parent=Overlay
   widget=VerticalBox_61
   slot {class=OverlaySlot, properties {horizontalAlignment=HAlign_Fill, verticalAlignment=VAlign_Fill}}
  }
  CommonHardwareVisibilityBorder_0 {
   widgetClassPath=&CommonHardwareVisibilityBorder
   parent=VerticalBox_61
   widget=CommonHardwareVisibilityBorder_0
   properties {
    hiddenType=Hidden
    visibilityQuery {
     autoDescription=" ANY( Input.MouseAndKeyboard, Input.Touch )"
     tokenStreamVersion=0
     userDescription=Input Type
     queryTokenStream [0,1,1,2,0,1]
     tagDictionary [{tagName=Input.MouseAndKeyboard}, {tagName=Input.Touch}]
    }
   }
   slot {class=VerticalBoxSlot, properties {horizontalAlignment=HAlign_Left, padding=<ltrb 32 0 0 0>}}
  }
  Button_Back {
   widgetClassPath=&WBP_UIKit_Button_Generic
   bIsVariable=true
   parent=CommonHardwareVisibilityBorder_0
   widget=Button_Back
   properties {
    icon=/CreativeEmporium/Materials/T_UI_Icon_CrRefresh_Chevron_Left
    iconMaterial=/FortUILibrary/UIKit/Buttons/MaterialInstances/CTAButtons/MI_UIKit_IconQuiet
    minWidth=200
    text=Back
   }
   slot {class=BorderSlot, properties {padding=<ltrb 0 0 0 0>}}
  }
  Overlay_3 {
   widgetClassPath=&Overlay
   parent=VerticalBox_61
   widget=Overlay_3
   slot {class=VerticalBoxSlot, properties {horizontalAlignment=HAlign_Left, verticalAlignment=VAlign_Top, padding=<ltrb 24 52 24 0>}}
  }
  PreviewImageBorder {
   widgetClassPath=&Image
   parent=Overlay_3
   widget=PreviewImageBorder
   properties {brush {
     bIsDynamicallyLoaded=false
     drawAs=RoundedBox
     imageType=NoImage
     mirroring=NoMirror
     resourceName=None
     resourceObject=null
     tiling=NoTile
     imageSize=<32 32>
     margin=<ltrb 0 0 0 0>
     outlineSettings {
      bUseBrushTransparency=false
      roundingType=FixedRadius
      width=0
      color {colorUseRule=UseColor_Specified, specifiedColor=<rgba 0 0 0 0>}
      cornerRadii=<10 10 10 10>
     }
     tintColor {colorUseRule=UseColor_Specified, specifiedColor=<rgba 1 1 1 0.1>}
     uVRegion {bIsValid=false, max=<0 0>, min=<0 0>}
    }}
   slot {class=OverlaySlot, properties {horizontalAlignment=HAlign_Fill, verticalAlignment=VAlign_Fill}}
  }
  ScaleBox_0 {
   widgetClassPath=&ScaleBox
   parent=Overlay_3
   widget=ScaleBox_0
   properties {stretch=ScaleToFitX, userSpecifiedScale=8}
   slot {class=OverlaySlot, properties {horizontalAlignment=HAlign_Center, verticalAlignment=VAlign_Center, padding=<ltrb 24 24 24 24>}}
  }
  SizeBox_1 {
   widgetClassPath=&SizeBox
   parent=ScaleBox_0
   widget=SizeBox_1
   properties {bOverride_HeightOverride=true, bOverride_WidthOverride=true, heightOverride=256, widthOverride=256}
   slot {class=ScaleBoxSlot}
  }
  Image_SelectedItem {
   widgetClassPath=&Image
   bIsVariable=true
   parent=SizeBox_1
   widget=Image_SelectedItem
   properties {brush {
     bIsDynamicallyLoaded=false
     drawAs=Image
     imageType=NoImage
     mirroring=NoMirror
     resourceName=None
     resourceObject=null
     tiling=NoTile
     imageSize=<48 48>
     margin=<ltrb 0 0 0 0>
     outlineSettings {
      bUseBrushTransparency=false
      roundingType=HalfHeightRadius
      width=0
      color {colorUseRule=UseColor_Specified, specifiedColor=<rgba 0 0 0 0>}
      cornerRadii=<0 0 0 0>
     }
     tintColor {colorUseRule=UseColor_Specified, specifiedColor=<rgba 1 1 1 1>}
     uVRegion {bIsValid=false, max=<0 0>, min=<0 0>}
    }}
   slot {class=SizeBoxSlot}
  }
  SpecialItemTagSizeBox {
   widgetClassPath=&SizeBox
   parent=Overlay_3
   widget=SpecialItemTagSizeBox
   slot {class=OverlaySlot, properties {horizontalAlignment=HAlign_Fill, verticalAlignment=VAlign_Bottom}}
  }
  Cr_SpecialItemTag {
   widgetClassPath=/Game/UI/Inventory/Cr_SpecialItemTag
   bIsVariable=true
   parent=SpecialItemTagSizeBox
   widget=Cr_SpecialItemTag
   slot {class=SizeBoxSlot, properties {horizontalAlignment=HAlign_Right, verticalAlignment=VAlign_Bottom}}
  }
  SizeBox {
   widgetClassPath=&SizeBox
   parent=VerticalBox_61
   widget=SizeBox
   properties {widthOverride=500}
   slot {class=VerticalBoxSlot, properties {verticalAlignment=VAlign_Bottom, padding=<ltrb 0 24 0 0>}}
  }
  Switcher_Details {
   widgetClassPath=&CommonVisibilitySwitcher
   bInherited=true
   parent=SizeBox
   widget=Switcher_Details
   slot {class=SizeBoxSlot}
  }
  ItemDetails {
   widgetClassPath=&/WBP_Creative_Emporium_ItemInfo
   bInherited=true
   bIsVariable=true
   parent=Switcher_Details
   widget=ItemDetails
   slot {class=CommonVisibilitySwitcherSlot}
  }
  LegacyItemDetails {
   widgetClassPath=/CreativeEmporium/Widgets/WBP_ContentBrowser_ItemInfo
   bInherited=true
   parent=Switcher_Details
   widget=LegacyItemDetails
   slot {class=CommonVisibilitySwitcherSlot}
  }
 ]
}
