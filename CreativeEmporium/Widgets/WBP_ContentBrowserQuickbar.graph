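# Quickbar widget for the Creative content browser, derived from FortCreativeMenuQuickbar, with a switcher between creative, trap and primary quickbars.
# On construct it binds slot-activation events on all three bars, plus drop events on the creative and trap bars.
# Creative and trap events route the current or dropped item to HandleAddItemToQuickbarAction, traps using a fixed Secondary slot, while primary activation calls SetItemInSlot.
# Switcher changes toggle the primary bar's owner-activation state, and world-inventory changes refresh its slot visibilities.
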WidgetBlueprint /CreativeEmporium/Widgets/WBP_ContentBrowserQuickbar
animations [BarActivated {binding_count=0, length_seconds=2, object=&self:BarActivated, sequence {bindings []}}]
graphs [
 EventGraph {nodes [
   "On Active Widget Changed (Switcher_QuickBars)"!() -> (OutputDelegate, then, ActiveWidget:Widget Object Reference, ActiveWidgetIndex:int=0) {is=title K2Node_ComponentBoundEvent}
   > UpdatePrimaryQuickBarState()
   n1 = OnSlotActivated!() -> (OutputDelegate, then, SlotId:int) {is=title K2Node_CustomEvent}
   > HandleAddItemToQuickbarAction!(
    execute
    self:Fort Creative Menu Quickbar Object Reference
    QuickbarType:EFortQuickBars Enum=Creative
    SlotNum:int=@n1.SlotId
    ItemDefinition:Item Definition Base Object Reference=Get CurrentItemDefinition() -> (CurrentItemDefinition:Item Definition Base Object Reference)
   ) -> (then)
   n2 = OnTrapSlotActivated() {is=title K2Node_CustomEvent}
   > HandleAddItemToQuickbarAction(QuickbarType=Secondary, SlotNum=4, ItemDefinition=Get CurrentItemDefinition())
   n3 = OnQuickBarDrop!() -> (OutputDelegate, then, SlotID:int, QuickbarType:EFortQuickBars Enum=Primary, Item:Fort Item Object Reference) {is=title K2Node_CustomEvent}
   > HandleAddItemToQuickbarAction(
    QuickbarType=@n3.QuickbarType
    SlotNum=@n3.SlotID
    ItemDefinition=GetItemDefinition!(self:Fort Item Object Reference=@n3.Item) -> (ReturnValue:Fort Item Definition Object Reference)
   )
   n4 = OnTrapQuickbarDrop!() -> (OutputDelegate, then, Item:Fort Item Object Reference) {is=title K2Node_CustomEvent}
   > HandleAddItemToQuickbarAction(QuickbarType=Secondary, SlotNum=4, ItemDefinition=GetItemDefinition(self=@n4.Item))
   OnWorldInventoryChanged!() -> (OutputDelegate, then, bInventoryActive:bool=~) {is=title K2Node_CustomEvent}
   > UpdateSlotVisibilities(self:WBP Content Browser Primary Quickbar Object Reference=Get PrimaryQuickbar() -> (PrimaryQuickbar:WBP Content Browser Primary Quickbar Object Reference))
   n5 = OnPrimarySlotActivated!() -> (OutputDelegate, then, SlotIndex:int) {is=title K2Node_CustomEvent}
   > SetItemInSlot!(
    execute
    self:WBP Content Browser Primary Quickbar Object Reference=Get PrimaryQuickbar()
    SlotIndex:int=@n5.SlotIndex
    ItemDefinition:Item Definition Base Object Reference=Get CurrentItemDefinition()
   ) -> (then)
   Event Construct()
   > UpdatePrimaryQuickBarState()
   > n6 = Sequence!(execute) -> (then_0, then_1, then_2:Exec, then_3:Exec)
   > n7 = Cast To FortPlayerControllerAthena(Object=GetOwningPlayer())
   > Bind Event to On World Inventory Changed Delegate(self:Fort Player Controller Gameplay Object Reference=@n7, Delegate=Create Event()) {is=title K2Node_AddDelegate}
   n8 = Cast To Cr_ContentBrowserCreativeQuickBar!(execute=@n6.then_1, Object=Get AthenaQuickbar_Creative() -> (AthenaQuickbar_Creative:Athena Quickbar Editor Base Object Reference)) -> (then, CastFailed, AsCr Content Browser Creative Quick Bar:Cr Content Browser Creative Quick Bar Object Reference) {is=title K2Node_DynamicCast}
   > Bind Event to On Quickbar Slot Activated(self:Cr Content Browser Creative Quick Bar Object Reference=@n8, Delegate=@n1) {is=title K2Node_AddDelegate}
   > Bind Event to On Quick Bar Drop(self:Cr Content Browser Creative Quick Bar Object Reference=@n8, Delegate=@n3) {is=title K2Node_AddDelegate}
   n9 = Cast To cr_ContentBrowserTrapQuickBar!(execute=@n6.then_2, Object=Get AthenaQuickbar_Trap() -> (AthenaQuickbar_Trap:Athena Quickbar Editor Base Object Reference)) -> (then, CastFailed, AsCr Content Browser Trap Quick Bar:Cr Content Browser Trap Quick Bar Object Reference) {is=title K2Node_DynamicCast}
   > Bind Event to On Quickbar Trap Slot Activated(self:Cr Content Browser Trap Quick Bar Object Reference=@n9, Delegate=@n2) {is=title K2Node_AddDelegate}
   > Bind Event to On Quickbar Dropped(self:Cr Content Browser Trap Quick Bar Object Reference=@n9, Delegate=@n4) {is=title K2Node_AddDelegate}
   Bind Event to On Quickbar Slot Activated(execute=@n6.then_3, self:WBP Content Browser Primary Quickbar Object Reference=Get PrimaryQuickbar(), Delegate=@n5)
  ]}
 UpdatePrimaryQuickBarState {nodes [
   UpdatePrimaryQuickBarState() {is=title K2Node_FunctionEntry}
   n1 = Get Switcher_QuickBars() -> (Switcher_QuickBars:Widget Switcher Object Reference)
   n2 = Get PrimaryQuickbar()
   > n3 = Branch(Condition="Equal (Object)"(A=GetActiveWidget!(self:Widget Switcher Object Reference=@n1) -> (ReturnValue:Widget Object Reference), B=@n2))
   > OnOwnerActivated(self:WBP Content Browser Primary Quickbar Object Reference=@n2)
   OnOwnerDeactivated(execute=@n3.else, self:WBP Content Browser Primary Quickbar Object Reference=@n2)
  ]}
]
variables []
widgets {
 info {inheritedWidgetCount=4, parentClass=/Script/FortniteUI.FortCreativeMenuQuickbar, rootWidgetClass=&Overlay, widgetCount=18}
 widgets [
  Overlay_0 {widgetClassPath=&Overlay, widget=Overlay_0}
  QuickbarBgColor {
   widgetClassPath=&Image
   parent=Overlay_0
   widget=QuickbarBgColor
   properties {colorAndOpacity=<rgba 1 1 1 0>}
   slot {class=OverlaySlot, properties {horizontalAlignment=HAlign_Fill, verticalAlignment=VAlign_Fill}}
  }
  Switcher_QuickBars {
   widgetClassPath=&CommonWidgetSwitcherLegacy
   bInherited=true
   bIsVariable=true
   parent=Overlay_0
   widget=Switcher_QuickBars
   properties {transitionDuration=0}
   slot {class=OverlaySlot}
  }
  Overlay_Cr_QuickBar {widgetClassPath=&Overlay, parent=Switcher_QuickBars, widget=Overlay_Cr_QuickBar, slot {class=WidgetSwitcherSlot}}
  SizeBox {
   widgetClassPath=&SizeBox
   parent=Overlay_Cr_QuickBar
   widget=SizeBox
   properties {bOverride_HeightOverride=true, heightOverride=45, widthOverride=1000}
   slot {class=OverlaySlot, properties {horizontalAlignment=HAlign_Fill, verticalAlignment=VAlign_Fill, padding=<ltrb 0 32 0 0>}}
  }
  HorizontalBox_0 {widgetClassPath=&HorizontalBox, parent=SizeBox, widget=HorizontalBox_0, slot {class=SizeBoxSlot}}
  BarHighlight_Buttons_L {
   widgetClassPath=&Image
   parent=HorizontalBox_0
   widget=BarHighlight_Buttons_L
   properties {
    brush {
     bIsDynamicallyLoaded=false
     drawAs=Image
     imageType=NoImage
     mirroring=NoMirror
     resourceName=None
     resourceObject=/Game/UI/Inventory/Art/BasicLinearUIGradient_Cr_ContentBrowserTileEdge_Horz
     tiling=NoTile
     imageSize=<32 32>
     margin=<ltrb 0 0 0 0>
     outlineSettings {
      bUseBrushTransparency=false
      roundingType=HalfHeightRadius
      width=0
      color {colorUseRule=UseColor_Specified, specifiedColor=<rgba 0 0 0 0>}
      cornerRadii=<0 0 0 0>
     }
     tintColor {colorUseRule=UseColor_Specified, specifiedColor=<rgba 1 1 1 1>}
     uVRegion {bIsValid=false, max=<0 0>, min=<0 0>}
    }
    colorAndOpacity=<rgba 0 0.402 1 0>
    renderTransform {angle=0, scale=<-1 1>, shear=<0 0>, translation=<0 0>}
   }
   slot {class=HorizontalBoxSlot, properties {size {sizeRule=Fill, value=1}}}
  }
  BarHighlight_Buttons_R {
   widgetClassPath=&Image
   parent=HorizontalBox_0
   widget=BarHighlight_Buttons_R
   properties {
    brush {
     bIsDynamicallyLoaded=false
     drawAs=Image
     imageType=NoImage
     mirroring=NoMirror
     resourceName=None
     resourceObject=&BasicLinearUIGradient_Cr_ContentBrowserTileEdge_Horz
     tiling=NoTile
     imageSize=<32 32>
     margin=<ltrb 0 0 0 0>
     outlineSettings {
      bUseBrushTransparency=false
      roundingType=HalfHeightRadius
      width=0
      color {colorUseRule=UseColor_Specified, specifiedColor=<rgba 0 0 0 0>}
      cornerRadii=<0 0 0 0>
     }
     tintColor {colorUseRule=UseColor_Specified, specifiedColor=<rgba 1 1 1 1>}
     uVRegion {bIsValid=false, max=<0 0>, min=<0 0>}
    }
    colorAndOpacity=<rgba 0 0.402 1 0>
   }
   slot {class=HorizontalBoxSlot, properties {size {sizeRule=Fill, value=1}}}
  }
  VerticalBox_0 {
   widgetClassPath=&VerticalBox
   parent=Overlay_Cr_QuickBar
   widget=VerticalBox_0
   slot {class=OverlaySlot, properties {horizontalAlignment=HAlign_Fill, verticalAlignment=VAlign_Fill}}
  }
  CommonTextBlock_Quickbar {
   widgetClassPath=&CommonTextBlock
   parent=VerticalBox_0
   widget=CommonTextBlock_Quickbar
   properties {
    text=QUICK BAR
    visibility=Collapsed
    colorAndOpacity {colorUseRule=UseColor_Specified, specifiedColor=<rgba 1 1 1 0.8>}
    font {
     bForceMonospaced=false
     bMaterialIsStencil=false
     fontMaterial=null
     fontObject=&NotoSans
     letterSpacing=0
     monospacedWidth=1
     size=12
     skewAmount=0
     typefaceFontName=NotoSans-Bold
     outlineSettings {
      bApplyOutlineToDropShadows=false
      bMiteredCorners=false
      bSeparateFillAlpha=false
      outlineMaterial=/Game/UI/Foundation/Materials/M-FontShadow
      outlineSize=0
      outlineColor=<rgba 0 0 0 1>
     }
    }
    shadowOffset=<0 0>
   }
   slot {class=VerticalBoxSlot, properties {padding=<ltrb 4 2 0 10>}}
  }
  Overlay_1 {widgetClassPath=&Overlay, parent=VerticalBox_0, widget=Overlay_1, slot {class=VerticalBoxSlot}}
  AthenaQuickbar_Creative {
   widgetClassPath=/Game/UI/Inventory/Cr_ContentBrowserCreativeQuickBar
   bInherited=true
   bIsVariable=true
   parent=Overlay_1
   widget=AthenaQuickbar_Creative
   properties {quickBarSlotButtonClass=/Game/Athena/HUD/Inventory/AthenaInventoryEquipSlot.AthenaInventoryEquipSlot_C}
   slot {class=OverlaySlot, properties {horizontalAlignment=HAlign_Fill, verticalAlignment=VAlign_Fill}}
  }
  Overlay_Trap_QuickBar {widgetClassPath=&Overlay, parent=Switcher_QuickBars, widget=Overlay_Trap_QuickBar, slot {class=WidgetSwitcherSlot}}
  VerticalBox_1 {widgetClassPath=&VerticalBox, parent=Overlay_Trap_QuickBar, widget=VerticalBox_1, slot {class=OverlaySlot}}
  CommonTextBlock_Traps {
   widgetClassPath=&CommonTextBlock
   parent=VerticalBox_1
   widget=CommonTextBlock_Traps
   properties {
    text=TRAPS
    visibility=Collapsed
    colorAndOpacity {colorUseRule=UseColor_Specified, specifiedColor=<rgba 1 1 1 0.8>}
    font {
     bForceMonospaced=false
     bMaterialIsStencil=false
     fontMaterial=null
     fontObject=&NotoSans
     letterSpacing=0
     monospacedWidth=1
     size=12
     skewAmount=0
     typefaceFontName=NotoSans-Bold
     outlineSettings {
      bApplyOutlineToDropShadows=false
      bMiteredCorners=false
      bSeparateFillAlpha=false
      outlineMaterial=&M-FontShadow
      outlineSize=0
      outlineColor=<rgba 0 0 0 1>
     }
    }
    shadowOffset=<0 0>
   }
   slot {class=VerticalBoxSlot, properties {padding=<ltrb 4 2 0 10>}}
  }
  AthenaQuickbar_Trap {
   widgetClassPath=/Game/UI/Inventory/cr_ContentBrowserTrapQuickBar
   bInherited=true
   bIsVariable=true
   parent=VerticalBox_1
   widget=AthenaQuickbar_Trap
   properties {quickBarSlotButtonClass=&AthenaInventoryEquipSlot_C}
   slot {class=VerticalBoxSlot}
  }
  PrimaryQuickbar {
   widgetClassPath=&/WBP_ContentBrowser_PrimaryQuickbar
   bIsVariable=true
   parent=Switcher_QuickBars
   widget=PrimaryQuickbar
   slot {class=WidgetSwitcherSlot}
  }
  AthenaQuickbar_Primary {widgetClassPath=/Script/FortniteUI.AthenaQuickbarEditorBase, bInherited=true, widget=null}
 ]
}
