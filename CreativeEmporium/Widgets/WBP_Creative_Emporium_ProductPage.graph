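# Product page widget for the Creative Emporium item details panel, showing an item's info, price and preview image.
# Anim_MainOut slides the main section up and fades it to reveal a secondary scroll panel, and plays in reverse to return.
# A visibility switcher there shows either additional info, tags and related content, or license text requested asynchronously with a fallback message.
# Navigation and focus handlers move between the price widget and the extended panel.
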WidgetBlueprint /CreativeEmporium/Widgets/WBP_Creative_Emporium_ProductPage
animations [Anim_MainOut {
  binding_count=2
  length_seconds=0.75
  object=&self:Anim_MainOut
  sequence {bindings [
    Secondary {
     id=#1
     display_name=SizeBox_0
     possessed_class=SizeBox
     tracks [{
       type=MovieScene2DTransformTrack
       sections [{
         type=MovieScene2DTransformSection
         end_frame=15
         start_frame=0
         channels [Translation.Y {type=MovieSceneScriptingFloatChannel, keys [[0,0],[4,0],[8,-686.158],[15,-860]]}]
        }]
      }]
    }
    Main {
     id=#2
     possessed_class=VerticalBox
     tracks [
      {
       type=MovieScene2DTransformTrack
       sections [{
         type=MovieScene2DTransformSection
         end_frame=15
         start_frame=0
         channels [Translation.Y {type=MovieSceneScriptingFloatChannel, keys [[0,0],[10,-1200]]}]
        }]
      }
      {
       type=MovieSceneFloatTrack
       display_name=Render Opacity
       sections [{type=MovieSceneFloatSection, end_frame=15, start_frame=0, channels [None {type=MovieSceneScriptingFloatChannel, keys [[4,1],[10,0]]}]}]
      }
     ]
    }
   ]}
 }]
graphs [
 EventGraph {nodes [
   "Last Item Navigation Down Event (WBP_ItemPrice)"() {is=title K2Node_ComponentBoundEvent}
   > OnShowAdditionalItemData!(execute, self, Item:Fort Item Object Reference) -> (then)
   Event On Main Scroll Out() {is=title K2Node_CustomEvent}
   > PlayAnimationForward(InAnimation=Get Anim_MainOut() -> (Anim_MainOut:Widget Animation Object Reference))
   > Set bShowingExtendedDataPanel(bShowingExtendedDataPanel:bool=true) -> (Output_Get:bool)
   OnShowAdditionalItemData!() -> (OutputDelegate, then, Item:Fort Item Object Reference) {is=title K2Node_CustomEvent}
   > SetActiveWidget(
    self=Get Switcher_ScrollBox() -> (Switcher_ScrollBox:Common Visibility Switcher Object Reference)
    Widget=Get VerticalBox_AdditionalContent() -> (VerticalBox_AdditionalContent:Vertical Box Object Reference)
   )
   > Event On Main Scroll Out() {is=title K2Node_CallFunction}
   Event On Main Scroll Back() {is=title K2Node_CustomEvent}
   > n1 = "Play Animation with Finished event (legacy)"!(
    execute:Exec
    Widget:User Widget Object Reference=Self-Reference()
    InAnimation:Widget Animation Object Reference=Get Anim_MainOut()
    StartAtTime:float=0
    NumLoopsToPlay:int=1
    PlayMode:EUMGSequencePlayMode Enum=Reverse
    PlaybackSpeed:float=1
   ) -> (then:Exec, Result:UMGSequence Player Object Reference, Finished:Exec) {is=title K2Node_PlayAnimation}
   > Set bShowingExtendedDataPanel(bShowingExtendedDataPanel=false)
   SetActiveWidget(execute=@n1.Finished, self=Get Switcher_ScrollBox(), Widget=Get VerticalBox_AdditionalContent())
   n2 = OnItemLicenseReceived!() -> (OutputDelegate, then, bSuccess:bool=~, LicenseText:String) {is=title K2Node_CustomEvent}
   > SetActiveWidget(self=Get Switcher_ScrollBox(), Widget=Get Text_Legal() -> (Text_Legal:Fort Rich Text Block Legacy Object Reference))
   n3 = NOT Boolean(A=IsEmpty(InString=@n2.LicenseText))
   > SetText(
    self:Fort Rich Text Block Legacy Object Reference=Get Text_Legal()
    InText:Text=Select(
     Option 0:Text="INVTEXT(\"Could Not Get Legal Text\")"
     Option 1:Text="To Text (String)"(InString=@n2.LicenseText)
     Index=AND Boolean(A=@n2.bSuccess, B=@n3)
    ) -> (ReturnValue:Text)
   )
   > Event On Main Scroll Out()
   n4 = OnShowItemLicense!() -> (OutputDelegate, then, ItemData:Fort Emporium Item Data Structure) {is=title K2Node_CustomEvent}
   > RequestLicenseText!(
    execute
    License:EFortEmporiumItemLicense Enum=Break Fort Emporium Item Data!(FortEmporiumItemData:"Fort Emporium Item Data Structure (by ref)"=@n4.ItemData) -> (LicenseType:EFortEmporiumItemLicense Enum) {is=title K2Node_BreakStruct}
    Callback:"Delegate (by ref)"=@n2
   ) -> (then)
   Event ShowExtendedDataPanel() {is=title K2Node_Event}
   > Event On Main Scroll Out()
   Event HideAdditionalButtons() {is=title K2Node_Event}
   > HideAdditionalButtons(self:WBP Creative Emporium Item Price Object Reference=Get WBP_ItemPrice() -> (WBP_ItemPrice:WBP Creative Emporium Item Price Object Reference))
   Event HideExtendedDataPanel() {is=title K2Node_Event}
   > Event On Main Scroll Back() {is=title K2Node_CallFunction}
   > SetFocus(self=Get WBP_ItemPrice())
   n5 = Event SetItemDetails!() -> (
    OutputDelegate
    then
    FortItem:Fort Item Object Reference
    ParentDefinition:Item Definition Base Object Reference
    bUseLargeThumbnail:bool
    bAllowInteractiveTags:bool
   ) {is=title K2Node_Event}
   > Is Valid(InputObject=@n5.FortItem)
   > UpdateWithFortItem!(
    execute
    self:Fort Emporium Item Info Object Reference=Get ItemDetails() -> (ItemDetails:Fort Emporium Item Info Object Reference)
    Item:Fort Item Object Reference=@n5.FortItem
    ParentDefinition:Item Definition Base Object Reference
    bAllowInteractiveTags:bool=false
   ) -> (then)
   > UpdateWithFortItem(self=Get WBP_ItemPrice(), Item=@n5.FortItem, bAllowInteractiveTags=false)
   n6 = Get Image_SelectedItem() -> (Image_SelectedItem:Image Object Reference)
   > SetBrushFromSoftTexture!(
    execute
    self:Image Object Reference=@n6
    SoftTexture:Texture 2D Soft Object Reference=Select(
     Option 0:Texture 2D Soft Object Reference=GetSmallPreviewImage!(self:Fort Item Object Reference=@n5.FortItem) -> (ReturnValue:Texture 2D Soft Object Reference)
     Option 1:Texture 2D Soft Object Reference=GetLargePreviewImage!(self:Fort Item Object Reference=@n5.FortItem) -> (ReturnValue:Texture 2D Soft Object Reference)
     Index=@n5.bUseLargeThumbnail
    ) -> (ReturnValue:Texture 2D Soft Object Reference)
    bMatchSize:bool=false
   ) -> (then)
   > InvalidateLayoutAndVolatility(self:Widget Object Reference=@n6)
   Event OnInitialized()
   n7 = Get WBP_ItemPrice()
   > Bind Event to Show Additional Item Data(self:WBP Creative Emporium Item Price Object Reference=@n7, Delegate=Create Event()) {is=title K2Node_AddDelegate}
   > Bind Event to Show Item License(self:WBP Creative Emporium Item Price Object Reference=@n7, Delegate=Create Event()) {is=title K2Node_AddDelegate}
   Event Construct()
   > ProvideReferenceToDownNavigatedWidget!(
    execute
    self:WBP Creative Emporium Item Price Object Reference=Get WBP_ItemPrice()
    Widget:Common User Widget Object Reference=Get WBP_Creative_Emporium_Item_AdditionalInformation() -> (WBP_Creative_Emporium_Item_AdditionalInformation:WBP Creative Emporium Item Additional Information Object Reference)
   ) -> (then)
  ]}
 IsShowingAdditionalButtons {nodes [
   IsShowingAdditionalButtons() {is=title K2Node_FunctionEntry}
   > n1 = IsShowingAdditionalButtons!(execute, self:WBP Creative Emporium Item Price Object Reference=Get WBP_ItemPrice()) -> (then, IsShowingAdditionalButtons:bool) {is=title K2Node_CallFunction}
   > Return Node!(execute, ReturnValue:bool=@n1)
  ]}
 IsShowingExtendedDataPanel {nodes [
   IsShowingExtendedDataPanel() {is=title K2Node_FunctionEntry}
   > Return Node(ReturnValue=Get bShowingExtendedDataPanel() -> (bShowingExtendedDataPanel:bool))
  ]}
 NagivateDownFromTileView {nodes [
   NagivateDownFromTileView!() -> (then, Navigation:EUINavigation Enum) {is=title K2Node_FunctionEntry}
   > OnShowAdditionalItemData() {is=title K2Node_CallFunction}
   > Return Node(ReturnValue:Widget Object Reference=Get WBP_Creative_Emporium_Item_AdditionalInformation())
  ]}
 NavigateUpFromAdditionalInformation {nodes [
   NavigateUpFromAdditionalInformation!() -> (then, Navigation:EUINavigation Enum) {is=title K2Node_FunctionEntry}
   > Event On Main Scroll Back()
   > Return Node(ReturnValue=Get WBP_ItemPrice())
  ]}
 OnFocusReceived {nodes [
   OnFocusReceived!() -> (then, MyGeometry:Geometry Structure, InFocusEvent:Focus Event Structure) {is=title K2Node_FunctionEntry}
   > n1 = Sequence()
   > Branch(Condition=Get bShowingExtendedDataPanel())
   > Event On Main Scroll Back()
   SetFocus(execute=@n1.then_1, self=Get WBP_ItemPrice())
   > Return Node(ReturnValue:Event Reply Structure=Handled())
  ]}
]
variables [bShowingExtendedDataPanel {type=boolean}]
widgets {
 info {
  inheritedWidgetCount=3
  parentClass=/Script/CreativeEmporiumUI.FortEmporiumItemDetailsPanel
  rootWidgetClass=&Overlay
  widgetCount=21
 }
 widgets [
  Root {widgetClassPath=&Overlay, widget=Root}
  Image_Background {
   widgetClassPath=&Image
   parent=Root
   widget=Image_Background
   properties {brush {
     bIsDynamicallyLoaded=false
     drawAs=Image
     imageType=NoImage
     mirroring=NoMirror
     resourceName=None
     resourceObject=/CreativeEmporium/Materials/MI_Creative_Emporium_Background
     tiling=NoTile
     imageSize=<32 32>
     margin=<ltrb 0 0 0 0>
     outlineSettings {
      bUseBrushTransparency=false
      roundingType=HalfHeightRadius
      width=0
      color {colorUseRule=UseColor_Specified, specifiedColor=<rgba 0 0 0 0>}
      cornerRadii=<0 0 0 0>
     }
     tintColor {colorUseRule=UseColor_Specified, specifiedColor=<rgba 1 1 1 1>}
     uVRegion {bIsValid=false, max=<0 0>, min=<0 0>}
    }}
   slot {class=OverlaySlot, properties {horizontalAlignment=HAlign_Fill, verticalAlignment=VAlign_Fill}}
  }
  Content {
   widgetClassPath=&VerticalBox
   parent=Root
   widget=Content
   slot {class=OverlaySlot, properties {horizontalAlignment=HAlign_Fill, padding=<ltrb 62 135 62 0>}}
  }
  Main {widgetClassPath=&VerticalBox, parent=Content, widget=Main, slot {class=VerticalBoxSlot}}
  HorizontalBox_Main {widgetClassPath=&HorizontalBox, parent=Main, widget=HorizontalBox_Main, slot {class=VerticalBoxSlot}}
  VerticalBox_LeftPanel {
   widgetClassPath=&VerticalBox
   parent=HorizontalBox_Main
   widget=VerticalBox_LeftPanel
   slot {class=HorizontalBoxSlot, properties {horizontalAlignment=HAlign_Left, size {sizeRule=Fill, value=1}}}
  }
  ItemDetails {
   widgetClassPath=&/WBP_Creative_Emporium_Item_Info
   bInherited=true
   bIsVariable=true
   parent=VerticalBox_LeftPanel
   widget=ItemDetails
   slot {class=VerticalBoxSlot, properties {horizontalAlignment=HAlign_Left, verticalAlignment=VAlign_Top}}
  }
  WBP_ItemPrice {
   widgetClassPath=&/WBP_Creative_Emporium_Item_Price
   bIsVariable=true
   parent=VerticalBox_LeftPanel
   widget=WBP_ItemPrice
   slot {
    class=VerticalBoxSlot
    properties {horizontalAlignment=HAlign_Left, verticalAlignment=VAlign_Bottom, padding=<ltrb 0 0 0 -65>, size {sizeRule=Fill, value=1}}
   }
  }
  VerticalBox_RightPanel {
   widgetClassPath=&VerticalBox
   parent=HorizontalBox_Main
   widget=VerticalBox_RightPanel
   slot {class=HorizontalBoxSlot, properties {horizontalAlignment=HAlign_Left, verticalAlignment=VAlign_Top, size {sizeRule=Fill, value=1}}}
  }
  Image_SelectedItem {
   widgetClassPath=&Image
   bIsVariable=true
   parent=VerticalBox_RightPanel
   widget=Image_SelectedItem
   properties {brush {
     bIsDynamicallyLoaded=false
     drawAs=Image
     imageType=NoImage
     mirroring=NoMirror
     resourceName=None
     resourceObject=null
     tiling=NoTile
     imageSize=<936 532>
     margin=<ltrb 0 0 0 0>
     outlineSettings {
      bUseBrushTransparency=false
      roundingType=HalfHeightRadius
      width=0
      color {colorUseRule=UseColor_Specified, specifiedColor=<rgba 0 0 0 0>}
      cornerRadii=<0 0 0 0>
     }
     tintColor {colorUseRule=UseColor_Specified, specifiedColor=<rgba 1 1 1 1>}
     uVRegion {bIsValid=false, max=<0 0>, min=<0 0>}
    }}
   slot {class=VerticalBoxSlot, properties {horizontalAlignment=HAlign_Center, verticalAlignment=VAlign_Center}}
  }
  AthenaCreativeItemTileView_RelatedPreviews {
   widgetClassPath=/Script/FortniteUI.AthenaCreativeItemTileView
   bIsVariable=true
   parent=VerticalBox_RightPanel
   widget=AthenaCreativeItemTileView_RelatedPreviews
   properties {
    bReturnFocusToSelection=true
    entryHeight=146
    entryWidgetClass=&/WBP_Creative_Emporium_Item_Preview.WBP_Creative_Emporium_Item_Preview_C
    entryWidth=230
    horizontalEntrySpacing=25
    navigation=&self:WidgetTree.AthenaCreativeItemTileView_RelatedPreviews.WidgetNavigation_0
    numDesignerPreviewEntries=4
    orientation=Orient_Horizontal
    tileAlignment=LeftAligned
    verticalEntrySpacing=25
    scrollBarStyle {
     thickness=0
     draggedThumbImage {
      bIsDynamicallyLoaded=false
      drawAs=NoDrawType
      imageType=FullColor
      mirroring=NoMirror
      resourceName=../../../Engine/Content/Slate/Common/Scrollbar_Thumb.png
      resourceObject=null
      tiling=NoTile
      imageSize=<0 0>
      margin=<ltrb 0.25 0.25 0.25 0.25>
      outlineSettings {
       bUseBrushTransparency=false
       roundingType=HalfHeightRadius
       width=0
       color {colorUseRule=UseColor_Specified, specifiedColor=<rgba 0 0 0 0>}
       cornerRadii=<0 0 0 0>
      }
      tintColor {colorUseRule=UseColor_Specified, specifiedColor=<rgba 1 1 1 1>}
      uVRegion {bIsValid=false, max=<0 0>, min=<0 0>}
     }
     horizontalBackgroundImage {
      bIsDynamicallyLoaded=false
      drawAs=NoDrawType
      imageType=NoImage
      mirroring=NoMirror
      resourceName=None
      resourceObject=null
      tiling=NoTile
      imageSize=<0 0>
      margin=<ltrb 0 0 0 0>
      outlineSettings {
       bUseBrushTransparency=false
       roundingType=HalfHeightRadius
       width=0
       color {colorUseRule=UseColor_Specified, specifiedColor=<rgba 0 0 0 0>}
       cornerRadii=<0 0 0 0>
      }
      tintColor {colorUseRule=UseColor_Specified, specifiedColor=<rgba 1 1 1 1>}
      uVRegion {bIsValid=false, max=<0 0>, min=<0 0>}
     }
     horizontalBottomSlotImage {
      bIsDynamicallyLoaded=false
      drawAs=NoDrawType
      imageType=FullColor
      mirroring=NoMirror
      resourceName=../../../Engine/Content/Slate/Common/Scrollbar_Background_Horizontal.png
      resourceObject=null
      tiling=NoTile
      imageSize=<8 8>
      margin=<ltrb 0 0 0 0>
      outlineSettings {
       bUseBrushTransparency=false
       roundingType=HalfHeightRadius
       width=0
       color {colorUseRule=UseColor_Specified, specifiedColor=<rgba 0 0 0 0>}
       cornerRadii=<0 0 0 0>
      }
      tintColor {colorUseRule=UseColor_Specified, specifiedColor=<rgba 1 1 1 1>}
      uVRegion {bIsValid=false, max=<0 0>, min=<0 0>}
     }
     horizontalTopSlotImage {
      bIsDynamicallyLoaded=false
      drawAs=NoDrawType
      imageType=FullColor
      mirroring=NoMirror
      resourceName=../../../Engine/Content/Slate/Common/Scrollbar_Background_Horizontal.png
      resourceObject=null
      tiling=NoTile
      imageSize=<8 8>
      margin=<ltrb 0 0 0 0>
      outlineSettings {
       bUseBrushTransparency=false
       roundingType=HalfHeightRadius
       width=0
       color {colorUseRule=UseColor_Specified, specifiedColor=<rgba 0 0 0 0>}
       cornerRadii=<0 0 0 0>
      }
      tintColor {colorUseRule=UseColor_Specified, specifiedColor=<rgba 1 1 1 1>}
      uVRegion {bIsValid=false, max=<0 0>, min=<0 0>}
     }
     hoveredThumbImage {
      bIsDynamicallyLoaded=false
      drawAs=NoDrawType
      imageType=FullColor
      mirroring=NoMirror
      resourceName=../../../Engine/Content/Slate/Common/Scrollbar_Thumb.png
      resourceObject=null
      tiling=NoTile
      imageSize=<0 0>
      margin=<ltrb 0.25 0.25 0.25 0.25>
      outlineSettings {
       bUseBrushTransparency=false
       roundingType=HalfHeightRadius
       width=0
       color {colorUseRule=UseColor_Specified, specifiedColor=<rgba 0 0 0 0>}
       cornerRadii=<0 0 0 0>
      }
      tintColor {colorUseRule=UseColor_Specified, specifiedColor=<rgba 1 1 1 1>}
      uVRegion {bIsValid=false, max=<0 0>, min=<0 0>}
     }
     normalThumbImage {
      bIsDynamicallyLoaded=false
      drawAs=NoDrawType
      imageType=FullColor
      mirroring=NoMirror
      resourceName=../../../Engine/Content/Slate/Common/Scrollbar_Thumb.png
      resourceObject=null
      tiling=NoTile
      imageSize=<0 0>
      margin=<ltrb 0.25 0.25 0.25 0.25>
      outlineSettings {
       bUseBrushTransparency=false
       roundingType=HalfHeightRadius
       width=0
       color {colorUseRule=UseColor_Specified, specifiedColor=<rgba 0 0 0 0>}
       cornerRadii=<0 0 0 0>
      }
      tintColor {colorUseRule=UseColor_Specified, specifiedColor=<rgba 1 1 1 1>}
      uVRegion {bIsValid=false, max=<0 0>, min=<0 0>}
     }
     verticalBackgroundImage {
      bIsDynamicallyLoaded=false
      drawAs=NoDrawType
      imageType=NoImage
      mirroring=NoMirror
      resourceName=None
      resourceObject=null
      tiling=NoTile
      imageSize=<0 0>
      margin=<ltrb 0 0 0 0>
      outlineSettings {
       bUseBrushTransparency=false
       roundingType=HalfHeightRadius
       width=0
       color {colorUseRule=UseColor_Specified, specifiedColor=<rgba 0 0 0 0>}
       cornerRadii=<0 0 0 0>
      }
      tintColor {colorUseRule=UseColor_Specified, specifiedColor=<rgba 1 1 1 1>}
      uVRegion {bIsValid=false, max=<0 0>, min=<0 0>}
     }
     verticalBottomSlotImage {
      bIsDynamicallyLoaded=false
      drawAs=NoDrawType
      imageType=FullColor
      mirroring=NoMirror
      resourceName=../../../Engine/Content/Slate/Common/Scrollbar_Background_Vertical.png
      resourceObject=null
      tiling=NoTile
      imageSize=<8 8>
      margin=<ltrb 0 0 0 0>
      outlineSettings {
       bUseBrushTransparency=false
       roundingType=HalfHeightRadius
       width=0
       color {colorUseRule=UseColor_Specified, specifiedColor=<rgba 0 0 0 0>}
       cornerRadii=<0 0 0 0>
      }
      tintColor {colorUseRule=UseColor_Specified, specifiedColor=<rgba 1 1 1 1>}
      uVRegion {bIsValid=false, max=<0 0>, min=<0 0>}
     }
     verticalTopSlotImage {
      bIsDynamicallyLoaded=false
      drawAs=NoDrawType
      imageType=FullColor
      mirroring=NoMirror
      resourceName=../../../Engine/Content/Slate/Common/Scrollbar_Background_Vertical.png
      resourceObject=null
      tiling=NoTile
      imageSize=<8 8>
      margin=<ltrb 0 0 0 0>
      outlineSettings {
       bUseBrushTransparency=false
       roundingType=HalfHeightRadius
       width=0
       color {colorUseRule=UseColor_Specified, specifiedColor=<rgba 0 0 0 0>}
       cornerRadii=<0 0 0 0>
      }
      tintColor {colorUseRule=UseColor_Specified, specifiedColor=<rgba 1 1 1 1>}
      uVRegion {bIsValid=false, max=<0 0>, min=<0 0>}
     }
    }
   }
   slot {class=VerticalBoxSlot, properties {horizontalAlignment=HAlign_Left, verticalAlignment=VAlign_Top, padding=<ltrb -11 16 0 0>}}
  }
  Secondary {
   widgetClassPath=&SizeBox
   parent=Content
   widget=Secondary
   properties {bOverride_HeightOverride=true, heightOverride=948}
   slot {class=VerticalBoxSlot, properties {verticalAlignment=VAlign_Top, padding=<ltrb 0 75 0 0>}}
  }
  ScrollBox {
   widgetClassPath=/Script/FortniteUI.AthenaScrollBox
   parent=Secondary
   widget=ScrollBox
   properties {scrollBarVisibility=Collapsed}
   slot {class=SizeBoxSlot}
  }
  Switcher_ScrollBox {
   widgetClassPath=&CommonVisibilitySwitcher
   bIsVariable=true
   parent=ScrollBox
   widget=Switcher_ScrollBox
   slot {class=ScrollBoxSlot}
  }
  VerticalBox_AdditionalContent {
   widgetClassPath=&VerticalBox
   bIsVariable=true
   parent=Switcher_ScrollBox
   widget=VerticalBox_AdditionalContent
   slot {class=CommonVisibilitySwitcherSlot}
  }
  WBP_Creative_Emporium_Item_AdditionalInformation {
   widgetClassPath=&/WBP_Creative_Emporium_Item_AdditionalInformation
   bIsVariable=true
   parent=VerticalBox_AdditionalContent
   widget=WBP_Creative_Emporium_Item_AdditionalInformation
   properties {navigation=&self:WidgetTree.WBP_Creative_Emporium_Item_AdditionalInformation.WidgetNavigation_0}
   slot {class=VerticalBoxSlot}
  }
  WBP_Creative_Emporium_Item_SubPanel_Tags {
   widgetClassPath=&/WBP_Creative_Emporium_Item_SubPanel_Tags
   parent=VerticalBox_AdditionalContent
   widget=WBP_Creative_Emporium_Item_SubPanel_Tags
   slot {class=VerticalBoxSlot, properties {padding=<ltrb 48 58 0 0>}}
  }
  WBP_Creative_Emporium_Item_SubPanel_RelatedContent {
   widgetClassPath=&/WBP_Creative_Emporium_Item_SubPanel_RelatedContent
   parent=VerticalBox_AdditionalContent
   widget=WBP_Creative_Emporium_Item_SubPanel_RelatedContent
   slot {class=VerticalBoxSlot, properties {padding=<ltrb 48 64 0 0>}}
  }
  Text_Legal {
   widgetClassPath=/Script/FortniteUI.FortRichTextBlockLegacy
   bIsVariable=true
   parent=Switcher_ScrollBox
   widget=Text_Legal
   properties {styleSet=/Game/UI/Foundation/Legal/LegalScreenRichTextStyles}
   slot {class=CommonVisibilitySwitcherSlot}
  }
  Switcher_Details {widgetClassPath=&CommonVisibilitySwitcher, bInherited=true, widget=null}
  LegacyItemDetails {widgetClassPath=/Script/FortniteUI.AthenaInventoryItemInfo, bInherited=true, widget=null}
 ]
}
