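# Sound cue for the Creative mode in-game UI sound played when adding to a llama, as its name indicates.
# A single wave player feeds the root, playing the Creative_AddToChest sound wave, so it reuses the chest sound with no randomization or modulation.
# It routes to the non-critical in-game UI sound class and sets a very large max distance.
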SoundCue /Game/Sounds/Creative/IngameUI/Creative_AddToLlama_Cue
root=WavePlayer_0
nodes [{id=WavePlayer_0, properties {soundWaveAssetPtr=&/Creative_AddToChest}}]
properties {maxDistance=2097152, soundClassObject=&Fort_UI_InGame_NonCritical}
