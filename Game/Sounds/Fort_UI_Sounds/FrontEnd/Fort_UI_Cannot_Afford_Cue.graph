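# Sound cue for the front-end UI sound played, by its name, when the player cannot afford something.
# A single wave player feeds the output with the shared gameplay error sound Fort_UI_Error_02, with no randomization or modulation.
# It routes through the Fort_UI sound class and overrides concurrency to block new instances at the limit rather than cut off playing ones.
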SoundCue /Game/Sounds/Fort_UI_Sounds/FrontEnd/Fort_UI_Cannot_Afford_Cue
root=WavePlayer_1995
nodes [{id=WavePlayer_1995, properties {soundWaveAssetPtr=/Game/Sounds/Fort_UI_Sounds/Gameplay/Fort_UI_Error_02}}]
properties {
 bOverrideConcurrency=true
 maxDistance=2097152
 soundClassObject=&Fort_UI
 volumeMultiplier=1
 concurrencyOverrides {resolutionRule=PreventNew}
}
