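# Creative mode content browser menu widget for equipping, placing and dropping items, or filling a chest list that spawns as chest or llama.
# Left and right side panels open on hover and dock on click; slide animations play forward or reversed from their current time.
# A constructed UIKit tab view model switches the right panel between Details and Filters.
# Container classes come from Data Registry lookups, with hard-coded fallbacks.
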WidgetBlueprint /CreativeEmporium/WBP_Creative_Emporium_Browser_Tabs
animations [
 OpenPanel_Left {
  binding_count=2
  length_seconds=0.25
  object=&self:OpenPanel_Left
  sequence {
   bindings [
    Left_Background {
     id=#1
     possessed_class=BackgroundBlur
     tracks [
      {
       type=MovieSceneFloatTrack
       display_name=Render Opacity
       sections [{type=MovieSceneFloatSection, end_frame=5, start_frame=0, channels [None {type=MovieSceneScriptingFloatChannel, keys [[0,0],[5,0.7]]}]}]
      }
      {
       type=MovieScene2DTransformTrack
       sections [{
         type=MovieScene2DTransformSection
         end_frame=5
         start_frame=0
         channels [
          Translation.X {type=MovieSceneScriptingFloatChannel, keys [[0,-450],[4,20],[5,0]]}
          Translation.Y {type=MovieSceneScriptingFloatChannel}
          Angle {type=MovieSceneScriptingFloatChannel}
          Scale.X {type=MovieSceneScriptingFloatChannel}
          Scale.Y {type=MovieSceneScriptingFloatChannel}
          Shear.X {type=MovieSceneScriptingFloatChannel}
          Shear.Y {type=MovieSceneScriptingFloatChannel}
         ]
        }]
      }
      {
       type=MovieSceneByteTrack
       display_name=Visibility
       sections [{type=MovieSceneByteSection, end_frame=5, start_frame=0, channels [None {type=MovieSceneScriptingByteChannel, keys [[0,2],[1,0],[5,0]]}]}]
      }
     ]
    }
    Overlay_SearchFilters {
     id=#2
     possessed_class=Overlay
     tracks [
      {
       type=MovieScene2DTransformTrack
       sections [{
         type=MovieScene2DTransformSection
         end_frame=5
         start_frame=0
         channels [
          Translation.X {type=MovieSceneScriptingFloatChannel, keys [[0,-450],[4,20],[5,0]]}
          Translation.Y {type=MovieSceneScriptingFloatChannel}
          Angle {type=MovieSceneScriptingFloatChannel}
          Scale.X {type=MovieSceneScriptingFloatChannel}
          Scale.Y {type=MovieSceneScriptingFloatChannel}
          Shear.X {type=MovieSceneScriptingFloatChannel}
          Shear.Y {type=MovieSceneScriptingFloatChannel}
         ]
        }]
      }
      {
       type=MovieSceneFloatTrack
       display_name=Render Opacity
       sections [{type=MovieSceneFloatSection, end_frame=5, start_frame=0, channels [None {type=MovieSceneScriptingFloatChannel, keys [[0,0],[5,1]]}]}]
      }
      {
       type=MovieSceneByteTrack
       display_name=Visibility
       sections [{type=MovieSceneByteSection, end_frame=5, start_frame=0, channels [None {type=MovieSceneScriptingByteChannel, keys [[0,2],[1,0],[5,0]]}]}]
      }
     ]
    }
   ]
   tracks [{
     type=MovieSceneEventTrack
     sections [{
       type=MovieSceneEventTriggerSection
       channels [None {type=MovieSceneScriptingEventChannel, const ["<Struct 'MovieSceneEvent' (0x466c15c7940) {}>", 1]}]
      }]
    }]
  }
 }
 OpenPanel_Right {
  binding_count=2
  length_seconds=0.25
  object=&self:OpenPanel_Right
  sequence {
   bindings [
    Right_Background {
     id=#3
     possessed_class=BackgroundBlur
     tracks [
      {
       type=MovieScene2DTransformTrack
       sections [{
         type=MovieScene2DTransformSection
         end_frame=5
         start_frame=0
         channels [
          Translation.X {type=MovieSceneScriptingFloatChannel, keys [[0,300],[4,-20],[5,0]]}
          Translation.Y {type=MovieSceneScriptingFloatChannel}
          Angle {type=MovieSceneScriptingFloatChannel}
          Scale.X {type=MovieSceneScriptingFloatChannel}
          Scale.Y {type=MovieSceneScriptingFloatChannel}
          Shear.X {type=MovieSceneScriptingFloatChannel}
          Shear.Y {type=MovieSceneScriptingFloatChannel}
         ]
        }]
      }
      {
       type=MovieSceneFloatTrack
       display_name=Render Opacity
       sections [{type=MovieSceneFloatSection, end_frame=5, start_frame=0, channels [None {type=MovieSceneScriptingFloatChannel, keys [[0,0],[5,0.7]]}]}]
      }
      {
       type=MovieSceneByteTrack
       display_name=Visibility
       sections [{type=MovieSceneByteSection, end_frame=5, start_frame=0, channels [None {type=MovieSceneScriptingByteChannel, keys [[0,2],[1,0],[5,0]]}]}]
      }
     ]
    }
    Right_Content_Overlay {
     id=#4
     possessed_class=Overlay
     tracks [
      {
       type=MovieSceneFloatTrack
       display_name=Render Opacity
       sections [{type=MovieSceneFloatSection, end_frame=5, start_frame=0, channels [None {type=MovieSceneScriptingFloatChannel, keys [[0,0],[5,1]]}]}]
      }
      {
       type=MovieScene2DTransformTrack
       sections [{
         type=MovieScene2DTransformSection
         end_frame=5
         start_frame=0
         channels [
          Translation.X {type=MovieSceneScriptingFloatChannel, keys [[0,300],[4,-20],[5,0]]}
          Translation.Y {type=MovieSceneScriptingFloatChannel}
          Angle {type=MovieSceneScriptingFloatChannel}
          Scale.X {type=MovieSceneScriptingFloatChannel}
          Scale.Y {type=MovieSceneScriptingFloatChannel}
          Shear.X {type=MovieSceneScriptingFloatChannel}
          Shear.Y {type=MovieSceneScriptingFloatChannel}
         ]
        }]
      }
      {
       type=MovieSceneByteTrack
       display_name=Visibility
       sections [{type=MovieSceneByteSection, end_frame=5, start_frame=0, channels [None {type=MovieSceneScriptingByteChannel, keys [[0,2],[1,0],[5,0]]}]}]
      }
     ]
    }
   ]
   tracks [{
     type=MovieSceneEventTrack
     sections [{
       type=MovieSceneEventTriggerSection
       channels [None {type=MovieSceneScriptingEventChannel, const ["<Struct 'MovieSceneEvent' (0x466c15c7080) {}>", 1]}]
      }]
    }]
  }
 }
 Dock_RightPanel {
  binding_count=1
  length_seconds=0.25
  object=&self:Dock_RightPanel
  sequence {bindings [Right_Button_Overlay {
     id=#5
     possessed_class=Overlay
     tracks [
      {
       type=MovieScene2DTransformTrack
       sections [{
         type=MovieScene2DTransformSection
         end_frame=5
         start_frame=0
         channels [
          Translation.X {type=MovieSceneScriptingFloatChannel}
          Translation.Y {type=MovieSceneScriptingFloatChannel}
          Angle {type=MovieSceneScriptingFloatChannel}
          Scale.X {type=MovieSceneScriptingFloatChannel}
          Scale.Y {type=MovieSceneScriptingFloatChannel}
          Shear.X {type=MovieSceneScriptingFloatChannel}
          Shear.Y {type=MovieSceneScriptingFloatChannel}
         ]
        }]
      }
      {
       type=MovieSceneFloatTrack
       display_name=Render Opacity
       sections [{type=MovieSceneFloatSection, end_frame=5, start_frame=0, channels [None {type=MovieSceneScriptingFloatChannel, keys [[0,0],[5,1]]}]}]
      }
     ]
    }]}
 }
 Dock_LeftPanel {
  binding_count=1
  length_seconds=0.25
  object=&self:Dock_LeftPanel
  sequence {bindings [Left_Button_Overlay {
     id=#6
     display_name=Button_Left_Overlay
     possessed_class=Overlay
     tracks [
      {
       type=MovieScene2DTransformTrack
       sections [{
         type=MovieScene2DTransformSection
         end_frame=5
         start_frame=0
         channels [
          Translation.X {type=MovieSceneScriptingFloatChannel}
          Translation.Y {type=MovieSceneScriptingFloatChannel}
          Angle {type=MovieSceneScriptingFloatChannel}
          Scale.X {type=MovieSceneScriptingFloatChannel}
          Scale.Y {type=MovieSceneScriptingFloatChannel}
          Shear.X {type=MovieSceneScriptingFloatChannel}
          Shear.Y {type=MovieSceneScriptingFloatChannel}
         ]
        }]
      }
      {
       type=MovieSceneFloatTrack
       display_name=Render Opacity
       sections [{type=MovieSceneFloatSection, end_frame=5, start_frame=0, channels [None {type=MovieSceneScriptingFloatChannel, keys [[0,0],[5,1]]}]}]
      }
     ]
    }]}
 }
 OpenPanel_Left_NoBackground {
  binding_count=2
  length_seconds=0.25
  object=&self:OpenPanel_Left_NoBackground
  sequence {
   bindings [
    Left_Background {
     id=#7
     possessed_class=BackgroundBlur
     tracks [
      {
       type=MovieSceneFloatTrack
       display_name=Render Opacity
       sections [{type=MovieSceneFloatSection, end_frame=5, start_frame=0, channels [None {type=MovieSceneScriptingFloatChannel, const [0,2]}]}]
      }
      {
       type=MovieScene2DTransformTrack
       sections [{
         type=MovieScene2DTransformSection
         end_frame=5
         start_frame=0
         channels [
          Translation.X {type=MovieSceneScriptingFloatChannel, keys [[0,-450],[4,20],[5,0]]}
          Translation.Y {type=MovieSceneScriptingFloatChannel}
          Angle {type=MovieSceneScriptingFloatChannel}
          Scale.X {type=MovieSceneScriptingFloatChannel}
          Scale.Y {type=MovieSceneScriptingFloatChannel}
          Shear.X {type=MovieSceneScriptingFloatChannel}
          Shear.Y {type=MovieSceneScriptingFloatChannel}
         ]
        }]
      }
      {
       type=MovieSceneByteTrack
       display_name=Visibility
       sections [{type=MovieSceneByteSection, end_frame=5, start_frame=0, channels [None {type=MovieSceneScriptingByteChannel, keys [[0,2],[1,0],[5,0]]}]}]
      }
     ]
    }
    Overlay_SearchFilters {
     id=#8
     possessed_class=Overlay
     tracks [
      {
       type=MovieScene2DTransformTrack
       sections [{
         type=MovieScene2DTransformSection
         end_frame=5
         start_frame=0
         channels [
          Translation.X {type=MovieSceneScriptingFloatChannel, keys [[0,-450],[4,20],[5,0]]}
          Translation.Y {type=MovieSceneScriptingFloatChannel}
          Angle {type=MovieSceneScriptingFloatChannel}
          Scale.X {type=MovieSceneScriptingFloatChannel}
          Scale.Y {type=MovieSceneScriptingFloatChannel}
          Shear.X {type=MovieSceneScriptingFloatChannel}
          Shear.Y {type=MovieSceneScriptingFloatChannel}
         ]
        }]
      }
      {
       type=MovieSceneFloatTrack
       display_name=Render Opacity
       sections [{type=MovieSceneFloatSection, end_frame=5, start_frame=0, channels [None {type=MovieSceneScriptingFloatChannel, keys [[0,0],[5,1]]}]}]
      }
      {
       type=MovieSceneByteTrack
       display_name=Visibility
       sections [{type=MovieSceneByteSection, end_frame=5, start_frame=0, channels [None {type=MovieSceneScriptingByteChannel, keys [[0,2],[1,0],[5,0]]}]}]
      }
     ]
    }
   ]
   tracks [{
     type=MovieSceneEventTrack
     sections [{
       type=MovieSceneEventTriggerSection
       channels [None {type=MovieSceneScriptingEventChannel, const ["<Struct 'MovieSceneEvent' (0x466c15cf500) {}>", 1]}]
      }]
    }]
  }
 }
 OpenPanel_Left_NoFocus {
  binding_count=2
  length_seconds=0.25
  object=&self:OpenPanel_Left_NoFocus
  sequence {bindings [
    Left_Background {
     id=#1
     possessed_class=BackgroundBlur
     tracks [
      {
       type=MovieSceneFloatTrack
       display_name=Render Opacity
       sections [{type=MovieSceneFloatSection, end_frame=5, start_frame=0, channels [None {type=MovieSceneScriptingFloatChannel, keys [[0,0],[5,0.7]]}]}]
      }
      {
       type=MovieScene2DTransformTrack
       sections [{
         type=MovieScene2DTransformSection
         end_frame=5
         start_frame=0
         channels [
          Translation.X {type=MovieSceneScriptingFloatChannel, keys [[0,-450],[4,20],[5,0]]}
          Translation.Y {type=MovieSceneScriptingFloatChannel}
          Angle {type=MovieSceneScriptingFloatChannel}
          Scale.X {type=MovieSceneScriptingFloatChannel}
          Scale.Y {type=MovieSceneScriptingFloatChannel}
          Shear.X {type=MovieSceneScriptingFloatChannel}
          Shear.Y {type=MovieSceneScriptingFloatChannel}
         ]
        }]
      }
      {
       type=MovieSceneByteTrack
       display_name=Visibility
       sections [{type=MovieSceneByteSection, end_frame=5, start_frame=0, channels [None {type=MovieSceneScriptingByteChannel, keys [[0,2],[1,0],[5,0]]}]}]
      }
     ]
    }
    Overlay_SearchFilters {
     id=#2
     possessed_class=Overlay
     tracks [
      {
       type=MovieScene2DTransformTrack
       sections [{
         type=MovieScene2DTransformSection
         end_frame=5
         start_frame=0
         channels [
          Translation.X {type=MovieSceneScriptingFloatChannel, keys [[0,-450],[4,20],[5,0]]}
          Translation.Y {type=MovieSceneScriptingFloatChannel}
          Angle {type=MovieSceneScriptingFloatChannel}
          Scale.X {type=MovieSceneScriptingFloatChannel}
          Scale.Y {type=MovieSceneScriptingFloatChannel}
          Shear.X {type=MovieSceneScriptingFloatChannel}
          Shear.Y {type=MovieSceneScriptingFloatChannel}
         ]
        }]
      }
      {
       type=MovieSceneFloatTrack
       display_name=Render Opacity
       sections [{type=MovieSceneFloatSection, end_frame=5, start_frame=0, channels [None {type=MovieSceneScriptingFloatChannel, keys [[0,0],[5,1]]}]}]
      }
      {
       type=MovieSceneByteTrack
       display_name=Visibility
       sections [{type=MovieSceneByteSection, end_frame=5, start_frame=0, channels [None {type=MovieSceneScriptingByteChannel, keys [[0,2],[1,0],[5,0]]}]}]
      }
     ]
    }
   ]}
 }
 OpenPanel_Right_NoFocus {
  binding_count=2
  length_seconds=0.25
  object=&self:OpenPanel_Right_NoFocus
  sequence {bindings [
    Right_Background {
     id=#3
     possessed_class=BackgroundBlur
     tracks [
      {
       type=MovieScene2DTransformTrack
       sections [{
         type=MovieScene2DTransformSection
         end_frame=5
         start_frame=0
         channels [
          Translation.X {type=MovieSceneScriptingFloatChannel, keys [[0,300],[4,-20],[5,0]]}
          Translation.Y {type=MovieSceneScriptingFloatChannel}
          Angle {type=MovieSceneScriptingFloatChannel}
          Scale.X {type=MovieSceneScriptingFloatChannel}
          Scale.Y {type=MovieSceneScriptingFloatChannel}
          Shear.X {type=MovieSceneScriptingFloatChannel}
          Shear.Y {type=MovieSceneScriptingFloatChannel}
         ]
        }]
      }
      {
       type=MovieSceneFloatTrack
       display_name=Render Opacity
       sections [{type=MovieSceneFloatSection, end_frame=5, start_frame=0, channels [None {type=MovieSceneScriptingFloatChannel, keys [[0,0],[5,0.7]]}]}]
      }
      {
       type=MovieSceneByteTrack
       display_name=Visibility
       sections [{type=MovieSceneByteSection, end_frame=5, start_frame=0, channels [None {type=MovieSceneScriptingByteChannel, keys [[0,2],[1,0],[5,0]]}]}]
      }
     ]
    }
    Right_Content_Overlay {
     id=#4
     possessed_class=Overlay
     tracks [
      {
       type=MovieSceneFloatTrack
       display_name=Render Opacity
       sections [{type=MovieSceneFloatSection, end_frame=5, start_frame=0, channels [None {type=MovieSceneScriptingFloatChannel, keys [[0,0],[5,1]]}]}]
      }
      {
       type=MovieScene2DTransformTrack
       sections [{
         type=MovieScene2DTransformSection
         end_frame=5
         start_frame=0
         channels [
          Translation.X {type=MovieSceneScriptingFloatChannel, keys [[0,300],[4,-20],[5,0]]}
          Translation.Y {type=MovieSceneScriptingFloatChannel}
          Angle {type=MovieSceneScriptingFloatChannel}
          Scale.X {type=MovieSceneScriptingFloatChannel}
          Scale.Y {type=MovieSceneScriptingFloatChannel}
          Shear.X {type=MovieSceneScriptingFloatChannel}
          Shear.Y {type=MovieSceneScriptingFloatChannel}
         ]
        }]
      }
      {
       type=MovieSceneByteTrack
       display_name=Visibility
       sections [{type=MovieSceneByteSection, end_frame=5, start_frame=0, channels [None {type=MovieSceneScriptingByteChannel, keys [[0,2],[1,0],[5,0]]}]}]
      }
     ]
    }
   ]}
 }
 ShowFilterToast {
  binding_count=1
  length_seconds=3.15
  object=&self:ShowFilterToast
  sequence {bindings [FilterModeToast {
     id=#9
     possessed_class=SizeBox
     tracks [
      {
       type=MovieSceneFloatTrack
       display_name=Render Opacity
       sections [{
         type=MovieSceneFloatSection
         end_frame=63
         start_frame=0
         channels [None {type=MovieSceneScriptingFloatChannel, keys [[0,0],[5,1],[50,1],[63,0]]}]
        }]
      }
      {
       type=MovieScene2DTransformTrack
       sections [{
         type=MovieScene2DTransformSection
         end_frame=5
         start_frame=0
         channels [
          Translation.X {type=MovieSceneScriptingFloatChannel, keys [[0,56],[5,0]]}
          Translation.Y {type=MovieSceneScriptingFloatChannel}
          Angle {type=MovieSceneScriptingFloatChannel}
          Scale.X {type=MovieSceneScriptingFloatChannel}
          Scale.Y {type=MovieSceneScriptingFloatChannel}
          Shear.X {type=MovieSceneScriptingFloatChannel}
          Shear.Y {type=MovieSceneScriptingFloatChannel}
         ]
        }]
      }
     ]
    }]}
 }
]
graphs [
 AdjustGridSlotBottomPadding {nodes [
   n1 = AdjustGridSlotBottomPadding!() -> (then, Grid Slot Widget:Widget Object Reference, BottomPadding:double=~) {is=title K2Node_FunctionEntry}
   n2 = SlotAsGridSlot(Widget=@n1)
   n3 = Break Margin!(Margin:"Margin Structure (by ref)"=Get Padding!(self:Grid Slot Object Reference=@n2) -> (Padding:Margin Structure)) -> (Left:float=~, Top:float=~, Right:float=~, Bottom:float=~) {is=title K2Node_BreakStruct}
   > Set Padding!(
    execute
    Padding:Margin Structure=Make Margin(Left=@n3, Top=@n3.Top, Right=@n3.Right, Bottom=@n1.BottomPadding)
    self:Grid Slot Object Reference=@n2
   ) -> (then, Output_Get:Margin Structure)
  ]}
 Broadcast Panel Focused State {nodes [
   n1 = Broadcast Panel Focused State!() -> (then, Panel:bool=~, Focused State:bool=~) {is=title K2Node_FunctionEntry}
   > Broadcast Event!(
    execute
    self:Gameplay Event Router Component Object Reference=GetGlobalEventRouter()
    EventData:Gameplay Event Side Panel Focused Structure=Make Gameplay Event Side Panel Focused!(Panel_1_#10:bool=@n1, FocusedState_3_#11:bool=@n1.Focused State) -> (GameplayEvent_SidePanelFocused:Gameplay Event Side Panel Focused Structure) {is=title K2Node_MakeStruct}
    EventContext:Object Reference
   ) -> (then, ReturnValue:bool)
  ]}
 ConstructTabViewModel {nodes [
   ConstructTabViewModel() {is=title K2Node_FunctionEntry}
   > n1 = Construct Object from Class(Class=/Script/FortniteUI.UIKitTabGroupViewModel, self=GetOwningPlayer()) -> (ReturnValue:UI Kit Tab Group View Model Object Reference)
   > n2 = Sequence!(execute) -> (then_0, then_1, then_2:Exec)
   > n3 = Construct Object from Class(Class=/Script/FortniteUI.UIKitTabGroupButtonViewModel) -> (ReturnValue:UI Kit Tab Group Button View Model Object Reference)
   > Set Label!(execute, Label:Text, self:UI Kit Selectable Button View Model Object Reference=@n3) -> (then, Output_Get:Text)
   > Set Icon!(
    execute
    Icon:Texture 2D Object Reference=/FortUILibrary/Textures/Icons/T_UI_Icon_UIKit_Information
    self:UI Kit Selectable Button View Model Object Reference=@n3
   ) -> (then, Output_Get:Texture 2D Object Reference)
   > Set TabID!(execute, TabID:Name=Details, self:UI Kit Tab Group Button View Model Object Reference=@n3) -> (then, Output_Get:Name=None)
   > Add(
    TargetArray:Array of UI Kit Tab Group Button View Model Object References=Get TabsDataInfo() -> (TabsDataInfo:Array of UI Kit Tab Group Button View Model Object References)
    NewItem:"UI Kit Tab Group Button View Model Object Reference (by ref)"=@n3
   ) {is=title K2Node_CallArrayFunction}
   Set Tabs!(
    execute=@n2.then_2
    Tabs:Array of UI Kit Tab Group Button View Model Object References=Get TabsDataInfo()
    self:UI Kit Tab Group View Model Object Reference=@n1
   ) -> (then, Output_Get:Array of UI Kit Tab Group Button View Model Object References)
   n4 = Get RightSidePanel_Tabs() -> (RightSidePanel_Tabs:WBP UIKit Tabs Object Reference)
   > Set UIKitTabGroupViewModel!(execute, UIKitTabGroupViewModel:UI Kit Tab Group View Model Object Reference=@n1, self:WBP UIKit Tabs Object Reference=@n4) -> (then, Output_Get:UI Kit Tab Group View Model Object Reference)
   > SelectTabByIndex!(execute, self:UIKit Tab Group Object Reference=@n4, Index:int=1, bSuppressClickFeedback:bool=true) -> (then, ReturnValue:bool)
   n5 = Construct Object from Class(execute=@n2.then_1, Class=&UIKitTabGroupButtonViewModel)
   > Set Label(self=@n5)
   > Set Icon(Icon=/FortUILibrary/Textures/Icons/T_UI_Icon_UIKit_Options, self=@n5)
   > Set TabID(TabID=Filters, self=@n5) -> (Output_Get=None)
   > Add(TargetArray=Get TabsDataInfo(), NewItem=@n5)
  ]}
 EventGraph {nodes [
   "On Clicked (SidePanelButton_Left)"!() -> (OutputDelegate, then, Button:Common Button Base Object Reference) {is=title K2Node_ComponentBoundEvent}
   n1 = IsUsingTouch(PlayerController=GetOwningPlayer())
   > n2 = Branch(Condition=@n1)
   n3 = Set PanelDocked_Left(execute=[@n52.else, @n2.else, @n53.then], PanelDocked_Left:bool=NOT Boolean(A=Get PanelDocked_Left() -> (PanelDocked_Left:bool))) -> (Output_Get:bool)
   > Attempt Dock Panel!(execute, self, Right Panel:bool=false, Dock:bool=@n3, HIdeBackplate:bool=false, LeavingSubtab:bool=false) -> (then)
   "On Clicked (SearchPanel)"!() -> (OutputDelegate, then, Button:Common Button Base Object Reference) {is=title K2Node_ComponentBoundEvent}
   n4 = Get SearchPanel() -> (SearchPanel:WBP Creative Emporium Search Bar Object Reference)
   > SetFocus(self=Get EditableText!(self:WBP UIKit Block Editable Text Object Reference=Get Block_EditableText!(self:WBP UIKit Input Field Base Object Reference=@n4) -> (Block_EditableText:WBP UIKit Block Editable Text Object Reference)) -> (EditableText:UIKit Editable Text Object Reference))
   "On Clicked (Button_CreateLlama)"!() -> (OutputDelegate, then, Button:Common Button Base Object Reference) {is=title K2Node_ComponentBoundEvent}
   n5 = Get LlamaContainer() -> (LlamaContainer:Fort Container Table Row Structure)
   > n6 = Get Data Registry Item!(execute, ItemId:Data Registry Id Structure=FNE.Containers:Llama, OutItem:Fort Container Table Row Structure=@n5) -> (then, ReturnValue:bool) {is=title K2Node_CallFunction}
   > n7 = Branch(Condition=@n6)
   > n8 = SpawnContainer!(
    execute
    self:Fort Emporium Item List Menu Object Reference
    SupplyDropClass:Actor Class Reference=Break Fort Container Table Row!(FortContainerTableRow:"Fort Container Table Row Structure (by ref)"=@n5) -> (ContainerClass:Actor Class Reference) {is=title K2Node_BreakStruct}
   ) -> (then)
   n9 = PlaySound2D(execute=[@n11.then, @n8.then], Sound=/Game/Sounds/Creative/IngameUI/Creative_AddToLlama_Cue)
   n10 = Cast To WBP_Creative_Emporium_Browser_ListView!(execute=[@n9.then, @n41.then], Object=Get ChestTabPanel() -> (ChestTabPanel:Fort Emporium Item List Tab Panel Object Reference)) -> (then, CastFailed, AsWBP Creative Emporium Browser List View:WBP Creative Emporium Browser List View Object Reference) {is=title K2Node_DynamicCast}
   > Shine All Items(self:WBP Creative Emporium Browser List View Object Reference=@n10)
   n11 = SpawnContainer(execute=@n7.else, SupplyDropClass=/Game/Athena/SupplyDrops/Llama/AthenaCreativeSpawnDrop.AthenaCreativeSpawnDrop_C)
   "On Clicked (Button_RemoveFromChest)"!() -> (OutputDelegate, then, Button:Common Button Base Object Reference) {is=title K2Node_ComponentBoundEvent}
   > PlaySound2D(Sound=/Game/Sounds/Fort_UI_Sounds/FrontEnd/New_FrontEnd/Player_WarpOut_Cue)
   n12 = Get ChestTabPanel()
   > RemoveSelectedItem(self:Fort Emporium Item List Tab Panel Object Reference=@n12)
   > UpdateDetailsPanel(self:Fort Emporium Item List Menu Object Reference)
   "On Clicked (Button_ResetChest)"!() -> (OutputDelegate, then, Button:Common Button Base Object Reference) {is=title K2Node_ComponentBoundEvent}
   > PlaySound2D(Sound=/Game/Sounds/Fort_UI_Sounds/FrontEnd/New_FrontEnd/Fortnite_UI_QuestJournal_Menu_01_Cue)
   > ClearAllItems(self:Fort Emporium Item List Tab Panel Object Reference=@n12)
   > UpdateDetailsPanel()
   "On Clicked (Button_AddToChest)"!() -> (OutputDelegate, then, Button:Common Button Base Object Reference) {is=title K2Node_ComponentBoundEvent}
   n13 = GetCurrentSelection!(self:Fort Emporium Item List Menu Object Reference) -> (ReturnValue:Fort Item Entry Structure)
   > n14 = Branch(Condition=CanAddToChest!(self:Fort Emporium Item List Menu Object Reference, SelectedItem:"Fort Item Entry Structure (by ref)"=@n13) -> (ReturnValue:bool))
   > PlaySound2D(Sound=/Game/Sounds/Fort_UI_Sounds/FrontEnd/New_FrontEnd/Fortnite_UI_MainMenu_Generic_03_Cue)
   > AddItemToSource!(execute, self:Fort Emporium Item List Tab Panel Object Reference=Get ChestTabPanel(), ItemToAdd:"Fort Item Entry Structure (by ref)"=@n13) -> (then)
   > n15 = Cast To WBP_Creative_Emporium_Browser_ListView(Object=GetCurrentTab!(self:Fort Emporium Item List Menu Object Reference) -> (ReturnValue:Fort Creative Content Browser Tab Panel Base Object Reference))
   > Shine Selected Item(self:WBP Creative Emporium Browser List View Object Reference=@n15)
   PlaySound2D(execute=@n14.else, Sound=/Game/Sounds/Fort_UI_Sounds/FrontEnd/Fort_UI_Cannot_Afford_Cue)
   n16 = "On UIKit Tab Selected (RightSidePanel_Tabs)"!() -> (OutputDelegate, then, TabID:Name=None, TabVM:MVVM Base Viewmodel Object Reference) {is=title K2Node_ComponentBoundEvent}
   > Set Selected SubFilter Tab(Selected SubFilter Tab:Widget Object Reference=Select(
     Option 0:Widget Object Reference=Get FiltersSubPanel() -> (FiltersSubPanel:Fort Emporium Filters Sub Panel Object Reference)
     Option 1:Widget Object Reference=Get ItemDetailsSidePanel_DefaultSidePanel() -> (ItemDetailsSidePanel_DefaultSidePanel:Fort Emporium Item Details Panel Object Reference)
     Index="Equal (Name)"(A=@n16.TabID, B=Details)
    ) -> (ReturnValue:Widget Object Reference)) -> (Output_Get:Widget Object Reference)
   > UpdateSelectedDetailsPanelOption()
   > Update Tooltip Text()
   "On Hovered (SidePanelButton_Left)"!() -> (OutputDelegate, then, Button:Common Button Base Object Reference) {is=title K2Node_ComponentBoundEvent}
   > Branch(Condition=NOT Boolean(A=IsLeftDocked!() -> (PanelDocked_Left:bool)))
   > Open Panel!(execute, self, IsRightPanel:bool=false, Open:bool=true, HideBackplate:bool=false, LeavingSubTab:bool=false) -> (then)
   > n17 = Update Side Button Visuals()
   Delay(execute=[@n17.then, @n21.then], Duration=0.125)
   n18 = Get SidePanelButton_Left() -> (SidePanelButton_Left:WBP UIKit Button Generic Object Reference)
   n19 = Get Border_Left() -> (Border_Left:Border Object Reference)
   n20 = Get FiltersPanel() -> (FiltersPanel:Emporium Browser Filters Panel Object Reference)
   > n21 = Branch(Condition=OR Boolean!(A=IsHovered(self=@n20), B=IsHovered(self=@n18), C:bool=IsHovered(self=@n19)) -> (ReturnValue))
   n22 = Branch(execute=@n21.else, Condition=NOT Boolean(A=IsLeftDocked()))
   > n23 = Open Panel(IsRightPanel=false, Open=false, HideBackplate=false, LeavingSubTab=false)
   Update Side Button Visuals(execute=[@n23.then, @n22.else])
   "On Clicked (SidePanelButton_Right)"!() -> (OutputDelegate, then, Button:Common Button Base Object Reference) {is=title K2Node_ComponentBoundEvent}
   n24 = IsUsingTouch(PlayerController=GetOwningPlayer())
   > n25 = Branch(Condition=@n24)
   n26 = Set PanelDocked_Right(execute=[@n52.then, @n25.else, @n60.then], PanelDocked_Right:bool=NOT Boolean(A=Get PanelDocked_Right() -> (PanelDocked_Right:bool))) -> (Output_Get:bool)
   > Attempt Dock Panel(Right Panel=true, Dock=@n26, HIdeBackplate=false, LeavingSubtab=false)
   "On Hovered (SidePanelButton_Right)"!() -> (OutputDelegate, then, Button:Common Button Base Object Reference) {is=title K2Node_ComponentBoundEvent}
   > Branch(Condition=NOT Boolean(A=IsRightDocked!() -> (PanelDocked_Right:bool)))
   > Open Panel(IsRightPanel=true, Open=true, HideBackplate=false, LeavingSubTab=false)
   > n27 = Update Side Button Visuals()
   Delay(execute=[@n27.then, @n31.then], Duration=0.125)
   n28 = Get Right_Content_Overlay() -> (Right_Content_Overlay:Overlay Object Reference)
   n29 = Get SidePanelButton_Right() -> (SidePanelButton_Right:WBP UIKit Button Generic Object Reference)
   n30 = Get Border_Right() -> (Border_Right:Border Object Reference)
   > n31 = Branch(Condition=OR Boolean(A=IsHovered(self=@n28), B=IsHovered(self=@n29), C=IsHovered(self=@n30)))
   n32 = Branch(execute=@n31.else, Condition=NOT Boolean(A=IsRightDocked()))
   > n33 = Open Panel(IsRightPanel=true, Open=false, HideBackplate=false, LeavingSubTab=false)
   Update Side Button Visuals(execute=[@n33.then, @n32.else])
   "On Button Clicked (Button_HitBlocker_Right)"!() -> (OutputDelegate, then, Button:Common Button Legacy Object Reference) {is=title K2Node_ComponentBoundEvent}
   > HandleModalHitBlockerClicked(self:Fort Emporium Item List Menu Object Reference)
   "On Clicked (Button_PlaceNow)"!() -> (OutputDelegate, then, Button:Common Button Base Object Reference) {is=title K2Node_ComponentBoundEvent}
   > PlaceSelectionInMoveTool!(execute, self:Fort Emporium Item List Menu Object Reference) -> (then, ReturnValue:bool)
   "On Clicked (Button_OpenItem)"!() -> (OutputDelegate, then, Button:Common Button Base Object Reference) {is=title K2Node_ComponentBoundEvent}
   > OpenSelection!(execute, self:Fort Emporium Item List Menu Object Reference) -> (then, ReturnValue:bool)
   "On Clicked (Button_Equip)"!() -> (OutputDelegate, then, Button:Common Button Base Object Reference) {is=title K2Node_ComponentBoundEvent}
   > EquipSelection!(execute, self:Fort Emporium Item List Menu Object Reference) -> (then, ReturnValue:bool)
   "On Clicked (Button_AddToQuickBar)"!() -> (OutputDelegate, then, Button:Common Button Base Object Reference) {is=title K2Node_ComponentBoundEvent}
   > AddSelectionToQuickBar!(execute, self:Fort Emporium Item List Menu Object Reference) -> (then, ReturnValue:bool)
   "On Clicked (Button_Drop)"!() -> (OutputDelegate, then, Button:Common Button Base Object Reference) {is=title K2Node_ComponentBoundEvent}
   > n34 = DropItem!(execute, self:Fort Emporium Item List Menu Object Reference, SelectedItem:"Fort Item Entry Structure (by ref)"=GetCurrentSelection()) -> (then, ReturnValue:bool)
   > n35 = Branch(Condition=@n34)
   > PlaySound2D(Sound=&Fortnite_UI_MainMenu_Generic_03_Cue)
   > n36 = Cast To WBP_Creative_Emporium_Browser_ListView(Object=GetCurrentTab())
   > Shine Selected Item(self=@n36)
   > RestoreLastSelection(self:Fort Emporium Item List Tab Panel Object Reference=@n36)
   PlaySound2D(execute=@n35.else, Sound=&Fort_UI_Cannot_Afford_Cue)
   "On Clicked (Button_CreateChest)"!() -> (OutputDelegate, then, Button:Common Button Base Object Reference) {is=title K2Node_ComponentBoundEvent}
   n37 = Get ChestContainer() -> (ChestContainer:Fort Container Table Row Structure)
   > n38 = Get Data Registry Item(ItemId=FNE.Containers:Chest, OutItem=@n37)
   > n39 = Branch(Condition=@n38)
   > n40 = SpawnContainer(SupplyDropClass=Break Fort Container Table Row(FortContainerTableRow=@n37))
   n41 = PlaySound2D(execute=[@n40.then, @n42.then], Sound=/Game/Sounds/Creative/IngameUI/Creative_AddToChest_Cue)
   n42 = SpawnContainer(execute=@n39.else, SupplyDropClass=/Game/Building/ActorBlueprints/Containers/Creative_Tiered_Chest.Creative_Tiered_Chest_C)
   n43 = CreativeModeEnabledChanged!() -> (OutputDelegate, then, bIsEnabled:bool=~) {is=title K2Node_CustomEvent}
   > n44 = Branch(Condition=@n43.bIsEnabled)
   DeactivateWidget(execute=@n44.else)
   n45 = Open Panel!() -> (OutputDelegate, then, IsRightPanel:bool=~, Open:bool=~, HideBackplate:bool=~, LeavingSubTab:bool=~) {is=title K2Node_CustomEvent}
   > n46 = PanelShouldOpen!(execute, self, IsRightPanel:bool=@n45.IsRightPanel, RequestedOpen:bool=@n45.Open) -> (then, Open:bool)
   n47 = Select(
    Option 0:bool=Get PanelOpen_Left() -> (PanelOpen_Left:bool)
    Option 1:bool=Get PanelOpen_Right() -> (PanelOpen_Right:bool)
    Index=@n45.IsRightPanel
   ) -> (ReturnValue:bool)
   > Branch(Condition="Not Equal (Boolean)"(A=@n47, B=@n46) -> (ReturnValue=~))
   n48 = Get OpenPanel_Left_NoFocus() -> (OpenPanel_Left_NoFocus:Widget Animation Object Reference)
   n49 = Get OpenPanel_Left() -> (OpenPanel_Left:Widget Animation Object Reference)
   n50 = Select(
    Option 0:Widget Animation Object Reference=Select(
     Option 0=Select(Option 0=@n49, Option 1=@n48, Index=@n45.LeavingSubTab)
     Option 1=Get OpenPanel_Left_NoBackground() -> (OpenPanel_Left_NoBackground:Widget Animation Object Reference)
     Index=@n45.HideBackplate
    )
    Option 1:Widget Animation Object Reference=Select(
     Option 0=Get OpenPanel_Right() -> (OpenPanel_Right:Widget Animation Object Reference)
     Option 1=Get OpenPanel_Right_NoFocus() -> (OpenPanel_Right_NoFocus:Widget Animation Object Reference)
     Index=@n45.LeavingSubTab
    )
    Index=@n45.IsRightPanel
   ) -> (ReturnValue:Widget Animation Object Reference)
   > "Play Animation with Finished event (legacy)"!(
    execute:Exec
    Widget:User Widget Object Reference=Self-Reference()
    InAnimation:Widget Animation Object Reference=@n50
    StartAtTime:float=GetAnimationCurrentTime!(self:User Widget Object Reference, InAnimation:Widget Animation Object Reference=@n50) -> (ReturnValue:float)
    NumLoopsToPlay:int=1
    PlayMode:EUMGSequencePlayMode Enum=Select(Option 0:EUMGSequencePlayMode Enum=Reverse, Option 1:EUMGSequencePlayMode Enum, Index=@n46) -> (ReturnValue:EUMGSequencePlayMode Enum)
    PlaybackSpeed:float=1
   ) -> (then:Exec, Result:UMGSequence Player Object Reference, Finished:Exec) {is=title K2Node_PlayAnimation}
   > "Set Boolean (by ref)"!(execute:Exec, Target:"Boolean (by ref)"=@n47, Value:bool=@n46) -> (then:Exec) {is=title K2Node_VariableSetRef}
   > Update Tooltip Text()
   n51 = Toggle Dock Panel!() -> (OutputDelegate, then, RightPanel:bool=~) {is=title K2Node_CustomEvent}
   > n52 = Branch(Condition=@n51.RightPanel)
   Update Right Focus() {is=title K2Node_CustomEvent}
   > FocusSidePanel!(execute, self:Fort Emporium Item List Menu Object Reference, bIsRightPanel:bool=true, bFocus:bool=Get PanelOpen_Right()) -> (then, ReturnValue:Widget Object Reference)
   Update Left Focus() {is=title K2Node_CustomEvent}
   > FocusSidePanel(bIsRightPanel=false, bFocus=Get PanelOpen_Left())
   OnTagFilterModeToggle() {is=title K2Node_CustomEvent}
   > CycleTagFilterMode(self:Fort Emporium Item List Menu Object Reference)
   "On Released (SidePanelButton_Left)"!() -> (OutputDelegate, then, Button:WBP UIKit Button Generic Base Object Reference) {is=title K2Node_CustomEvent}
   > Branch(Condition=@n1)
   > Branch(Condition=Get SidePanelButton_Left_Pressed() -> (SidePanelButton_Left_Pressed:bool))
   > n53 = Set SidePanelButton_Left_Pressed(SidePanelButton_Left_Pressed:bool=false) -> (Output_Get:bool)
   Refresh Panel Details() {is=title K2Node_CustomEvent}
   > UpdateSidePanelKeybinds()
   > Update Panel Detail()
   > Update Side Button Visuals()
   > Update Tooltip Text()
   n54 = Attempt Dock Panel!() -> (OutputDelegate, then, Right Panel:bool=~, Dock:bool=~, HIdeBackplate:bool=~, LeavingSubtab:bool=~) {is=title K2Node_CustomEvent}
   > Open Panel(IsRightPanel=@n54.Right Panel, Open=@n54.Dock, HideBackplate=@n54.HIdeBackplate, LeavingSubTab=@n54.LeavingSubtab) {is=title K2Node_CallFunction}
   > Update Center Panel Location()
   > Refresh Panel Details() {is=title K2Node_CallFunction}
   > PlayAnimation(InAnimation=Select(
     Option 0:Widget Animation Object Reference=Get Dock_LeftPanel() -> (Dock_LeftPanel:Widget Animation Object Reference)
     Option 1:Widget Animation Object Reference=Get Dock_RightPanel() -> (Dock_RightPanel:Widget Animation Object Reference)
     Index=@n54.Right Panel
    ) -> (ReturnValue:Widget Animation Object Reference))
   SetPanelLayout() {is=title K2Node_CustomEvent}
   > SetVisibility(
    self=[
     Get CategoryHeader() -> (CategoryHeader:WBP Creative Emporium Browser Header Object Reference)
     Get Left_Background() -> (Left_Background:Background Blur Object Reference)
    ]
    InVisibility=SelfHitTestInvisible
   )
   > InputModeChanged!(execute, self, Index:ECommonInputType Enum=GetCurrentInputType(self=Get CommonInputSubsystem())) -> (then)
   > n55 = Sequence!(execute) -> (then_0, then_1)
   > UpdateSelectedDetailsPanelOption()
   > SetVisibility(self=Get RightSideBacking_Image() -> (RightSideBacking_Image:Image Object Reference), InVisibility=Visible)
   > Update Center Panel Location()
   > SetPadding!(
    execute
    self:Horizontal Box Slot Object Reference=SlotAsHorizontalBoxSlot!(Widget:Widget Object Reference=Get CategoryHeader()) -> (ReturnValue:Horizontal Box Slot Object Reference)
    InPadding:Margin Structure=Make Margin(Left=60, Top=0, Right=0, Bottom=0)
   ) -> (then)
   n56 = IsMobileGame()
   > SetGridSlotNudge!(
    execute
    Widget:Widget Object Reference=Get MenuQuickbar_QuickBars() -> (MenuQuickbar_QuickBars:Fort Creative Menu Quickbar Object Reference)
    Nudge:Vector 2D Structure=MakeVector2D(X=0, Y=Select(Option 0:double=30, Option 1:double=0, Index=@n56) -> (ReturnValue:double))
   ) -> (then)
   SetSideButtonLayout(execute=@n55.then_1)
   SetupNav() {is=title K2Node_CustomEvent}
   n57 = GetCurrentTab()
   > n58 = Is Valid(InputObject=@n57)
   > n59 = HandleTabSelected!(
    execute
    self:Fort Emporium Item List Menu Object Reference
    TabId:Name=Get TabName!(self:Fort Creative Content Browser Tab Panel Base Object Reference=@n57) -> (TabName:Name=None)
    bCaptureFocus:bool=false
   ) -> (then)
   Refresh Panel Details(execute=[@n59.then, @n58.Is Not Valid])
   "On Released (SidePanelButton_Right)"!() -> (OutputDelegate, then, Button:WBP UIKit Button Generic Base Object Reference) {is=title K2Node_CustomEvent}
   > Branch(Condition=@n24)
   > Branch(Condition=Get SidePanelButton_Right_Pressed() -> (SidePanelButton_Right_Pressed:bool))
   > n60 = Set SidePanelButton_Right_Pressed(SidePanelButton_Right_Pressed:bool=false) -> (Output_Get:bool)
   "On Pressed (SidePanelButton_Left)"() {is=title K2Node_CustomEvent}
   > Set SidePanelButton_Left_Pressed(SidePanelButton_Left_Pressed=true)
   "On Pressed (SidePanelButton_Right)"() {is=title K2Node_CustomEvent}
   > Set SidePanelButton_Right_Pressed(SidePanelButton_Right_Pressed=true)
   Open() {is=title K2Node_CustomEvent}
   > TryPushWidgetAsModalPanel!(execute, PushWidget:User Widget Object Reference=Self-Reference()) -> (then)
   n61 = Get FortHUDContext!() -> (ReturnValue:Fort HUDContext Object Reference) {is=title K2Node_GetBlueprintContextEx}
   > EnterCursorMode!(execute, self:Fort HUDContext Object Reference=@n61, ActionName:Name=None, CursorModeWidget:User Widget Object Reference) -> (then)
   > SetCursorModeLocked!(execute, self:Fort HUDContext Object Reference=@n61, bLocked:bool=true) -> (then)
   > PlaySound2D(Sound=Get MenuOpenedSound() -> (MenuOpenedSound:Sound Base Object Reference))
   n62 = OnChestItemCountChanged!() -> (OutputDelegate, then, Count:int) {is=title K2Node_CustomEvent}
   > UpdateChestItemCount!(execute, self, NumItems:int=@n62.Count) -> (then)
   > RefreshTabWarningState(self:Fort Emporium Item List Menu Object Reference)
   n63 = HandleInputModeChanged!() -> (OutputDelegate, then, bNewInputType:ECommonInputType Enum) {is=title K2Node_CustomEvent}
   > InputModeChanged(Index=@n63.bNewInputType)
   n64 = OnSelectionChangeInTab!() -> (OutputDelegate, then, TabPanel:Fort Creative Content Browser Tab Panel Base Object Reference) {is=title K2Node_CustomEvent}
   > UpdateDetailsPanel()
   n65 = InputModeChanged!() -> (OutputDelegate, then, Index:ECommonInputType Enum) {is=title K2Node_CustomEvent}
   > SetVisibility(
    self=[
     Get ChestCommandsBackplate_Image() -> (ChestCommandsBackplate_Image:Image Object Reference)
     Get VendCommandsBackplate_Image() -> (VendCommandsBackplate_Image:Image Object Reference)
    ]
    InVisibility=Select!(
     MouseAndKeyboard:ESlateVisibility Enum=Hidden
     Gamepad:ESlateVisibility Enum=SelfHitTestInvisible
     Touch:ESlateVisibility Enum=SelfHitTestInvisible
     Count:ESlateVisibility Enum=Visible
     Index:ECommonInputType Enum=@n65.Index
    ) -> (ReturnValue:ESlateVisibility Enum)
   )
   > Refresh Panel Details()
   > SetMobileStyles!(
    execute
    self
    IsUsingTouch:bool=Select(MouseAndKeyboard:bool=false, Gamepad:bool=false, Touch:bool=true, Count:bool=false, Index=@n65.Index) -> (ReturnValue:bool)
   ) -> (then)
   OnKeybindDockVisibilityChanged!() -> (OutputDelegate, then, InVisibility:ESlateVisibility Enum=Visible) {is=title K2Node_CustomEvent}
   > Update Tooltip Text()
   n66 = On Category Header Tag Selected!() -> (OutputDelegate, then, Tag ID:Name) {is=title K2Node_CustomEvent}
   > HandleItemTagSelected!(
    execute
    self:Fort Emporium Item Details Panel Object Reference=Get ItemDetailsSidePanel_DefaultSidePanel()
    TagID:Name=@n66.Tag ID
    bSelected:bool=false
   ) -> (then)
   Event OnItemsLoaded() {is=title K2Node_Event}
   > Bind Event to On Selection Changed(self:Fort Creative Content Browser Tab Panel Base Object Reference=GetCurrentTab(), Delegate=@n64) {is=title K2Node_AddDelegate}
   > UpdateDetailsPanel()
   Event OnExitSubTab() {is=title K2Node_Event}
   > SetSubTabOpened!(execute, self:Fort Emporium Item List Menu Object Reference, ItemDefinition:Item Definition Base Object Reference) -> (then)
   > SetSidePanelsForSubtab!(execute, self, EnteringSubtab:bool=false) -> (then)
   Event OnInitialized()
   > n67 = Cast To WBP_Creative_Emporium_PlaysetDetailsPanel!(execute, Object=Get SubTabItemDetailsPanel() -> (SubTabItemDetailsPanel:Fort Emporium Item Details Panel Object Reference)) -> (then, CastFailed, AsWBP Creative Emporium Playset Details Panel:WBP Creative Emporium Playset Details Panel Object Reference) {is=title K2Node_DynamicCast}
   > Bind Event to On Exit Subtab(self:WBP Creative Emporium Playset Details Panel Object Reference=@n67, Delegate=Create Event()) {is=title K2Node_AddDelegate}
   n68 = Get CategoryHeader()
   > Bind Event to Item Detail Tag Selected(self:WBP Creative Emporium Browser Header Object Reference=@n68, Delegate=Create Event()) {is=title K2Node_AddDelegate}
   > Bind Event to On Tag Filter Mode Toggle(self:WBP Creative Emporium Browser Header Object Reference=@n68, Delegate=Create Event()) {is=title K2Node_AddDelegate}
   n69 = Get FiltersSubPanel()
   > Set Selected SubFilter Tab(Selected SubFilter Tab=@n69)
   > Update Panel Detail()
   > n70 = Cast To WBP_Creative_Emporium_Filters_SubPanel!(execute, Object=@n69) -> (then, CastFailed, AsWBP Creative Emporium Filters Sub Panel:WBP Creative Emporium Filters Sub Panel Object Reference) {is=title K2Node_DynamicCast}
   > Bind Event to On Cycle Tag Filter Mode(self:WBP Creative Emporium Filters Sub Panel Object Reference=@n70, Delegate=Create Event()) {is=title K2Node_AddDelegate}
   > Bind Event to On Visibility Changed(
    self:User Widget Object Reference=Get Keybind_DockLeft_ScreenEdge() -> (Keybind_DockLeft_ScreenEdge:Keybind Widget Object Reference)
    Delegate=Create Event()
   ) {is=title K2Node_AddDelegate}
   > Bind Event to On Visibility Changed(self=Get Keybind_DockRight_ScreenEdge() -> (Keybind_DockRight_ScreenEdge:Keybind Widget Object Reference), Delegate=Create Event())
   Event On Activated()
   > SetupNav() {is=title K2Node_CallFunction}
   > SetHideInputAction!(
    execute
    self:Common Button Base Object Reference=Get SearchPanel()
    bInHideInputAction:bool=NOT Boolean(A=DoesPlatformRequireVirtualKeyboard!() -> (ReturnValue:bool))
   ) -> (then)
   > SetPanelLayout() {is=title K2Node_CallFunction}
   > n71 = Cast To FortPlayerControllerAthena(Object=GetOwningPlayer())
   > Bind Event to On Creative Mode Enabled Changed Delegate(self:Fort Player Controller Gameplay Object Reference=@n71, Delegate=Create Event()) {is=title K2Node_AddDelegate}
   > n72 = Cast To WBP_Creative_Emporium_Browser_Filters!(execute, Object=Get FiltersPanel()) -> (then, CastFailed, AsWBP Creative Emporium Browser Filters:WBP Creative Emporium Browser Filters Object Reference) {is=title K2Node_DynamicCast}
   > UpdateSearchActionIcon(self:WBP Creative Emporium Browser Filters Object Reference=@n72)
   Event PreConstruct()
   > SetMobileStyles(IsUsingTouch=IsUsingTouch(PlayerController=GetOwningPlayer()))
   n73 = Event OnUpdateCategoryLabel!() -> (OutputDelegate, then, SectionLabel:"Text (by ref)", CategoryLabel:"Text (by ref)") {is=title K2Node_Event}
   n74 = Get CategoryHeader()
   > UpdateCategoryLabel!(
    execute
    self:WBP Creative Emporium Browser Header Object Reference=@n74
    Section Label:Text=@n73.SectionLabel
    Category Label:Text=@n73.CategoryLabel
   ) -> (then)
   n75 = Get FiltersPanel()
   > SetActiveItemDetailTags!(
    execute
    self:WBP Creative Emporium Browser Header Object Reference=@n74
    Category Tags:Array of Names=GetIncludedTags!(self:Emporium Browser Filters Panel Object Reference=@n75) -> (ReturnValue:Array of Names)
    Creative Tags:Array of Names=GetItemDetailTags!(self:Emporium Browser Filters Panel Object Reference=@n75) -> (ReturnValue:Array of Names)
   ) -> (then)
   > UpdateSelectedDetailsPanelOption()
   Event OnSubPanelModeChanged() {is=title K2Node_Event}
   n76 = Get RightSidePanel_Tabs()
   > SelectNextTab!(execute, self:UIKit Tab Group Object Reference=@n76, bSuppressClickFeedback:bool=true) -> (then)
   > Set Selected SubFilter Tab(Selected SubFilter Tab=Select!(
     Option 0:Widget Object Reference=Get ItemDetailsSidePanel_DefaultSidePanel()
     Option 1:Widget Object Reference=Get FiltersSubPanel()
     Index:int=GetSelectedTabIndex!(self:UIKit Tab Group Object Reference=@n76) -> (ReturnValue:int=0)
    ) -> (ReturnValue:Widget Object Reference))
   > UpdateSelectedDetailsPanelOption()
   > Update Tooltip Text()
   n77 = Event OnShowCategoryModal!() -> (OutputDelegate, then, bShowModal:bool) {is=title K2Node_Event}
   > SetVisibility(
    self=[
     Get Blur_HitBlocker_Right() -> (Blur_HitBlocker_Right:Background Blur Object Reference)
     Get Button_HitBlocker_Right() -> (Button_HitBlocker_Right:Fort Text Button Object Reference)
    ]
    InVisibility=Select(Option 0:ESlateVisibility Enum=Collapsed, Option 1:ESlateVisibility Enum=Visible, Index:bool=@n77.bShowModal) -> (ReturnValue:ESlateVisibility Enum)
   )
   Event OnTabChanged() {is=title K2Node_Event}
   > UpdateSelectedDetailsPanelOption()
   n78 = Event OnShowGridBlocker!() -> (OutputDelegate, then, bShow:bool) {is=title K2Node_Event}
   > SetVisibility(self=Get Blur_HitBlocker_Right(), InVisibility=Select(Option 0=Collapsed, Option 1=SelfHitTestInvisible, Index=@n78.bShow))
   Event OnBuiltInItemsLoaded() {is=title K2Node_Event}
   > InitializeTabs(self:Fort Emporium Item List Menu Object Reference)
   > SetupNav()
   > Bind Event to On Item Count Changed(self:Fort Emporium Item List Tab Panel Object Reference=Get ChestTabPanel(), Delegate=@n62) {is=title K2Node_AddDelegate}
   > RefreshTabWarningState()
   n79 = Event OnUpdateTagFilterButton!() -> (OutputDelegate, then, Mode:EFortEmporiumItemTagFilterMode Enum=MatchAny) {is=title K2Node_Event}
   > UpdateTagFilterMode!(
    execute
    self:WBP Creative Emporium Browser Header Object Reference=Get CategoryHeader()
    NewParam:EFortEmporiumItemTagFilterMode Enum=@n79.Mode
   ) -> (then)
   > n80 = Cast To WBP_Creative_Emporium_Filters_SubPanel(Object=Get FiltersSubPanel())
   > ConfigureFilters!(execute, self:WBP Creative Emporium Filters Sub Panel Object Reference=@n80, TagFilterMode:EFortEmporiumItemTagFilterMode Enum=@n79.Mode) -> (then)
   > n81 = Branch(Condition="Equal (Enum)"(
     A:EFortEmporiumItemTagFilterMode Enum=@n79.Mode
     B:EFortEmporiumItemTagFilterMode Enum=Get LastFilterModeUsed() -> (LastFilterModeUsed:EFortEmporiumItemTagFilterMode Enum=MatchAny)
    ))
   Set LastFilterModeUsed(execute=[@n82.then, @n81.then], LastFilterModeUsed:EFortEmporiumItemTagFilterMode Enum=@n79.Mode) -> (Output_Get:EFortEmporiumItemTagFilterMode Enum=MatchAny)
   n82 = ShowFilterModeToast!(execute=@n81.else, self, FilterMode:EFortEmporiumItemTagFilterMode Enum=@n79.Mode) -> (then)
   n83 = Event OnAnimateSidePanel!() -> (OutputDelegate, then, bIsRightPanel:bool, bShow:bool) {is=title K2Node_Event}
   > Broadcast Panel Focused State!(execute, self, Panel:bool=@n83.bIsRightPanel, Focused State:bool=@n83.bShow) -> (then) {is=title K2Node_CallFunction}
   > n84 = Sequence!(execute) -> (then_0, then_1, then_2:Exec)
   > "Set Boolean (by ref)"(
    Target=Select(
     Option 0:bool=Get PanelFocused_Left() -> (PanelFocused_Left:bool)
     Option 1:bool=Get PanelFocused_Right() -> (PanelFocused_Right:bool)
     Index=@n83.bIsRightPanel
    ) -> (ReturnValue:bool)
    Value=@n83.bShow
   )
   n85 = IsRightDocked()
   n86 = IsLeftDocked()
   > n87 = Branch(Condition=NOT Boolean(A=Select(Option 0=@n86, Option 1=@n85, Index=@n83.bIsRightPanel)))
   > Open Panel(IsRightPanel=@n83.bIsRightPanel, Open=@n83.bShow, HideBackplate=false, LeavingSubTab=false)
   FocusSidePanel(execute=@n87.else, bIsRightPanel=@n83.bIsRightPanel, bFocus=@n83.bShow)
   > Branch(Condition=@n83.bIsRightPanel)
   > Not Home Tab!(execute:Exec) -> (True:Exec, false:Exec) {is=title K2Node_MacroInstance}
   > Force Filter View!(execute, self, On:bool=@n83.bShow) -> (then)
   Branch(execute=@n84.then_1, Condition=NOT Boolean(A=@n83.bShow))
   > FocusSidePanel(bIsRightPanel=@n83.bShow, bFocus=false)
   > UpdateSelectedDetailsPanelOption()
   Update Tooltip Text(execute=@n84.then_2)
   > Update Side Button Visuals()
   n88 = Event OnSelectionOpened!() -> (OutputDelegate, then, SelectedItemDefinition:Item Definition Base Object Reference) {is=title K2Node_Event}
   > SetSubTabOpened(ItemDefinition=@n88.SelectedItemDefinition)
   > SetSidePanelsForSubtab(EnteringSubtab=true)
   Event On Deactivated()
   > n89 = Cast To FortPlayerControllerAthena(Object=GetOwningPlayer())
   > Unbind Event from On Creative Mode Enabled Changed Delegate(self:Fort Player Controller Gameplay Object Reference=@n89, Delegate=Create Event()) {is=title K2Node_RemoveDelegate}
   Event OnSidePanelDockToggled() {is=title K2Node_Event}
   > n90 = Sequence!(execute) -> (then_0, then_1)
   > Branch(Condition=Get PanelFocused_Left())
   > Toggle Dock Panel!(execute, self, RightPanel:bool=false) -> (then) {is=title K2Node_CallFunction}
   Branch(execute=@n90.then_1, Condition=Get PanelFocused_Right())
   > Toggle Dock Panel(RightPanel=true)
   n91 = Event OnUpdateCommandBarVisibility!() -> (OutputDelegate, then, bShowBar:bool) {is=title K2Node_Event}
   > SetVisibility(
    self=Get Overlay_Switcher() -> (Overlay_Switcher:Overlay Object Reference)
    InVisibility=Select(Option 0:ESlateVisibility Enum=Hidden, Option 1:ESlateVisibility Enum=SelfHitTestInvisible, Index=@n91.bShowBar) -> (ReturnValue:ESlateVisibility Enum)
   )
   Event Construct()
   n92 = Get CommonInputSubsystem()
   > Bind Event to On Input Method Changed(self=@n92, Delegate=@n63)
   > InputModeChanged(Index=GetCurrentInputType(self=@n92)) {is=title K2Node_CallFunction}
   > n93 = Cast To WBP_Creative_Emporium_Browser_Filters(Object=Get FiltersPanel())
   > Override Search Bar!(
    execute
    self:WBP Creative Emporium Browser Filters Object Reference=@n93
    ParentSearchBar:WBP Creative Emporium Search Bar Object Reference=Get SearchPanel()
   ) -> (then)
   > Update Tooltip Text()
   > Update Side Button Visuals()
   > ConstructTabViewModel() {is=title K2Node_CallFunction}
   > n94 = GetDynamicMaterial(self:Border Object Reference=Get Border_Left())
   > Set DM_LeftSideNavigator_Backplate(DM_LeftSideNavigator_Backplate:Material Instance Dynamic Object Reference=@n94) -> (Output_Get:Material Instance Dynamic Object Reference)
   n95 = Get SidePanelButton_Left()
   > Bind Event to On Pressed(self:WBP UIKit Button Generic Object Reference=@n95, Delegate=Create Event()) {is=title K2Node_AddDelegate}
   > Bind Event to On Released(self:WBP UIKit Button Generic Base Object Reference=@n95, Delegate=Create Event()) {is=title K2Node_AddDelegate}
   n96 = Get SidePanelButton_Right()
   > Bind Event to On Pressed(self=@n96, Delegate=Create Event())
   > Bind Event to On Released(self=@n96, Delegate=Create Event())
   Close() {is=title K2Node_CustomEvent}
  ]}
 FilterToItemGridNavigation {nodes [
   n1 = FilterToItemGridNavigation!() -> (then, Navigation:EUINavigation Enum) {is=title K2Node_FunctionEntry}
   > Set IsRightPanel(IsRightPanel:bool="Equal (Enum)"(A:EUINavigation Enum=@n1, B:EUINavigation Enum=Left)) -> (Output_Get:bool)
   > n2 = Sequence()
   n3 = NOT Boolean(A=Get PanelFocused_Right())
   n4 = Get IsRightPanel() -> (IsRightPanel:bool)
   n5 = NOT Boolean(A=Get PanelFocused_Left())
   n6 = NOT Boolean(A=Get IsRightPanel())
   > Branch(Condition=OR Boolean!(A=AND Boolean(A=@n6, B=@n5), B=AND Boolean(A=@n4, B=@n3)) -> (ReturnValue))
   n7 = GetVisibility!(self:Widget Object Reference=Get Blur_HitBlocker_Right()) -> (ReturnValue:ESlateVisibility Enum=Visible)
   n8 = OR Boolean(A="Equal (Enum)"(A:ESlateVisibility Enum=@n7, B:ESlateVisibility Enum=Collapsed), B="Equal (Enum)"(A=@n7, B=Hidden))
   n9 = Get IsRightPanel()
   n10 = NOT Boolean(A=Get IsRightPanel())
   n11 = GetVisibility(self=Get Blur_HitBlocker_Left() -> (Blur_HitBlocker_Left:Background Blur Object Reference)) -> (ReturnValue=Visible)
   n12 = OR Boolean(A="Equal (Enum)"(A=@n11, B=Collapsed), B="Equal (Enum)"(A=@n11, B=Hidden))
   > Branch(Condition=OR Boolean(A=AND Boolean(A=@n12, B=@n10), B=AND Boolean(A=@n8, B=@n9)))
   n13 = GetCurrentTab()
   > n14 = Cast To WBP_Creative_Emporium_Browser_ListView(Object=@n13)
   > Set FocusWidget(FocusWidget:Widget Object Reference=Get CurrentEntryWidget!(self:WBP Creative Emporium Browser List View Object Reference=@n14) -> (CurrentEntryWidget:WBP Creative Emporium Item Entry Object Reference)) -> (Output_Get:Widget Object Reference)
   n15 = Cast To WBP_Creative_Emporium_HomeTabPanel!(execute=@n14.CastFailed, Object=@n13) -> (then, CastFailed, AsWBP Creative Emporium Home Tab Panel:WBP Creative Emporium Home Tab Panel Object Reference) {is=title K2Node_DynamicCast}
   > n16 = GetFocusWidget!(
    execute
    self:Fort Emporium Home List View Object Reference=Get FortEmporiumHomeListView_Options!(self:Fort Emporium Home Tab Panel Object Reference=@n15) -> (FortEmporiumHomeListView_Options:Fort Emporium Home List View Object Reference)
   ) -> (then, ReturnValue:Widget Object Reference)
   > Set FocusWidget(FocusWidget=@n16)
   n17 = Cast To WBP_Creative_Emporium_LevelManagerPanel!(execute=@n15.CastFailed, Object=@n13) -> (then, CastFailed, AsWBP Creative Emporium Level Manager Panel:WBP Creative Emporium Level Manager Panel Object Reference) {is=title K2Node_DynamicCast}
   > n18 = RefreshFocusWidget!(execute, self:Fort Creative Level Management Tab Panel Object Reference=@n17) -> (then, ReturnValue:Widget Object Reference)
   > n19 = Set FocusWidget(FocusWidget=@n18)
   > Is Valid(InputObject=@n19)
   n20 = GetOwningPlayer()
   > Branch(Condition=NOT Boolean(A=HasUserFocus!(self:Widget Object Reference=@n19, PlayerController:Player Controller Object Reference=@n20) -> (ReturnValue:bool)))
   > SetFocus(self=@n19)
   UpdateBackButtonText!(execute=@n2.then_1, self:Fort Emporium Item List Menu Object Reference, bFilterFocus:bool=false) -> (then)
   > Return Node!(execute, ReturnValue:Widget Object Reference=Get FocusWidget() -> (FocusWidget:Widget Object Reference))
  ]}
 Force Filter View {nodes [
   n1 = Force Filter View!() -> (then, On:bool=~) {is=title K2Node_FunctionEntry}
   > n2 = Set Filter View Forced(Filter View Forced:bool=@n1) -> (Output_Get:bool)
   > n3 = Branch(Condition=@n2)
   > SetActiveWidget(self=Get Switcher_SidePanels() -> (Switcher_SidePanels:Common Visibility Switcher Object Reference), Widget=Get FiltersSubPanel())
   > Set Filter View Forced(Filter View Forced=false)
   > UnregisterScrollRecipientExternal!(execute, self:Common User Widget Object Reference, AnalogScrollRecipient:Widget Object Reference=Get ItemDetailsSidePanel_DefaultSidePanel()) -> (then)
   UpdateSelectedDetailsPanelOption(execute=@n3.else)
  ]}
 GetEmptyTabFocusWidget {nodes [
   n1 = GetEmptyTabFocusWidget!() -> (then, Tab:Fort Emporium Item List Tab Panel Object Reference) {is=title K2Node_FunctionEntry}
   n2 = Get Button_Mode!(self:WBP Creative Emporium Browser Header Object Reference=Get CategoryHeader()) -> (Button_Mode:WBP UIKit Button Generic Object Reference)
   > n3 = Branch(Condition=IsVisible(self=@n2))
   > Return Node(ReturnValue=@n2)
   n4 = Get ChestTabPanel()
   Return Node(
    execute=@n3.else
    ReturnValue=Select(
     Option 0:Widget Object Reference=Get Button_TabEmpty() -> (Button_TabEmpty:WBP UIKit Button Quiet Object Reference)
     Option 1:Widget Object Reference=Get Button_ChestEmpty() -> (Button_ChestEmpty:WBP UIKit Button Quiet Object Reference)
     Index="Equal (Object)"(A=@n1, B=@n4)
    ) -> (ReturnValue:Widget Object Reference)
   )
  ]}
 IsLeftDocked {nodes [
   IsLeftDocked() {is=title K2Node_FunctionEntry}
   > n1 = IsUsingSidePanels!(execute, self) -> (then, ReturnValue:bool)
   > Return Node!(execute, PanelDocked_Left:bool=OR Boolean(A=NOT Boolean(A=@n1), B=Get PanelDocked_Left()))
  ]}
 IsRightDocked {nodes [
   IsRightDocked() {is=title K2Node_FunctionEntry}
   > n1 = IsUsingSidePanels()
   > Return Node!(execute, PanelDocked_Right:bool=AND Boolean(A=@n1, B=Get PanelDocked_Right()))
  ]}
 IsShowingModal {nodes [
   IsShowingModal() {is=title K2Node_FunctionEntry}
   n1 = GetCurrentTab()
   > n2 = Cast To WBP_Creative_Emporium_Browser_ListView(Object=@n1)
   > Return Node!(
    execute
    ReturnValue:bool=IsShowingModal!(self:WBP Creative Emporium Browser List View Object Reference=@n2) -> (Showing:bool) {is=title K2Node_CallFunction}
   )
   n3 = Cast To WBP_Creative_Emporium_HomeTabPanel(execute=@n2.CastFailed, Object=@n1)
   > Return Node(ReturnValue=IsShowingModal(self:WBP Creative Emporium Home Tab Panel Object Reference=@n3))
   Return Node(execute=@n3.CastFailed, ReturnValue=false)
  ]}
 IsSidePanelOpen {nodes [
   n1 = IsSidePanelOpen!() -> (then, bRightPanel:bool) {is=title K2Node_FunctionEntry}
   > Return Node(ReturnValue=Select(Option 0:bool=Get PanelOpen_Left(), Option 1:bool=Get PanelOpen_Right(), Index=@n1) -> (ReturnValue:bool))
  ]}
 IsUsingSidePanels {nodes [
   IsUsingSidePanels() {is=title K2Node_FunctionEntry}
   n1 = GetCurrentTab()
   > n2 = Is Valid(InputObject=@n1)
   Return Node(execute=@n2.Is Not Valid, ReturnValue=true)
   > Return Node(ReturnValue=IsUsingSidePanel!(self:Fort Creative Content Browser Tab Panel Base Object Reference=@n1) -> (ReturnValue:bool))
  ]}
 Leaving Filter Left Bound Custom {nodes [
   n1 = Leaving Filter Left Bound Custom!() -> (then, Navigation:EUINavigation Enum) {is=title K2Node_FunctionEntry}
   n2 = Get FortGlobalUIContext() -> (ReturnValue:Fort Global UIContext Object Reference) {is=title K2Node_GetSubsystem}
   > n3 = Branch(Condition=AND Boolean(
     A="Equal (Enum)"(A:EUINavigation Enum=@n1, B:EUINavigation Enum=Left)
     B=IsUsingGamepad!(self:Fort Global UIContext Object Reference=@n2) -> (ReturnValue:bool)
    ))
   > n4 = FocusSidePanel(bIsRightPanel=false, bFocus=Get PanelOpen_Left())
   > Return Node(ReturnValue:Widget Object Reference=@n4)
   Return Node(execute=@n3.else)
  ]}
 NavigateToCategoryHeader {nodes [
   NavigateToCategoryHeader!() -> (then, Navigation:EUINavigation Enum) {is=title K2Node_FunctionEntry}
   n1 = Get Button_Mode(self=Get CategoryHeader())
   > Return Node(ReturnValue=Select(Option 0:Widget Object Reference, Option 1:Widget Object Reference=@n1, Index=IsVisible(self=@n1)) -> (ReturnValue:Widget Object Reference))
  ]}
 Not Home Tab {nodes [
   Inputs()
   > n1 = Branch(Condition="Not Equal (Object)"(A:Object Reference=GetCurrentTab(), B:Object Reference=Get HomeTabPanel() -> (HomeTabPanel:Fort Emporium Home Tab Panel Object Reference)) -> (ReturnValue:bool=~) {is=title K2Node_PromotableOperator})
   > Outputs!(True:Exec, false:Exec=@n1.else)
  ]}
 OnRefreshFocusWidget {nodes [
   OnRefreshFocusWidget() {is=title K2Node_FunctionEntry}
   n1 = Get EditableText(self=Get Block_EditableText(self=Get SearchPanel()))
   > n2 = Branch(Condition=HasUserFocus(self=@n1, PlayerController=GetOwningPlayer()))
   > Return Node(ReturnValue=@n1)
   Return Node(execute=@n2.else)
  ]}
 OnShouldAllowCommandBar {nodes [
   n1 = OnShouldAllowCommandBar!() -> (then, FocusPathWidget:User Widget Object Reference) {is=title K2Node_FunctionEntry}
   > n2 = Sequence!(execute) -> (then_0, then_1, then_2:Exec)
   > Branch(Condition="Equal (Object)"(A=@n1, B=Get CategoryHeader()))
   > Return Node!(execute, bAllowCommandBar:bool=false, ReturnValue:bool=true)
   Branch(execute=@n2.then_1, Condition="Equal (Object)"(A=@n1, B=Get SearchPanel()))
   > Return Node(bAllowCommandBar=false, ReturnValue=true)
   Return Node(execute=@n2.then_2, bAllowCommandBar=false, ReturnValue=false)
  ]}
 OnShouldAllowModifierKeys {nodes [
   OnShouldAllowModifierKeys() {is=title K2Node_FunctionEntry}
   n1 = HasAnyUserFocusedDescendants!(self:Widget Object Reference=Get SearchPanel()) -> (ReturnValue:bool)
   n2 = GetCurrentTab()
   n3 = Get SubItemsTabPanel() -> (SubItemsTabPanel:Fort Emporium Item List Tab Panel Object Reference)
   > Return Node!(execute, ReturnValue=AND Boolean(A="Not Equal (Object)"(A=@n2, B=@n3) -> (ReturnValue=~), B=NOT Boolean(A=@n1)))
  ]}
 PanelShouldOpen {nodes [
   n1 = PanelShouldOpen!() -> (then, IsRightPanel:bool=~, RequestedOpen:bool=~) {is=title K2Node_FunctionEntry}
   > n2 = IsUsingSidePanels() {is=title K2Node_CallFunction}
   n3 = NOT Boolean(A=@n2)
   > Return Node!(
    execute
    Open:bool=Select(Option 0:bool=OR Boolean(A=@n3, B=@n1.RequestedOpen), Option 1:bool=AND Boolean(A=@n1.RequestedOpen, B=@n2), Index=@n1) -> (ReturnValue:bool)
   )
  ]}
 SetMobileStyles {nodes [
   SetMobileStyles!() -> (then, IsUsingTouch:bool=~) {is=title K2Node_FunctionEntry}
   n1 = Get IsUsingTouch() -> (IsUsingTouch:bool)
   n2 = Select(Option 0:double=Get BottomBarOffset() -> (BottomBarOffset:double), Option 1:double=0, Index=@n1) -> (ReturnValue:double)
   > AdjustGridSlotBottomPadding!(
    execute
    self
    Grid Slot Widget:Widget Object Reference=Get Left_Keybinds_Overlay() -> (Left_Keybinds_Overlay:Overlay Object Reference)
    BottomPadding:double=@n2
   ) -> (then) {is=title K2Node_CallFunction}
   > AdjustGridSlotBottomPadding(Grid Slot Widget=Get Right_Keybinds_Overlay() -> (Right_Keybinds_Overlay:Overlay Object Reference), BottomPadding=@n2)
   n3 = Select(Option 0=0, Option 1=Get BottomBarOffset(), Index=@n1)
   > AdjustGridSlotBottomPadding(Grid Slot Widget=Get Left_Content_Overlay() -> (Left_Content_Overlay:Overlay Object Reference), BottomPadding=@n3)
   > AdjustGridSlotBottomPadding(Grid Slot Widget=Get Left_Background(), BottomPadding=@n3)
   > AdjustGridSlotBottomPadding(Grid Slot Widget=Get Right_Content_Overlay(), BottomPadding=@n3)
   > AdjustGridSlotBottomPadding(Grid Slot Widget=Get Right_Background() -> (Right_Background:Background Blur Object Reference), BottomPadding=@n3)
   > Branch(Condition=IsMobileGame())
   > SetVisibility(
    self=[
     Get SizeBox_108() -> (SizeBox_108:Size Box Object Reference)
     Get SizeBox_106() -> (SizeBox_106:Size Box Object Reference)
     Get SizeBox_6() -> (SizeBox_6:Size Box Object Reference)
     Get SizeBox_3() -> (SizeBox_3:Size Box Object Reference)
     Get SizeBox_2() -> (SizeBox_2:Size Box Object Reference)
     Get SizeBox_5() -> (SizeBox_5:Size Box Object Reference)
    ]
    InVisibility=Select(Option 0:ESlateVisibility Enum=SelfHitTestInvisible, Option 1:ESlateVisibility Enum=Collapsed, Index=Get IsUsingTouch()) -> (ReturnValue:ESlateVisibility Enum)
   )
  ]}
 SetSideButtonLayout {nodes [
   SetSideButtonLayout() {is=title K2Node_FunctionEntry}
   n1 = IsMobileGame()
   > SetGridSlotPadding!(
    execute
    Widget:Widget Object Reference=Get Left_Button_Overlay() -> (Left_Button_Overlay:Overlay Object Reference)
    Padding:Margin Structure=Make Margin(Left=Select(Option 0:float=-132, Option 1:float=-96, Index=@n1) -> (ReturnValue:float), Top=42, Right=0, Bottom=0)
   ) -> (then)
   > SetGridSlotPadding(
    Widget=Get Right_Button_Overlay() -> (Right_Button_Overlay:Overlay Object Reference)
    Padding=Make Margin(Left=0, Top=42, Right=Select(Option 0=-132, Option 1=-32, Index=@n1), Bottom=0)
   )
   > SetGridSlotPadding(
    Widget=Get Block_Icon!(self:WBP UIKit Button Generic Base Object Reference=Get SidePanelButton_Right()) -> (Block_Icon:WBP UIKit Block Image Object Reference)
    Padding=Make Margin(Left=Select(Option 0=-46, Option 1=0, Index=@n1), Top=0, Right=0, Bottom=0)
   )
  ]}
 SetSidePanelDocked {nodes [
   n1 = SetSidePanelDocked!() -> (then, Right Side:bool=~, Docked:bool=~) {is=title K2Node_FunctionEntry}
   > n2 = Branch(Condition=@n1)
   > Branch(Condition="Not Equal (Boolean)"(A=Get PanelDocked_Right(), B=@n1.Docked) -> (ReturnValue=~))
   > Toggle Dock Panel(RightPanel=true)
   Branch(execute=@n2.else, Condition="Not Equal (Boolean)"(A=@n1.Docked, B=Get PanelDocked_Left()) -> (ReturnValue=~))
   > Toggle Dock Panel(RightPanel=false)
  ]}
 SetSidePanelsForSubtab {nodes [
   n1 = SetSidePanelsForSubtab!() -> (then, EnteringSubtab:bool=~) {is=title K2Node_FunctionEntry}
   > SetRenderOpacity(self=Get Left_Background(), InOpacity=Select(Option 0=1, Option 1=0, Index=@n1))
   > UpdateDetailsPanel()
   n2 = NOT Boolean(A=@n1)
   > Attempt Dock Panel(Right Panel=true, Dock=IsRightDocked() {is=title K2Node_CallFunction}, HIdeBackplate=@n1, LeavingSubtab=@n2) {is=title K2Node_CallFunction}
   > Attempt Dock Panel(Right Panel=false, Dock=IsLeftDocked() {is=title K2Node_CallFunction}, HIdeBackplate=@n1, LeavingSubtab=@n2)
   > Update Side Button Visuals()
  ]}
 ShowBackgroundBlurHeader {nodes [
   ShowBackgroundBlurHeader() {is=title K2Node_FunctionEntry}
   > Return Node!(execute, bShowBackgroundBlurHeader:bool=false)
  ]}
 ShowFilterModeMessage {nodes [ShowFilterModeMessage() {is=title K2Node_FunctionEntry}]}
 ShowFilterModeToast {nodes [
   n1 = ShowFilterModeToast!() -> (then, FilterMode:EFortEmporiumItemTagFilterMode Enum=MatchAny) {is=title K2Node_FunctionEntry}
   > "SetText (Text)"(
    self=Get FilterName_Text() -> (FilterName_Text:Common Text Object Reference)
    InText=Select!(
     MatchAny:Text=Get MatchAnyText() -> (MatchAnyText:Text)
     MatchAll:Text=Get MatchAllText() -> (MatchAllText:Text)
     Index:EFortEmporiumItemTagFilterMode Enum=@n1
    ) -> (ReturnValue:Text)
   )
   > "SetText (Text)"(
    self=Get FilterModeDescrtiption_Text() -> (FilterModeDescrtiption_Text:Common Text Object Reference)
    InText=Select(
     MatchAny=Get MatchAnyDescription() -> (MatchAnyDescription:Text)
     MatchAll=Get MatchAllDescription() -> (MatchAllDescription:Text)
     Index=@n1
    )
   )
   > SetBrush(
    self=Get FilterMode_Image() -> (FilterMode_Image:Image Object Reference)
    InBrush=Make Slate Brush!(
     TintColor:Slate Color Structure=(SpecifiedColor=(R=1,G=1,B=1,A=1),ColorUseRule=UseColor_Specified)
     DrawAs:ESlateBrushDrawType Enum=Image
     Tiling:ESlateBrushTileType Enum=NoTile
     Mirroring:ESlateBrushMirrorType Enum=NoMirror
     ImageSize:Vector 2D Structure=(X=32,Y=32)
     Margin:Margin Structure=(Left=0,Top=0,Right=0,Bottom=0)
     ResourceObject:Object Reference=Select(
      MatchAny:Object Reference=/CreativeEmporium/Materials/T_UI_Icon_FilterUnion
      MatchAll:Object Reference=/CreativeEmporium/Materials/T_UI_Icon_FilterIntersect
      Index=@n1
     ) -> (ReturnValue:Object Reference)
     OutlineSettings:Slate Brush Outline Settings Structure=(CornerRadii=(X=0,Y=0,Z=0,W=0),Color=(SpecifiedColor=(R=0,G=0,B=0,A=0),ColorUseRule=UseColor_Specified),Width=0,RoundingType=HalfHeightRadius,bUseBrushTransparency=False)
    ) -> (SlateBrush:Slate Brush Structure) {is=title K2Node_MakeStruct}
   )
   > PlayAnimation(InAnimation=Get ShowFilterToast() -> (ShowFilterToast:Widget Animation Object Reference), bRestoreState=true)
  ]}
 ShowHeaderBackground {nodes [
   ShowHeaderBackground() {is=title K2Node_FunctionEntry}
   > Return Node!(execute, bShowHeaderBackground:bool=true)
  ]}
 Update Center Panel Location {nodes [
   Update Center Panel Location() {is=title K2Node_FunctionEntry}
   n1 = NOT Boolean(A=IsLeftDocked())
   n2 = NOT Boolean(A=IsRightDocked())
   > SetGridSlotColumnAndRow!(
    execute
    Widget:Widget Object Reference=Get Center_Overlay() -> (Center_Overlay:Overlay Object Reference)
    Row:int=0
    Row Span:int=0
    Column:int=SelectInt(A=1, B=0, bPickA=IsLeftDocked())
    Column Span:int=int + int!(A="To Integer (Boolean)"!(InBool:bool=@n2) -> (ReturnValue:int=0), B="To Integer (Boolean)"(InBool=@n1) -> (ReturnValue=0), C:int=1) -> (ReturnValue)
   ) -> (then)
   > SetGridSlotPadding!(
    execute
    Widget=Get Center_Overlay()
    Padding_Left:float=SelectFloat(
     A=Get Inventory Padding Open() -> (Inventory Padding Open:double)
     B=Get Inventory Padding Closed() -> (Inventory Padding Closed:double)
     bPickA=IsLeftDocked()
    )
    Padding_Top:float=Get InventoryPadding_Version2() -> (InventoryPadding_Version2:float)
    Padding_Right:float=SelectFloat(A=Get Inventory Padding Open(), B=Get Inventory Padding Closed(), bPickA=IsRightDocked())
    Padding_Bottom:float=Select!(Option 0:float=-38, Option 1:float=0, Index:bool=IsMobileGame()) -> (ReturnValue:float)
   ) -> (then)
  ]}
 Update Panel Detail {nodes [
   Update Panel Detail() {is=title K2Node_FunctionEntry}
   > SetBrushColor(
    self=Get Border_Right()
    InBrushColor=SelectColor(
     A=Get RightPanelColorDocked() -> (RightPanelColorDocked:Linear Color Structure)
     B=Get RightPanelColorFlyout() -> (RightPanelColorFlyout:Linear Color Structure)
     bPickA=IsRightDocked()
    )
   )
   n1 = Get Left_Background()
   n2 = Get!(self:Background Blur Object Reference=@n1) -> (Padding_Left:float=~, Padding_Top:float=~, Padding_Right:float=~, Padding_Bottom:float=~)
   > SetPadding!(
    execute
    self:Background Blur Object Reference=@n1
    InPadding_Left:float=SelectFloat(A=0, B=-100, bPickA=IsLeftDocked())
    InPadding_Top:float=@n2.Padding_Top
    InPadding_Right:float=@n2.Padding_Right
    InPadding_Bottom:float=@n2.Padding_Bottom
   ) -> (then)
   > SetWidthOverride(
    self=Get BorderLeftSizeBox() -> (BorderLeftSizeBox:Size Box Object Reference)
    InWidthOverride=SelectFloat(A=350, B=450, bPickA=IsLeftDocked())
   )
  ]}
 Update Side Button Visuals {nodes [
   Update Side Button Visuals() {is=title K2Node_FunctionEntry}
   n1 = IsHovered(self=Get SidePanelButton_Right())
   n2 = Get PanelButtonIcon_Pin() -> (PanelButtonIcon_Pin:Texture 2D Object Reference)
   n3 = Get PanelButtonIcon_Show() -> (PanelButtonIcon_Show:Texture 2D Object Reference)
   n4 = IsRightDocked()
   n5 = Get PanelButtonIcon_Hide() -> (PanelButtonIcon_Hide:Texture 2D Object Reference)
   > SetIcon!(
    execute
    self:WBP UIKit Button Generic Base Object Reference=Get SidePanelButton_Right()
    Icon:Object Reference=SelectObject!(
     A:Object Reference=SelectObject(A=@n5, B=@n2, bSelectA=@n4)
     B:Object Reference=SelectObject(A=@n2, B=@n3, bSelectA=@n1)
     bSelectA:bool=Get PanelOpen_Right()
    ) -> (ReturnValue:Object Reference)
   ) -> (then)
   n6 = IsLeftDocked()
   n7 = Get PanelButtonIcon_Pin()
   n8 = Get PanelButtonIconInverted_Hide() -> (PanelButtonIconInverted_Hide:Texture 2D Object Reference)
   > SetIcon(
    self=Get SidePanelButton_Left()
    Icon=SelectObject(A=SelectObject(A=@n8, B=@n7, bSelectA=@n6), B=Get PanelButtonIcon_Hide(), bSelectA=Get PanelOpen_Left())
   )
   > n9 = IsUsingSidePanels()
   > Set ForceHideButtons(ForceHideButtons:bool=NOT Boolean(A=@n9)) -> (Output_Get:bool)
   n10 = "Equal (Enum)"(A:ECommonInputType Enum=GetCurrentInputType(self=Get CommonInputSubsystem()), B:ECommonInputType Enum=Gamepad)
   n11 = AND Boolean(A=IsLeftDocked(), B=@n10)
   n12 = Get ForceHideButtons() -> (ForceHideButtons:bool)
   > SetVisibility(
    self=Get Left_Button_Overlay()
    InVisibility=Select(Option 0:ESlateVisibility Enum=Visible, Option 1:ESlateVisibility Enum=Collapsed, Index=OR Boolean(A=@n11, B=@n12)) -> (ReturnValue:ESlateVisibility Enum)
   )
   n13 = AND Boolean(A=IsRightDocked(), B=@n10)
   n14 = Get ForceHideButtons()
   > SetVisibility(self=Get Right_Button_Overlay(), InVisibility=Select(Option 0=Visible, Option 1=Collapsed, Index=OR Boolean(A=@n13, B=@n14)))
   n15 = IsLeftDocked()
   n16 = Get PanelOpen_Left()
   > SetVisibility(
    self=Get SidePanelLeftHitZone_Image() -> (SidePanelLeftHitZone_Image:Image Object Reference)
    InVisibility=Select(Option 0=SelfHitTestInvisible, Option 1=Collapsed, Index=OR Boolean(A=@n15, B=@n16))
   )
   n17 = IsRightDocked()
   n18 = Get PanelOpen_Right()
   > SetVisibility(
    self=Get SidePanelRightHitZone_Image() -> (SidePanelRightHitZone_Image:Image Object Reference)
    InVisibility=Select(Option 0=SelfHitTestInvisible, Option 1=Collapsed, Index=OR Boolean(A=@n17, B=@n18))
   )
  ]}
 Update Tooltip Text {nodes [
   Update Tooltip Text() {is=title K2Node_FunctionEntry}
   > n1 = IsUsingSidePanels()
   > Set UsingSidePanels(UsingSidePanels:bool=@n1) -> (Output_Get:bool)
   n2 = GetCurrentInputType(self=Get CommonInputSubsystem())
   n3 = "Equal (Enum)"(A=@n2, B=MouseAndKeyboard)
   > n4 = Branch(Condition=OR Boolean(A=@n3, B="Equal (Enum)"(A=@n2, B=Gamepad)))
   > SetVisibility(
    self=[
     Get RightSideCollapseArrow_Image() -> (RightSideCollapseArrow_Image:Image Object Reference)
     Get LeftSideCollapseArrow_Image() -> (LeftSideCollapseArrow_Image:Image Object Reference)
    ]
    InVisibility=Select(Option 0=SelfHitTestInvisible, Option 1=Hidden, Index=@n3)
   )
   > n5 = Sequence()
   n6 = Set IsRightPanel(execute=@n5.then_1, IsRightPanel=true)
   n7 = IsLeftDocked()
   n8 = IsRightDocked()
   n9 = Get IsRightPanel()
   "SetText (Text)"(
    execute=[@n5.then_0, @n6.then]
    self=Select(
     Option 0:Common Text Object Reference=Get Text_DockLeft() -> (Text_DockLeft:Common Text Object Reference)
     Option 1:Common Text Object Reference=Get Text_DockRight() -> (Text_DockRight:Common Text Object Reference)
     Index=Get IsRightPanel()
    ) -> (ReturnValue:Common Text Object Reference)
    InText=SelectText(
     A=Get UndockText() -> (UndockText:Text)
     B=Get DockText() -> (DockText:Text)
     bPickA=Select(Option 0:bool=@n7, Option 1:bool=@n8, Index=@n9) -> (ReturnValue:bool)
    )
   )
   n10 = Get IsRightPanel()
   n11 = OR Boolean(
    A=NOT Boolean(A=@n10)
    B=IsSidePanelOpen!(self:Fort Emporium Item List Menu Object Reference, bRightPanel:bool=@n10) -> (ReturnValue:bool) {is=title K2Node_CallFunction}
   )
   n12 = Get UsingSidePanels() -> (UsingSidePanels:bool)
   n13 = Select(Option 0=Get PanelFocused_Left(), Option 1=Get PanelFocused_Right(), Index=Get IsRightPanel())
   > SetVisibility(
    self=Select(
     Option 0:Widget Object Reference=Get HBox_DockLeft() -> (HBox_DockLeft:Horizontal Box Object Reference)
     Option 1:Widget Object Reference=Get DockRight_Overlay() -> (DockRight_Overlay:Overlay Object Reference)
     Index=Get IsRightPanel()
    ) -> (ReturnValue:Widget Object Reference)
    InVisibility=Select(
     Option 0:ESlateVisibility Enum=Hidden
     Option 1:ESlateVisibility Enum=HitTestInvisible
     Index=AND Boolean!(A=@n11, B=@n12, C:bool=@n13) -> (ReturnValue)
    ) -> (ReturnValue:ESlateVisibility Enum)
   )
   > SetVisibility(
    self=Select(
     Option 0:Widget Object Reference=Get FocusLeft_Overlay() -> (FocusLeft_Overlay:Overlay Object Reference)
     Option 1:Widget Object Reference=Get FocusRight_Overlay() -> (FocusRight_Overlay:Overlay Object Reference)
     Index=Get IsRightPanel()
    ) -> (ReturnValue:Widget Object Reference)
    InVisibility=Select(Option 0:ESlateVisibility Enum=SelfHitTestInvisible, Option 1:ESlateVisibility Enum=Hidden, Index=@n13) -> (ReturnValue:ESlateVisibility Enum)
   )
   n14 = Select(
    Option 0:bool=OR Boolean(A=IsLeftDocked(), B=Get PanelFocused_Left())
    Option 1:bool=OR Boolean(A=IsRightDocked(), B=Get PanelFocused_Right())
    Index=Get IsRightPanel()
   ) -> (ReturnValue:bool)
   n15 = Get UsingSidePanels()
   > SetVisibility(
    self=Select(
     Option 0:Widget Object Reference=Get Left_Keybinds_Overlay()
     Option 1:Widget Object Reference=Get Right_Keybinds_Overlay()
     Index=Get IsRightPanel()
    ) -> (ReturnValue:Widget Object Reference)
    InVisibility=Select(
     Option 0:ESlateVisibility Enum=Hidden
     Option 1:ESlateVisibility Enum=SelfHitTestInvisible
     Index=AND Boolean!(A=@n15, B=@n14) -> (ReturnValue)
    ) -> (ReturnValue:ESlateVisibility Enum)
   )
   n16 = NOT Boolean(A=Get UsingSidePanels())
   > SetVisibility(
    self=Select(
     Option 0:Widget Object Reference=Get Keybind_ButtonLeft() -> (Keybind_ButtonLeft:Keybind Widget Object Reference)
     Option 1:Widget Object Reference=Get Keybind_ButtonRight() -> (Keybind_ButtonRight:Keybind Widget Object Reference)
     Index=Get IsRightPanel()
    ) -> (ReturnValue:Widget Object Reference)
    InVisibility=Select(Option 0:ESlateVisibility Enum=SelfHitTestInvisible, Option 1:ESlateVisibility Enum=Hidden, Index=OR Boolean(A=@n14, B=@n16)) -> (ReturnValue:ESlateVisibility Enum)
   )
   n17 = IsLeftDocked()
   n18 = IsRightDocked()
   n19 = Get IsRightPanel()
   > SetVisibility(
    self=Get RightSideEndcapButtonsBacking_Image() -> (RightSideEndcapButtonsBacking_Image:Image Object Reference)
    InVisibility=Select(Option 0=Hidden, Option 1=SelfHitTestInvisible, Index=Select(Option 0:bool=@n17, Option 1:bool=@n18, Index=@n19) -> (ReturnValue:bool))
   )
   n20 = IsRightDocked()
   n21 = Get PanelFocused_Right()
   > SetVisibility(
    self=Get CollapseKeybind_Right_VBox() -> (CollapseKeybind_Right_VBox:Vertical Box Object Reference)
    InVisibility=Select(Option 0:ESlateVisibility Enum=Hidden, Option 1:ESlateVisibility Enum=HitTestInvisible, Index=AND Boolean(A=@n21, B=@n20)) -> (ReturnValue:ESlateVisibility Enum)
   )
   n22 = NOT Boolean(A=IsRightDocked())
   n23 = Get PanelFocused_Right()
   > SetVisibility(self=Get Keybind_DockRight_ScreenEdge(), InVisibility=Select(Option 0=Hidden, Option 1=HitTestInvisible, Index=AND Boolean(A=@n23, B=@n22)))
   n24 = IsLeftDocked()
   n25 = Get PanelFocused_Left()
   > SetVisibility(
    self=Get CollapseKeybind_Left_VBox() -> (CollapseKeybind_Left_VBox:Vertical Box Object Reference)
    InVisibility=Select(Option 0=Hidden, Option 1=HitTestInvisible, Index=AND Boolean(A=@n25, B=@n24))
   )
   n26 = NOT Boolean(A=IsLeftDocked())
   n27 = Get PanelFocused_Left()
   > SetVisibility(self=Get Keybind_DockLeft_ScreenEdge(), InVisibility=Select(Option 0=Hidden, Option 1=HitTestInvisible, Index=AND Boolean(A=@n27, B=@n26)))
   > SetVectorParameterValue(
    self=Get DM_LeftSideNavigator_Backplate() -> (DM_LeftSideNavigator_Backplate:Material Instance Dynamic Object Reference)
    ParameterName=Corner Radius
    Value=Select(
     Option 0:Linear Color Structure=Make Linear Color!(R:float=16, G:float=0, B:float=16, A:float=0) -> (LinearColor:Linear Color Structure) {is=title K2Node_MakeStruct}
     Option 1:Linear Color Structure=Make Linear Color(R=16, G=16, B=16, A=16)
     Index=Get PanelFocused_Left()
    ) -> (ReturnValue:Linear Color Structure)
   )
   n28 = IsRightDocked()
   n29 = "Equal (Object)"(A=Get Selected SubFilter Tab() -> (Selected SubFilter Tab:Widget Object Reference), B=Get FiltersSubPanel())
   "SetText (Text)"(
    execute=@n5.then_2
    self=Get Text_FocusRight() -> (Text_FocusRight:Common Text Object Reference)
    InText=SelectText(A=Get FilterText() -> (FilterText:Text), B=Get DetailsText() -> (DetailsText:Text), bPickA=OR Boolean(A=@n28, B=@n29))
   )
   SetVisibility(
    execute=@n4.else
    self=[Get Left_Keybinds_Overlay(), Get Right_Keybinds_Overlay(), Get Keybind_ButtonLeft(), Get Keybind_ButtonRight()]
    InVisibility=Collapsed
   )
  ]}
 UpdateChestItemCount {nodes [UpdateChestItemCount!() -> (then, NumItems:int) {is=title K2Node_FunctionEntry}]}
 UpdateKeybind {nodes [
   n1 = UpdateKeybind!() -> (then, Keybind:Keybind Widget Object Reference, Label:Common Text Object Reference, InputAction:Data Table Row Handle Structure) {is=title K2Node_FunctionEntry}
   n2 = GetMapActionKey!(self:Fort Emporium Item List Menu Object Reference, InInputActionRow:"Data Table Row Handle Structure (by ref)"=@n1.InputAction) -> (ReturnValue:Key Structure)
   > n3 = Branch(Condition=NOT Boolean(A="Equal (Key)"(A:Key Structure=@n2, B:Key Structure) -> (ReturnValue:bool=~) {is=title K2Node_PromotableOperator}))
   > SetVisibility(self=@n1, InVisibility=HitTestInvisible)
   > SetBoundKey!(execute, self:Fort Keybind Widget Object Reference=@n1, NewBoundAction:Key Structure=@n2) -> (then)
   > Is Valid(InputObject=@n1.Label)
   > SetVisibility(self=@n1.Label, InVisibility=HitTestInvisible)
   SetVisibility(execute=@n3.else, self=@n1, InVisibility=Collapsed)
   > Is Valid(InputObject=@n1.Label)
   > SetVisibility(self=@n1.Label, InVisibility=Collapsed)
  ]}
 UpdateRightSideTabs {nodes [
   n1 = UpdateRightSideTabs!() -> (then, Show:bool=~, SelectedIndex:int=-1) {is=title K2Node_FunctionEntry}
   > SetVisibility(
    self=Get RightSideTabs_Overlay() -> (RightSideTabs_Overlay:Overlay Object Reference)
    InVisibility=Select(Option 0:ESlateVisibility Enum=Collapsed, Option 1:ESlateVisibility Enum=SelfHitTestInvisible, Index=@n1) -> (ReturnValue:ESlateVisibility Enum)
   )
   n2 = Get SelectedIndex() -> (SelectedIndex:int=0)
   > Branch(Condition="integer > integer"(A=@n2, B=-1))
   > SelectTabByIndex(self=Get RightSidePanel_Tabs(), Index=@n2, bSuppressClickFeedback=true)
  ]}
 UpdateSelectedDetailsPanelOption {nodes [
   UpdateSelectedDetailsPanelOption() {is=title K2Node_FunctionEntry}
   > n1 = Sequence!(execute) -> (then_0, then_1)
   > n2 = Not Home Tab()
   n3 = Get Switcher_SidePanels()
   > SetActiveWidget(
    self=@n3
    Widget=Select(
     Option 0:Widget Object Reference=Get Selected SubFilter Tab()
     Option 1:Widget Object Reference=Get FiltersSubPanel()
     Index=Get Filter View Forced() -> (Filter View Forced:bool)
    ) -> (ReturnValue:Widget Object Reference)
   )
   > UpdateRightSideTabs!(execute, self, Show:bool=true, SelectedIndex:int=-1) -> (then) {is=title K2Node_CallFunction}
   SetActiveWidget(execute=@n2.false, self=@n3, Widget=Get ItemDetailsSidePanel_DefaultSidePanel())
   > UpdateRightSideTabs(Show=false, SelectedIndex=-1)
   n4 = Get Switcher_SidePanels()
   n5 = Branch(
    execute=@n1.then_1
    Condition="Equal (Object)"(
     A=GetActiveWidget!(self:Common Visibility Switcher Object Reference=@n4) -> (ReturnValue:Widget Object Reference)
     B=Get ItemDetailsSidePanel_DefaultSidePanel()
    )
   )
   n6 = Get ItemDetailsSidePanel_DefaultSidePanel()
   > RegisterScrollRecipientExternal!(execute, self:Common User Widget Object Reference, AnalogScrollRecipient:Widget Object Reference=@n6) -> (then)
   UnregisterScrollRecipientExternal(execute=@n5.else, AnalogScrollRecipient=@n6)
  ]}
 UpdateSidePanelKeybinds {nodes [
   UpdateSidePanelKeybinds() {is=title K2Node_FunctionEntry}
   > UpdateKeybind!(
    execute
    self
    Keybind:Keybind Widget Object Reference=Get Keybind_ButtonLeft()
    Label:Common Text Object Reference
    InputAction:Data Table Row Handle Structure=Get ToggleFilterFocusActionRow() -> (ToggleFilterFocusActionRow:Data Table Row Handle Structure)
   ) -> (then) {is=title K2Node_CallFunction}
   > UpdateKeybind(
    Keybind=Get Keybind_ButtonRight()
    InputAction=Get ToggleFilterSubPanelFocusActionRow() -> (ToggleFilterSubPanelFocusActionRow:Data Table Row Handle Structure)
   )
   > UpdateKeybind(
    Keybind=Get Keybind_FocusLeft() -> (Keybind_FocusLeft:Keybind Widget Object Reference)
    Label=Get Text_FocusLeft() -> (Text_FocusLeft:Common Text Object Reference)
    InputAction=Get ToggleFilterFocusActionRow()
   )
   > UpdateKeybind(
    Keybind=Get Keybind_FocusRight() -> (Keybind_FocusRight:Keybind Widget Object Reference)
    Label=Get Text_FocusRight()
    InputAction=Get ToggleFilterSubPanelFocusActionRow()
   )
   > UpdateKeybind(
    Keybind=Get Keybind_CycleTags() -> (Keybind_CycleTags:Keybind Widget Object Reference)
    Label=Get Text_CycleTags() -> (Text_CycleTags:Common Text Object Reference)
    InputAction=Get CycleTagFilterModeModifierActionRow() -> (CycleTagFilterModeModifierActionRow:Data Table Row Handle Structure)
   )
   > UpdateKeybind(
    Keybind=Get Keybind_ClearTags() -> (Keybind_ClearTags:Keybind Widget Object Reference)
    Label=Get Text_ClearTags() -> (Text_ClearTags:Common Text Object Reference)
    InputAction=Get ClearTagFiltersModifierActionRow() -> (ClearTagFiltersModifierActionRow:Data Table Row Handle Structure)
   )
   > UpdateKeybind(
    Keybind=Get Keybind_DockLeft_ScreenEdge()
    Label=Get Text_DockLeft()
    InputAction=Get ToggleSidePanelDockingModifierActionRow() -> (ToggleSidePanelDockingModifierActionRow:Data Table Row Handle Structure)
   )
   > UpdateKeybind(Keybind=Get Keybind_DockRight_ScreenEdge(), Label=Get Text_DockRight(), InputAction=Get ToggleSidePanelDockingModifierActionRow())
   > UpdateKeybind(
    Keybind=Get Keybind_CollapseRight_ScreenEdge() -> (Keybind_CollapseRight_ScreenEdge:Keybind Widget Object Reference)
    Label=Get Text_DockRight()
    InputAction=Get ToggleSidePanelDockingModifierActionRow()
   )
   > UpdateKeybind(
    Keybind=Get Keybind_CollapseLeft_ScreenEdge() -> (Keybind_CollapseLeft_ScreenEdge:Keybind Widget Object Reference)
    Label=Get Text_DockLeft()
    InputAction=Get ToggleSidePanelDockingModifierActionRow()
   )
  ]}
]
variables [
 BottomBarOffset {type=number, default=32}
 ChestContainer {type=FortContainerTableRow, default {containerClass=null}}
 DataRegistryLookup {type=DataRegistryLookup, default {}}
 DetailsText {type=string, default=Item Details}
 DM_LeftSideNavigator_Backplate {type=MaterialInstanceDynamic}
 DockText {type=string, default=Pin}
 Filter View Forced {type=boolean}
 FilterText {type=string, default=Filters}
 Inventory Padding Closed {type=number, default=64}
 Inventory Padding Open {type=number, default=48}
 InventoryPadding_Original {type=number, default=-80}
 InventoryPadding_Version2 {type=number, default=36}
 LastFilterModeUsed {type=EFortEmporiumItemTagFilterMode, default=MatchAny}
 LlamaContainer {type=FortContainerTableRow, default {containerClass=null}}
 MenuOpenedSound {type=SoundBase, default=/Game/Sounds/Fort_UI_Sounds/FrontEnd/New_FrontEnd/Fortnite_UI_MainMenu_Tab_03_Cue}
 PanelButtonIcon_Hide {type=Texture2D, default=/CreativeEmporium/Materials/T_UI_Icon_CrRefresh_Chevron_Right}
 PanelButtonIcon_Pin {type=Texture2D, default=/CreativeEmporium/Materials/T_UI_Icon_CrRefresh_Pin}
 PanelButtonIcon_Show {type=Texture2D, default=/CreativeEmporium/Materials/T_UI_Icon_CrRefresh_Chevron_Left}
 PanelButtonIconInverted_Hide {type=Texture2D, default=&T_UI_Icon_CrRefresh_Chevron_Left}
 PanelDocked_Left {type=boolean, default=true}
 PanelDocked_Right {type=boolean, default=true}
 PanelFocused_Left {type=boolean}
 PanelFocused_Right {type=boolean}
 PanelOpen_Left {type=boolean, default=true}
 PanelOpen_Right {type=boolean, default=true}
 RightPanelColorDocked {type=LinearColor, default=<rgba 0 0 0 0>}
 RightPanelColorFlyout {type=LinearColor, default=<rgba 0.0048 0.004 0.0116 0.7>}
 Selected SubFilter Tab {type=Widget, default=null}
 ShowFilterModeMessage {}
 SidePanelButton_Left_Pressed {type=boolean}
 SidePanelButton_Right_Pressed {type=boolean}
 TabsDataInfo {type="array<UIKitTabGroupButtonViewModel>", default []}
 UndockText {type=string, default=Hide}
 Viewmodel {type=UIKitTabGroupViewModel, default=null}
]
widgets {
 info {
  inheritedWidgetCount=36
  namedSlotCount=4
  parentClass=/Script/CreativeEmporiumUI.FortEmporiumItemListMenu
  rootWidgetClass=&Overlay
  widgetCount=188
 }
 widgets [
  Overlay_4 {widgetClassPath=&Overlay, widget=Overlay_4}
  BrowserBgColorEffects_Image {
   widgetClassPath=&Image
   parent=Overlay_4
   widget=BrowserBgColorEffects_Image
   properties {brush {
     bIsDynamicallyLoaded=false
     drawAs=Image
     imageType=NoImage
     mirroring=NoMirror
     resourceName=None
     resourceObject=/CreativeEmporium/Materials/MI_UI_CrBrowser_NebulousBackground
     tiling=NoTile
     imageSize=<64 696.8439>
     margin=<ltrb 0 0 0 0>
     outlineSettings {
      bUseBrushTransparency=false
      roundingType=HalfHeightRadius
      width=0
      color {colorUseRule=UseColor_Specified, specifiedColor=<rgba 0 0 0 0>}
      cornerRadii=<0 0 0 0>
     }
     tintColor {colorUseRule=UseColor_Specified, specifiedColor=<rgba 1 1 1 1>}
     uVRegion {bIsValid=false, max=<0 0>, min=<0 0>}
    }}
   slot {class=OverlaySlot, properties {horizontalAlignment=HAlign_Fill, verticalAlignment=VAlign_Fill}}
  }
  WBP_CaptureForPostBufferUpdate {
   widgetClassPath=&WBP_CaptureForPostBufferUpdate
   bIsVariable=true
   parent=Overlay_4
   slot {class=OverlaySlot, properties {horizontalAlignment=HAlign_Fill, verticalAlignment=VAlign_Fill}}
  }
  Switcher_InventoryPermission {
   widgetClassPath=&WidgetSwitcher
   bInherited=true
   bIsVariable=true
   parent=Overlay_4
   widget=Switcher_InventoryPermission
   slot {class=OverlaySlot, properties {horizontalAlignment=HAlign_Fill, verticalAlignment=VAlign_Fill, padding=<ltrb 32 24 32 32>}}
  }
  Overlay_EmporiumPanel {
   widgetClassPath=&Overlay
   bInherited=true
   bIsVariable=true
   parent=Switcher_InventoryPermission
   widget=Overlay_EmporiumPanel
   slot {class=WidgetSwitcherSlot}
  }
  Switcher_ProductPage {
   widgetClassPath=&CommonVisibilitySwitcher
   bInherited=true
   parent=Overlay_EmporiumPanel
   widget=Switcher_ProductPage
   slot {class=OverlaySlot, properties {horizontalAlignment=HAlign_Fill, verticalAlignment=VAlign_Fill}}
  }
  Overlay_ItemBrowser {
   widgetClassPath=&Overlay
   bInherited=true
   parent=Switcher_ProductPage
   widget=Overlay_ItemBrowser
   slot {class=CommonVisibilitySwitcherSlot}
  }
  VBox_ItemBrowser {
   widgetClassPath=&VerticalBox
   parent=Overlay_ItemBrowser
   widget=VBox_ItemBrowser
   slot {class=OverlaySlot, properties {horizontalAlignment=HAlign_Fill, verticalAlignment=VAlign_Fill}}
  }
  SizeBox_TopBar {
   widgetClassPath=&SizeBox
   parent=VBox_ItemBrowser
   widget=SizeBox_TopBar
   properties {bOverride_HeightOverride=true, heightOverride=72}
   slot {class=VerticalBoxSlot, properties {padding=<ltrb 0 0 0 12>}}
  }
  HBox_Top {widgetClassPath=&HorizontalBox, parent=SizeBox_TopBar, widget=HBox_Top, slot {class=SizeBoxSlot}}
  SizeBox_SearchPanel {
   widgetClassPath=&SizeBox
   parent=HBox_Top
   widget=SizeBox_SearchPanel
   properties {bOverride_HeightOverride=true, bOverride_WidthOverride=true, heightOverride=72, widthOverride=350}
   slot {class=HorizontalBoxSlot, properties {horizontalAlignment=HAlign_Left, verticalAlignment=VAlign_Top, padding=<ltrb 32 10 0 0>}}
  }
  SearchPanel {
   widgetClassPath=&/WBP_Creative_Emporium_SearchBar
   bIsVariable=true
   parent=SizeBox_SearchPanel
   widget=SearchPanel
   properties {
    bHideInputAction=false
    hint Text=Search...
    holdData=&CommonUI_HoldData_Base_C
    minHeight=64
    visibility=Visible
   }
   slot {class=SizeBoxSlot}
  }
  SizeBox_103 {
   widgetClassPath=&SizeBox
   parent=HBox_Top
   widget=SizeBox_103
   properties {bOverride_WidthOverride=true, widthOverride=450}
   slot {class=HorizontalBoxSlot, properties {horizontalAlignment=HAlign_Left}}
  }
  Blade_SortAndFilter {
   widgetClassPath=/Game/Athena/UI/Customization/AthenaCustomizationPickerSortAndFilterBlade
   bInherited=true
   bIsVariable=true
   parent=SizeBox_103
   widget=Blade_SortAndFilter
   properties {visibility=Collapsed}
   slot {class=SizeBoxSlot}
  }
  CategoryHeader {
   widgetClassPath=&/WBP_Creative_Emporium_Browser_Header
   bIsVariable=true
   parent=HBox_Top
   widget=CategoryHeader
   slot {class=HorizontalBoxSlot, properties {padding=<ltrb 68 0 0 0>}}
  }
  SafeZone_CenterContent {
   widgetClassPath=&SafeZone
   bIsVariable=true
   parent=VBox_ItemBrowser
   widget=SafeZone_CenterContent
   properties {padBottom=false, padTop=false}
   slot {class=VerticalBoxSlot, properties {size {sizeRule=Fill, value=1}}}
  }
  Grid_Center {
   widgetClassPath=/Script/FortniteUI.UIKitGridPanel
   parent=SafeZone_CenterContent
   widget=Grid_Center
   properties {columnFill [0,1,0], rowFill [0.65,0.25]}
   slot {class=SafeZoneSlot}
  }
  Left_Background {
   widgetClassPath=&BackgroundBlur
   bIsVariable=true
   parent=Grid_Center
   widget=Left_Background
   properties {
    blurStrength=10
    renderOpacity=0.7
    lowQualityFallbackBrush {
     bIsDynamicallyLoaded=false
     drawAs=Image
     imageType=NoImage
     mirroring=NoMirror
     resourceName=None
     resourceObject=null
     tiling=NoTile
     imageSize=<0 0>
     margin=<ltrb 0 0 0 0>
     outlineSettings {
      bUseBrushTransparency=false
      roundingType=HalfHeightRadius
      width=0
      color {colorUseRule=UseColor_Specified, specifiedColor=<rgba 0 0 0 0>}
      cornerRadii=<0 0 0 0>
     }
     tintColor {colorUseRule=UseColor_Specified, specifiedColor=<rgba 0 0.0627 0.1529 0.7>}
     uVRegion {bIsValid=false, max=<0 0>, min=<0 0>}
    }
   }
   slot {class=GridSlot, properties {horizontalAlignment=HAlign_Left, layer=1, rowSpan=2, padding=<ltrb 32 42 0 0>}}
  }
  BorderLeftSizeBox {
   widgetClassPath=&SizeBox
   bIsVariable=true
   parent=Left_Background
   widget=BorderLeftSizeBox
   properties {bOverride_WidthOverride=true, widthOverride=350}
   slot {class=BackgroundBlurSlot, properties {padding=<ltrb 0 0 0 0>}}
  }
  Border_Left {
   widgetClassPath=&Border
   bIsVariable=true
   parent=BorderLeftSizeBox
   widget=Border_Left
   properties {
    background {
     bIsDynamicallyLoaded=false
     drawAs=Image
     imageType=NoImage
     mirroring=NoMirror
     resourceName=None
     resourceObject=/CreativeEmporium/Materials/MI_CrBrowser_Backplate_Nav
     tiling=NoTile
     imageSize=<32 32>
     margin=<ltrb 0 0 0 0>
     outlineSettings {
      bUseBrushTransparency=false
      roundingType=HalfHeightRadius
      width=0
      color {colorUseRule=UseColor_Specified, specifiedColor=<rgba 0 0 0 0>}
      cornerRadii=<0 0 0 0>
     }
     tintColor {colorUseRule=UseColor_Specified, specifiedColor=<rgba 1 1 1 1>}
     uVRegion {bIsValid=false, max=<0 0>, min=<0 0>}
    }
    brushColor=<rgba 0.0048 0.004 0.0116 0.85>
    padding=<ltrb 0 0 0 0>
   }
   slot {class=SizeBoxSlot}
  }
  Left_Button_Overlay {
   widgetClassPath=&Overlay
   bIsVariable=true
   parent=Grid_Center
   widget=Left_Button_Overlay
   properties {visibility=Collapsed}
   slot {class=GridSlot, properties {horizontalAlignment=HAlign_Left, layer=3, rowSpan=2, padding=<ltrb -132 42 0 0>}}
  }
  SidePanelButton_Left {
   widgetClassPath=&WBP_UIKit_Button_Generic
   bIsVariable=true
   parent=Left_Button_Overlay
   widget=SidePanelButton_Left
   properties {
    bHideInputAction=true
    icon=/CreativeEmporium/Materials/T_UI_Icon_CrRefresh_Chevron_Right2
    iconMaterial=/CreativeEmporium/Materials/MI_CrBrowser_ArrowIcons_Color
    minWidth=164
    shouldTextImpactLayout=false
    text=""
    useBackgroundBlock=false
    useTextBlock=false
    iconPadding=<ltrb 46 0 0 0>
    iconSize=<32 32>
   }
   slot {class=OverlaySlot, properties {horizontalAlignment=HAlign_Fill, verticalAlignment=VAlign_Fill}}
  }
  Left_ButtonKeybind_Overlay {widgetClassPath=&Overlay, namedSlotHost=SidePanelButton_Left, widget=Left_ButtonKeybind_Overlay}
  SizeBox_5 {
   widgetClassPath=&SizeBox
   bIsVariable=true
   parent=Left_ButtonKeybind_Overlay
   widget=SizeBox_5
   properties {bOverride_HeightOverride=true, bOverride_WidthOverride=true, heightOverride=24, widthOverride=24}
   slot {class=OverlaySlot, properties {horizontalAlignment=HAlign_Right, verticalAlignment=VAlign_Center, padding=<ltrb 0 32 24 0>}}
  }
  Keybind_ButtonLeft {
   widgetClassPath=/FortUILibrary/Input/KeybindWidget
   bIsVariable=true
   parent=SizeBox_5
   widget=Keybind_ButtonLeft
   properties {bUseKeyboardTextureForKeys=true, boundKeyFallback=LeftControl}
   slot {class=SizeBoxSlot}
  }
  SizeBox_2 {
   widgetClassPath=&SizeBox
   bIsVariable=true
   parent=Left_ButtonKeybind_Overlay
   widget=SizeBox_2
   properties {bOverride_HeightOverride=true, bOverride_WidthOverride=true, heightOverride=24, widthOverride=24}
   slot {class=OverlaySlot, properties {horizontalAlignment=HAlign_Right, verticalAlignment=VAlign_Center, padding=<ltrb 0 32 24 0>}}
  }
  Keybind_DockLeft_ScreenEdge {
   widgetClassPath=&KeybindWidget
   bIsVariable=true
   parent=SizeBox_2
   widget=Keybind_DockLeft_ScreenEdge
   properties {bUseKeyboardTextureForKeys=true, boundKeyFallback=D, designTimeKey=D, visibility=HitTestInvisible}
   slot {class=SizeBoxSlot, properties {horizontalAlignment=HAlign_Center, verticalAlignment=VAlign_Center}}
  }
  SidePanelLeftHitZone_Image {
   widgetClassPath=&Image
   bIsVariable=true
   namedSlotHost=SidePanelButton_Left
   widget=SidePanelLeftHitZone_Image
   properties {
    visibility=SelfHitTestInvisible
    brush {
     bIsDynamicallyLoaded=false
     drawAs=RoundedBox
     imageType=NoImage
     mirroring=NoMirror
     resourceName=None
     resourceObject=null
     tiling=NoTile
     imageSize=<32 32>
     margin=<ltrb 0 0 0 0>
     outlineSettings {
      bUseBrushTransparency=false
      roundingType=FixedRadius
      width=0
      color {colorUseRule=UseColor_Specified, specifiedColor=<rgba 0 0 0 0>}
      cornerRadii=<50 50 50 50>
     }
     tintColor {colorUseRule=UseColor_Specified, specifiedColor=<rgba 0 0 0 0.3>}
     uVRegion {bIsValid=false, max=<0 0>, min=<0 0>}
    }
    renderTransform {angle=0, scale=<1 1>, shear=<0 0>, translation=<6 0>}
   }
  }
  Left_Content_Overlay {
   widgetClassPath=&Overlay
   bIsVariable=true
   parent=Grid_Center
   widget=Left_Content_Overlay
   properties {navigation=&self:WidgetTree.Left_Content_Overlay.WidgetNavigation_0}
   slot {class=GridSlot, properties {layer=2, rowSpan=2, padding=<ltrb 32 42 0 0>}}
  }
  Overlay_SearchFilters {
   widgetClassPath=&Overlay
   bInherited=true
   bIsVariable=true
   parent=Left_Content_Overlay
   widget=Overlay_SearchFilters
   slot {class=OverlaySlot, properties {horizontalAlignment=HAlign_Fill, verticalAlignment=VAlign_Fill}}
  }
  LeftSearchFiltersSizeBox {
   widgetClassPath=&SizeBox
   parent=Overlay_SearchFilters
   widget=LeftSearchFiltersSizeBox
   properties {bOverride_WidthOverride=true, heightOverride=586.9089, maxDesiredWidth=400, minDesiredWidth=200, widthOverride=350}
   slot {class=OverlaySlot, properties {verticalAlignment=VAlign_Fill}}
  }
  Switcher_LeftPanel {
   widgetClassPath=/Script/CommonUI.CommonAnimatedSwitcher
   bInherited=true
   bIsVariable=true
   parent=LeftSearchFiltersSizeBox
   widget=Switcher_LeftPanel
   properties {transitionDuration=0}
   slot {class=SizeBoxSlot}
  }
  FiltersPanel {
   widgetClassPath=&/WBP_Creative_Emporium_Browser_Filters
   bInherited=true
   bIsVariable=true
   parent=Switcher_LeftPanel
   widget=FiltersPanel
   properties {bIsFocusable=true, navigation=&self:WidgetTree.FiltersPanel.WidgetNavigation_1}
   slot {class=WidgetSwitcherSlot}
  }
  SubTabItemDetailsPanel {
   widgetClassPath=&/WBP_Creative_Emporium_PlaysetDetailsPanel
   bInherited=true
   bIsVariable=true
   parent=Switcher_LeftPanel
   widget=SubTabItemDetailsPanel
   slot {class=WidgetSwitcherSlot}
  }
  CanvasPanel_0 {
   widgetClassPath=&CanvasPanel
   parent=Left_Content_Overlay
   widget=CanvasPanel_0
   slot {class=OverlaySlot, properties {horizontalAlignment=HAlign_Fill, verticalAlignment=VAlign_Fill}}
  }
  Blur_HitBlocker_Left {
   widgetClassPath=&BackgroundBlur
   bIsVariable=true
   parent=CanvasPanel_0
   widget=Blur_HitBlocker_Left
   properties {blurStrength=1, visibility=Collapsed}
   slot {class=CanvasPanelSlot, properties=~}
  }
  Button_HitBlocker_Left {
   widgetClassPath=/FortUILibrary/Foundation/Buttons/FortTextButton
   bIsVariable=true
   parent=CanvasPanel_0
   widget=Button_HitBlocker_Left
   properties {
    defaultText=""
    onButtonBaseClicked=(&self:WidgetTree.Button_HitBlocker_Left.HandleOnButtonClicked)
    onButtonBaseDoubleClicked=(&self:WidgetTree.Button_HitBlocker_Left.HandleOnButtonDoubleClicked)
    onButtonBaseHovered=(&self:WidgetTree.Button_HitBlocker_Left.HandleOnButtonHovered)
    onButtonBaseUnhovered=(&self:WidgetTree.Button_HitBlocker_Left.HandleOnButtonUnhovered)
    onSelectedChangedBase=(&self:WidgetTree.Button_HitBlocker_Left.HandleOnSelectedChanged)
    visibility=Collapsed
    colorAndOpacity=<rgba 1 1 1 0>
   }
   slot {class=CanvasPanelSlot, properties=~}
  }
  CollapseKeybind_Left_VBox {
   widgetClassPath=&VerticalBox
   bIsVariable=true
   parent=Left_Content_Overlay
   widget=CollapseKeybind_Left_VBox
   properties {visibility=Collapsed, renderTransform {angle=0, scale=<1 1>, shear=<0 0>, translation=<-52 12>}}
   slot {class=OverlaySlot, properties {verticalAlignment=VAlign_Center}}
  }
  LeftSideCollapseArrow_Image {
   widgetClassPath=&Image
   bIsVariable=true
   parent=CollapseKeybind_Left_VBox
   widget=LeftSideCollapseArrow_Image
   properties {brush {
     bIsDynamicallyLoaded=false
     drawAs=Image
     imageType=FullColor
     mirroring=NoMirror
     resourceName=None
     resourceObject=&T_UI_Icon_CrRefresh_Chevron_Left
     tiling=NoTile
     imageSize=<32 32>
     margin=<ltrb 0 0 0 0>
     outlineSettings {
      bUseBrushTransparency=false
      roundingType=HalfHeightRadius
      width=0
      color {colorUseRule=UseColor_Specified, specifiedColor=<rgba 0 0 0 0>}
      cornerRadii=<0 0 0 0>
     }
     tintColor {colorUseRule=UseColor_Specified, specifiedColor=<rgba 1 1 1 1>}
     uVRegion {bIsValid=false, max=<0 0>, min=<0 0>}
    }}
   slot {class=VerticalBoxSlot}
  }
  SizeBox_3 {
   widgetClassPath=&SizeBox
   bIsVariable=true
   parent=CollapseKeybind_Left_VBox
   widget=SizeBox_3
   properties {bOverride_HeightOverride=true, bOverride_WidthOverride=true, heightOverride=24, widthOverride=24}
   slot {class=VerticalBoxSlot, properties {horizontalAlignment=HAlign_Center, verticalAlignment=VAlign_Center}}
  }
  Keybind_CollapseLeft_ScreenEdge {
   widgetClassPath=&KeybindWidget
   bIsVariable=true
   parent=SizeBox_3
   widget=Keybind_CollapseLeft_ScreenEdge
   properties {bUseKeyboardTextureForKeys=true, boundKeyFallback=D, designTimeKey=D, visibility=HitTestInvisible}
   slot {class=SizeBoxSlot}
  }
  Left_Keybinds_Overlay {
   widgetClassPath=&Overlay
   bIsVariable=true
   parent=Grid_Center
   widget=Left_Keybinds_Overlay
   slot {class=GridSlot, properties {row=2, verticalAlignment=VAlign_Bottom, padding=<ltrb 32 0 0 0>}}
  }
  FocusLeft_Overlay {
   widgetClassPath=&Overlay
   bIsVariable=true
   parent=Left_Keybinds_Overlay
   widget=FocusLeft_Overlay
   slot {class=OverlaySlot, properties {horizontalAlignment=HAlign_Fill}}
  }
  Image_2 {
   widgetClassPath=&Image
   bIsVariable=true
   parent=FocusLeft_Overlay
   widget=Image_2
   properties {brush {
     bIsDynamicallyLoaded=false
     drawAs=RoundedBox
     imageType=NoImage
     mirroring=NoMirror
     resourceName=None
     resourceObject=null
     tiling=NoTile
     imageSize=<32 32>
     margin=<ltrb 0 0 0 0>
     outlineSettings {
      bUseBrushTransparency=false
      roundingType=FixedRadius
      width=0
      color {colorUseRule=UseColor_Specified, specifiedColor=<rgba 0 0 0 0>}
      cornerRadii=<0 0 20 20>
     }
     tintColor {colorUseRule=UseColor_Specified, specifiedColor=<rgba 1 1 1 0.15>}
     uVRegion {bIsValid=false, max=<0 0>, min=<0 0>}
    }}
   slot {class=OverlaySlot, properties {horizontalAlignment=HAlign_Fill, verticalAlignment=VAlign_Fill}}
  }
  HBox_FocusLeft {
   widgetClassPath=&HorizontalBox
   bIsVariable=true
   parent=FocusLeft_Overlay
   widget=HBox_FocusLeft
   slot {class=OverlaySlot, properties {horizontalAlignment=HAlign_Center, padding=<ltrb 0 2 0 2>}}
  }
  Keybind_FocusLeft {
   widgetClassPath=&KeybindWidget
   bIsVariable=true
   parent=HBox_FocusLeft
   widget=Keybind_FocusLeft
   properties {bUseKeyboardTextureForKeys=true, boundKeyFallback=LeftControl, designTimeKey=LeftControl}
   slot {class=HorizontalBoxSlot}
  }
  Text_FocusLeft {
   widgetClassPath=&CommonTextBlock
   bIsVariable=true
   parent=HBox_FocusLeft
   widget=Text_FocusLeft
   properties {
    bAutoCollapseWithEmptyText=true
    text=Categories
    textTransformPolicy=ToUpper
    colorAndOpacity {colorUseRule=UseColor_Specified, specifiedColor=<rgba 1 1 1 0.7>}
    font {
     bForceMonospaced=false
     bMaterialIsStencil=false
     fontMaterial=null
     fontObject=&HeadingNow
     letterSpacing=0
     monospacedWidth=1
     size=13.5
     skewAmount=0
     typefaceFontName=75 Medium
     outlineSettings {
      bApplyOutlineToDropShadows=false
      bMiteredCorners=false
      bSeparateFillAlpha=false
      outlineMaterial=null
      outlineSize=0
      outlineColor=<rgba 0 0 0 1>
     }
    }
    shadowOffset=<0 0>
   }
   slot {class=HorizontalBoxSlot, properties {horizontalAlignment=HAlign_Left, verticalAlignment=VAlign_Center, padding=<ltrb 12 0 0 0>}}
  }
  Delete_This_Left {
   widgetClassPath=&Overlay
   parent=Left_Keybinds_Overlay
   widget=Delete_This_Left
   properties {visibility=Collapsed}
   slot {class=OverlaySlot}
  }
  HBox_DockLeft {widgetClassPath=&HorizontalBox, bIsVariable=true, parent=Delete_This_Left, widget=HBox_DockLeft, slot {class=OverlaySlot}}
  Keybind_DockLeft {
   widgetClassPath=&KeybindWidget
   bIsVariable=true
   parent=HBox_DockLeft
   widget=Keybind_DockLeft
   properties {bUseKeyboardTextureForKeys=true, boundKeyFallback=Gamepad_FaceButton_Top, designTimeKey=Gamepad_FaceButton_Top}
   slot {class=HorizontalBoxSlot}
  }
  Text_DockLeft {
   widgetClassPath=&CommonTextBlock
   bIsVariable=true
   parent=HBox_DockLeft
   widget=Text_DockLeft
   properties {
    bAutoCollapseWithEmptyText=true
    text=Dock
    textTransformPolicy=ToUpper
    colorAndOpacity {colorUseRule=UseColor_Specified, specifiedColor=<rgba 1 1 1 0.7>}
    font {
     bForceMonospaced=false
     bMaterialIsStencil=false
     fontMaterial=null
     fontObject=&HeadingNow
     letterSpacing=0
     monospacedWidth=1
     size=13.5
     skewAmount=0
     typefaceFontName=75 Medium
     outlineSettings {
      bApplyOutlineToDropShadows=false
      bMiteredCorners=false
      bSeparateFillAlpha=false
      outlineMaterial=null
      outlineSize=0
      outlineColor=<rgba 0 0 0 1>
     }
    }
    shadowOffset=<0 0>
   }
   slot {class=HorizontalBoxSlot, properties {horizontalAlignment=HAlign_Left, verticalAlignment=VAlign_Center, padding=<ltrb 12 0 0 0>}}
  }
  Center_Overlay {
   widgetClassPath=&Overlay
   bIsVariable=true
   parent=Grid_Center
   widget=Center_Overlay
   properties {
    clipping=ClipToBoundsWithoutIntersecting
    navigation=&self:WidgetTree.Center_Overlay.WidgetNavigation_0
   }
   slot {class=GridSlot, properties {column=1, padding=<ltrb 48 12 48 -20>}}
  }
  Switcher_ItemTabWarnings {
   widgetClassPath=&WidgetSwitcher
   bInherited=true
   bIsVariable=true
   parent=Center_Overlay
   widget=Switcher_ItemTabWarnings
   slot {class=OverlaySlot, properties {horizontalAlignment=HAlign_Fill, verticalAlignment=VAlign_Fill}}
  }
  Switcher_Categories {
   widgetClassPath=&CommonVisibilitySwitcher
   bInherited=true
   bIsVariable=true
   parent=Switcher_ItemTabWarnings
   widget=Switcher_Categories
   properties {bAutoActivateSlot=false}
   slot {class=WidgetSwitcherSlot}
  }
  Overlay_ChestEmpty {
   widgetClassPath=&Overlay
   bInherited=true
   parent=Switcher_ItemTabWarnings
   widget=Overlay_ChestEmpty
   slot {class=WidgetSwitcherSlot, properties {padding=<ltrb 0 40 0 0>}}
  }
  StackBox {
   widgetClassPath=&StackBox
   parent=Overlay_ChestEmpty
   widget=StackBox
   properties {orientation=Orient_Vertical}
   slot {class=OverlaySlot, properties {horizontalAlignment=HAlign_Fill, padding=<ltrb 24 0 0 0>}}
  }
  Spacer {widgetClassPath=&Spacer, parent=StackBox, widget=Spacer, properties {size=<1 64>}, slot {class=StackBoxSlot}}
  StackBox_3 {widgetClassPath=&StackBox, parent=StackBox, widget=StackBox_3, slot {class=StackBoxSlot}}
  Info_Image_1 {
   widgetClassPath=&Image
   parent=StackBox_3
   widget=Info_Image_1
   properties {brush {
     bIsDynamicallyLoaded=false
     drawAs=Image
     imageType=NoImage
     mirroring=NoMirror
     resourceName=None
     resourceObject=/FortUILibrary/Textures/MaterialInstances/MI_UIKit_Icon_Information
     tiling=NoTile
     imageSize=<32 32>
     margin=<ltrb 0 0 0 0>
     outlineSettings {
      bUseBrushTransparency=false
      roundingType=HalfHeightRadius
      width=0
      color {colorUseRule=UseColor_Specified, specifiedColor=<rgba 0 0 0 0>}
      cornerRadii=<0 0 0 0>
     }
     tintColor {colorUseRule=UseColor_Specified, specifiedColor=<rgba 1 1 1 0.8>}
     uVRegion {bIsValid=false, max=<0 0>, min=<0 0>}
    }}
   slot {class=StackBoxSlot, properties {horizontalAlignment=HAlign_Left, verticalAlignment=VAlign_Top, padding=<ltrb 0 0 14 0>}}
  }
  DisplayMessage_Chest_Text {
   widgetClassPath=&CommonTextBlock
   parent=StackBox_3
   widget=DisplayMessage_Chest_Text
   properties {
    autoWrapText=true
    text=Add items to the chest and then create a chest or llama.
    wrapTextAt=900
    colorAndOpacity {colorUseRule=UseColor_Specified, specifiedColor=<rgba 1 1 1 0.8>}
    font {
     bForceMonospaced=false
     bMaterialIsStencil=false
     fontMaterial=null
     fontObject=&HeadingNow
     letterSpacing=0
     monospacedWidth=1
     size=18
     skewAmount=0
     typefaceFontName=75 Medium Italic
     outlineSettings {
      bApplyOutlineToDropShadows=false
      bMiteredCorners=false
      bSeparateFillAlpha=false
      outlineMaterial=null
      outlineSize=1
      outlineColor=<rgba 0.0052 0.0052 0.0056 0.2>
     }
    }
   }
   slot {class=StackBoxSlot, properties {verticalAlignment=VAlign_Center, size {sizeRule=Fill, value=1}}}
  }
  Spacer_3 {widgetClassPath=&Spacer, parent=StackBox, widget=Spacer_3, properties {size=<1 64>}, slot {class=StackBoxSlot}}
  Button_ChestEmpty {
   widgetClassPath=/FortUILibrary/UIKit/Buttons/WBP_UIKit_Button_Quiet
   bIsVariable=true
   parent=StackBox
   widget=Button_ChestEmpty
   properties {text=Ok, visibility=HitTestInvisible, colorAndOpacity=<rgba 1 1 1 0>}
   slot {class=StackBoxSlot, properties {horizontalAlignment=HAlign_Center, verticalAlignment=VAlign_Center}}
  }
  Overlay_TabEmpty {
   widgetClassPath=&Overlay
   bInherited=true
   parent=Switcher_ItemTabWarnings
   widget=Overlay_TabEmpty
   slot {class=WidgetSwitcherSlot}
  }
  StackBox_0 {
   widgetClassPath=&StackBox
   parent=Overlay_TabEmpty
   widget=StackBox_0
   properties {orientation=Orient_Vertical}
   slot {class=OverlaySlot, properties {horizontalAlignment=HAlign_Center, verticalAlignment=VAlign_Center}}
  }
  Spacer_77 {widgetClassPath=&Spacer, parent=StackBox_0, widget=Spacer_77, properties {size=<1 48>}, slot {class=StackBoxSlot}}
  HorizontalBox_5 {
   widgetClassPath=&HorizontalBox
   parent=StackBox_0
   widget=HorizontalBox_5
   slot {class=StackBoxSlot, properties {horizontalAlignment=HAlign_Center, verticalAlignment=VAlign_Center}}
  }
  Info_Image {
   widgetClassPath=&Image
   parent=HorizontalBox_5
   widget=Info_Image
   properties {brush {
     bIsDynamicallyLoaded=false
     drawAs=Image
     imageType=NoImage
     mirroring=NoMirror
     resourceName=None
     resourceObject=&MI_UIKit_Icon_Information
     tiling=NoTile
     imageSize=<32 32>
     margin=<ltrb 0 0 0 0>
     outlineSettings {
      bUseBrushTransparency=false
      roundingType=HalfHeightRadius
      width=0
      color {colorUseRule=UseColor_Specified, specifiedColor=<rgba 0 0 0 0>}
      cornerRadii=<0 0 0 0>
     }
     tintColor {colorUseRule=UseColor_Specified, specifiedColor=<rgba 1 1 1 0.8>}
     uVRegion {bIsValid=false, max=<0 0>, min=<0 0>}
    }}
   slot {class=HorizontalBoxSlot, properties {verticalAlignment=VAlign_Center, padding=<ltrb 0 0 14 0>}}
  }
  DisplayMessage_Empty_Text {
   widgetClassPath=&CommonTextBlock
   parent=HorizontalBox_5
   widget=DisplayMessage_Empty_Text
   properties {
    text=No Items To Show
    wrapTextAt=558.7021
    colorAndOpacity {colorUseRule=UseColor_Specified, specifiedColor=<rgba 1 1 1 0.8>}
    font {
     bForceMonospaced=false
     bMaterialIsStencil=false
     fontMaterial=null
     fontObject=&HeadingNow
     letterSpacing=0
     monospacedWidth=1
     size=18
     skewAmount=0
     typefaceFontName=76 Bold
     outlineSettings {
      bApplyOutlineToDropShadows=false
      bMiteredCorners=false
      bSeparateFillAlpha=false
      outlineMaterial=null
      outlineSize=1
      outlineColor=<rgba 0.0052 0.0052 0.0056 0.2>
     }
    }
   }
   slot {class=HorizontalBoxSlot, properties {horizontalAlignment=HAlign_Left, verticalAlignment=VAlign_Center}}
  }
  Spacer_4 {widgetClassPath=&Spacer, parent=StackBox_0, widget=Spacer_4, properties {size=<1 12>}, slot {class=StackBoxSlot}}
  HorizontalBox_6 {
   widgetClassPath=&HorizontalBox
   parent=StackBox_0
   widget=HorizontalBox_6
   slot {class=StackBoxSlot, properties {horizontalAlignment=HAlign_Center, verticalAlignment=VAlign_Center, size {sizeRule=Automatic, value=0.2}}}
  }
  DisplayMessage_Empty_Text_1 {
   widgetClassPath=&CommonTextBlock
   parent=HorizontalBox_6
   widget=DisplayMessage_Empty_Text_1
   properties {
    justification=Center
    text="Consider disabling category filters, tags, or search terms."
    wrapTextAt=500
    colorAndOpacity {colorUseRule=UseColor_Specified, specifiedColor=<rgba 1 1 1 0.8>}
    font {
     bForceMonospaced=false
     bMaterialIsStencil=false
     fontMaterial=null
     fontObject=&HeadingNow
     letterSpacing=0
     monospacedWidth=1
     size=18
     skewAmount=0
     typefaceFontName=74 Regular
     outlineSettings {
      bApplyOutlineToDropShadows=false
      bMiteredCorners=false
      bSeparateFillAlpha=false
      outlineMaterial=null
      outlineSize=1
      outlineColor=<rgba 0.0052 0.0052 0.0056 0.2>
     }
    }
   }
   slot {class=HorizontalBoxSlot, properties {horizontalAlignment=HAlign_Center, verticalAlignment=VAlign_Center}}
  }
  Button_TabEmpty {
   widgetClassPath=&WBP_UIKit_Button_Quiet
   bIsVariable=true
   parent=StackBox_0
   widget=Button_TabEmpty
   properties {text=Ok, visibility=HitTestInvisible, colorAndOpacity=<rgba 1 1 1 0>}
   slot {class=StackBoxSlot, properties {horizontalAlignment=HAlign_Center, verticalAlignment=VAlign_Center}}
  }
  Overlay_DownloadingFabCategories {
   widgetClassPath=&Overlay
   bInherited=true
   bIsVariable=true
   parent=Switcher_ItemTabWarnings
   widget=Overlay_DownloadingFabCategories
   slot {class=WidgetSwitcherSlot, properties {padding=<ltrb 0 40 0 0>}}
  }
  StackBox_4 {
   widgetClassPath=&StackBox
   parent=Overlay_DownloadingFabCategories
   widget=StackBox_4
   properties {orientation=Orient_Vertical}
   slot {class=OverlaySlot, properties {horizontalAlignment=HAlign_Fill, padding=<ltrb 24 0 0 0>}}
  }
  Spacer_1 {widgetClassPath=&Spacer, parent=StackBox_4, widget=Spacer_1, properties {size=<1 64>}, slot {class=StackBoxSlot}}
  StackBox_5 {widgetClassPath=&StackBox, parent=StackBox_4, widget=StackBox_5, slot {class=StackBoxSlot}}
  Info_Image_2 {
   widgetClassPath=&Image
   parent=StackBox_5
   widget=Info_Image_2
   properties {brush {
     bIsDynamicallyLoaded=false
     drawAs=Image
     imageType=NoImage
     mirroring=NoMirror
     resourceName=None
     resourceObject=&MI_UIKit_Icon_Information
     tiling=NoTile
     imageSize=<32 32>
     margin=<ltrb 0 0 0 0>
     outlineSettings {
      bUseBrushTransparency=false
      roundingType=HalfHeightRadius
      width=0
      color {colorUseRule=UseColor_Specified, specifiedColor=<rgba 0 0 0 0>}
      cornerRadii=<0 0 0 0>
     }
     tintColor {colorUseRule=UseColor_Specified, specifiedColor=<rgba 1 1 1 0.8>}
     uVRegion {bIsValid=false, max=<0 0>, min=<0 0>}
    }}
   slot {class=StackBoxSlot, properties {horizontalAlignment=HAlign_Left, verticalAlignment=VAlign_Top, padding=<ltrb 0 0 14 0>}}
  }
  QuantityTextBlock_2 {
   widgetClassPath=&CommonTextBlock
   bIsVariable=true
   parent=StackBox_5
   widget=QuantityTextBlock_2
   properties {
    autoWrapText=true
    text=Downloading Category List
    wrapTextAt=900
    colorAndOpacity {colorUseRule=UseColor_Specified, specifiedColor=<rgba 1 1 1 0.8>}
    font {
     bForceMonospaced=false
     bMaterialIsStencil=false
     fontMaterial=null
     fontObject=&HeadingNow
     letterSpacing=0
     monospacedWidth=1
     size=18
     skewAmount=0
     typefaceFontName=75 Medium Italic
     outlineSettings {
      bApplyOutlineToDropShadows=false
      bMiteredCorners=false
      bSeparateFillAlpha=false
      outlineMaterial=null
      outlineSize=1
      outlineColor=<rgba 0.0052 0.0052 0.0056 0.2>
     }
    }
   }
   slot {class=StackBoxSlot, properties {verticalAlignment=VAlign_Center, size {sizeRule=Fill, value=1}}}
  }
  Overlay_FabCategoriesNotLoaded {
   widgetClassPath=&Overlay
   bInherited=true
   bIsVariable=true
   parent=Switcher_ItemTabWarnings
   widget=Overlay_FabCategoriesNotLoaded
   slot {class=WidgetSwitcherSlot, properties {padding=<ltrb 0 40 0 0>}}
  }
  StackBox_6 {
   widgetClassPath=&StackBox
   parent=Overlay_FabCategoriesNotLoaded
   widget=StackBox_6
   properties {orientation=Orient_Vertical}
   slot {class=OverlaySlot, properties {horizontalAlignment=HAlign_Fill, padding=<ltrb 24 0 0 0>}}
  }
  Spacer_2 {widgetClassPath=&Spacer, parent=StackBox_6, widget=Spacer_2, properties {size=<1 64>}, slot {class=StackBoxSlot}}
  StackBox_7 {widgetClassPath=&StackBox, parent=StackBox_6, widget=StackBox_7, slot {class=StackBoxSlot}}
  Info_Image_3 {
   widgetClassPath=&Image
   parent=StackBox_7
   widget=Info_Image_3
   properties {brush {
     bIsDynamicallyLoaded=false
     drawAs=Image
     imageType=NoImage
     mirroring=NoMirror
     resourceName=None
     resourceObject=&MI_UIKit_Icon_Information
     tiling=NoTile
     imageSize=<32 32>
     margin=<ltrb 0 0 0 0>
     outlineSettings {
      bUseBrushTransparency=false
      roundingType=HalfHeightRadius
      width=0
      color {colorUseRule=UseColor_Specified, specifiedColor=<rgba 0 0 0 0>}
      cornerRadii=<0 0 0 0>
     }
     tintColor {colorUseRule=UseColor_Specified, specifiedColor=<rgba 1 1 1 0.8>}
     uVRegion {bIsValid=false, max=<0 0>, min=<0 0>}
    }}
   slot {class=StackBoxSlot, properties {horizontalAlignment=HAlign_Left, verticalAlignment=VAlign_Top, padding=<ltrb 0 0 14 0>}}
  }
  QuantityTextBlock_4 {
   widgetClassPath=&CommonTextBlock
   bIsVariable=true
   parent=StackBox_7
   widget=QuantityTextBlock_4
   properties {
    autoWrapText=true
    text=Category List Not Downloaded
    wrapTextAt=900
    colorAndOpacity {colorUseRule=UseColor_Specified, specifiedColor=<rgba 1 1 1 0.8>}
    font {
     bForceMonospaced=false
     bMaterialIsStencil=false
     fontMaterial=null
     fontObject=&HeadingNow
     letterSpacing=0
     monospacedWidth=1
     size=18
     skewAmount=0
     typefaceFontName=75 Medium Italic
     outlineSettings {
      bApplyOutlineToDropShadows=false
      bMiteredCorners=false
      bSeparateFillAlpha=false
      outlineMaterial=null
      outlineSize=1
      outlineColor=<rgba 0.0052 0.0052 0.0056 0.2>
     }
    }
   }
   slot {class=StackBoxSlot, properties {verticalAlignment=VAlign_Center, size {sizeRule=Fill, value=1}}}
  }
  MenuQuickbar_QuickBars {
   widgetClassPath=/CreativeEmporium/Widgets/WBP_ContentBrowserQuickbar
   bInherited=true
   bIsVariable=true
   parent=Grid_Center
   widget=MenuQuickbar_QuickBars
   properties {bSupportsActivationFocus=false}
   slot {class=GridSlot, properties {column=1, horizontalAlignment=HAlign_Center, row=1, verticalAlignment=VAlign_Bottom, nudge=<0 30>}}
  }
  Right_Button_Overlay {
   widgetClassPath=&Overlay
   bIsVariable=true
   parent=Grid_Center
   widget=Right_Button_Overlay
   properties {visibility=Collapsed}
   slot {class=GridSlot, properties {column=2, horizontalAlignment=HAlign_Right, layer=3, rowSpan=2, padding=<ltrb 0 42 -132 0>}}
  }
  SidePanelButton_Right {
   widgetClassPath=&WBP_UIKit_Button_Generic
   bIsVariable=true
   parent=Right_Button_Overlay
   widget=SidePanelButton_Right
   properties {
    bHideInputAction=true
    icon=&T_UI_Icon_CrRefresh_Chevron_Left
    iconMaterial=&MI_CrBrowser_ArrowIcons_Color
    minWidth=164
    shouldTextImpactLayout=false
    text=""
    useBackgroundBlock=false
    useTextBlock=false
    foregroundColor {colorUseRule=UseColor_Specified, specifiedColor=<rgba 0.8699 1 0.9 1>}
    iconPadding=<ltrb -46 0 0 0>
    iconSize=<32 32>
   }
   slot {class=OverlaySlot, properties {verticalAlignment=VAlign_Fill}}
  }
  Overlay_5 {widgetClassPath=&Overlay, namedSlotHost=SidePanelButton_Right, widget=Overlay_5}
  SizeBox_6 {
   widgetClassPath=&SizeBox
   bIsVariable=true
   parent=Overlay_5
   widget=SizeBox_6
   properties {bOverride_HeightOverride=true, bOverride_WidthOverride=true, heightOverride=24, widthOverride=24}
   slot {class=OverlaySlot, properties {verticalAlignment=VAlign_Center, padding=<ltrb 24 32 0 0>}}
  }
  Keybind_ButtonRight {
   widgetClassPath=&KeybindWidget
   bIsVariable=true
   parent=SizeBox_6
   widget=Keybind_ButtonRight
   properties {bUseKeyboardTextureForKeys=true, boundKeyFallback=LeftAlt, designTimeKey=LeftAlt}
   slot {class=SizeBoxSlot}
  }
  SizeBox_106 {
   widgetClassPath=&SizeBox
   bIsVariable=true
   parent=Overlay_5
   widget=SizeBox_106
   properties {bOverride_HeightOverride=true, bOverride_WidthOverride=true, heightOverride=24, widthOverride=24}
   slot {class=OverlaySlot, properties {verticalAlignment=VAlign_Center, padding=<ltrb 24 32 0 0>}}
  }
  Keybind_DockRight_ScreenEdge {
   widgetClassPath=&KeybindWidget
   bIsVariable=true
   parent=SizeBox_106
   widget=Keybind_DockRight_ScreenEdge
   properties {bUseKeyboardTextureForKeys=true, boundKeyFallback=D, designTimeKey=D, visibility=HitTestInvisible}
   slot {class=SizeBoxSlot, properties {horizontalAlignment=HAlign_Center, verticalAlignment=VAlign_Center}}
  }
  SidePanelRightHitZone_Image {
   widgetClassPath=&Image
   bIsVariable=true
   namedSlotHost=SidePanelButton_Right
   widget=SidePanelRightHitZone_Image
   properties {
    visibility=SelfHitTestInvisible
    brush {
     bIsDynamicallyLoaded=false
     drawAs=RoundedBox
     imageType=NoImage
     mirroring=NoMirror
     resourceName=None
     resourceObject=null
     tiling=NoTile
     imageSize=<32 32>
     margin=<ltrb 0 0 0 0>
     outlineSettings {
      bUseBrushTransparency=false
      roundingType=FixedRadius
      width=0
      color {colorUseRule=UseColor_Specified, specifiedColor=<rgba 0 0 0 0>}
      cornerRadii=<50 50 50 50>
     }
     tintColor {colorUseRule=UseColor_Specified, specifiedColor=<rgba 0 0 0 0.3>}
     uVRegion {bIsValid=false, max=<0 0>, min=<0 0>}
    }
    renderTransform {angle=0, scale=<1 1>, shear=<0 0>, translation=<6 0>}
   }
  }
  Right_Content_Overlay {
   widgetClassPath=&Overlay
   bIsVariable=true
   parent=Grid_Center
   widget=Right_Content_Overlay
   properties {navigation=&self:WidgetTree.Right_Content_Overlay.WidgetNavigation_0}
   slot {class=GridSlot, properties {column=2, horizontalAlignment=HAlign_Left, layer=2, rowSpan=2, padding=<ltrb 0 42 32 0>}}
  }
  SizeBox_RightPanel {
   widgetClassPath=&SizeBox
   parent=Right_Content_Overlay
   widget=SizeBox_RightPanel
   properties {
    bOverride_MaxDesiredWidth=true
    bOverride_WidthOverride=true
    heightOverride=717
    maxDesiredWidth=350
    minDesiredWidth=200
    widthOverride=350
   }
   slot {class=OverlaySlot, properties {horizontalAlignment=HAlign_Fill, verticalAlignment=VAlign_Fill}}
  }
  Overlay_RightFilter {widgetClassPath=&Overlay, bIsVariable=true, parent=SizeBox_RightPanel, widget=Overlay_RightFilter, slot {class=SizeBoxSlot}}
  RightSideBacking_Image {
   widgetClassPath=&Image
   bIsVariable=true
   parent=Overlay_RightFilter
   widget=RightSideBacking_Image
   properties {brush {
     bIsDynamicallyLoaded=false
     drawAs=Image
     imageType=FullColor
     mirroring=NoMirror
     resourceName=None
     resourceObject=/CreativeEmporium/Materials/MI_CreativeBrowser_PanelBacking
     tiling=NoTile
     imageSize=<32 32>
     margin=<ltrb 0 0 0 0>
     outlineSettings {
      bUseBrushTransparency=false
      roundingType=HalfHeightRadius
      width=0
      color {colorUseRule=UseColor_Specified, specifiedColor=<rgba 0 0 0 0>}
      cornerRadii=<0 0 0 0>
     }
     tintColor {colorUseRule=UseColor_Specified, specifiedColor=<rgba 1 1 1 1>}
     uVRegion {bIsValid=false, max=<0 0>, min=<0 0>}
    }}
   slot {class=OverlaySlot, properties {horizontalAlignment=HAlign_Fill, verticalAlignment=VAlign_Fill}}
  }
  VerticalBox_0 {
   widgetClassPath=&VerticalBox
   parent=Overlay_RightFilter
   widget=VerticalBox_0
   slot {class=OverlaySlot, properties {horizontalAlignment=HAlign_Fill, verticalAlignment=VAlign_Fill, padding=<ltrb 4 4 4 4>}}
  }
  RightSideTabs_Overlay {
   widgetClassPath=&Overlay
   bIsVariable=true
   parent=VerticalBox_0
   widget=RightSideTabs_Overlay
   properties {visibility=Collapsed}
   slot {class=VerticalBoxSlot}
  }
  HorizontalBox_1 {
   widgetClassPath=&HorizontalBox
   parent=RightSideTabs_Overlay
   widget=HorizontalBox_1
   slot {class=OverlaySlot, properties {horizontalAlignment=HAlign_Fill}}
  }
  RightSidePanel_Tabs {
   widgetClassPath=/FortUILibrary/UIKit/Tabs/WBP_UIKit_Tabs
   bIsVariable=true
   parent=HorizontalBox_1
   widget=RightSidePanel_Tabs
   properties {
    buttonClass=/CreativeEmporium/Widgets/WBP_Creative_Emporium_Button_SideNav_TabFilters.WBP_Creative_Emporium_Button_SideNav_TabFilters_C
    initialSelectedTabIndex=1
    inputActions=None
    minButtonWidth=148
    tabGroupHorizontalAlignment=HAlign_Center
    useBackplate=true
   }
   slot {class=HorizontalBoxSlot, properties {size {sizeRule=Fill, value=1}}}
  }
  RightPanelTabs_HardwareVisibility {
   widgetClassPath=&CommonHardwareVisibilityBorder
   parent=RightSideTabs_Overlay
   widget=RightPanelTabs_HardwareVisibility
   properties {
    horizontalAlignment=HAlign_Center
    verticalAlignment=VAlign_Center
    visibilityQuery {
     autoDescription=" ANY( Input.Gamepad )"
     tokenStreamVersion=0
     userDescription=""
     queryTokenStream [0,1,1,1,0]
     tagDictionary [{tagName=Input.Gamepad}]
    }
   }
   slot {class=OverlaySlot, properties {horizontalAlignment=HAlign_Center, verticalAlignment=VAlign_Center, padding=<ltrb 0 4 0 0>}}
  }
  Border_0 {
   widgetClassPath=&Border
   parent=RightPanelTabs_HardwareVisibility
   widget=Border_0
   properties {
    background {
     bIsDynamicallyLoaded=false
     drawAs=RoundedBox
     imageType=NoImage
     mirroring=NoMirror
     resourceName=None
     resourceObject=null
     tiling=NoTile
     imageSize=<32 32>
     margin=<ltrb 0 0 0 0>
     outlineSettings {
      bUseBrushTransparency=false
      roundingType=FixedRadius
      width=0
      color {colorUseRule=UseColor_Specified, specifiedColor=<rgba 0 0 0 0>}
      cornerRadii=<20 20 20 20>
     }
     tintColor {colorUseRule=UseColor_Specified, specifiedColor=<rgba 1 1 1 1>}
     uVRegion {bIsValid=false, max=<0 0>, min=<0 0>}
    }
    brushColor=<rgba 0 0 0 0.65>
    padding=<ltrb 3 3 3 3>
   }
   slot {class=BorderSlot, properties {horizontalAlignment=HAlign_Center, verticalAlignment=VAlign_Center, padding=<ltrb 0 0 0 0>}}
  }
  KeybindRightPanel_Tabs {
   widgetClassPath=&KeybindWidget
   parent=Border_0
   widget=KeybindRightPanel_Tabs
   properties {bUseKeyboardTextureForKeys=true, boundKeyFallback=Gamepad_RightThumbstick, designTimeKey=Gamepad_RightThumbstick}
   slot {class=BorderSlot, properties {padding=<ltrb 3 3 3 3>}}
  }
  Switcher_SidePanels {
   widgetClassPath=&CommonVisibilitySwitcher
   bInherited=true
   bIsVariable=true
   parent=VerticalBox_0
   widget=Switcher_SidePanels
   properties {clipping=ClipToBounds}
   slot {class=VerticalBoxSlot, properties {size {sizeRule=Fill, value=1}}}
  }
  ItemDetailsSidePanel_DefaultSidePanel {
   widgetClassPath=&/WBP_Creative_Emporium_DetailsPanel
   bInherited=true
   bIsVariable=true
   parent=Switcher_SidePanels
   widget=ItemDetailsSidePanel_DefaultSidePanel
   slot {class=CommonVisibilitySwitcherSlot}
  }
  FiltersSubPanel {
   widgetClassPath=&/WBP_Creative_Emporium_Filters_SubPanel
   bInherited=true
   bIsVariable=true
   parent=Switcher_SidePanels
   widget=FiltersSubPanel
   properties {filterEntryWidgetClass=&/WBP_CrBrowser_FilterEntryWidget.WBP_CrBrowser_FilterEntryWidget_C}
   slot {class=CommonVisibilitySwitcherSlot, properties {padding=<ltrb 4 24 4 24>}}
  }
  Overlay_1 {
   widgetClassPath=&Overlay
   parent=Overlay_RightFilter
   widget=Overlay_1
   slot {class=OverlaySlot, properties {horizontalAlignment=HAlign_Fill, verticalAlignment=VAlign_Fill}}
  }
  Blur_HitBlocker_Right {
   widgetClassPath=&BackgroundBlur
   bIsVariable=true
   parent=Overlay_1
   widget=Blur_HitBlocker_Right
   properties {blurStrength=10, visibility=Collapsed}
   slot {class=OverlaySlot, properties {horizontalAlignment=HAlign_Fill, verticalAlignment=VAlign_Fill}}
  }
  Button_HitBlocker_Right {
   widgetClassPath=&FortTextButton
   bIsVariable=true
   parent=Overlay_1
   widget=Button_HitBlocker_Right
   properties {
    defaultText=""
    onButtonBaseClicked=(&self:WidgetTree.Button_HitBlocker_Right.HandleOnButtonClicked)
    onButtonBaseDoubleClicked=(&self:WidgetTree.Button_HitBlocker_Right.HandleOnButtonDoubleClicked)
    onButtonBaseHovered=(&self:WidgetTree.Button_HitBlocker_Right.HandleOnButtonHovered)
    onButtonBaseUnhovered=(&self:WidgetTree.Button_HitBlocker_Right.HandleOnButtonUnhovered)
    onSelectedChangedBase=(&self:WidgetTree.Button_HitBlocker_Right.HandleOnSelectedChanged)
    visibility=Collapsed
    colorAndOpacity=<rgba 1 1 1 0>
   }
   slot {class=OverlaySlot, properties {horizontalAlignment=HAlign_Fill, verticalAlignment=VAlign_Fill}}
  }
  CollapseKeybind_Right_VBox {
   widgetClassPath=&VerticalBox
   bIsVariable=true
   parent=Right_Content_Overlay
   widget=CollapseKeybind_Right_VBox
   properties {visibility=Collapsed, renderTransform {angle=0, scale=<1 1>, shear=<0 0>, translation=<52 12>}}
   slot {class=OverlaySlot, properties {horizontalAlignment=HAlign_Right, verticalAlignment=VAlign_Center}}
  }
  RightSideCollapseArrow_Image {
   widgetClassPath=&Image
   bIsVariable=true
   parent=CollapseKeybind_Right_VBox
   widget=RightSideCollapseArrow_Image
   properties {
    brush {
     bIsDynamicallyLoaded=false
     drawAs=Image
     imageType=FullColor
     mirroring=NoMirror
     resourceName=None
     resourceObject=&T_UI_Icon_CrRefresh_Chevron_Left
     tiling=NoTile
     imageSize=<32 32>
     margin=<ltrb 0 0 0 0>
     outlineSettings {
      bUseBrushTransparency=false
      roundingType=HalfHeightRadius
      width=0
      color {colorUseRule=UseColor_Specified, specifiedColor=<rgba 0 0 0 0>}
      cornerRadii=<0 0 0 0>
     }
     tintColor {colorUseRule=UseColor_Specified, specifiedColor=<rgba 1 1 1 1>}
     uVRegion {bIsValid=false, max=<0 0>, min=<0 0>}
    }
    renderTransform {angle=0, scale=<-1 1>, shear=<0 0>, translation=<0 0>}
   }
   slot {class=VerticalBoxSlot}
  }
  SizeBox_108 {
   widgetClassPath=&SizeBox
   bIsVariable=true
   parent=CollapseKeybind_Right_VBox
   widget=SizeBox_108
   properties {bOverride_HeightOverride=true, bOverride_WidthOverride=true, heightOverride=24, widthOverride=24}
   slot {class=VerticalBoxSlot, properties {horizontalAlignment=HAlign_Center, verticalAlignment=VAlign_Center}}
  }
  Keybind_CollapseRight_ScreenEdge {
   widgetClassPath=&KeybindWidget
   bIsVariable=true
   parent=SizeBox_108
   widget=Keybind_CollapseRight_ScreenEdge
   properties {bUseKeyboardTextureForKeys=true, boundKeyFallback=D, designTimeKey=D, visibility=HitTestInvisible}
   slot {class=SizeBoxSlot}
  }
  Right_Keybinds_Overlay {
   widgetClassPath=&Overlay
   bIsVariable=true
   parent=Grid_Center
   widget=Right_Keybinds_Overlay
   slot {class=GridSlot, properties {column=2, row=2, verticalAlignment=VAlign_Bottom, padding=<ltrb 0 0 32 0>}}
  }
  FocusRight_Overlay {
   widgetClassPath=&Overlay
   bIsVariable=true
   parent=Right_Keybinds_Overlay
   widget=FocusRight_Overlay
   slot {class=OverlaySlot, properties {horizontalAlignment=HAlign_Fill}}
  }
  Image_80 {
   widgetClassPath=&Image
   bIsVariable=true
   parent=FocusRight_Overlay
   widget=Image_80
   properties {brush {
     bIsDynamicallyLoaded=false
     drawAs=RoundedBox
     imageType=NoImage
     mirroring=NoMirror
     resourceName=None
     resourceObject=null
     tiling=NoTile
     imageSize=<32 32>
     margin=<ltrb 0 0 0 0>
     outlineSettings {
      bUseBrushTransparency=false
      roundingType=FixedRadius
      width=0
      color {colorUseRule=UseColor_Specified, specifiedColor=<rgba 0 0 0 0>}
      cornerRadii=<0 0 20 20>
     }
     tintColor {colorUseRule=UseColor_Specified, specifiedColor=<rgba 1 1 1 0.15>}
     uVRegion {bIsValid=false, max=<0 0>, min=<0 0>}
    }}
   slot {class=OverlaySlot, properties {horizontalAlignment=HAlign_Fill, verticalAlignment=VAlign_Fill}}
  }
  HBox_FocusRight {
   widgetClassPath=&HorizontalBox
   bIsVariable=true
   parent=FocusRight_Overlay
   widget=HBox_FocusRight
   slot {class=OverlaySlot, properties {horizontalAlignment=HAlign_Center, padding=<ltrb 0 2 0 2>}}
  }
  Keybind_FocusRight {
   widgetClassPath=&KeybindWidget
   bIsVariable=true
   parent=HBox_FocusRight
   widget=Keybind_FocusRight
   properties {bUseKeyboardTextureForKeys=true, boundKeyFallback=LeftAlt, designTimeKey=LeftAlt}
   slot {class=HorizontalBoxSlot}
  }
  Text_FocusRight {
   widgetClassPath=&CommonTextBlock
   bIsVariable=true
   parent=HBox_FocusRight
   widget=Text_FocusRight
   properties {
    bAutoCollapseWithEmptyText=true
    text=Filters
    textTransformPolicy=ToUpper
    colorAndOpacity {colorUseRule=UseColor_Specified, specifiedColor=<rgba 1 1 1 0.7>}
    font {
     bForceMonospaced=false
     bMaterialIsStencil=false
     fontMaterial=null
     fontObject=&HeadingNow
     letterSpacing=0
     monospacedWidth=1
     size=13.5
     skewAmount=0
     typefaceFontName=75 Medium
     outlineSettings {
      bApplyOutlineToDropShadows=false
      bMiteredCorners=false
      bSeparateFillAlpha=false
      outlineMaterial=null
      outlineSize=0
      outlineColor=<rgba 0 0 0 1>
     }
    }
    shadowOffset=<0 0>
   }
   slot {class=HorizontalBoxSlot, properties {horizontalAlignment=HAlign_Right, verticalAlignment=VAlign_Center, padding=<ltrb 3 0 6 0>}}
  }
  SizeBox_DockRight {
   widgetClassPath=&SizeBox
   parent=Right_Keybinds_Overlay
   widget=SizeBox_DockRight
   slot {class=OverlaySlot, properties {horizontalAlignment=HAlign_Fill}}
  }
  DockRight_Overlay {widgetClassPath=&Overlay, bIsVariable=true, parent=SizeBox_DockRight, widget=DockRight_Overlay, slot {class=SizeBoxSlot}}
  RightSideEndcapButtonsBacking_Image {
   widgetClassPath=&Image
   bIsVariable=true
   parent=DockRight_Overlay
   widget=RightSideEndcapButtonsBacking_Image
   properties {
    visibility=SelfHitTestInvisible
    brush {
     bIsDynamicallyLoaded=false
     drawAs=RoundedBox
     imageType=NoImage
     mirroring=NoMirror
     resourceName=None
     resourceObject=null
     tiling=NoTile
     imageSize=<32 32>
     margin=<ltrb 0 0 0 0>
     outlineSettings {
      bUseBrushTransparency=false
      roundingType=FixedRadius
      width=0
      color {colorUseRule=UseColor_Specified, specifiedColor=<rgba 0 0 0 0>}
      cornerRadii=<0 0 20 20>
     }
     tintColor {colorUseRule=UseColor_Specified, specifiedColor=<rgba 1 1 1 0.15>}
     uVRegion {bIsValid=false, max=<0 0>, min=<0 0>}
    }
   }
   slot {class=OverlaySlot, properties {horizontalAlignment=HAlign_Fill, verticalAlignment=VAlign_Fill}}
  }
  HBox_DockRight {
   widgetClassPath=&HorizontalBox
   bIsVariable=true
   parent=DockRight_Overlay
   widget=HBox_DockRight
   slot {class=OverlaySlot, properties {horizontalAlignment=HAlign_Center, padding=<ltrb 0 1 0 1>}}
  }
  Delete_This {
   widgetClassPath=&Overlay
   parent=HBox_DockRight
   widget=Delete_This
   properties {visibility=Collapsed}
   slot {class=HorizontalBoxSlot, properties {horizontalAlignment=HAlign_Center, verticalAlignment=VAlign_Center}}
  }
  Keybind_DockRight {
   widgetClassPath=&KeybindWidget
   bIsVariable=true
   parent=Delete_This
   widget=Keybind_DockRight
   properties {bUseKeyboardTextureForKeys=true, boundKeyFallback=D, designTimeKey=D}
   slot {class=OverlaySlot}
  }
  Text_DockRight {
   widgetClassPath=&CommonTextBlock
   bIsVariable=true
   parent=Delete_This
   widget=Text_DockRight
   properties {
    bAutoCollapseWithEmptyText=true
    text=Dock
    textTransformPolicy=ToUpper
    colorAndOpacity {colorUseRule=UseColor_Specified, specifiedColor=<rgba 1 1 1 0.8>}
    font {
     bForceMonospaced=false
     bMaterialIsStencil=false
     fontMaterial=null
     fontObject=&HeadingNow
     letterSpacing=0
     monospacedWidth=1
     size=12
     skewAmount=0
     typefaceFontName=74 Regular
     outlineSettings {
      bApplyOutlineToDropShadows=false
      bMiteredCorners=false
      bSeparateFillAlpha=false
      outlineMaterial=null
      outlineSize=0
      outlineColor=<rgba 0 0 0 1>
     }
    }
    shadowOffset=<0 0>
   }
   slot {class=OverlaySlot}
  }
  Border {
   widgetClassPath=&Border
   parent=HBox_DockRight
   widget=Border
   properties {
    background {
     bIsDynamicallyLoaded=false
     drawAs=RoundedBox
     imageType=NoImage
     mirroring=NoMirror
     resourceName=None
     resourceObject=null
     tiling=NoTile
     imageSize=<32 32>
     margin=<ltrb 0 0 0 0>
     outlineSettings {
      bUseBrushTransparency=false
      roundingType=HalfHeightRadius
      width=0
      color {colorUseRule=UseColor_Specified, specifiedColor=<rgba 0 0 0 0>}
      cornerRadii=<10 10 10 10>
     }
     tintColor {colorUseRule=UseColor_Specified, specifiedColor=<rgba 1 1 1 1>}
     uVRegion {bIsValid=false, max=<0 0>, min=<0 0>}
    }
    brushColor=<rgba 1 1 1 0>
    padding=<ltrb 6 1 8 1>
   }
   slot {class=HorizontalBoxSlot, properties {padding=<ltrb 2 0 0 0>}}
  }
  HorizontalBox {widgetClassPath=&HorizontalBox, parent=Border, widget=HorizontalBox, slot {class=BorderSlot, properties {padding=<ltrb 6 1 8 1>}}}
  Keybind_ClearTags {
   widgetClassPath=&KeybindWidget
   bIsVariable=true
   parent=HorizontalBox
   widget=Keybind_ClearTags
   properties {bUseKeyboardTextureForKeys=true, boundKeyFallback=Gamepad_FaceButton_Top, designTimeKey=Gamepad_FaceButton_Top}
   slot {class=HorizontalBoxSlot, properties {padding=<ltrb 0 0 2 0>}}
  }
  Text_ClearTags {
   widgetClassPath=&CommonTextBlock
   bIsVariable=true
   parent=HorizontalBox
   widget=Text_ClearTags
   properties {
    bAutoCollapseWithEmptyText=true
    text=Clear
    textTransformPolicy=ToUpper
    colorAndOpacity {colorUseRule=UseColor_Specified, specifiedColor=<rgba 1 1 1 0.8>}
    font {
     bForceMonospaced=false
     bMaterialIsStencil=false
     fontMaterial=null
     fontObject=&HeadingNow
     letterSpacing=0
     monospacedWidth=1
     size=12
     skewAmount=0
     typefaceFontName=74 Regular
     outlineSettings {
      bApplyOutlineToDropShadows=false
      bMiteredCorners=false
      bSeparateFillAlpha=false
      outlineMaterial=null
      outlineSize=0
      outlineColor=<rgba 0 0 0 1>
     }
    }
    shadowOffset=<0 0>
   }
   slot {class=HorizontalBoxSlot, properties {horizontalAlignment=HAlign_Right, verticalAlignment=VAlign_Center}}
  }
  Border_1 {
   widgetClassPath=&Border
   parent=HBox_DockRight
   widget=Border_1
   properties {
    background {
     bIsDynamicallyLoaded=false
     drawAs=RoundedBox
     imageType=NoImage
     mirroring=NoMirror
     resourceName=None
     resourceObject=null
     tiling=NoTile
     imageSize=<32 32>
     margin=<ltrb 0 0 0 0>
     outlineSettings {
      bUseBrushTransparency=false
      roundingType=HalfHeightRadius
      width=0
      color {colorUseRule=UseColor_Specified, specifiedColor=<rgba 0 0 0 0>}
      cornerRadii=<10 10 10 10>
     }
     tintColor {colorUseRule=UseColor_Specified, specifiedColor=<rgba 1 1 1 1>}
     uVRegion {bIsValid=false, max=<0 0>, min=<0 0>}
    }
    brushColor=<rgba 1 1 1 0>
    padding=<ltrb 6 1 8 1>
   }
   slot {class=HorizontalBoxSlot, properties {padding=<ltrb 2 0 0 0>}}
  }
  HorizontalBox_0 {
   widgetClassPath=&HorizontalBox
   parent=Border_1
   widget=HorizontalBox_0
   slot {class=BorderSlot, properties {padding=<ltrb 6 1 8 1>}}
  }
  Keybind_CycleTags {
   widgetClassPath=&KeybindWidget
   bIsVariable=true
   parent=HorizontalBox_0
   widget=Keybind_CycleTags
   properties {bUseKeyboardTextureForKeys=true, boundKeyFallback=Gamepad_FaceButton_Top, designTimeKey=Gamepad_FaceButton_Top}
   slot {class=HorizontalBoxSlot, properties {padding=<ltrb 0 0 2 0>}}
  }
  Text_CycleTags {
   widgetClassPath=&CommonTextBlock
   bIsVariable=true
   parent=HorizontalBox_0
   widget=Text_CycleTags
   properties {
    bAutoCollapseWithEmptyText=true
    text=Mode
    textTransformPolicy=ToUpper
    colorAndOpacity {colorUseRule=UseColor_Specified, specifiedColor=<rgba 1 1 1 0.8>}
    font {
     bForceMonospaced=false
     bMaterialIsStencil=false
     fontMaterial=null
     fontObject=&HeadingNow
     letterSpacing=0
     monospacedWidth=1
     size=12
     skewAmount=0
     typefaceFontName=74 Regular
     outlineSettings {
      bApplyOutlineToDropShadows=false
      bMiteredCorners=false
      bSeparateFillAlpha=false
      outlineMaterial=null
      outlineSize=0
      outlineColor=<rgba 0 0 0 1>
     }
    }
    shadowOffset=<0 0>
   }
   slot {class=HorizontalBoxSlot, properties {horizontalAlignment=HAlign_Right, verticalAlignment=VAlign_Center}}
  }
  Right_Background {
   widgetClassPath=&BackgroundBlur
   bIsVariable=true
   parent=Grid_Center
   widget=Right_Background
   properties {
    blurStrength=10
    renderOpacity=0.7
    lowQualityFallbackBrush {
     bIsDynamicallyLoaded=false
     drawAs=Image
     imageType=NoImage
     mirroring=NoMirror
     resourceName=None
     resourceObject=null
     tiling=NoTile
     imageSize=<0 0>
     margin=<ltrb 0 0 0 0>
     outlineSettings {
      bUseBrushTransparency=false
      roundingType=HalfHeightRadius
      width=0
      color {colorUseRule=UseColor_Specified, specifiedColor=<rgba 0 0 0 0>}
      cornerRadii=<0 0 0 0>
     }
     tintColor {colorUseRule=UseColor_Specified, specifiedColor=<rgba 0 0.0627 0.1529 0.7>}
     uVRegion {bIsValid=false, max=<0 0>, min=<0 0>}
    }
   }
   slot {class=GridSlot, properties {column=2, layer=1, rowSpan=2, padding=<ltrb 0 42 -100 0>}}
  }
  Border_Right {
   widgetClassPath=&Border
   bIsVariable=true
   parent=Right_Background
   widget=Border_Right
   properties {
    background {
     bIsDynamicallyLoaded=false
     drawAs=Image
     imageType=NoImage
     mirroring=NoMirror
     resourceName=None
     resourceObject=&MI_CrBrowser_Backplate_Nav
     tiling=NoTile
     imageSize=<32 32>
     margin=<ltrb 0 0 0 0>
     outlineSettings {
      bUseBrushTransparency=false
      roundingType=HalfHeightRadius
      width=0
      color {colorUseRule=UseColor_Specified, specifiedColor=<rgba 0 0 0 0>}
      cornerRadii=<0 0 0 0>
     }
     tintColor {colorUseRule=UseColor_Specified, specifiedColor=<rgba 1 1 1 1>}
     uVRegion {bIsValid=false, max=<0 0>, min=<0 0>}
    }
    brushColor=<rgba 0.0048 0.004 0.0116 0.85>
    padding=<ltrb 0 0 0 0>
   }
   slot {class=BackgroundBlurSlot, properties {padding=<ltrb 0 0 0 0>}}
  }
  HBox_Bottom {
   widgetClassPath=&HorizontalBox
   parent=VBox_ItemBrowser
   widget=HBox_Bottom
   slot {class=VerticalBoxSlot, properties {verticalAlignment=VAlign_Bottom}}
  }
  Overlay_Switcher {
   widgetClassPath=&Overlay
   bIsVariable=true
   parent=HBox_Bottom
   widget=Overlay_Switcher
   slot {
    class=HorizontalBoxSlot
    properties {horizontalAlignment=HAlign_Center, verticalAlignment=VAlign_Bottom, padding=<ltrb 0 12 0 0>, size {sizeRule=Fill, value=1}}
   }
  }
  SizeBox_0 {
   widgetClassPath=&SizeBox
   parent=Overlay_Switcher
   widget=SizeBox_0
   properties {bOverride_HeightOverride=true, heightOverride=70}
   slot {class=OverlaySlot, properties {horizontalAlignment=HAlign_Center, verticalAlignment=VAlign_Center}}
  }
  Box_CommandBar {
   widgetClassPath=&HorizontalBox
   parent=SizeBox_0
   widget=Box_CommandBar
   slot {class=SizeBoxSlot, properties {horizontalAlignment=HAlign_Center}}
  }
  Switcher_CommandBar {
   widgetClassPath=&WidgetSwitcher
   bInherited=true
   bIsVariable=true
   parent=Box_CommandBar
   widget=Switcher_CommandBar
   slot {class=HorizontalBoxSlot, properties {horizontalAlignment=HAlign_Center}}
  }
  Overlay_ChestCommands {
   widgetClassPath=&Overlay
   bInherited=true
   parent=Switcher_CommandBar
   widget=Overlay_ChestCommands
   slot {class=WidgetSwitcherSlot}
  }
  ChestCommands_Overlay {
   widgetClassPath=&Overlay
   parent=Overlay_ChestCommands
   widget=ChestCommands_Overlay
   slot {class=OverlaySlot, properties {horizontalAlignment=HAlign_Center, verticalAlignment=VAlign_Center}}
  }
  ChestCommandsBackplate_Image {
   widgetClassPath=&Image
   bIsVariable=true
   parent=ChestCommands_Overlay
   widget=ChestCommandsBackplate_Image
   properties {
    visibility=SelfHitTestInvisible
    brush {
     bIsDynamicallyLoaded=false
     drawAs=Image
     imageType=NoImage
     mirroring=NoMirror
     resourceName=None
     resourceObject=/CreativeEmporium/Materials/MI_CrBrowser_Backplate_BottomBar
     tiling=NoTile
     imageSize=<32 32>
     margin=<ltrb 0 0 0 0>
     outlineSettings {
      bUseBrushTransparency=false
      roundingType=HalfHeightRadius
      width=0
      color {colorUseRule=UseColor_Specified, specifiedColor=<rgba 0 0 0 0>}
      cornerRadii=<10 10 10 10>
     }
     tintColor {colorUseRule=UseColor_Specified, specifiedColor=<rgba 1 1 1 1>}
     uVRegion {bIsValid=false, max=<0 0>, min=<0 0>}
    }
    colorAndOpacity=<rgba 0 0 0 1>
   }
   slot {class=OverlaySlot, properties {horizontalAlignment=HAlign_Fill, verticalAlignment=VAlign_Fill}}
  }
  Box_ChestCommands {
   widgetClassPath=&HorizontalBox
   parent=ChestCommands_Overlay
   widget=Box_ChestCommands
   slot {class=OverlaySlot, properties {horizontalAlignment=HAlign_Center, verticalAlignment=VAlign_Center}}
  }
  Button_CreateChest {
   widgetClassPath=/FortUILibrary/UIKit/Buttons/WBP_UIKit_Button_Loud
   bInherited=true
   bIsVariable=true
   parent=Box_ChestCommands
   widget=Button_CreateChest
   properties {
    bIsFocusable=false
    buttonSize=Small
    minHeight=56
    text=Create Chest
    triggeringInputAction {dataTable=/FortUILibrary/Input/CreativeInputActionDataTable, rowName=InventoryV2_CreateChest}
   }
   slot {class=HorizontalBoxSlot, properties {verticalAlignment=VAlign_Center, padding=<ltrb 5 0 5 0>}}
  }
  Button_CreateLlama {
   widgetClassPath=&WBP_UIKit_Button_Loud
   bInherited=true
   bIsVariable=true
   parent=Box_ChestCommands
   widget=Button_CreateLlama
   properties {
    bIsFocusable=false
    buttonSize=Small
    minHeight=56
    text=Create Llama
    triggeringInputAction {dataTable=&CreativeInputActionDataTable, rowName=InventoryV2_CreateLlama}
   }
   slot {class=HorizontalBoxSlot, properties {verticalAlignment=VAlign_Center, padding=<ltrb 5 0 5 0>}}
  }
  Button_RemoveFromChest {
   widgetClassPath=&WBP_UIKit_Button_Quiet
   bInherited=true
   bIsVariable=true
   parent=Box_ChestCommands
   widget=Button_RemoveFromChest
   properties {
    bIsFocusable=false
    minHeight=48
    minWidth=172
    text=Remove
    triggeringInputAction {dataTable=&CreativeInputActionDataTable, rowName=InventoryV2_RemoveItem}
   }
   slot {class=HorizontalBoxSlot, properties {verticalAlignment=VAlign_Center, padding=<ltrb 5 0 5 0>}}
  }
  Button_ResetChest {
   widgetClassPath=&WBP_UIKit_Button_Quiet
   bInherited=true
   bIsVariable=true
   parent=Box_ChestCommands
   widget=Button_ResetChest
   properties {
    bIsFocusable=false
    minWidth=172
    text=Reset
    triggeringInputAction {dataTable=&CreativeInputActionDataTable, rowName=InventoryV2_RemoveAll}
   }
   slot {class=HorizontalBoxSlot, properties {verticalAlignment=VAlign_Center, padding=<ltrb 5 0 5 0>}}
  }
  Overlay_VendCommands {
   widgetClassPath=&Overlay
   bInherited=true
   parent=Switcher_CommandBar
   widget=Overlay_VendCommands
   slot {class=WidgetSwitcherSlot}
  }
  VendCommands_Overlay {
   widgetClassPath=&Overlay
   parent=Overlay_VendCommands
   widget=VendCommands_Overlay
   slot {class=OverlaySlot, properties {horizontalAlignment=HAlign_Center, verticalAlignment=VAlign_Center}}
  }
  VendCommandsBackplate_Image {
   widgetClassPath=&Image
   bIsVariable=true
   parent=VendCommands_Overlay
   widget=VendCommandsBackplate_Image
   properties {
    visibility=SelfHitTestInvisible
    brush {
     bIsDynamicallyLoaded=false
     drawAs=Image
     imageType=NoImage
     mirroring=NoMirror
     resourceName=None
     resourceObject=&MI_CrBrowser_Backplate_BottomBar
     tiling=NoTile
     imageSize=<32 32>
     margin=<ltrb 0 0 0 0>
     outlineSettings {
      bUseBrushTransparency=false
      roundingType=HalfHeightRadius
      width=0
      color {colorUseRule=UseColor_Specified, specifiedColor=<rgba 0 0 0 0>}
      cornerRadii=<10 10 10 10>
     }
     tintColor {colorUseRule=UseColor_Specified, specifiedColor=<rgba 1 1 1 1>}
     uVRegion {bIsValid=false, max=<0 0>, min=<0 0>}
    }
   }
   slot {class=OverlaySlot, properties {horizontalAlignment=HAlign_Fill, verticalAlignment=VAlign_Fill}}
  }
  Box_VendCommands {
   widgetClassPath=&HorizontalBox
   parent=VendCommands_Overlay
   widget=Box_VendCommands
   slot {class=OverlaySlot, properties {verticalAlignment=VAlign_Center}}
  }
  LeftButtons {
   widgetClassPath=&HorizontalBox
   parent=Box_VendCommands
   widget=LeftButtons
   slot {class=HorizontalBoxSlot, properties {horizontalAlignment=HAlign_Left, verticalAlignment=VAlign_Center}}
  }
  Button_OpenItem {
   widgetClassPath=&WBP_UIKit_Button_Loud
   bInherited=true
   bIsVariable=true
   parent=LeftButtons
   widget=Button_OpenItem
   properties {
    bIsFocusable=false
    buttonSize=Small
    minHeight=56
    minWidth=200
    text=Open
    triggeringInputAction {dataTable=&CreativeInputActionDataTable, rowName=InventoryV2_Open}
   }
   slot {class=HorizontalBoxSlot, properties {verticalAlignment=VAlign_Center, padding=<ltrb 8 0 8 0>}}
  }
  Button_PlaceNow {
   widgetClassPath=&WBP_UIKit_Button_Loud
   bInherited=true
   bIsVariable=true
   parent=LeftButtons
   widget=Button_PlaceNow
   properties {
    bIsFocusable=false
    buttonSize=Small
    minWidth=200
    text=Place Now
    triggeringInputAction {dataTable=&CreativeInputActionDataTable, rowName=InventoryV2_PlaceNow}
   }
   slot {class=HorizontalBoxSlot, properties {verticalAlignment=VAlign_Center, padding=<ltrb 8 0 8 0>}}
  }
  CenterButtons {
   widgetClassPath=&HorizontalBox
   parent=Box_VendCommands
   widget=CenterButtons
   slot {class=HorizontalBoxSlot, properties {horizontalAlignment=HAlign_Left, verticalAlignment=VAlign_Center}}
  }
  Button_Equip {
   widgetClassPath=&WBP_UIKit_Button_Quiet
   bInherited=true
   bIsVariable=true
   parent=CenterButtons
   widget=Button_Equip
   properties {
    bIsFocusable=false
    minWidth=200
    text=Equip
    triggeringInputAction {dataTable=&CreativeInputActionDataTable, rowName=InventoryV2_Equip}
   }
   slot {class=HorizontalBoxSlot, properties {verticalAlignment=VAlign_Center, padding=<ltrb 8 0 8 0>}}
  }
  Button_AddToQuickBar {
   widgetClassPath=&WBP_UIKit_Button_Quiet
   bInherited=true
   bIsVariable=true
   parent=CenterButtons
   widget=Button_AddToQuickBar
   properties {
    bIsFocusable=false
    minWidth=200
    text=Add to Quick Bar
    triggeringInputAction {dataTable=&CreativeInputActionDataTable, rowName=InventoryV2_AddToQuickBar}
   }
   slot {class=HorizontalBoxSlot, properties {verticalAlignment=VAlign_Center, padding=<ltrb 8 0 8 0>}}
  }
  RightButtons {
   widgetClassPath=&HorizontalBox
   parent=Box_VendCommands
   widget=RightButtons
   slot {class=HorizontalBoxSlot, properties {horizontalAlignment=HAlign_Left, verticalAlignment=VAlign_Center}}
  }
  Button_AddToChest {
   widgetClassPath=&WBP_UIKit_Button_Quiet
   bInherited=true
   bIsVariable=true
   parent=RightButtons
   widget=Button_AddToChest
   properties {
    bIsFocusable=false
    text=Add to Chest
    triggeringInputAction {dataTable=&CreativeInputActionDataTable, rowName=InventoryV2_SendItemToChest}
   }
   slot {class=HorizontalBoxSlot, properties {verticalAlignment=VAlign_Center, padding=<ltrb 8 0 8 0>}}
  }
  Button_Drop {
   widgetClassPath=&WBP_UIKit_Button_Quiet
   bInherited=true
   bIsVariable=true
   parent=RightButtons
   widget=Button_Drop
   properties {bIsFocusable=false, text=Drop, triggeringInputAction {dataTable=&CreativeInputActionDataTable, rowName=InventoryV2_Drop}}
   slot {class=HorizontalBoxSlot, properties {verticalAlignment=VAlign_Center, padding=<ltrb 8 0 8 0>}}
  }
  Experimental_Text {
   widgetClassPath=&CommonTextBlock
   bIsVariable=true
   parent=Overlay_ItemBrowser
   widget=Experimental_Text
   properties {
    justification=Right
    mobileFontSizeMultiplier=1.34
    text=This version will be disabled soon and replaced with the new and improved Content Browser.
    visibility=Collapsed
    wrapTextAt=400
    colorAndOpacity {colorUseRule=UseColor_Specified, specifiedColor=<rgba 0.7843 0.7843 0.7843 1>}
    font {
     bForceMonospaced=false
     bMaterialIsStencil=false
     fontMaterial=null
     fontObject=&HeadingNow
     letterSpacing=0
     monospacedWidth=1
     size=11.25
     skewAmount=0
     typefaceFontName=74 Regular
     outlineSettings {
      bApplyOutlineToDropShadows=false
      bMiteredCorners=false
      bSeparateFillAlpha=false
      outlineMaterial=null
      outlineSize=0
      outlineColor=<rgba 0 0 0 1>
     }
    }
   }
   slot {class=OverlaySlot, properties {horizontalAlignment=HAlign_Right, padding=<ltrb 0 20 0 0>}}
  }
  FilterModeToast {
   widgetClassPath=&SizeBox
   bIsVariable=true
   parent=Overlay_ItemBrowser
   widget=FilterModeToast
   properties {
    bOverride_WidthOverride=true
    heightOverride=86
    renderOpacity=0
    visibility=HitTestInvisible
    widthOverride=600
    renderTransform {angle=0, scale=<1 1>, shear=<0 0>, translation=<56 0>}
   }
   slot {class=OverlaySlot, properties {horizontalAlignment=HAlign_Right}}
  }
  Overlay_0 {
   widgetClassPath=&Overlay
   parent=FilterModeToast
   widget=Overlay_0
   properties {visibility=HitTestInvisible}
   slot {class=SizeBoxSlot}
  }
  WBP_UIKit_Backplate {
   widgetClassPath=&WBP_UIKit_Backplate
   parent=Overlay_0
   widget=WBP_UIKit_Backplate
   properties {brightness=Obscure}
   slot {class=OverlaySlot, properties {horizontalAlignment=HAlign_Fill, verticalAlignment=VAlign_Fill}}
  }
  HorizontalBox_4 {
   widgetClassPath=&HorizontalBox
   parent=Overlay_0
   widget=HorizontalBox_4
   properties {visibility=HitTestInvisible}
   slot {class=OverlaySlot, properties {horizontalAlignment=HAlign_Fill, padding=<ltrb 12 12 12 12>}}
  }
  FilterMode_Image {
   widgetClassPath=&Image
   bIsVariable=true
   parent=HorizontalBox_4
   widget=FilterMode_Image
   properties {
    renderOpacity=0.9
    visibility=HitTestInvisible
    brush {
     bIsDynamicallyLoaded=false
     drawAs=Image
     imageType=FullColor
     mirroring=NoMirror
     resourceName=None
     resourceObject=&T_UI_Icon_FilterUnion
     tiling=NoTile
     imageSize=<42 42>
     margin=<ltrb 0 0 0 0>
     outlineSettings {
      bUseBrushTransparency=false
      roundingType=HalfHeightRadius
      width=0
      color {colorUseRule=UseColor_Specified, specifiedColor=<rgba 0 0 0 0>}
      cornerRadii=<0 0 0 0>
     }
     tintColor {colorUseRule=UseColor_Specified, specifiedColor=<rgba 1 1 1 1>}
     uVRegion {bIsValid=false, max=<0 0>, min=<0 0>}
    }
   }
   slot {class=HorizontalBoxSlot, properties {verticalAlignment=VAlign_Top}}
  }
  VerticalBox_1 {
   widgetClassPath=&VerticalBox
   parent=HorizontalBox_4
   widget=VerticalBox_1
   properties {visibility=HitTestInvisible}
   slot {class=HorizontalBoxSlot, properties {padding=<ltrb 12 0 0 0>, size {sizeRule=Fill, value=1}}}
  }
  FilterName_Text {
   widgetClassPath=&CommonTextBlock
   bIsVariable=true
   parent=VerticalBox_1
   widget=FilterName_Text
   properties {
    autoWrapText=true
    renderOpacity=0.9
    text="{FilterMode}"
    visibility=HitTestInvisible
    font {
     bForceMonospaced=false
     bMaterialIsStencil=false
     fontMaterial=null
     fontObject=&HeadingNow
     letterSpacing=0
     monospacedWidth=1
     size=15
     skewAmount=0
     typefaceFontName=74 Regular
     outlineSettings {
      bApplyOutlineToDropShadows=false
      bMiteredCorners=false
      bSeparateFillAlpha=false
      outlineMaterial=null
      outlineSize=0
      outlineColor=<rgba 0 0 0 1>
     }
    }
    shadowOffset=<0 0>
   }
   slot {class=VerticalBoxSlot, properties {padding=<ltrb 0 0 0 2>}}
  }
  FilterModeDescrtiption_Text {
   widgetClassPath=&CommonTextBlock
   bIsVariable=true
   parent=VerticalBox_1
   widget=FilterModeDescrtiption_Text
   properties {
    autoWrapText=true
    renderOpacity=0.9
    text="{Description}"
    visibility=HitTestInvisible
    font {
     bForceMonospaced=false
     bMaterialIsStencil=false
     fontMaterial=null
     fontObject=&HeadingNow
     letterSpacing=0
     monospacedWidth=1
     size=15
     skewAmount=0
     typefaceFontName=74 Regular
     outlineSettings {
      bApplyOutlineToDropShadows=false
      bMiteredCorners=false
      bSeparateFillAlpha=false
      outlineMaterial=null
      outlineSize=0
      outlineColor=<rgba 0 0 0 1>
     }
    }
    shadowOffset=<0 0>
   }
   slot {class=VerticalBoxSlot}
  }
  ItemProductPagePanel {
   widgetClassPath=/CreativeEmporium/Widgets/WBP_Creative_Emporium_ProductPage
   bInherited=true
   bIsVariable=true
   parent=Switcher_ProductPage
   widget=ItemProductPagePanel
   slot {class=CommonVisibilitySwitcherSlot}
  }
  Overlay_InventoryNotAllowed {
   widgetClassPath=&Overlay
   bInherited=true
   bIsVariable=true
   parent=Switcher_InventoryPermission
   widget=Overlay_InventoryNotAllowed
   slot {class=WidgetSwitcherSlot}
  }
  SizeBox_7 {
   widgetClassPath=&SizeBox
   parent=Overlay_InventoryNotAllowed
   widget=SizeBox_7
   properties {bOverride_WidthOverride=true, widthOverride=1024}
   slot {class=OverlaySlot, properties {horizontalAlignment=HAlign_Center, verticalAlignment=VAlign_Center, padding=<ltrb 0 -128 0 0>}}
  }
  HorizontalBox_9 {widgetClassPath=&HorizontalBox, parent=SizeBox_7, widget=HorizontalBox_9, slot {class=SizeBoxSlot}}
  Image_401 {
   widgetClassPath=&Image
   parent=HorizontalBox_9
   widget=Image_401
   properties {
    brush {
     bIsDynamicallyLoaded=false
     drawAs=Image
     imageType=NoImage
     mirroring=NoMirror
     resourceName=None
     resourceObject=/Game/Athena/HUD/Art/Materials/M_UI_CreativeLogo
     tiling=NoTile
     imageSize=<164 164>
     margin=<ltrb 0 0 0 0>
     outlineSettings {
      bUseBrushTransparency=false
      roundingType=HalfHeightRadius
      width=0
      color {colorUseRule=UseColor_Specified, specifiedColor=<rgba 0 0 0 0>}
      cornerRadii=<0 0 0 0>
     }
     tintColor {colorUseRule=UseColor_Specified, specifiedColor=<rgba 1 1 1 1>}
     uVRegion {bIsValid=false, max=<0 0>, min=<0 0>}
    }
    renderTransform {angle=-4, scale=<1 1>, shear=<0 0>, translation=<0 0>}
   }
   slot {class=HorizontalBoxSlot, properties {horizontalAlignment=HAlign_Left, verticalAlignment=VAlign_Center}}
  }
  QuantityTextBlock {
   widgetClassPath=&CommonTextBlock
   bIsVariable=true
   parent=HorizontalBox_9
   widget=QuantityTextBlock
   properties {
    autoWrapText=true
    text=This menu is only available where you can build.
    colorAndOpacity {colorUseRule=UseColor_Specified, specifiedColor=<rgba 0.5647 0.7922 0.9412 1>}
    font {
     bForceMonospaced=false
     bMaterialIsStencil=false
     fontMaterial=null
     fontObject=&BurbankSmall
     letterSpacing=0
     monospacedWidth=1
     size=26
     skewAmount=0
     typefaceFontName=Black
     outlineSettings {
      bApplyOutlineToDropShadows=false
      bMiteredCorners=false
      bSeparateFillAlpha=false
      outlineMaterial=null
      outlineSize=0
      outlineColor=<rgba 0 0 0 1>
     }
    }
    renderTransform {angle=0, scale=<1 1>, shear=<-8 0>, translation=<0 0>}
    shadowOffset=<0 0>
   }
   slot {class=HorizontalBoxSlot, properties {verticalAlignment=VAlign_Center, padding=<ltrb -32 0 0 0>}}
  }
  Overlay_InitializingInventory {
   widgetClassPath=&Overlay
   bInherited=true
   bIsVariable=true
   parent=Switcher_InventoryPermission
   widget=Overlay_InitializingInventory
   slot {class=WidgetSwitcherSlot}
  }
  SizeBox_1 {
   widgetClassPath=&SizeBox
   parent=Overlay_InitializingInventory
   widget=SizeBox_1
   properties {bOverride_WidthOverride=true, widthOverride=1024}
   slot {class=OverlaySlot, properties {horizontalAlignment=HAlign_Center, verticalAlignment=VAlign_Center, padding=<ltrb 0 -128 0 0>}}
  }
  HorizontalBox_3 {widgetClassPath=&HorizontalBox, parent=SizeBox_1, widget=HorizontalBox_3, slot {class=SizeBoxSlot}}
  Image_1 {
   widgetClassPath=&Image
   parent=HorizontalBox_3
   widget=Image_1
   properties {
    brush {
     bIsDynamicallyLoaded=false
     drawAs=Image
     imageType=NoImage
     mirroring=NoMirror
     resourceName=None
     resourceObject=&M_UI_CreativeLogo
     tiling=NoTile
     imageSize=<164 164>
     margin=<ltrb 0 0 0 0>
     outlineSettings {
      bUseBrushTransparency=false
      roundingType=HalfHeightRadius
      width=0
      color {colorUseRule=UseColor_Specified, specifiedColor=<rgba 0 0 0 0>}
      cornerRadii=<0 0 0 0>
     }
     tintColor {colorUseRule=UseColor_Specified, specifiedColor=<rgba 1 1 1 1>}
     uVRegion {bIsValid=false, max=<0 0>, min=<0 0>}
    }
    renderTransform {angle=-4, scale=<1 1>, shear=<0 0>, translation=<0 0>}
   }
   slot {class=HorizontalBoxSlot, properties {horizontalAlignment=HAlign_Left, verticalAlignment=VAlign_Center}}
  }
  QuantityTextBlock_1 {
   widgetClassPath=&CommonTextBlock
   bIsVariable=true
   parent=HorizontalBox_3
   widget=QuantityTextBlock_1
   properties {
    autoWrapText=true
    text=Initializing Inventory Display
    colorAndOpacity {colorUseRule=UseColor_Specified, specifiedColor=<rgba 0.5647 0.7922 0.9412 1>}
    font {
     bForceMonospaced=false
     bMaterialIsStencil=false
     fontMaterial=null
     fontObject=&BurbankSmall
     letterSpacing=0
     monospacedWidth=1
     size=26
     skewAmount=0
     typefaceFontName=Black
     outlineSettings {
      bApplyOutlineToDropShadows=false
      bMiteredCorners=false
      bSeparateFillAlpha=false
      outlineMaterial=null
      outlineSize=0
      outlineColor=<rgba 0 0 0 1>
     }
    }
    renderTransform {angle=0, scale=<1 1>, shear=<-8 0>, translation=<0 0>}
    shadowOffset=<0 0>
   }
   slot {class=HorizontalBoxSlot, properties {verticalAlignment=VAlign_Center, padding=<ltrb -32 0 0 0>}}
  }
  Overlay_EditorDisconnected {
   widgetClassPath=&Overlay
   bInherited=true
   bIsVariable=true
   parent=Switcher_InventoryPermission
   widget=Overlay_EditorDisconnected
   slot {class=WidgetSwitcherSlot}
  }
  SizeBox {
   widgetClassPath=&SizeBox
   parent=Overlay_EditorDisconnected
   widget=SizeBox
   properties {bOverride_WidthOverride=true, widthOverride=1024}
   slot {class=OverlaySlot, properties {horizontalAlignment=HAlign_Center, verticalAlignment=VAlign_Center, padding=<ltrb 0 -128 0 0>}}
  }
  HorizontalBox_2 {widgetClassPath=&HorizontalBox, parent=SizeBox, widget=HorizontalBox_2, slot {class=SizeBoxSlot}}
  Image {
   widgetClassPath=&Image
   parent=HorizontalBox_2
   widget=Image
   properties {
    brush {
     bIsDynamicallyLoaded=false
     drawAs=Image
     imageType=NoImage
     mirroring=NoMirror
     resourceName=None
     resourceObject=&M_UI_CreativeLogo
     tiling=NoTile
     imageSize=<164 164>
     margin=<ltrb 0 0 0 0>
     outlineSettings {
      bUseBrushTransparency=false
      roundingType=HalfHeightRadius
      width=0
      color {colorUseRule=UseColor_Specified, specifiedColor=<rgba 0 0 0 0>}
      cornerRadii=<0 0 0 0>
     }
     tintColor {colorUseRule=UseColor_Specified, specifiedColor=<rgba 1 1 1 1>}
     uVRegion {bIsValid=false, max=<0 0>, min=<0 0>}
    }
    renderTransform {angle=-4, scale=<1 1>, shear=<0 0>, translation=<0 0>}
   }
   slot {class=HorizontalBoxSlot, properties {horizontalAlignment=HAlign_Left, verticalAlignment=VAlign_Center}}
  }
  QuantityTextBlock_3 {
   widgetClassPath=&CommonTextBlock
   bIsVariable=true
   parent=HorizontalBox_2
   widget=QuantityTextBlock_3
   properties {
    autoWrapText=true
    text=This Project can not be modified when the Editor is disconnected
    colorAndOpacity {colorUseRule=UseColor_Specified, specifiedColor=<rgba 0.5647 0.7922 0.9412 1>}
    font {
     bForceMonospaced=false
     bMaterialIsStencil=false
     fontMaterial=null
     fontObject=&BurbankSmall
     letterSpacing=0
     monospacedWidth=1
     size=26
     skewAmount=0
     typefaceFontName=Black
     outlineSettings {
      bApplyOutlineToDropShadows=false
      bMiteredCorners=false
      bSeparateFillAlpha=false
      outlineMaterial=null
      outlineSize=0
      outlineColor=<rgba 0 0 0 1>
     }
    }
    renderTransform {angle=0, scale=<1 1>, shear=<-8 0>, translation=<0 0>}
    shadowOffset=<0 0>
   }
   slot {
    class=HorizontalBoxSlot
    properties {horizontalAlignment=HAlign_Left, verticalAlignment=VAlign_Center, padding=<ltrb -32 0 0 0>, size {sizeRule=Fill, value=1}}
   }
  }
 ]
}
