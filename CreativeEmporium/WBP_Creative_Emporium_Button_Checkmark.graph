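# Checkable filter button for the Creative Emporium browser, built on EmporiumBrowserFilterEntry, with label, item count, optional icon and input prompt.
# Parent events drive checked state, which toggles the backplate, recolors text, hides the prompt and plays a pitch-shifted sound, and the count, which shows only above zero.
# Hover shows an outline, and input-method changes swap mouse/keyboard and gamepad styles.
# PreConstruct adjusts mobile layout and previews checked state at design time.
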WidgetBlueprint /CreativeEmporium/WBP_Creative_Emporium_Button_Checkmark
graphs [
 EventGraph {nodes [
   n1 = "On Input Method Changed (InputActionWidget)"!() -> (OutputDelegate, then, bUsingGamepad:bool) {is=title K2Node_ComponentBoundEvent}
   > UpdateTextAndStyle!(execute, self, bUsingGamepad:bool=@n1.bUsingGamepad) -> (then)
   n2 = Event UpdateCheckMarkState!() -> (OutputDelegate, then, bIsChecked:bool) {is=title K2Node_Event}
   > UpdateCheckmark!(execute, self, Checked:bool=@n2.bIsChecked) -> (then)
   n3 = Event UpdateItemCount!() -> (OutputDelegate, then, ItemCount:int=0) {is=title K2Node_Event}
   n4 = "To Text (Integer)"(Value=@n3.ItemCount)
   n5 = "integer > integer"(A=@n3.ItemCount)
   > "SetText (Text)"(
    self=Get Text_NumItems() -> (Text_NumItems:Common Text Object Reference)
    InText=Format Text!(Format="LOCTEXT(\"({count})\")", count:Text=Select(Option 0:Text, Option 1:Text=@n4, Index=@n5) -> (ReturnValue:Text)) -> (Result)
   )
   > SetVisibility(
    self=[Get Count_Overlay() -> (Count_Overlay:Overlay Object Reference), Get Count_SizeBox() -> (Count_SizeBox:Size Box Object Reference)]
    InVisibility=Select(Option 0:ESlateVisibility Enum=Hidden, Option 1:ESlateVisibility Enum=HitTestInvisible, Index=@n5) -> (ReturnValue:ESlateVisibility Enum)
   )
   n6 = Event SetTabLabelInfo!() -> (OutputDelegate, then, TabLabelInfo:"Fort Tab Button Label Info Structure (by ref)") {is=title K2Node_Event}
   > Set Text!(
    execute
    self
    ButtonText:Text=Break Fort Tab Button Label Info!(FortTabButtonLabelInfo:"Fort Tab Button Label Info Structure (by ref)"=@n6.TabLabelInfo) -> (DisplayName:Text, IconBrush:Slate Brush Structure, IconResource:Soft Object Reference) {is=title K2Node_BreakStruct}
   ) -> (then) {is=title K2Node_CallFunction}
   Event Destruct()
   > Set OverrideButtonText(OverrideButtonText:Text) -> (Output_Get:Text)
   Event OnFocusLost!() -> (OutputDelegate, then, InFocusEvent:Focus Event Structure) {is=title K2Node_Event}
   > SetVisibility(self=Get Image_Backplate_hover() -> (Image_Backplate_hover:Image Object Reference), InVisibility=Hidden)
   Event OnTriggeredInputActionChanged!() -> (OutputDelegate, then, NewTriggeredAction:"Data Table Row Handle Structure (by ref)") {is=title K2Node_Event}
   n7 = Get CommonInputSubsystem!() -> (ReturnValue:Common Input Object Reference) {is=title K2Node_GetBlueprintContextEx}
   > UpdateTextAndStyle(bUsingGamepad="Equal (Enum)"(A=GetCurrentInputType(self=@n7), B=Gamepad))
   Event Construct()
   > SetInitialMouseKeyboardStyle()
   n8 = Get CommonInputSubsystem()
   > UpdateTextAndStyle(bUsingGamepad="Equal (Enum)"(A=GetCurrentInputType(self=@n8), B=Gamepad))
   Event On Hovered()
   > SetVisibility(self=Get Image_Backplate_hover(), InVisibility=SelfHitTestInvisible)
   n9 = Event PreConstruct()
   > n10 = Sequence!(execute) -> (then_0, then_1, then_2:Exec)
   Branch(execute=@n10.then_1, Condition=@n9.IsDesignTime)
   > UpdateCheckmark(Checked=Get PreviewCheck() -> (PreviewCheck:bool))
   UpdateButtonText(execute=@n10.then_2, self:Emporium Browser Filter Entry Object Reference)
   n11 = Branch(execute=@n10.then_0, Condition=IsMobileGame())
   n12 = SlotAsHorizontalBoxSlot!(Widget:Widget Object Reference=Get TextBlock_FilterName() -> (TextBlock_FilterName:Common Text Object Reference)) -> (ReturnValue:Horizontal Box Slot Object Reference)
   > SetPadding!(execute, self:Horizontal Box Slot Object Reference=@n12, InPadding:Margin Structure=Make Margin(Left=0, Top=14, Right=5, Bottom=14)) -> (then)
   > ClearHeightOverride(self:Size Box Object Reference=Get SizeBox_0() -> (SizeBox_0:Size Box Object Reference))
   SetPadding(execute=@n11.else, self=@n12, InPadding=Make Margin(Left=0, Top=-6, Right=5, Bottom=-6))
   Event On Unhovered()
   > SetVisibility(self=Get Image_Backplate_hover(), InVisibility=Hidden)
   SetCustomPaddingOnCheck!() -> (OutputDelegate, then, NewPadding:Margin Structure) {is=title K2Node_CustomEvent}
   Event OnActionProgress!() -> (OutputDelegate, then, HeldPercent:float) {is=title K2Node_Event}
   Event OnActionComplete() {is=title K2Node_Event}
  ]}
 InitializeButton {nodes [
   InitializeButton() {is=title K2Node_FunctionEntry}
   n1 = Get IconBrush() -> (IconBrush:Slate Brush Structure)
   > n2 = Is Valid(InputObject=GetBrushResource!(Brush:"Slate Brush Structure (by ref)"=@n1) -> (ReturnValue:Object Reference))
   > n3 = Branch(Condition="Not Equal (Enum)"!(
     A:ESlateBrushDrawType Enum=Break Slate Brush!(SlateBrush:"Slate Brush Structure (by ref)"=@n1) -> (
      TintColor:Slate Color Structure
      DrawAs:ESlateBrushDrawType Enum
      Tiling:ESlateBrushTileType Enum
      Mirroring:ESlateBrushMirrorType Enum
      ImageSize:Vector 2D Structure
      Margin:Margin Structure
      ResourceObject:Object Reference
      OutlineSettings:Slate Brush Outline Settings Structure
     ) {is=title K2Node_BreakStruct}.DrawAs
     B:ESlateBrushDrawType Enum=NoDrawType
    ) -> (ReturnValue:bool))
   > SetBrush(self=Get LeftSideImage() -> (LeftSideImage:Image Object Reference), InBrush=Get IconBrush())
   SetVisibility(execute=[@n3.else, @n2.Is Not Valid], self=Get LeftSideImage(), InVisibility=Collapsed)
  ]}
 Set Icon {nodes [
   n1 = Set Icon!() -> (then, IconBrush:Slate Brush Structure) {is=title K2Node_FunctionEntry}
   > n2 = Set IconBrush(IconBrush:Slate Brush Structure=@n1) -> (Output_Get:Slate Brush Structure)
   > SetBrush(self=Get LeftSideImage(), InBrush=@n2)
  ]}
 Set Text {nodes [
   n1 = Set Text!() -> (then, ButtonText:Text) {is=title K2Node_FunctionEntry}
   > n2 = Set OverrideButtonText(OverrideButtonText=@n1)
   > SetTextInternal!(execute, self, InButtonText:Text=@n2) -> (then)
  ]}
 SetContentAlignment {nodes [
   n1 = SetContentAlignment!() -> (then, ContentAlignment:EHorizontalAlignment Enum=HAlign_Fill) {is=title K2Node_FunctionEntry}
   > Set ContentAlignment(ContentAlignment:EHorizontalAlignment Enum=@n1) -> (Output_Get:EHorizontalAlignment Enum=HAlign_Fill)
   > UpdateContentAlignment()
  ]}
 SetControllerStyle {nodes [
   n1 = SetControllerStyle!() -> (then, ControllerInputStyle:Common Button Style Class Reference) {is=title K2Node_FunctionEntry}
   > Set ControllerInputStyle(ControllerInputStyle:Common Button Style Class Reference=@n1) -> (Output_Get:Common Button Style Class Reference)
   n2 = Get CommonInputSubsystem()
   > UpdateStyle!(execute, self, UsingGamepad:bool="Equal (Enum)"(A=GetCurrentInputType(self=@n2), B=Gamepad)) -> (then)
  ]}
 SetInitialMouseKeyboardStyle {nodes [
   SetInitialMouseKeyboardStyle() {is=title K2Node_FunctionEntry}
   > n1 = Branch(Condition=Get bMouseKeyboardStyleSet() -> (bMouseKeyboardStyleSet:bool))
   Set bMouseKeyboardStyleSet(execute=@n1.else, bMouseKeyboardStyleSet:bool=true) -> (Output_Get:bool)
   > Set MouseKeyboardStyle(MouseKeyboardStyle:Common Button Style Class Reference=Get Style() -> (Style:Common Button Style Class Reference)) -> (Output_Get:Common Button Style Class Reference)
  ]}
 SetMouseKeyboardStyle {nodes [
   n1 = SetMouseKeyboardStyle!() -> (then, ControllerInputStyle:Common Button Style Class Reference) {is=title K2Node_FunctionEntry}
   > Set MouseKeyboardStyle(MouseKeyboardStyle=@n1)
   n2 = Get CommonInputSubsystem()
   > UpdateStyle(UsingGamepad="Equal (Enum)"(A=GetCurrentInputType(self=@n2), B=Gamepad))
  ]}
 SetTextInternal {nodes [
   SetTextInternal!() -> (then, InButtonText:Text) {is=title K2Node_FunctionEntry}
   > UpdateButtonText()
  ]}
 ShowIcon {nodes [
   n1 = ShowIcon!() -> (then, bShouldShow:bool=true) {is=title K2Node_FunctionEntry}
   > n2 = Branch(Condition=@n1)
   n3 = Get LeftSideImage()
   > SetVisibility(self=@n3, InVisibility=SelfHitTestInvisible)
   SetVisibility(execute=@n2.else, self=@n3, InVisibility=Collapsed)
  ]}
 UpdateCheckmark {nodes [
   n1 = UpdateCheckmark!() -> (then, Checked:bool=~) {is=title K2Node_FunctionEntry}
   > n2 = Branch(Condition=@n1)
   n3 = Get Overlay_Backplate() -> (Overlay_Backplate:Overlay Object Reference)
   > n4 = SetVisibility(self=@n3, InVisibility=SelfHitTestInvisible)
   PlaySound2D(
    execute=[@n5.then, @n4.then]
    Sound=/Game/Sounds/Fort_UI_Sounds/FrontEnd/New_FrontEnd/Fortnite_UI_MainMenu_Generic_01_Cue
    PitchMultiplier=Select(Option 0:double=0.85, Option 1:double=1, Index=@n1) -> (ReturnValue:double)
   )
   > SetColorAndOpacity(
    self=[Get Text_NumItems(), Get TextBlock_FilterName()]
    InColorAndOpacity=Select(
     Option 0:Slate Color Structure=Get UncheckBrush() -> (UncheckBrush:Slate Color Structure)
     Option 1:Slate Color Structure=Get CheckBrush() -> (CheckBrush:Slate Color Structure)
     Index=Get Checked() -> (Checked:bool)
    ) -> (ReturnValue:Slate Color Structure)
   )
   > SetVisibility(
    self=Get InputAction_Overlay() -> (InputAction_Overlay:Overlay Object Reference)
    InVisibility=Select(Option 0:ESlateVisibility Enum=SelfHitTestInvisible, Option 1:ESlateVisibility Enum=Hidden, Index=Get Checked()) -> (ReturnValue:ESlateVisibility Enum)
   )
   n5 = SetVisibility(execute=@n2.else, self=@n3, InVisibility=Collapsed)
  ]}
 UpdateContentAlignment {nodes [
   UpdateContentAlignment() {is=title K2Node_FunctionEntry}
   > n1 = Cast To BorderSlot!(
    execute
    Object=GetContentSlot!(self:Content Widget Object Reference=Get ContentBorder() -> (ContentBorder:Border Object Reference)) -> (ReturnValue:Panel Slot Object Reference)
   ) -> (then, CastFailed, AsBorder Slot:Border Slot Object Reference) {is=title K2Node_DynamicCast}
   > SetHorizontalAlignment(
    self:Border Slot Object Reference=@n1
    InHorizontalAlignment=Get ContentAlignment() -> (ContentAlignment:EHorizontalAlignment Enum=HAlign_Fill)
   )
  ]}
 UpdateStyle {nodes [
   n1 = UpdateStyle!() -> (then, UsingGamepad:bool=~) {is=title K2Node_FunctionEntry}
   > n2 = Branch(Condition=@n1)
   > SetStyle!(
    execute
    self:Common Button Base Object Reference
    InStyle:Common Button Style Class Reference=Get ControllerInputStyle() -> (ControllerInputStyle:Common Button Style Class Reference)
   ) -> (then)
   SetStyle(execute=@n2.else, InStyle=Get MouseKeyboardStyle() -> (MouseKeyboardStyle:Common Button Style Class Reference))
  ]}
 UpdateTextAndStyle {nodes [
   n1 = UpdateTextAndStyle!() -> (then, bUsingGamepad:bool=~) {is=title K2Node_FunctionEntry}
   > UpdateStyle(UsingGamepad=@n1) {is=title K2Node_CallFunction}
   > UpdateTextStyle()
   > UpdateButtonText()
  ]}
 UpdateTextStyle {nodes [
   SetStyle(
    self:Common Text Object Reference=Get TextBlock_FilterName()
    InStyle:Common Text Style Class Reference=GetCurrentTextStyleClass!(self:Common Button Base Object Reference) -> (ReturnValue:Common Text Style Class Reference)
   )
   UpdateTextStyle() {is=title K2Node_FunctionEntry}
  ]}
]
variables [
 bMouseKeyboardStyleSet {type=boolean}
 CheckBrush {type=SlateColor, default {colorUseRule=UseColor_Specified, specifiedColor=<rgba 0 0 0 1>}}
 ContentAlignment {type=EHorizontalAlignment, default=HAlign_Fill}
 ControllerInputStyle {type=Class, default=null}
 CustomCheckPadding {type=Margin, default=<ltrb 0 0 6 0>}
 IconBrush {
  type=SlateBrush
  default {
   bIsDynamicallyLoaded=false
   drawAs=NoDrawType
   imageType=NoImage
   mirroring=NoMirror
   resourceName=None
   resourceObject=null
   tiling=NoTile
   imageSize=<0 0>
   margin=<ltrb 0 0 0 0>
   outlineSettings {
    bUseBrushTransparency=false
    roundingType=HalfHeightRadius
    width=0
    color {colorUseRule=UseColor_Specified, specifiedColor=<rgba 0 0 0 0>}
    cornerRadii=<0 0 0 1>
   }
   tintColor {colorUseRule=UseColor_Specified, specifiedColor=<rgba 1 1 1 1>}
   uVRegion {bIsValid=false, max=<0 0>, min=<0 0>}
  }
 }
 MouseKeyboardStyle {type=Class, default=null}
 OverrideButtonText {type=string, default=""}
 PreviewCheck {type=boolean}
 UncheckBrush {type=SlateColor, default {colorUseRule=UseColor_Specified, specifiedColor=<rgba 1 1 1 0.8>}}
]
widgets {
 info {
  inheritedWidgetCount=2
  parentClass=/Script/CreativeEmporiumUI.EmporiumBrowserFilterEntry
  rootWidgetClass=&Border
  widgetCount=19
 }
 widgets [
  ContentBorder {
   widgetClassPath=&Border
   bIsVariable=true
   widget=ContentBorder
   properties {
    verticalAlignment=VAlign_Center
    background {
     bIsDynamicallyLoaded=false
     drawAs=NoDrawType
     imageType=NoImage
     mirroring=NoMirror
     resourceName=None
     resourceObject=null
     tiling=NoTile
     imageSize=<32 32>
     margin=<ltrb 0 0 0 0>
     outlineSettings {
      bUseBrushTransparency=false
      roundingType=HalfHeightRadius
      width=0
      color {colorUseRule=UseColor_Specified, specifiedColor=<rgba 0 0 0 0>}
      cornerRadii=<0 0 0 0>
     }
     tintColor {colorUseRule=UseColor_Specified, specifiedColor=<rgba 1 1 1 1>}
     uVRegion {bIsValid=false, max=<0 0>, min=<0 0>}
    }
    padding=<ltrb 0 0 0 0>
   }
  }
  SizeBox_0 {
   widgetClassPath=&SizeBox
   bIsVariable=true
   parent=ContentBorder
   widget=SizeBox_0
   properties {bOverride_HeightOverride=true, heightOverride=56}
   slot {class=BorderSlot, properties {verticalAlignment=VAlign_Center, padding=<ltrb 0 0 0 0>}}
  }
  Overlay_0 {widgetClassPath=&Overlay, parent=SizeBox_0, widget=Overlay_0, slot {class=SizeBoxSlot}}
  Image_Backplate_hover {
   widgetClassPath=&Image
   bIsVariable=true
   parent=Overlay_0
   widget=Image_Backplate_hover
   properties {
    visibility=Hidden
    brush {
     bIsDynamicallyLoaded=false
     drawAs=Image
     imageType=NoImage
     mirroring=NoMirror
     resourceName=None
     resourceObject=/CreativeEmporium/Materials/MI_CrBrowser_TabOutline
     tiling=NoTile
     imageSize=<156 64>
     margin=<ltrb 0 0 0 0>
     outlineSettings {
      bUseBrushTransparency=false
      roundingType=HalfHeightRadius
      width=0
      color {colorUseRule=UseColor_Specified, specifiedColor=<rgba 0 0 0 0>}
      cornerRadii=<0 0 0 0>
     }
     tintColor {colorUseRule=UseColor_Specified, specifiedColor=<rgba 1 1 1 0.5>}
     uVRegion {bIsValid=false, max=<0 0>, min=<0 0>}
    }
   }
   slot {class=OverlaySlot, properties {horizontalAlignment=HAlign_Fill, verticalAlignment=VAlign_Fill, padding=<ltrb -6 -4 -6 -4>}}
  }
  Overlay_Backplate {
   widgetClassPath=&Overlay
   bIsVariable=true
   parent=Overlay_0
   widget=Overlay_Backplate
   properties {clipping=ClipToBounds, visibility=Collapsed}
   slot {class=OverlaySlot, properties {horizontalAlignment=HAlign_Fill, verticalAlignment=VAlign_Fill, padding=<ltrb 8 10 8 10>}}
  }
  Image_Backplate {
   widgetClassPath=&Image
   parent=Overlay_Backplate
   widget=Image_Backplate
   properties {
    visibility=SelfHitTestInvisible
    brush {
     bIsDynamicallyLoaded=false
     drawAs=Image
     imageType=NoImage
     mirroring=NoMirror
     resourceName=None
     resourceObject=/CreativeEmporium/Materials/MI_CrBrowser_TabButton_BackgroundStyle
     tiling=NoTile
     imageSize=<156 12>
     margin=<ltrb 0 0 0 0>
     outlineSettings {
      bUseBrushTransparency=false
      roundingType=HalfHeightRadius
      width=0
      color {colorUseRule=UseColor_Specified, specifiedColor=<rgba 0 0 0 0>}
      cornerRadii=<0 0 0 0>
     }
     tintColor {colorUseRule=UseColor_Specified, specifiedColor=<rgba 1 1 1 0.95>}
     uVRegion {bIsValid=false, max=<0 0>, min=<0 0>}
    }
   }
   slot {class=OverlaySlot, properties {horizontalAlignment=HAlign_Fill, verticalAlignment=VAlign_Fill}}
  }
  HorizontalBox {
   widgetClassPath=&HorizontalBox
   parent=Overlay_0
   widget=HorizontalBox
   slot {class=OverlaySlot, properties {horizontalAlignment=HAlign_Fill, verticalAlignment=VAlign_Fill, padding=<ltrb 12 6 12 6>}}
  }
  Checkmark {
   widgetClassPath=&Image
   bIsVariable=true
   parent=HorizontalBox
   widget=Checkmark
   properties {
    visibility=Collapsed
    brush {
     bIsDynamicallyLoaded=false
     drawAs=Image
     imageType=NoImage
     mirroring=NoMirror
     resourceName=None
     resourceObject=null
     tiling=NoTile
     imageSize=<6 12>
     margin=<ltrb 0 0 0 0>
     outlineSettings {
      bUseBrushTransparency=false
      roundingType=HalfHeightRadius
      width=0
      color {colorUseRule=UseColor_Specified, specifiedColor=<rgba 0 0 0 0>}
      cornerRadii=<0 0 0 0>
     }
     tintColor {colorUseRule=UseColor_Specified, specifiedColor=<rgba 0.5127 0 1 0>}
     uVRegion {bIsValid=false, max=<0 0>, min=<0 0>}
    }
   }
   slot {class=HorizontalBoxSlot, properties {horizontalAlignment=HAlign_Left}}
  }
  Spacer_98 {
   widgetClassPath=&Spacer
   parent=HorizontalBox
   widget=Spacer_98
   properties {visibility=Collapsed, size=<24 1>}
   slot {class=HorizontalBoxSlot}
  }
  Count_SizeBox {
   widgetClassPath=&SizeBox
   bIsVariable=true
   parent=HorizontalBox
   widget=Count_SizeBox
   properties {bOverride_HeightOverride=true, bOverride_WidthOverride=true, heightOverride=12, minDesiredWidth=12, visibility=Hidden, widthOverride=12}
   slot {class=HorizontalBoxSlot, properties {horizontalAlignment=HAlign_Left, verticalAlignment=VAlign_Center, padding=<ltrb 20 0 6 0>}}
  }
  CommonBorder {
   widgetClassPath=&CommonBorder
   parent=Count_SizeBox
   widget=CommonBorder
   properties {
    horizontalAlignment=HAlign_Center
    verticalAlignment=VAlign_Center
    background {
     bIsDynamicallyLoaded=false
     drawAs=RoundedBox
     imageType=NoImage
     mirroring=NoMirror
     resourceName=None
     resourceObject=null
     tiling=NoTile
     imageSize=<32 32>
     margin=<ltrb 0 0 0 0>
     outlineSettings {
      bUseBrushTransparency=false
      roundingType=HalfHeightRadius
      width=0
      color {colorUseRule=UseColor_Specified, specifiedColor=<rgba 0 0 0 0>}
      cornerRadii=<0 0 0 0>
     }
     tintColor {colorUseRule=UseColor_Specified, specifiedColor=<rgba 1 0.063 0.3372 1>}
     uVRegion {bIsValid=false, max=<0 0>, min=<0 0>}
    }
    padding=<ltrb 0 0 0 0>
   }
   slot {class=SizeBoxSlot}
  }
  TextBlock_FilterName {
   widgetClassPath=&CommonTextBlock
   bInherited=true
   bIsVariable=true
   parent=HorizontalBox
   widget=TextBlock_FilterName
   properties {
    mobileFontSizeMultiplier=1.5
    scrollStyle=&TextScrollStyle-Base_C
    colorAndOpacity {colorUseRule=UseColor_Specified, specifiedColor=<rgba 1 1 1 0.95>}
    font {
     bForceMonospaced=false
     bMaterialIsStencil=false
     fontMaterial=/CreativeEmporium/Materials/MI_CrRefresh_TabButton_Text_Create
     fontObject=&HeadingNow
     letterSpacing=0
     monospacedWidth=1
     size=18
     skewAmount=0
     typefaceFontName=75 Medium
     outlineSettings {
      bApplyOutlineToDropShadows=false
      bMiteredCorners=false
      bSeparateFillAlpha=false
      outlineMaterial=null
      outlineSize=0
      outlineColor=<rgba 0 0 0 1>
     }
    }
    shadowOffset=<0 0>
   }
   slot {class=HorizontalBoxSlot, properties {verticalAlignment=VAlign_Center, padding=<ltrb 0 -6 5 -6>, size {sizeRule=Fill, value=1}}}
  }
  HorizontalBox_0 {
   widgetClassPath=&HorizontalBox
   parent=HorizontalBox
   widget=HorizontalBox_0
   slot {class=HorizontalBoxSlot, properties {padding=<ltrb 0 0 4 0>}}
  }
  RightExtraContentSlot {
   widgetClassPath=&NamedSlot
   bIsVariable=true
   parent=HorizontalBox_0
   widget=RightExtraContentSlot
   properties {bExposeOnInstanceOnly=true}
   slot {class=HorizontalBoxSlot, properties {padding=<ltrb 2 0 0 0>}}
  }
  Count_Overlay {
   widgetClassPath=&Overlay
   bIsVariable=true
   parent=HorizontalBox_0
   widget=Count_Overlay
   properties {visibility=Hidden}
   slot {class=HorizontalBoxSlot, properties {horizontalAlignment=HAlign_Left, verticalAlignment=VAlign_Center, padding=<ltrb 6 0 0 0>}}
  }
  Text_NumItems {
   widgetClassPath=&CommonTextBlock
   bIsVariable=true
   parent=Count_Overlay
   widget=Text_NumItems
   properties {
    justification=Center
    minDesiredWidth=18
    mobileFontSizeMultiplier=1.34
    text=8
    colorAndOpacity {colorUseRule=UseColor_Specified, specifiedColor=<rgba 1 1 1 0.9>}
    font {
     bForceMonospaced=false
     bMaterialIsStencil=false
     fontMaterial=null
     fontObject=&HeadingNow
     letterSpacing=0
     monospacedWidth=1
     size=15.75
     skewAmount=0
     typefaceFontName=75 Medium
     outlineSettings {
      bApplyOutlineToDropShadows=false
      bMiteredCorners=false
      bSeparateFillAlpha=false
      outlineMaterial=null
      outlineSize=0
      outlineColor=<rgba 0 0 0 1>
     }
    }
   }
   slot {class=OverlaySlot, properties {verticalAlignment=VAlign_Center}}
  }
  InputAction_Overlay {
   widgetClassPath=&Overlay
   bIsVariable=true
   parent=Overlay_0
   widget=InputAction_Overlay
   slot {class=OverlaySlot, properties {horizontalAlignment=HAlign_Right, verticalAlignment=VAlign_Center, padding=<ltrb 8 0 8 0>}}
  }
  InputActionWidget {
   widgetClassPath=&CommonActionWidget
   bInherited=true
   bIsVariable=true
   parent=InputAction_Overlay
   widget=InputActionWidget
   properties {
    progressMaterialParam=Percentage
    visibility=Collapsed
    progressMaterialBrush {
     bIsDynamicallyLoaded=false
     drawAs=Image
     imageType=NoImage
     mirroring=NoMirror
     resourceName=None
     resourceObject=/Game/UI/Materials/HollowCircularProgressBar
     tiling=NoTile
     imageSize=<36 36>
     margin=<ltrb 0 0 0 0>
     outlineSettings {
      bUseBrushTransparency=false
      roundingType=HalfHeightRadius
      width=0
      color {colorUseRule=UseColor_Specified, specifiedColor=<rgba 0 0 0 0>}
      cornerRadii=<0 0 0 0>
     }
     tintColor {colorUseRule=UseColor_Specified, specifiedColor=<rgba 1 1 0 1>}
     uVRegion {bIsValid=false, max=<0 0>, min=<0 0>}
    }
   }
   slot {class=OverlaySlot}
  }
  LeftSideImage {
   widgetClassPath=&Image
   bIsVariable=true
   parent=Overlay_0
   widget=LeftSideImage
   properties {brush {
     bIsDynamicallyLoaded=false
     drawAs=NoDrawType
     imageType=NoImage
     mirroring=NoMirror
     resourceName=None
     resourceObject=null
     tiling=NoTile
     imageSize=<0 0>
     margin=<ltrb 0 0 0 0>
     outlineSettings {
      bUseBrushTransparency=false
      roundingType=HalfHeightRadius
      width=0
      color {colorUseRule=UseColor_Specified, specifiedColor=<rgba 0 0 0 0>}
      cornerRadii=<0 0 0 0>
     }
     tintColor {colorUseRule=UseColor_Specified, specifiedColor=<rgba 1 1 1 1>}
     uVRegion {bIsValid=false, max=<0 0>, min=<0 0>}
    }}
   slot {class=OverlaySlot, properties {horizontalAlignment=HAlign_Right, verticalAlignment=VAlign_Center}}
  }
 ]
}
