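# Category header button for filter lists in the Creative Emporium, a CommonButtonBase widget showing a category name, entry count and expand arrow.
# ShowCategoryActive stores bIsActive, which drives the fill backplate's opacity and, through SetTextColors, selected or unselected text colors.
# Hover shows an outline backplate; unhover or focus loss hides it, both reapplying the active-state styling.
# OnExpansionChanged plays the arrow-rotation animation in reverse when expanded, forward when collapsed, then plays a UI sound.
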WidgetBlueprint /CreativeEmporium/WBP_Creative_Emporium_Button_FilterCategoryHeader
animations [
 OnHover {
  binding_count=7
  display_label=""
  length_seconds=0.2
  object=&self:OnHover
  sequence {bindings [
    ExpandDirectionTick {
     id=#1
     possessed_class=Image
     tracks [
      {
       type=MovieScene2DTransformTrack
       sections [{
         type=MovieScene2DTransformSection
         channels [
          Translation.X {type=MovieSceneScriptingFloatChannel}
          Translation.Y {type=MovieSceneScriptingFloatChannel, keys [[0,0],[3,-6],[4,0]]}
          Angle {type=MovieSceneScriptingFloatChannel}
          Scale.X {type=MovieSceneScriptingFloatChannel, keys [[0,1],[3,1.125],[4,1]]}
          Scale.Y {type=MovieSceneScriptingFloatChannel, keys [[0,1],[3,1.125],[4,1]]}
          Shear.X {type=MovieSceneScriptingFloatChannel}
          Shear.Y {type=MovieSceneScriptingFloatChannel}
         ]
        }]
      }
      {
       type=MovieSceneColorTrack
       display_name=Color and Opacity
       sections [{
         type=MovieSceneColorSection
         end_frame=3
         start_frame=0
         channels [
          Color.R {type=MovieSceneScriptingFloatChannel, keys [[0,0.5089],[3,1]]}
          Color.G {type=MovieSceneScriptingFloatChannel, keys [[0,0.5583],[3,1]]}
          Color.B {type=MovieSceneScriptingFloatChannel, keys [[0,0.7084],[3,1]]}
          Color.A {type=MovieSceneScriptingFloatChannel, const [1,1]}
         ]
        }]
      }
     ]
    }
    Text_ListName {id=#2, possessed_class=CommonTextBlock, tracks []}
    Text_NumEntries {id=#3, possessed_class=CommonTextBlock, tracks []}
    Image_EndCapBackplate {
     id=#4
     possessed_class=Image
     tracks [{
       type=MovieSceneFloatTrack
       display_name=Render Opacity
       sections [{type=MovieSceneFloatSection, end_frame=4, start_frame=0, channels [None {type=MovieSceneScriptingFloatChannel, keys [[0,0],[4,1]]}]}]
      }]
    }
    EntryBorder {
     id=#5
     possessed_class=Border
     tracks [
      {
       type=MovieSceneWidgetMaterialTrack
       sections [{
         type=MovieSceneParameterSection
         end_frame=4
         start_frame=0
         channels [PrimaryAlpha {type=MovieSceneScriptingFloatChannel, keys [[0,0],[3,0.5],[4,0.5]]}]
        }]
      }
      {
       type=MovieScene2DTransformTrack
       sections [{
         type=MovieScene2DTransformSection
         end_frame=4
         start_frame=0
         channels [
          Angle {type=MovieSceneScriptingFloatChannel}
          Scale.X {type=MovieSceneScriptingFloatChannel}
          Scale.Y {type=MovieSceneScriptingFloatChannel}
          Shear.X {type=MovieSceneScriptingFloatChannel}
          Shear.Y {type=MovieSceneScriptingFloatChannel}
         ]
        }]
      }
      {
       type=MovieSceneColorTrack
       display_name=Brush Color
       sections [{
         type=MovieSceneColorSection
         end_frame=0
         start_frame=0
         channels [
          Color.R {type=MovieSceneScriptingFloatChannel, const [1,1]}
          Color.G {type=MovieSceneScriptingFloatChannel, const [1,1]}
          Color.B {type=MovieSceneScriptingFloatChannel, const [1,1]}
          Color.A {type=MovieSceneScriptingFloatChannel, const [1,1]}
         ]
        }]
      }
     ]
    }
    HorizontalBoxSlot_2 {id=#6, parent_id=#2, possessed_class=HorizontalBoxSlot, tracks []}
    InputActionWidget {
     id=#7
     possessed_class=CommonActionWidget
     tracks [{
       type=MovieSceneFloatTrack
       display_name=Render Opacity
       sections [{type=MovieSceneFloatSection, end_frame=4, start_frame=0, channels [None {type=MovieSceneScriptingFloatChannel, keys [[0,0],[1,0],[2,1]]}]}]
      }]
    }
   ]}
 }
 OnExpansion {
  binding_count=1
  display_label=""
  length_seconds=0.2
  object=&self:OnExpansion
  sequence {bindings [ExpandDirectionTick {
     id=#8
     possessed_class=Image
     tracks [{
       type=MovieScene2DTransformTrack
       sections [{
         type=MovieScene2DTransformSection
         channels [
          Angle {type=MovieSceneScriptingFloatChannel, keys [[0,0],[4,90]]}
          Scale.X {type=MovieSceneScriptingFloatChannel, keys [[0,1],[3,1.5],[4,1]]}
          Scale.Y {type=MovieSceneScriptingFloatChannel, keys [[0,1],[3,1.5],[4,1]]}
          Shear.X {type=MovieSceneScriptingFloatChannel}
          Shear.Y {type=MovieSceneScriptingFloatChannel}
         ]
        }]
      }]
    }]}
 }
 Pulse {
  binding_count=1
  length_seconds=0.2
  object=&self:Pulse
  sequence {bindings [ExpandDirectionTick {
     id=#1
     possessed_class=Image
     tracks [
      {
       type=MovieScene2DTransformTrack
       sections [{
         type=MovieScene2DTransformSection
         channels [
          Translation.X {type=MovieSceneScriptingFloatChannel, keys [[3,-6],[4,0]]}
          Translation.Y {type=MovieSceneScriptingFloatChannel}
          Angle {type=MovieSceneScriptingFloatChannel, const [0,1]}
          Scale.X {type=MovieSceneScriptingFloatChannel, keys [[0,1],[3,1.125],[4,1]]}
          Scale.Y {type=MovieSceneScriptingFloatChannel, keys [[0,1],[3,1.125],[4,1]]}
          Shear.X {type=MovieSceneScriptingFloatChannel}
          Shear.Y {type=MovieSceneScriptingFloatChannel}
         ]
        }]
      }
      {
       type=MovieSceneColorTrack
       display_name=Color and Opacity
       sections [{
         type=MovieSceneColorSection
         end_frame=4
         start_frame=0
         channels [
          Color.R {type=MovieSceneScriptingFloatChannel, keys [[0,0.5089],[4,0]]}
          Color.G {type=MovieSceneScriptingFloatChannel, keys [[0,0.5583],[4,0]]}
          Color.B {type=MovieSceneScriptingFloatChannel, keys [[0,0.7084],[4,0]]}
          Color.A {type=MovieSceneScriptingFloatChannel, keys [[0,1],[4,0.8]]}
         ]
        }]
      }
     ]
    }]}
 }
]
graphs [
 EventGraph {nodes [
   n1 = Get Pulse() -> (Pulse:Widget Animation Object Reference)
   SetAnimationCurrentTime!(execute, self:User Widget Object Reference, InAnimation:Widget Animation Object Reference=@n1, InTime:float=0) -> (then)
   > StopAnimation(InAnimation=@n1)
   n2 = ShowCategoryActive!() -> (OutputDelegate, then, Active:bool=~) {is=title K2Node_CustomEvent}
   > Set bIsActive(bIsActive:bool=@n2.Active) -> (Output_Get:bool)
   > SetRenderOpacity(
    self=Get Image_BackplateFill() -> (Image_BackplateFill:Image Object Reference)
    InOpacity=Select(Option 0:float=0, Option 1:float=1, Index=Get bIsActive() -> (bIsActive:bool)) -> (ReturnValue:float)
   )
   > SetTextColors()
   Event On Hovered()
   > SetRenderOpacity(self=Get Overlay_Backplate() -> (Overlay_Backplate:Overlay Object Reference), InOpacity=1)
   > SetRenderOpacity(self=Get Image_BackplateFill(), InOpacity=Select(Option 0=0, Option 1=1, Index=Get bIsActive()))
   > SetTextColors()
   n3 = Event OnFocusLost!() -> (OutputDelegate, then, InFocusEvent:Focus Event Structure) {is=title K2Node_Event}
   SetRenderOpacity(execute=[@n3.then, @n4.then], self=Get Overlay_Backplate(), InOpacity=0)
   > SetRenderOpacity(self=Get Image_BackplateFill(), InOpacity=Select(Option 0=0, Option 1=1, Index=Get bIsActive()))
   > SetTextColors()
   n4 = Event On Unhovered()
   K2Node_Event_2 = OnExpansionChanged!() -> (OutputDelegate, then, bIsExpanded:bool=~) {is=title K2Node_CustomEvent}
   n5 = Branch(Condition=@K2Node_Event_2.bIsExpanded)
   n6 = Get ExpansionAnimationSpeed() -> (ExpansionAnimationSpeed:double)
   n7 = Get OnExpansion() -> (OnExpansion:Widget Animation Object Reference)
   > n8 = PlayAnimationReverse(InAnimation=@n7, PlaybackSpeed=@n6)
   PlaySound2D(execute=[@n9.then, @n8.then], Sound=/Game/Sounds/Fort_UI_Sounds/FrontEnd/New_FrontEnd/Fortnite_UI_MainMenu_Tab_Hover_02_Cue)
   n9 = PlayAnimationForward(execute=@n5.else, InAnimation=@n7, PlaybackSpeed=@n6)
   PlayAnimation(InAnimation=Get Pulse(), PlaybackSpeed=0.2)
  ]}
 SetTextColors {nodes [
   SetTextColors() {is=title K2Node_FunctionEntry}
   > SetColorAndOpacity(
    self=Get Text_ListName() -> (Text_ListName:Common Text Object Reference)
    InColorAndOpacity=Select(
     Option 0:Slate Color Structure=Get UnselectedColor() -> (UnselectedColor:Slate Color Structure)
     Option 1:Slate Color Structure=Get SelectedColor() -> (SelectedColor:Slate Color Structure)
     Index=Get bIsActive()
    ) -> (ReturnValue:Slate Color Structure)
   )
   > SetColorAndOpacity(
    self=Get Text_NumEntries() -> (Text_NumEntries:Common Text Object Reference)
    InColorAndOpacity=Select(
     Option 0=Get Color_CountUnselected() -> (Color_CountUnselected:Slate Color Structure)
     Option 1=Get Color_CountSelected() -> (Color_CountSelected:Slate Color Structure)
     Index=Get bIsActive()
    )
   )
  ]}
]
variables [
 bIsActive {type=boolean}
 Color_CountSelected {type=SlateColor, default {colorUseRule=UseColor_Specified, specifiedColor=<rgba 0.0144 0.0152 0.0152 1>}}
 Color_CountUnselected {type=SlateColor, default {colorUseRule=UseColor_Specified, specifiedColor=<rgba 0.5089 0.5583 0.7084 1>}}
 ExpansionAnimationSpeed {type=number, default=1.4}
 SelectedColor {type=SlateColor, default {colorUseRule=UseColor_Specified, specifiedColor=<rgba 0.0144 0.0152 0.0152 1>}}
 UnselectedColor {type=SlateColor, default {colorUseRule=UseColor_Specified, specifiedColor=<rgba 0.7758 0.7605 0.7913 1>}}
]
widgets {
 info {inheritedWidgetCount=1, parentClass=&CommonButtonBase, rootWidgetClass=&SizeBox, widgetCount=18}
 widgets [
  SB_EntryContent {
   widgetClassPath=&SizeBox
   bIsVariable=true
   widget=SB_EntryContent
   properties {bOverride_HeightOverride=true, heightOverride=58, minDesiredHeight=72}
  }
  EntryBorder {
   widgetClassPath=&Border
   parent=SB_EntryContent
   widget=EntryBorder
   properties {
    background {
     bIsDynamicallyLoaded=false
     drawAs=NoDrawType
     imageType=NoImage
     mirroring=NoMirror
     resourceName=None
     resourceObject=null
     tiling=NoTile
     imageSize=<32 32>
     margin=<ltrb 0 0 0 0>
     outlineSettings {
      bUseBrushTransparency=false
      roundingType=HalfHeightRadius
      width=0
      color {colorUseRule=UseColor_Specified, specifiedColor=<rgba 0 0 0 0>}
      cornerRadii=<0 0 0 0>
     }
     tintColor {colorUseRule=UseColor_Specified, specifiedColor=<rgba 1 1 1 1>}
     uVRegion {bIsValid=false, max=<0 0>, min=<0 0>}
    }
    padding=<ltrb 0 0 0 0>
   }
   slot {class=SizeBoxSlot, properties {padding=<ltrb 0 0 0 -20>}}
  }
  Overlay_0 {widgetClassPath=&Overlay, parent=EntryBorder, widget=Overlay_0, slot {class=BorderSlot, properties {padding=<ltrb 0 0 0 0>}}}
  Overlay_Backplate {
   widgetClassPath=&Overlay
   bIsVariable=true
   parent=Overlay_0
   widget=Overlay_Backplate
   properties {clipping=ClipToBounds, renderOpacity=0}
   slot {class=OverlaySlot, properties {horizontalAlignment=HAlign_Fill, verticalAlignment=VAlign_Fill}}
  }
  Image_Backplate {
   widgetClassPath=&Image
   parent=Overlay_Backplate
   widget=Image_Backplate
   properties {
    visibility=SelfHitTestInvisible
    brush {
     bIsDynamicallyLoaded=false
     drawAs=Image
     imageType=NoImage
     mirroring=NoMirror
     resourceName=None
     resourceObject=/CreativeEmporium/Materials/MI_CrBrowser_TabOutline
     tiling=NoTile
     imageSize=<32 32>
     margin=<ltrb 0 0 0 0>
     outlineSettings {
      bUseBrushTransparency=false
      roundingType=HalfHeightRadius
      width=0
      color {colorUseRule=UseColor_Specified, specifiedColor=<rgba 0 0 0 0>}
      cornerRadii=<0 0 0 0>
     }
     tintColor {colorUseRule=UseColor_Specified, specifiedColor=<rgba 1 1 1 1>}
     uVRegion {bIsValid=false, max=<0 0>, min=<0 0>}
    }
   }
   slot {class=OverlaySlot, properties {horizontalAlignment=HAlign_Fill, verticalAlignment=VAlign_Fill}}
  }
  Image_BackplateFill {
   widgetClassPath=&Image
   bIsVariable=true
   parent=Overlay_0
   widget=Image_BackplateFill
   properties {
    renderOpacity=0
    visibility=SelfHitTestInvisible
    brush {
     bIsDynamicallyLoaded=false
     drawAs=Image
     imageType=NoImage
     mirroring=NoMirror
     resourceName=None
     resourceObject=/CreativeEmporium/Materials/MI_CrBrowser_TabButton_BackgroundStyle
     tiling=NoTile
     imageSize=<364.5901 41.9209>
     margin=<ltrb 0 0 0 0>
     outlineSettings {
      bUseBrushTransparency=false
      roundingType=HalfHeightRadius
      width=0
      color {colorUseRule=UseColor_Specified, specifiedColor=<rgba 0 0 0 0>}
      cornerRadii=<0 0 0 0>
     }
     tintColor {colorUseRule=UseColor_Specified, specifiedColor=<rgba 1 1 1 1>}
     uVRegion {bIsValid=false, max=<0 0>, min=<0 0>}
    }
   }
   slot {class=OverlaySlot, properties {horizontalAlignment=HAlign_Fill, verticalAlignment=VAlign_Fill, padding=<ltrb 14 14 14 14>}}
  }
  HorizontalBox_0 {
   widgetClassPath=&HorizontalBox
   parent=Overlay_0
   widget=HorizontalBox_0
   slot {class=OverlaySlot, properties {horizontalAlignment=HAlign_Fill, verticalAlignment=VAlign_Center, padding=<ltrb 42 0 24 0>}}
  }
  Text_ListName {
   widgetClassPath=&CommonTextBlock
   bIsVariable=true
   parent=HorizontalBox_0
   widget=Text_ListName
   properties {
    mobileFontSizeMultiplier=1.34
    scrollStyle=&TextScrollStyle-Base_C
    text="{CATEGORY NAME}"
    colorAndOpacity {colorUseRule=UseColor_Specified, specifiedColor=<rgba 1 1 1 0.8>}
    font {
     bForceMonospaced=false
     bMaterialIsStencil=false
     fontMaterial=/CreativeEmporium/Materials/MI_CrRefresh_TabButton_Text_Create
     fontObject=&HeadingNow
     letterSpacing=0
     monospacedWidth=1
     size=18
     skewAmount=0
     typefaceFontName=75 Medium
     outlineSettings {
      bApplyOutlineToDropShadows=false
      bMiteredCorners=false
      bSeparateFillAlpha=false
      outlineMaterial=null
      outlineSize=0
      outlineColor=<rgba 0 0 0 1>
     }
    }
   }
   slot {class=HorizontalBoxSlot, properties {verticalAlignment=VAlign_Center, size {sizeRule=Fill, value=1}}}
  }
  HoverBubble {
   widgetClassPath=&SizeBox
   parent=HorizontalBox_0
   widget=HoverBubble
   properties {bOverride_MinDesiredWidth=true, minDesiredWidth=80}
   slot {class=HorizontalBoxSlot, properties {padding=<ltrb 2 0 0 0>}}
  }
  Overlay_1 {widgetClassPath=&Overlay, parent=HoverBubble, widget=Overlay_1, slot {class=SizeBoxSlot}}
  Image_EndCapBackplate {
   widgetClassPath=&Image
   parent=Overlay_1
   widget=Image_EndCapBackplate
   properties {
    renderOpacity=0
    visibility=Collapsed
    brush {
     bIsDynamicallyLoaded=false
     drawAs=NoDrawType
     imageType=NoImage
     mirroring=NoMirror
     resourceName=None
     resourceObject=/CreativeEmporium/Materials/MI_UI_SideMenu
     tiling=NoTile
     imageSize=<100 12>
     margin=<ltrb 0 0 0 0>
     outlineSettings {
      bUseBrushTransparency=false
      roundingType=HalfHeightRadius
      width=0
      color {colorUseRule=UseColor_Specified, specifiedColor=<rgba 0 0 0 0>}
      cornerRadii=<0 0 0 0>
     }
     tintColor {colorUseRule=UseColor_Specified, specifiedColor=<rgba 1 1 1 0.3>}
     uVRegion {bIsValid=false, max=<0 0>, min=<0 0>}
    }
    renderTransform {angle=0, scale=<-1 1>, shear=<0 0>, translation=<0 0>}
   }
   slot {class=OverlaySlot, properties {horizontalAlignment=HAlign_Fill, verticalAlignment=VAlign_Fill}}
  }
  AnimatedInputBorder {
   widgetClassPath=&Border
   parent=Overlay_1
   widget=AnimatedInputBorder
   properties {
    horizontalAlignment=HAlign_Right
    verticalAlignment=VAlign_Center
    visibility=SelfHitTestInvisible
    background {
     bIsDynamicallyLoaded=false
     drawAs=Image
     imageType=NoImage
     mirroring=NoMirror
     resourceName=None
     resourceObject=null
     tiling=NoTile
     imageSize=<64 64>
     margin=<ltrb 1 0 0 0>
     outlineSettings {
      bUseBrushTransparency=false
      roundingType=HalfHeightRadius
      width=0
      color {colorUseRule=UseColor_Specified, specifiedColor=<rgba 0 0 0 0>}
      cornerRadii=<0 0 0 0>
     }
     tintColor {colorUseRule=UseColor_Specified, specifiedColor=<rgba 1 1 1 0.1>}
     uVRegion {bIsValid=false, max=<0 0>, min=<0 0>}
    }
    brushColor=<rgba 1 1 1 0>
    padding=<ltrb 24 0 4 0>
   }
   slot {class=OverlaySlot, properties {horizontalAlignment=HAlign_Right, verticalAlignment=VAlign_Center}}
  }
  HorizontalBox_6 {
   widgetClassPath=&HorizontalBox
   parent=AnimatedInputBorder
   widget=HorizontalBox_6
   slot {class=BorderSlot, properties {horizontalAlignment=HAlign_Right, verticalAlignment=VAlign_Center, padding=<ltrb 24 0 4 0>}}
  }
  Text_NumEntries {
   widgetClassPath=&CommonTextBlock
   bIsVariable=true
   parent=HorizontalBox_6
   widget=Text_NumEntries
   properties {
    minDesiredWidth=18
    mobileFontSizeMultiplier=1.34
    colorAndOpacity {colorUseRule=UseColor_Specified, specifiedColor=<rgba 0.5089 0.5583 0.7084 1>}
    font {
     bForceMonospaced=false
     bMaterialIsStencil=false
     fontMaterial=null
     fontObject=&BurbankSmall
     letterSpacing=0
     monospacedWidth=1
     size=14
     skewAmount=0
     typefaceFontName=Black
     outlineSettings {
      bApplyOutlineToDropShadows=false
      bMiteredCorners=false
      bSeparateFillAlpha=false
      outlineMaterial=null
      outlineSize=0
      outlineColor=<rgba 0 0 0 1>
     }
    }
   }
   slot {class=HorizontalBoxSlot, properties {horizontalAlignment=HAlign_Left, verticalAlignment=VAlign_Center}}
  }
  ArrowSpacingSB {
   widgetClassPath=&SizeBox
   parent=HorizontalBox_6
   widget=ArrowSpacingSB
   properties {bOverride_MinDesiredHeight=true, heightOverride=72, minDesiredHeight=16, widthOverride=72}
   slot {class=HorizontalBoxSlot, properties {horizontalAlignment=HAlign_Left, verticalAlignment=VAlign_Center, padding=<ltrb 6 0 0 0>}}
  }
  ExpandDirectionTick {
   widgetClassPath=/Script/FortniteUI.FortMobileImage
   bIsVariable=true
   parent=ArrowSpacingSB
   widget=ExpandDirectionTick
   properties {
    mobileSizeMultiplier=1.2
    brush {
     bIsDynamicallyLoaded=false
     drawAs=Image
     imageType=NoImage
     mirroring=NoMirror
     resourceName=None
     resourceObject=/Game/UI/SocialPanel/Art/Textures/T_UI_Icon_TreeArrow_64
     tiling=NoTile
     imageSize=<16 16>
     margin=<ltrb 0 0 0 0>
     outlineSettings {
      bUseBrushTransparency=false
      roundingType=HalfHeightRadius
      width=0
      color {colorUseRule=UseColor_Specified, specifiedColor=<rgba 0 0 0 0>}
      cornerRadii=<0 0 0 0>
     }
     tintColor {colorUseRule=UseColor_Specified, specifiedColor=<rgba 1 1 1 1>}
     uVRegion {bIsValid=false, max=<0 0>, min=<0 0>}
    }
    colorAndOpacity=<rgba 0.5089 0.5583 0.7084 1>
   }
   slot {class=SizeBoxSlot, properties {horizontalAlignment=HAlign_Left, verticalAlignment=VAlign_Center}}
  }
  Overlay_2 {
   widgetClassPath=&Overlay
   parent=Overlay_0
   widget=Overlay_2
   properties {renderTransform {angle=0, scale=<1 1>, shear=<0 0>, translation=<2 0>}}
   slot {class=OverlaySlot, properties {horizontalAlignment=HAlign_Right, verticalAlignment=VAlign_Center}}
  }
  InputActionWidget {
   widgetClassPath=&CommonActionWidget
   bInherited=true
   parent=Overlay_2
   widget=InputActionWidget
   properties {
    visibility=Collapsed
    inputActions [{dataTable=&InputActionDataTable, rowName=Select}]
    progressMaterialBrush {
     bIsDynamicallyLoaded=false
     drawAs=Image
     imageType=NoImage
     mirroring=NoMirror
     resourceName=None
     resourceObject=null
     tiling=NoTile
     imageSize=<16 16>
     margin=<ltrb 0 0 0 0>
     outlineSettings {
      bUseBrushTransparency=false
      roundingType=HalfHeightRadius
      width=0
      color {colorUseRule=UseColor_Specified, specifiedColor=<rgba 0 0 0 0>}
      cornerRadii=<0 0 0 0>
     }
     tintColor {colorUseRule=UseColor_Specified, specifiedColor=<rgba 1 1 1 1>}
     uVRegion {bIsValid=false, max=<0 0>, min=<0 0>}
    }
   }
   slot {class=OverlaySlot}
  }
 ]
}
