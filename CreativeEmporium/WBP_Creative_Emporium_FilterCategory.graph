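# Collapsible filter category widget for the Creative Emporium, pairing a header button with a vertical box of subcategory widgets.
# Implements the FortEmporiumFilterCategoryHeader events to set title and entry count, add or clear subcategories, and show active state.
# Header clicks call ActivateCategory with the toggled expansion, derived from box visibility; ExpandCategory applies visibility and notifies the header.
# EnableCategoryExpansion shows the arrow or locks the button and hides its input action.
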WidgetBlueprint /CreativeEmporium/WBP_Creative_Emporium_FilterCategory
graphs [
 EventGraph {nodes [
   "On Clicked (Button_CategoryHeader)"!() -> (OutputDelegate, then, Button:Common Button Base Object Reference) {is=title K2Node_ComponentBoundEvent}
   n1 = Get Box_SubcategoryWidgets() -> (Box_SubcategoryWidgets:Vertical Box Object Reference)
   > ActivateCategory!(
    execute
    self:Fort Emporium Filter Category Header Object Reference
    bExpanded:bool="Equal (Enum)"(
     A:ESlateVisibility Enum=GetVisibility!(self:Widget Object Reference=@n1) -> (ReturnValue:ESlateVisibility Enum=Visible)
     B:ESlateVisibility Enum=Collapsed
    )
   ) -> (then)
   n2 = Event SetSubcategoryCount!() -> (OutputDelegate, then, NewCount:int=0) {is=title K2Node_Event}
   > "SetText (Text)"(
    self=Get Text_NumEntries!(self:WBP Creative Emporium Button Filter Category Header Object Reference=Get Button_CategoryHeader() -> (Button_CategoryHeader:WBP Creative Emporium Button Filter Category Header Object Reference)) -> (Text_NumEntries:Common Text Object Reference)
    InText=Select(
     Option 0:Text=GetEmptyText!() -> (ReturnValue:Text)
     Option 1:Text="To Text (Integer)"(Value=@n2.NewCount)
     Index="integer > integer"(A=@n2.NewCount)
    ) -> (ReturnValue:Text)
   )
   n3 = Event SetCategoryText!() -> (OutputDelegate, then, NewButtonText:"Text (by ref)") {is=title K2Node_Event}
   > "SetText (Text)"(
    self=Get Text_ListName!(self:WBP Creative Emporium Button Filter Category Header Object Reference=Get Button_CategoryHeader()) -> (Text_ListName:Common Text Object Reference)
    InText=@n3.NewButtonText
   )
   n4 = Event UpdateCategoryCount!() -> (OutputDelegate, then, NewCount:int=0, bHideIfEmpty:bool) {is=title K2Node_Event}
   n5 = "To Text (Integer)"(Value=@n4.NewCount)
   > "SetText (Text)"(
    self=Get Text_NumEntries(self=Get Button_CategoryHeader())
    InText=Select(Option 0=Select(Option 0=@n5, Index=@n4.bHideIfEmpty), Option 1=@n5, Index="integer > integer"(A=@n4.NewCount))
   )
   Event ClearSubcategoryWidgets() {is=title K2Node_Event}
   > ClearChildren(self=Get Box_SubcategoryWidgets())
   n6 = Event AddSubcategoryWidget!() -> (OutputDelegate, then, SubcategoryWidget:Widget Object Reference) {is=title K2Node_Event}
   > Is Valid(InputObject=@n6.SubcategoryWidget)
   > AddChildToVerticalBox(self=Get Box_SubcategoryWidgets(), Content=@n6.SubcategoryWidget)
   n7 = Event ExpandCategory!() -> (OutputDelegate, then, bExpanded:bool) {is=title K2Node_Event}
   > SetVisibility(
    self=Get Box_SubcategoryWidgets()
    InVisibility=Select(Option 0:ESlateVisibility Enum=Collapsed, Option 1:ESlateVisibility Enum=SelfHitTestInvisible, Index=@n7.bExpanded) -> (ReturnValue:ESlateVisibility Enum)
   )
   > OnExpansionChanged!(
    execute
    self:WBP Creative Emporium Button Filter Category Header Object Reference=Get Button_CategoryHeader()
    bIsExpanded:bool=@n7.bExpanded
   ) -> (then)
   n8 = Event EnableCategoryExpansion!() -> (OutputDelegate, then, bShow:bool) {is=title K2Node_Event}
   n9 = Get Button_CategoryHeader()
   > SetVisibility(
    self=Get ExpandDirectionTick!(self:WBP Creative Emporium Button Filter Category Header Object Reference=@n9) -> (ExpandDirectionTick:Fort Mobile Image Object Reference)
    InVisibility=Select(Option 0=Hidden, Option 1=HitTestInvisible, Index=@n8.bShow)
   )
   n10 = NOT Boolean(A=@n8.bShow)
   > SetIsLocked!(execute, self:Common Button Base Object Reference=@n9, bInIsLocked:bool=@n10) -> (then)
   > SetHideInputAction!(execute, self:Common Button Base Object Reference=@n9, bInHideInputAction:bool=@n10) -> (then)
   n11 = Event ShowCategoryActive!() -> (OutputDelegate, then, bIsActive:bool) {is=title K2Node_Event}
   > ShowCategoryActive!(execute, self:WBP Creative Emporium Button Filter Category Header Object Reference=Get Button_CategoryHeader(), Active:bool=@n11.bIsActive) -> (then)
  ]}
 GetCategoryButton {nodes [
   GetCategoryButton() {is=title K2Node_FunctionEntry}
   > Return Node!(execute, ReturnValue:Common Button Base Object Reference=Get Button_CategoryHeader())
  ]}
 GetSubcategoryWidgets {nodes [
   GetSubcategoryWidgets() {is=title K2Node_FunctionEntry}
   > Return Node(ReturnValue:Array of Widget Object References=GetAllChildren(self=Get Box_SubcategoryWidgets()))
  ]}
]
variables []
widgets {
 info {
  inheritedWidgetCount=0
  parentClass=/Script/CreativeEmporiumUI.FortEmporiumFilterCategoryHeader
  rootWidgetClass=&VerticalBox
  widgetCount=3
 }
 widgets [
  VerticalBox_30 {widgetClassPath=&VerticalBox, widget=VerticalBox_30}
  Button_CategoryHeader {
   widgetClassPath=&/WBP_Creative_Emporium_Button_FilterCategoryHeader
   bIsVariable=true
   parent=VerticalBox_30
   widget=Button_CategoryHeader
   properties {bSelectable=false, minHeight=56}
   slot {class=VerticalBoxSlot}
  }
  Box_SubcategoryWidgets {
   widgetClassPath=&VerticalBox
   bIsVariable=true
   parent=VerticalBox_30
   widget=Box_SubcategoryWidgets
   slot {class=VerticalBoxSlot}
  }
 ]
}
