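# Filter category header widget for the Creative Emporium's filter home, derived from FortEmporiumFilterCategoryHeader.
# A vertical box holds one header button, whose click calls ActivateCategory with bExpanded false.
# The SetCategoryText event writes the incoming text directly to that button's Text_ListName label.
# GetSubcategoryWidgets returns an unfilled array, so this header exposes no subcategory widgets.
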WidgetBlueprint /CreativeEmporium/WBP_Creative_Emporium_FilterHome
graphs [
 EventGraph {nodes [
   "On Clicked (Button_CategoryHeader)"!() -> (OutputDelegate, then, Button:Common Button Base Object Reference) {is=title K2Node_ComponentBoundEvent}
   > ActivateCategory!(execute, self:Fort Emporium Filter Category Header Object Reference, bExpanded:bool=false) -> (then)
   n1 = Event SetCategoryText!() -> (OutputDelegate, then, NewButtonText:"Text (by ref)") {is=title K2Node_Event}
   > "SetText (Text)"(
    self=Get Text_ListName!(self:WBP Creative Emporium Button Filter Home Header Object Reference=Get Button_CategoryHeader() -> (Button_CategoryHeader:WBP Creative Emporium Button Filter Home Header Object Reference)) -> (Text_ListName:Common Text Object Reference)
    InText=@n1.NewButtonText
   )
  ]}
 GetSubcategoryWidgets {nodes [
   GetSubcategoryWidgets() {is=title K2Node_FunctionEntry}
   > Return Node!(execute, ReturnValue:Array of Widget Object References)
  ]}
]
variables []
widgets {
 info {
  inheritedWidgetCount=0
  parentClass=/Script/CreativeEmporiumUI.FortEmporiumFilterCategoryHeader
  rootWidgetClass=&VerticalBox
  widgetCount=2
 }
 widgets [
  VerticalBox_30 {widgetClassPath=&VerticalBox, widget=VerticalBox_30}
  Button_CategoryHeader {
   widgetClassPath=&/WBP_Creative_Emporium_Button_FilterHomeHeader
   bIsVariable=true
   parent=VerticalBox_30
   widget=Button_CategoryHeader
   slot {class=VerticalBoxSlot}
  }
 ]
}
