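# Creative Emporium filter section widget: a collapsible category header with an indented box of subcategory widgets, derived from FortEmporiumFilterCategoryHeader.
# A header click calls ExpandSection, while the ExpandCategory event toggles the box's visibility and informs the header button.
# Other parent events add or clear children, set header name and count text (optionally blank when empty), toggle the upper spacer, and forward cycle-section icon and prompt settings to the button.
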WidgetBlueprint /CreativeEmporium/WBP_Creative_Emporium_FilterSection
graphs [
 EventGraph {nodes [
   "On Clicked (Button_CategoryHeader)"!() -> (OutputDelegate, then, Button:Common Button Base Object Reference) {is=title K2Node_ComponentBoundEvent}
   > ExpandSection(self:Fort Emporium Filter Category Header Object Reference)
   n1 = Event SetSubcategoryCount!() -> (OutputDelegate, then, NewCount:int=0) {is=title K2Node_Event}
   > "SetText (Text)"(
    self=Get Text_NumEntries!(self:WBP Creative Emporium Button Filter Section Header Object Reference=Get Button_CategoryHeader() -> (Button_CategoryHeader:WBP Creative Emporium Button Filter Section Header Object Reference)) -> (Text_NumEntries:Common Text Object Reference)
    InText=Select(
     Option 0:Text=GetEmptyText!() -> (ReturnValue:Text)
     Option 1:Text="To Text (Integer)"(Value=@n1.NewCount)
     Index="integer > integer"(A=@n1.NewCount)
    ) -> (ReturnValue:Text)
   )
   n2 = Event SetCategoryText!() -> (OutputDelegate, then, NewButtonText:"Text (by ref)") {is=title K2Node_Event}
   > "SetText (Text)"(
    self=Get Text_ListName!(self:WBP Creative Emporium Button Filter Section Header Object Reference=Get Button_CategoryHeader()) -> (Text_ListName:Common Text Object Reference)
    InText=@n2.NewButtonText
   )
   n3 = Event UpdateCategoryCount!() -> (OutputDelegate, then, NewCount:int=0, bHideIfEmpty:bool) {is=title K2Node_Event}
   n4 = "To Text (Integer)"(Value=@n3.NewCount)
   > "SetText (Text)"(
    self=Get Text_NumEntries(self=Get Button_CategoryHeader())
    InText=Select(Option 0=Select(Option 0=@n4, Index=@n3.bHideIfEmpty), Option 1=@n4, Index="integer > integer"(A=@n3.NewCount))
   )
   Event ClearSubcategoryWidgets() {is=title K2Node_Event}
   > ClearChildren(self=Get Box_SubcategoryWidgets() -> (Box_SubcategoryWidgets:Vertical Box Object Reference))
   n5 = Event AddSubcategoryWidget!() -> (OutputDelegate, then, SubcategoryWidget:Widget Object Reference) {is=title K2Node_Event}
   > Is Valid(InputObject=@n5.SubcategoryWidget)
   > AddChild(self=Get Box_SubcategoryWidgets(), Content=@n5.SubcategoryWidget)
   n6 = Event ExpandCategory!() -> (OutputDelegate, then, bExpanded:bool) {is=title K2Node_Event}
   > SetVisibility(
    self=Get Box_SubcategoryWidgets()
    InVisibility=Select(Option 0:ESlateVisibility Enum=Collapsed, Option 1:ESlateVisibility Enum=SelfHitTestInvisible, Index=@n6.bExpanded) -> (ReturnValue:ESlateVisibility Enum)
   )
   > OnExpansionChanged!(
    execute
    self:WBP Creative Emporium Button Filter Section Header Object Reference=Get Button_CategoryHeader()
    bIsExpanded:bool=@n6.bExpanded
   ) -> (then)
   n7 = Event SetCycleSectionIcon!() -> (OutputDelegate, then, CycleSectionIcon:Slate Brush Structure) {is=title K2Node_Event}
   > SetCycleSectionIcon!(
    execute
    self:WBP Creative Emporium Button Filter Section Header Object Reference=Get Button_CategoryHeader()
    Cycle Section Icon:Slate Brush Structure=@n7.CycleSectionIcon
   ) -> (then)
   n8 = Event ShowCycleSectionPrompt!() -> (OutputDelegate, then, bShow:bool) {is=title K2Node_Event}
   > ShowCycleSectionPrompt!(execute, self:WBP Creative Emporium Button Filter Section Header Object Reference=Get Button_CategoryHeader(), Show:bool=@n8.bShow) -> (then)
   n9 = Event ShowUpperSpacer!() -> (OutputDelegate, then, bShow:bool) {is=title K2Node_Event}
   > SetVisibility(
    self=Get Spacer_Upper() -> (Spacer_Upper:Spacer Object Reference)
    InVisibility=Select(Option 0=Collapsed, Option 1=SelfHitTestInvisible, Index=@n9.bShow)
   )
  ]}
 GetCategoryButton {nodes [
   GetCategoryButton() {is=title K2Node_FunctionEntry}
   > Return Node!(execute, ReturnValue:Common Button Base Object Reference=Get Button_CategoryHeader())
  ]}
 GetSubcategoryWidgets {nodes [
   GetSubcategoryWidgets() {is=title K2Node_FunctionEntry}
   > Return Node(ReturnValue:Array of Widget Object References=GetAllChildren(self=Get Box_SubcategoryWidgets()))
  ]}
]
variables [Hide if Empty {type=boolean}]
widgets {
 info {
  inheritedWidgetCount=0
  parentClass=/Script/CreativeEmporiumUI.FortEmporiumFilterCategoryHeader
  rootWidgetClass=&VerticalBox
  widgetCount=4
 }
 widgets [
  VerticalBox_30 {widgetClassPath=&VerticalBox, widget=VerticalBox_30}
  Spacer_Upper {
   widgetClassPath=&Spacer
   bIsVariable=true
   parent=VerticalBox_30
   widget=Spacer_Upper
   properties {visibility=Collapsed, size=<1 24>}
   slot {class=VerticalBoxSlot}
  }
  Button_CategoryHeader {
   widgetClassPath=&/WBP_Creative_Emporium_Button_FilterSectionHeader
   bIsVariable=true
   parent=VerticalBox_30
   widget=Button_CategoryHeader
   properties {bIsEnabled=false}
   slot {class=VerticalBoxSlot}
  }
  Box_SubcategoryWidgets {
   widgetClassPath=&VerticalBox
   bIsVariable=true
   parent=VerticalBox_30
   widget=Box_SubcategoryWidgets
   properties {visibility=Collapsed}
   slot {class=VerticalBoxSlot, properties {padding=<ltrb 20 0 0 0>}}
  }
 ]
}
