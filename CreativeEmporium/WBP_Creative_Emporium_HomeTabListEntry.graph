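# Category row widget for the Creative Emporium home tab, showing a heading and a tile view of item entries.
# Generated entries get hidden labels, category-based card size and border padding, and label alignment by column via index modulo List Width.
# The parent class's UpdateWarningMessage event switches between tiles and an empty-list message sized to entry height.
# UpdateOverflowDisplay shows the overflow count on a tile, adding a "Go To" tab label when applicable.
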WidgetBlueprint /CreativeEmporium/WBP_Creative_Emporium_HomeTabListEntry
graphs [
 EventGraph {nodes [
   n1 = "On Entry Generated (HomeListView_ItemOptions)"!() -> (OutputDelegate, then, Widget:User Widget Object Reference) {is=title K2Node_ComponentBoundEvent}
   > n2 = Cast To WBP_Creative_Emporium_ItemEntry!(execute, Object=@n1.Widget) -> (then, CastFailed, AsWBP Creative Emporium Item Entry:WBP Creative Emporium Item Entry Object Reference) {is=title K2Node_DynamicCast}
   > UpdateLabelOffset!(execute, self, Entry:WBP Creative Emporium Item Entry Object Reference=@n2) -> (then)
   > SetLabelVisibility!(execute, self:WBP Creative Browser Item List Entry Widget Object Reference=@n2, bShow:bool=false) -> (then)
   n3 = Get WeaponCard!(self:WBP Creative Browser Item List Entry Widget Object Reference=@n2) -> (WeaponCard:Fort Multi Size Item Card Object Reference)
   > SetCardSize!(
    execute
    self:Fort Multi Size Item Card Object Reference=@n3
    CardSize:EFortItemCardSize Enum=GetCardSizeForCategory!(self:Fort Emporium Home List Entry Object Reference) -> (ReturnValue:EFortItemCardSize Enum=XXS)
   ) -> (then)
   > SetShouldCollapseBorderPadding!(
    execute
    self:Fort Multi Size Item Card Object Reference=@n3
    Value:bool=GetCollapseBorderPadFlagForCategory!(self:Fort Emporium Home List Entry Object Reference) -> (ReturnValue:bool)
   ) -> (then)
   n4 = Event UpdateWarningMessage!() -> (OutputDelegate, then, bShowMessage:bool) {is=title K2Node_Event}
   > SetActiveWidget(
    self:Widget Switcher Object Reference=Get Switcher_WarningMessages() -> (Switcher_WarningMessages:Common Animated Switcher Object Reference)
    Widget=Select(
     Option 0:Widget Object Reference=Get HomeListView_ItemOptions() -> (HomeListView_ItemOptions:Athena Creative Item Tile View Object Reference)
     Option 1:Widget Object Reference=Get Overlay_WarningListEmpty() -> (Overlay_WarningListEmpty:Overlay Object Reference)
     Index=@n4.bShowMessage
    ) -> (ReturnValue:Widget Object Reference)
   )
   > SetHeightOverride(
    self=Get SizeBox_WarningListEmpty() -> (SizeBox_WarningListEmpty:Size Box Object Reference)
    InHeightOverride=GetEntryHeight!(self:Tile View Object Reference=Get HomeListView_ItemOptions()) -> (ReturnValue:float)
   )
   n5 = Event UpdateOverflowDisplay!() -> (OutputDelegate, then, OverflowTile:Athena Creative Item Tile Button Object Reference, OverflowCount:int=0) {is=title K2Node_Event}
   > n6 = Cast To WBP_Creative_Emporium_ItemEntry(Object=@n5.OverflowTile)
   > ShowOverflowCount!(execute, self:WBP Creative Browser Item List Entry Widget Object Reference=@n6, Count:int=@n5.OverflowCount) -> (then)
   n7 = Get TabPanel() -> (TabPanel:Fort Emporium Item List Tab Panel Object Reference)
   > n8 = Branch(Condition=AND Boolean(A=IsValid(Object=@n7), B="integer > integer"(A=@n5.OverflowCount)))
   > SetLabelOverride!(
    execute
    self:WBP Creative Browser Item List Entry Widget Object Reference=@n6
    LabelOverride:Text=Format Text!(
     Format="LOCTEXT(\"Go To {0}\")"
     0:Text=Get TabLabel!(self:Fort Creative Content Browser Tab Panel Base Object Reference=@n7) -> (TabLabel:Text)
    ) -> (Result)
   ) -> (then)
   SetLabelOverride(execute=@n8.else, self=@n6, LabelOverride=GetEmptyText!() -> (ReturnValue:Text))
   SetVisibility(
    self=Self-Reference()
    InVisibility=Select(Option 0:ESlateVisibility Enum=SelfHitTestInvisible, Option 1:ESlateVisibility Enum=Collapsed, Index=false) -> (ReturnValue:ESlateVisibility Enum)
   )
  ]}
 UpdateLabelOffset {nodes [
   n1 = UpdateLabelOffset!() -> (then, Entry:WBP Creative Emporium Item Entry Object Reference) {is=title K2Node_FunctionEntry}
   n2 = GetListItemObject!(UserObjectListEntry:User Object List Entry Interface=@n1) -> (ReturnValue:Object Reference)
   > Is Valid(InputObject=@n2)
   n3 = Get List Width() -> (List Width:int=0)
   n4 = "% (Integer)"(A=GetIndexForItem!(self:List View Object Reference=Get HomeListView_ItemOptions(), Item:Object Reference=@n2) -> (ReturnValue:int=0), B=@n3) -> (ReturnValue=0)
   > n5 = Branch(Condition="Equal (Integer)"(A=@n4))
   > ShiftLabel!(execute, self:WBP Creative Browser Item List Entry Widget Object Reference=@n1, Alignment:EHorizontalAlignment Enum=HAlign_Right) -> (then)
   n6 = Branch(execute=@n5.else, Condition="Equal (Integer)"(A=@n4, B=int - int(A=@n3)))
   > ShiftLabel(self=@n1, Alignment=HAlign_Left)
   ShiftLabel(execute=@n6.else, self=@n1, Alignment=HAlign_Center)
  ]}
]
variables [
 HoveredBrush {
  type=SlateBrush
  default {
   bIsDynamicallyLoaded=false
   drawAs=Image
   imageType=NoImage
   mirroring=NoMirror
   resourceName=None
   resourceObject=null
   tiling=NoTile
   imageSize=<32 32>
   margin=<ltrb 0 0 0 0>
   outlineSettings {
    bUseBrushTransparency=false
    roundingType=HalfHeightRadius
    width=0
    color {colorUseRule=UseColor_Specified, specifiedColor=<rgba 0 0 0 0>}
    cornerRadii=<0 0 0 1>
   }
   tintColor {colorUseRule=UseColor_Specified, specifiedColor=<rgba 1 1 1 0>}
   uVRegion {bIsValid=false, max=<0 0>, min=<0 0>}
  }
 }
 IsSelected {type=boolean}
 List Width {type=integer, default=6}
 SelectedBrush {
  type=SlateBrush
  default {
   bIsDynamicallyLoaded=false
   drawAs=Image
   imageType=NoImage
   mirroring=NoMirror
   resourceName=None
   resourceObject=null
   tiling=NoTile
   imageSize=<48 48>
   margin=<ltrb 0 0 0 0>
   outlineSettings {
    bUseBrushTransparency=false
    roundingType=HalfHeightRadius
    width=0
    color {colorUseRule=UseColor_Specified, specifiedColor=<rgba 0 0 0 0>}
    cornerRadii=<0 0 0 1>
   }
   tintColor {colorUseRule=UseColor_Specified, specifiedColor=<rgba 1 1 1 0>}
   uVRegion {bIsValid=false, max=<0 0>, min=<0 0>}
  }
 }
]
widgets {
 info {
  inheritedWidgetCount=2
  parentClass=/Script/CreativeEmporiumUI.FortEmporiumHomeListEntry
  rootWidgetClass=&Overlay
  widgetCount=12
 }
 widgets [
  Overlay_42 {widgetClassPath=&Overlay, widget=Overlay_42}
  HoverBorder {
   widgetClassPath=&CommonBorder
   bIsVariable=true
   parent=Overlay_42
   widget=HoverBorder
   properties {
    background {
     bIsDynamicallyLoaded=false
     drawAs=Image
     imageType=NoImage
     mirroring=NoMirror
     resourceName=None
     resourceObject=/FortUILibrary/Styles/Border/Materials/M-Border-SolidBG-White
     tiling=NoTile
     imageSize=<32 32>
     margin=<ltrb 0 0 0 0>
     outlineSettings {
      bUseBrushTransparency=false
      roundingType=HalfHeightRadius
      width=0
      color {colorUseRule=UseColor_Specified, specifiedColor=<rgba 0 0 0 0>}
      cornerRadii=<0 0 0 0>
     }
     tintColor {colorUseRule=UseColor_Specified, specifiedColor=<rgba 1 1 1 1>}
     uVRegion {bIsValid=false, max=<0 0>, min=<0 0>}
    }
    brushColor=<rgba 1 1 1 0>
    padding=<ltrb 0 0 0 0>
   }
   slot {class=OverlaySlot, properties {horizontalAlignment=HAlign_Fill, verticalAlignment=VAlign_Fill, padding=<ltrb 0 0 0 24>}}
  }
  Overlay_0 {widgetClassPath=&Overlay, parent=HoverBorder, widget=Overlay_0, slot {class=BorderSlot, properties {padding=<ltrb 0 0 0 0>}}}
  Image_192 {
   widgetClassPath=&Image
   bIsVariable=true
   parent=Overlay_0
   widget=Image_192
   properties {
    visibility=SelfHitTestInvisible
    brush {
     bIsDynamicallyLoaded=false
     drawAs=NoDrawType
     imageType=NoImage
     mirroring=NoMirror
     resourceName=None
     resourceObject=/CreativeEmporium/Materials/MI_CreativeBrowser_HomeRow
     tiling=NoTile
     imageSize=<32 32>
     margin=<ltrb 0 0 0 0>
     outlineSettings {
      bUseBrushTransparency=false
      roundingType=HalfHeightRadius
      width=0
      color {colorUseRule=UseColor_Specified, specifiedColor=<rgba 0 0 0 0>}
      cornerRadii=<0 0 0 0>
     }
     tintColor {colorUseRule=UseColor_Specified, specifiedColor=<rgba 1 1 1 1>}
     uVRegion {bIsValid=false, max=<0 0>, min=<0 0>}
    }
    colorAndOpacity=<rgba 1 1 1 0.144>
   }
   slot {class=OverlaySlot, properties {horizontalAlignment=HAlign_Fill, verticalAlignment=VAlign_Fill}}
  }
  VerticalBox_0 {
   widgetClassPath=&VerticalBox
   parent=Overlay_0
   widget=VerticalBox_0
   slot {class=OverlaySlot, properties {horizontalAlignment=HAlign_Fill, verticalAlignment=VAlign_Center}}
  }
  HorizontalBox_42 {
   widgetClassPath=&HorizontalBox
   parent=VerticalBox_0
   widget=HorizontalBox_42
   slot {class=VerticalBoxSlot, properties {padding=<ltrb 12 0 12 0>}}
  }
  Text_CategoryLabel {
   widgetClassPath=&CommonTextBlock
   bInherited=true
   bIsVariable=true
   parent=HorizontalBox_42
   widget=Text_CategoryLabel
   properties {
    text="{Featured Items}"
    colorAndOpacity {colorUseRule=UseColor_Specified, specifiedColor=<rgba 1 1 1 0.8>}
    font {
     bForceMonospaced=false
     bMaterialIsStencil=false
     fontMaterial=null
     fontObject=&HeadingNow
     letterSpacing=0
     monospacedWidth=1
     size=22.5
     skewAmount=0
     typefaceFontName=75 Medium Italic
     outlineSettings {
      bApplyOutlineToDropShadows=false
      bMiteredCorners=false
      bSeparateFillAlpha=false
      outlineMaterial=null
      outlineSize=0
      outlineColor=<rgba 0 0 0 1>
     }
    }
    shadowOffset=<0 0>
   }
   slot {class=HorizontalBoxSlot, properties {verticalAlignment=VAlign_Bottom, size {sizeRule=Fill, value=0.6}}}
  }
  Switcher_WarningMessages {
   widgetClassPath=/Script/CommonUI.CommonAnimatedSwitcher
   bIsVariable=true
   parent=VerticalBox_0
   widget=Switcher_WarningMessages
   properties {transitionDuration=0}
   slot {class=VerticalBoxSlot}
  }
  HomeListView_ItemOptions {
   widgetClassPath=/Script/FortniteUI.AthenaCreativeItemTileView
   bInherited=true
   bIsVariable=true
   parent=Switcher_WarningMessages
   widget=HomeListView_ItemOptions
   properties {
    bEnableRightClickScrolling=false
    bEnableTouchScrolling=false
    bReturnFocusToSelection=true
    clipping=Inherit
    consumeMouseWheel=Never
    entryWidgetClass=&/WBP_Creative_Emporium_ItemEntry.WBP_Creative_Emporium_ItemEntry_C
    horizontalEntrySpacing=24
    numDesignerPreviewEntries=8
    tileAlignment=LeftAligned
    verticalEntrySpacing=24
   }
   slot {class=WidgetSwitcherSlot}
  }
  Overlay_WarningListEmpty {
   widgetClassPath=&Overlay
   bIsVariable=true
   parent=Switcher_WarningMessages
   widget=Overlay_WarningListEmpty
   slot {class=WidgetSwitcherSlot}
  }
  SizeBox_WarningListEmpty {
   widgetClassPath=&SizeBox
   bIsVariable=true
   parent=Overlay_WarningListEmpty
   widget=SizeBox_WarningListEmpty
   properties {bOverride_HeightOverride=true, heightOverride=160}
   slot {class=OverlaySlot, properties {verticalAlignment=VAlign_Fill, padding=<ltrb 50 0 0 0>}}
  }
  ListEmpty_TextBlock {
   widgetClassPath=&CommonTextBlock
   parent=SizeBox_WarningListEmpty
   widget=ListEmpty_TextBlock
   properties {
    justification=Center
    text=No Items To Show
    colorAndOpacity {colorUseRule=UseColor_Specified, specifiedColor=<rgba 1 1 1 0.8>}
    font {
     bForceMonospaced=false
     bMaterialIsStencil=false
     fontMaterial=null
     fontObject=&HeadingNow
     letterSpacing=0
     monospacedWidth=1
     size=15.75
     skewAmount=0
     typefaceFontName=74 Regular
     outlineSettings {
      bApplyOutlineToDropShadows=false
      bMiteredCorners=false
      bSeparateFillAlpha=false
      outlineMaterial=null
      outlineSize=0
      outlineColor=<rgba 0 0 0 1>
     }
    }
    shadowOffset=<0 0>
   }
   slot {class=SizeBoxSlot, properties {horizontalAlignment=HAlign_Left, verticalAlignment=VAlign_Center}}
  }
 ]
}
