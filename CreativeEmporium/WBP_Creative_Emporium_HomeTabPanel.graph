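# Widget Blueprint for the Creative Emporium home tab panel, built on the native FortEmporiumHomeTabPanel class to display a list of options.
# An overlay holds a collapsed backplate image and a size-boxed vertical box containing the inherited list view, which spawns WBP_Creative_Emporium_HomeTabListEntry rows.
# The event graph is empty, so behavior comes from the parent class, apart from IsShowingModal, which always returns false.
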WidgetBlueprint /CreativeEmporium/WBP_Creative_Emporium_HomeTabPanel
graphs [
 EventGraph {links [], nodes []}
 IsShowingModal {nodes [
   IsShowingModal() {is=title K2Node_FunctionEntry}
   > Return Node!(execute, Showing:bool=false)
  ]}
]
variables []
widgets {
 info {inheritedWidgetCount=1, parentClass=/Script/CreativeEmporiumUI.FortEmporiumHomeTabPanel, rootWidgetClass=&Overlay, widgetCount=6}
 widgets [
  Overlay_27 {widgetClassPath=&Overlay, widget=Overlay_27}
  Image_50 {
   widgetClassPath=&Image
   parent=Overlay_27
   widget=Image_50
   properties {
    visibility=Collapsed
    brush {
     bIsDynamicallyLoaded=false
     drawAs=NoDrawType
     imageType=NoImage
     mirroring=NoMirror
     resourceName=None
     resourceObject=/CreativeEmporium/Materials/MI_UI_EmporiumBackplate
     tiling=NoTile
     imageSize=<32 32>
     margin=<ltrb 0 0 0 0>
     outlineSettings {
      bUseBrushTransparency=false
      roundingType=HalfHeightRadius
      width=0
      color {colorUseRule=UseColor_Specified, specifiedColor=<rgba 0 0 0 0>}
      cornerRadii=<0 0 0 0>
     }
     tintColor {colorUseRule=UseColor_Specified, specifiedColor=<rgba 1 1 1 1>}
     uVRegion {bIsValid=false, max=<0 0>, min=<0 0>}
    }
   }
   slot {class=OverlaySlot, properties {horizontalAlignment=HAlign_Fill, verticalAlignment=VAlign_Fill, padding=<ltrb -50 0 0 0>}}
  }
  SizeBox_114 {
   widgetClassPath=&SizeBox
   parent=Overlay_27
   widget=SizeBox_114
   properties {bOverride_MinDesiredHeight=true, minDesiredHeight=250}
   slot {class=OverlaySlot, properties {horizontalAlignment=HAlign_Fill, verticalAlignment=VAlign_Fill, padding=<ltrb 0 16 0 0>}}
  }
  VerticalBox_0 {widgetClassPath=&VerticalBox, parent=SizeBox_114, widget=VerticalBox_0, slot {class=SizeBoxSlot}}
  SizeBox {
   widgetClassPath=&SizeBox
   parent=VerticalBox_0
   widget=SizeBox
   properties {heightOverride=660}
   slot {class=VerticalBoxSlot, properties {padding=<ltrb 0 0 25 0>, size {sizeRule=Fill, value=1}}}
  }
  FortEmporiumHomeListView_Options {
   widgetClassPath=/Script/CreativeEmporiumUI.FortEmporiumHomeListView
   bInherited=true
   bIsVariable=true
   parent=SizeBox
   widget=FortEmporiumHomeListView_Options
   properties {
    entryWidgetClass=&/WBP_Creative_Emporium_HomeTabListEntry.WBP_Creative_Emporium_HomeTabListEntry_C
    navigation=&self:WidgetTree.FortEmporiumHomeListView_Options.WidgetNavigation_0
   }
   slot {class=SizeBoxSlot}
  }
 ]
}
