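# Level list entry widget for Creative Emporium level management, showing a level/terrain icon, name and Delete button on a native LevelManagementListEntry parent.
# Clicking the button builds a Confirm/Decline dialog view model and pushes a UIKit dialog as a modal panel.
# A click matching the stored Confirm index calls HandleLevelDeleteClicked silently; either choice unbinds and deactivates the dialog.
# The inherited delete button, memory bar and count texts are collapsed.
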WidgetBlueprint /CreativeEmporium/WBP_Creative_Emporium_LevelManagementListEntry
graphs [
 CreateDialogViewModel {nodes [
   CreateDialogViewModel() {is=title K2Node_FunctionEntry}
   > n1 = Construct Object from Class(Class=/Script/FortniteUI.UIKitDialogViewModel) -> (ReturnValue:UI Kit Dialog Viewmodel Object Reference)
   > n2 = Set View Model Reference(View Model Reference:UI Kit Dialog Viewmodel Object Reference=@n1) -> (Output_Get:UI Kit Dialog Viewmodel Object Reference)
   > Set DialogTitle!(execute, DialogTitle:Text="LOCTEXT(\"DELETE LEVEL\")", self:UI Kit Dialog Viewmodel Object Reference=@n2) -> (then, Output_Get:Text)
   > Set DialogInfoText!(
    execute
    DialogInfoText:Text="LOCTEXT(\"Are you sure you want to delete this level? This action can not be undone.\")"
    self:UI Kit Dialog Viewmodel Object Reference=@n2
   ) -> (then, Output_Get:Text)
   > Clear(TargetArray:Array of UI Kit Dialog Button View Model Object References=Get ButtonViewModels() -> (ButtonViewModels:Array of UI Kit Dialog Button View Model Object References))
   > n3 = Sequence!(execute) -> (then_0, then_1, then_2:Exec)
   > n4 = Construct Object from Class(Class=/Script/FortniteUI.UIKitDialogButtonViewModel) -> (ReturnValue:UI Kit Dialog Button View Model Object Reference)
   > Set Text(Text="LOCTEXT(\"Confirm\")", self:UI Kit Dialog Button View Model Object Reference=@n4)
   > Set ActionType!(execute, ActionType:EFortDialogResult Enum=Confirmed, self:UI Kit Dialog Button View Model Object Reference=@n4) -> (then, Output_Get:EFortDialogResult Enum=Confirmed)
   > Set DataTable!(
    execute
    DataTable:Data Table Soft Object Reference=&InputActionDataTable
    self:UI Kit Dialog Button View Model Object Reference=@n4
   ) -> (then, Output_Get:Data Table Soft Object Reference)
   > Set InputAction!(execute, InputAction:Name=Confirm, self:UI Kit Dialog Button View Model Object Reference=@n4) -> (then, Output_Get:Name=None)
   > n5 = Add(
    TargetArray:Array of UI Kit Dialog Button View Model Object References=Get ButtonViewModels()
    NewItem:"UI Kit Dialog Button View Model Object Reference (by ref)"=@n4
   ) {is=title K2Node_CallArrayFunction}
   > Set ConfirmButtonIndex(ConfirmButtonIndex:int=@n5) -> (Output_Get:int=0)
   n6 = Construct Object from Class(execute=@n3.then_1, Class=&UIKitDialogButtonViewModel)
   > Set Text(Text="LOCTEXT(\"Decline\")", self=@n6)
   > Set ActionType(ActionType=Declined, self=@n6) -> (Output_Get=Confirmed)
   > Set DataTable(DataTable=&InputActionDataTable, self=@n6)
   > Set InputAction(InputAction=Decline, self=@n6) -> (Output_Get=None)
   > Add(TargetArray=Get ButtonViewModels(), NewItem=@n6)
   Set Buttons!(
    execute=@n3.then_2
    Buttons:Array of UI Kit Dialog Button View Model Object References=Get ButtonViewModels()
    self:UI Kit Dialog Viewmodel Object Reference=@n2
   ) -> (then, Output_Get:Array of UI Kit Dialog Button View Model Object References)
  ]}
 EventGraph {nodes [
   "On Clicked (Button)"!() -> (OutputDelegate, then, Button:Common Button Base Object Reference) {is=title K2Node_ComponentBoundEvent}
   > CreateDialogViewModel() {is=title K2Node_CallFunction}
   > n1 = Create Widget!(
    execute:Exec
    Class:User Widget Class Reference=/FortUILibrary/UIKit/Dialogs/WBP_UIKit_Dialog.WBP_UIKit_Dialog_C
    OwningPlayer:Player Controller Object Reference
    UIKitDialogViewModel:UI Kit Dialog Viewmodel Object Reference=Get View Model Reference() -> (View Model Reference:UI Kit Dialog Viewmodel Object Reference)
   ) -> (then:Exec, ReturnValue:WBP UIKit Dialog Object Reference)
   > n2 = Set Dialog Reference(Dialog Reference:WBP UIKit Dialog Object Reference=@n1) -> (Output_Get:WBP UIKit Dialog Object Reference)
   > SetUIKitDialogViewModel!(execute, self:WBP UIKit Dialog Object Reference=@n2, Viewmodel:UI Kit Dialog Viewmodel Object Reference=Get View Model Reference()) -> (then)
   n3 = Get Dialog Reference() -> (Dialog Reference:WBP UIKit Dialog Object Reference)
   > Bind Event to Dialog Button Clicked Dispatcher(
    self:WBP UIKit Dialog Internal Object Reference=Get Dialog_Internal!(self:WBP UIKit Dialog Object Reference=@n3) -> (Dialog_Internal:WBP UIKit Dialog Internal Object Reference)
    Delegate=Create Event()
   ) {is=title K2Node_AddDelegate}
   > TryPushWidgetAsModalPanel!(execute, PushWidget:User Widget Object Reference=@n3) -> (then)
   n4 = OnDialogButtonClicked!() -> (OutputDelegate, then, ButtonIndex:int) {is=title K2Node_CustomEvent}
   > n5 = Sequence!(execute) -> (then_0, then_1)
   > Branch(Condition="Equal (Integer)"(A=Get ConfirmButtonIndex() -> (ConfirmButtonIndex:int=0), B=@n4.ButtonIndex))
   > HandleLevelDeleteClicked!(execute, self:Level Management List Entry Object Reference, bDeleteSilently:bool=true) -> (then)
   Unbind Event from Dialog Button Clicked Dispatcher(
    execute=@n5.then_1
    self:WBP UIKit Dialog Internal Object Reference=Get Dialog_Internal(self=Get Dialog Reference())
    Delegate=Create Event()
   ) {is=title K2Node_RemoveDelegate}
   n6 = Get Dialog Reference()
   > Unbind all Events from Dialog Button Clicked Dispatcher!(execute:Exec, self:WBP UIKit Dialog Internal Object Reference=Get Dialog_Internal(self=@n6)) -> (then:Exec) {is=title K2Node_ClearDelegate}
   > DeactivateWidget(self=@n6)
  ]}
]
variables [
 ButtonViewModels {type="array<UIKitDialogButtonViewModel>", default []}
 ConfirmButtonIndex {type=integer}
 Dialog Reference {type=WBP_UIKit_Dialog_C, default=null}
 HoveredBrush {
  type=SlateBrush
  default {
   bIsDynamicallyLoaded=false
   drawAs=Image
   imageType=NoImage
   mirroring=NoMirror
   resourceName=None
   resourceObject=/CreativeEmporium/Materials/MI_CrBrowser_TabButton_BackgroundStyle
   tiling=NoTile
   imageSize=<32 32>
   margin=<ltrb 0 0 0 0>
   outlineSettings {
    bUseBrushTransparency=false
    roundingType=HalfHeightRadius
    width=0
    color {colorUseRule=UseColor_Specified, specifiedColor=<rgba 0 0 0 0>}
    cornerRadii=<0 0 0 1>
   }
   tintColor {colorUseRule=UseColor_Specified, specifiedColor=<rgba 1 1 1 0.05>}
   uVRegion {bIsValid=false, max=<0 0>, min=<0 0>}
  }
 }
 IsSelected {type=boolean}
 SelectedBrush {
  type=SlateBrush
  default {
   bIsDynamicallyLoaded=false
   drawAs=Image
   imageType=NoImage
   mirroring=NoMirror
   resourceName=None
   resourceObject=&MI_CrBrowser_TabButton_BackgroundStyle
   tiling=NoTile
   imageSize=<48 48>
   margin=<ltrb 0 0 0 0>
   outlineSettings {
    bUseBrushTransparency=false
    roundingType=HalfHeightRadius
    width=0
    color {colorUseRule=UseColor_Specified, specifiedColor=<rgba 0 0 0 0>}
    cornerRadii=<0 0 0 1>
   }
   tintColor {colorUseRule=UseColor_Specified, specifiedColor=<rgba 1 1 1 0.3>}
   uVRegion {bIsValid=false, max=<0 0>, min=<0 0>}
  }
 }
 View Model Reference {type=UIKitDialogViewModel, default=null}
]
widgets {
 info {
  inheritedWidgetCount=6
  namedSlotCount=1
  parentClass=/Script/FortniteUI.LevelManagementListEntry
  rootWidgetClass=&Overlay
  widgetCount=15
 }
 widgets [
  Overlay_42 {widgetClassPath=&Overlay, widget=Overlay_42}
  Button {
   widgetClassPath=&WBP_UIKit_Button_Generic
   bIsVariable=true
   parent=Overlay_42
   widget=Button
   properties {
    backgroundMaterial=/CreativeEmporium/Materials/MI_CrBrowserLevelsButton
    contentHorizontalAlignment=HAlign_Right
    icon=/Game/UI/Inventory/Art/T_UI_LevelManagement_Trashcan_Icon
    iconMaterial=/CreativeEmporium/Materials/MI_CrBrowser_DeleteButtonIcon
    outlineMaterial=/CreativeEmporium/Materials/MI_CrBrowserLevelsEntryOutline
    text=Delete
    buttonPadding=<ltrb 24 0 48 0>
    font {
     bForceMonospaced=false
     bMaterialIsStencil=false
     fontMaterial=/CreativeEmporium/Materials/MI_CrBrowser_DeleteButtonText
     fontObject=&HeadingNow
     letterSpacing=48
     monospacedWidth=1
     size=16.5
     skewAmount=0
     typefaceFontName=76 Bold
     outlineSettings {
      bApplyOutlineToDropShadows=false
      bMiteredCorners=false
      bSeparateFillAlpha=false
      outlineMaterial=null
      outlineSize=0
      outlineColor=<rgba 0 0 0 1>
     }
    }
    iconSize=<32 32>
    promptNudge=<0 0>
   }
   slot {class=OverlaySlot, properties {horizontalAlignment=HAlign_Fill, verticalAlignment=VAlign_Fill, padding=<ltrb 12 8 12 8>}}
  }
  Overlay_0 {widgetClassPath=&Overlay, namedSlotHost=Button, widget=Overlay_0}
  HorizontalBox_42 {
   widgetClassPath=&HorizontalBox
   parent=Overlay_0
   widget=HorizontalBox_42
   slot {class=OverlaySlot, properties {verticalAlignment=VAlign_Center, padding=<ltrb 32 8 12 8>}}
  }
  WidgetSwitcher_TerrainType {
   widgetClassPath=&WidgetSwitcher
   bInherited=true
   bIsVariable=true
   parent=HorizontalBox_42
   widget=WidgetSwitcher_TerrainType
   slot {
    class=HorizontalBoxSlot
    properties {horizontalAlignment=HAlign_Center, verticalAlignment=VAlign_Center, size {sizeRule=Automatic, value=0.05}}
   }
  }
  Image_Level {
   widgetClassPath=&Image
   bIsVariable=true
   parent=WidgetSwitcher_TerrainType
   widget=Image_Level
   properties {brush {
     bIsDynamicallyLoaded=false
     drawAs=Image
     imageType=NoImage
     mirroring=NoMirror
     resourceName=None
     resourceObject=/Game/UI/Inventory/Art/T_UI_LevelManagement_Level_Icon
     tiling=NoTile
     imageSize=<56 56>
     margin=<ltrb 0 0 0 0>
     outlineSettings {
      bUseBrushTransparency=false
      roundingType=HalfHeightRadius
      width=0
      color {colorUseRule=UseColor_Specified, specifiedColor=<rgba 0 0 0 0>}
      cornerRadii=<0 0 0 0>
     }
     tintColor {colorUseRule=UseColor_Specified, specifiedColor=<rgba 1 1 1 0.8>}
     uVRegion {bIsValid=false, max=<0 0>, min=<0 0>}
    }}
   slot {class=WidgetSwitcherSlot, properties {horizontalAlignment=HAlign_Center, verticalAlignment=VAlign_Center}}
  }
  Image_Terrain {
   widgetClassPath=&Image
   bIsVariable=true
   parent=WidgetSwitcher_TerrainType
   widget=Image_Terrain
   properties {brush {
     bIsDynamicallyLoaded=false
     drawAs=Image
     imageType=NoImage
     mirroring=NoMirror
     resourceName=None
     resourceObject=/Game/UI/Inventory/Art/T_UI_LevelManagement_Terrain_Icon
     tiling=NoTile
     imageSize=<64 64>
     margin=<ltrb 0 0 0 0>
     outlineSettings {
      bUseBrushTransparency=false
      roundingType=HalfHeightRadius
      width=0
      color {colorUseRule=UseColor_Specified, specifiedColor=<rgba 0 0 0 0>}
      cornerRadii=<0 0 0 0>
     }
     tintColor {colorUseRule=UseColor_Specified, specifiedColor=<rgba 1 1 1 0.8>}
     uVRegion {bIsValid=false, max=<0 0>, min=<0 0>}
    }}
   slot {class=WidgetSwitcherSlot, properties {horizontalAlignment=HAlign_Center, verticalAlignment=VAlign_Center}}
  }
  Text_LevelName {
   widgetClassPath=&CommonTextBlock
   bInherited=true
   bIsVariable=true
   parent=HorizontalBox_42
   widget=Text_LevelName
   properties {
    minDesiredWidth=450
    text=Text_LevelName
    font {
     bForceMonospaced=false
     bMaterialIsStencil=false
     fontMaterial=null
     fontObject=&HeadingNow
     letterSpacing=0
     monospacedWidth=1
     size=19.5
     skewAmount=0
     typefaceFontName=76 Bold
     outlineSettings {
      bApplyOutlineToDropShadows=false
      bMiteredCorners=false
      bSeparateFillAlpha=false
      outlineMaterial=null
      outlineSize=0
      outlineColor=<rgba 0 0 0 1>
     }
    }
    shadowOffset=<0 0>
   }
   slot {class=HorizontalBoxSlot, properties {verticalAlignment=VAlign_Center, padding=<ltrb 12 0 0 0>, size {sizeRule=Fill, value=1}}}
  }
  ProgressBar_MemoryUse {
   widgetClassPath=/Script/UMG.ProgressBar
   bIsVariable=true
   parent=HorizontalBox_42
   widget=ProgressBar_MemoryUse
   properties {percent=0.5, visibility=Collapsed, fillColorAndOpacity=<rgba 0 0.5 1 1>}
   slot {class=HorizontalBoxSlot, properties {verticalAlignment=VAlign_Center, size {sizeRule=Fill, value=1}}}
  }
  Text_NumPortals {
   widgetClassPath=&CommonTextBlock
   bInherited=true
   bIsVariable=true
   parent=HorizontalBox_42
   widget=Text_NumPortals
   properties {
    justification=Center
    text=-
    visibility=Collapsed
    colorAndOpacity {colorUseRule=UseColor_Specified, specifiedColor=<rgba 0.6 0.8667 1 1>}
    font {
     bForceMonospaced=false
     bMaterialIsStencil=false
     fontMaterial=null
     fontObject=&BurbankSmall
     letterSpacing=0
     monospacedWidth=1
     size=20
     skewAmount=0
     typefaceFontName=Bold
     outlineSettings {
      bApplyOutlineToDropShadows=false
      bMiteredCorners=false
      bSeparateFillAlpha=false
      outlineMaterial=null
      outlineSize=0
      outlineColor=<rgba 0 0 0 1>
     }
    }
    shadowOffset=<0 0>
   }
   slot {class=HorizontalBoxSlot, properties {verticalAlignment=VAlign_Center}}
  }
  Text_NumRefs {
   widgetClassPath=&CommonTextBlock
   bInherited=true
   bIsVariable=true
   parent=HorizontalBox_42
   widget=Text_NumRefs
   properties {
    justification=Center
    visibility=Collapsed
    colorAndOpacity {colorUseRule=UseColor_Specified, specifiedColor=<rgba 1 1 1 0.8>}
    font {
     bForceMonospaced=false
     bMaterialIsStencil=false
     fontMaterial=null
     fontObject=&HeadingNow
     letterSpacing=0
     monospacedWidth=1
     size=15.75
     skewAmount=0
     typefaceFontName=75 Medium
     outlineSettings {
      bApplyOutlineToDropShadows=false
      bMiteredCorners=false
      bSeparateFillAlpha=false
      outlineMaterial=null
      outlineSize=0
      outlineColor=<rgba 0 0 0 1>
     }
    }
    shadowOffset=<0 0>
   }
   slot {class=HorizontalBoxSlot, properties {verticalAlignment=VAlign_Center, size {sizeRule=Automatic, value=0.3}}}
  }
  Text_NumRefs_1 {
   widgetClassPath=&CommonTextBlock
   bIsVariable=true
   parent=HorizontalBox_42
   widget=Text_NumRefs_1
   properties {
    justification=Center
    text=References
    visibility=Collapsed
    colorAndOpacity {colorUseRule=UseColor_Specified, specifiedColor=<rgba 1 1 1 0.8>}
    font {
     bForceMonospaced=false
     bMaterialIsStencil=false
     fontMaterial=null
     fontObject=&HeadingNow
     letterSpacing=0
     monospacedWidth=1
     size=15.75
     skewAmount=0
     typefaceFontName=75 Medium
     outlineSettings {
      bApplyOutlineToDropShadows=false
      bMiteredCorners=false
      bSeparateFillAlpha=false
      outlineMaterial=null
      outlineSize=0
      outlineColor=<rgba 0 0 0 1>
     }
    }
    shadowOffset=<0 0>
   }
   slot {class=HorizontalBoxSlot, properties {verticalAlignment=VAlign_Center, padding=<ltrb 6 0 0 0>, size {sizeRule=Automatic, value=0.3}}}
  }
  Overlay_1 {widgetClassPath=&Overlay, parent=Overlay_42, widget=Overlay_1, properties {visibility=Collapsed}, slot {class=OverlaySlot}}
  IconTextButton_Delete {
   widgetClassPath=&IconTextButton
   bInherited=true
   bIsVariable=true
   parent=Overlay_1
   widget=IconTextButton_Delete
   properties {
    holdData=&CommonUI_HoldData_Base_C
    onButtonBaseClicked=(&self:WidgetTree.IconTextButton_Delete.HandleOnButtonClicked)
    onButtonBaseDoubleClicked=(&self:WidgetTree.IconTextButton_Delete.HandleOnButtonDoubleClicked)
    onButtonBaseHovered=(&self:WidgetTree.IconTextButton_Delete.HandleOnButtonHovered)
    onButtonBaseUnhovered=(&self:WidgetTree.IconTextButton_Delete.HandleOnButtonUnhovered)
    onSelectedChangedBase=(&self:WidgetTree.IconTextButton_Delete.HandleOnSelectedChanged)
    visibility=Collapsed
   }
   slot {class=OverlaySlot}
  }
  InputActionWidget {widgetClassPath=&CommonActionWidget, bInherited=true, widget=null}
 ]
}
