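# Level management tab panel widget for the Creative Emporium, showing a list of levels under a LEVELS heading.
# It is layout only: the event graph is empty, so behavior is left to the parent FortCreativeLevelManagementTabPanel class.
# The inherited list view sits in a fixed-height box and builds rows from WBP_Creative_Emporium_LevelManagementListEntry.
# The TYPE, MEMORY, PORTALS and REFERENCES column headers and the dark backplate are present but hidden or collapsed.
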WidgetBlueprint /CreativeEmporium/WBP_Creative_Emporium_LevelManagerPanel
graphs [EventGraph {links [], nodes []}]
variables []
widgets {
 info {
  inheritedWidgetCount=1
  parentClass=/Script/FortniteUI.FortCreativeLevelManagementTabPanel
  rootWidgetClass=&Overlay
  widgetCount=13
 }
 widgets [
  Overlay_27 {widgetClassPath=&Overlay, widget=Overlay_27}
  UIKit_Backplate {
   widgetClassPath=&WBP_UIKit_Backplate
   parent=Overlay_27
   widget=UIKit_Backplate
   properties {brightness=Dark, visibility=Collapsed}
   slot {class=OverlaySlot, properties {horizontalAlignment=HAlign_Fill, verticalAlignment=VAlign_Fill}}
  }
  SizeBox_114 {
   widgetClassPath=&SizeBox
   parent=Overlay_27
   widget=SizeBox_114
   properties {bOverride_MinDesiredHeight=true, minDesiredHeight=250}
   slot {class=OverlaySlot, properties {horizontalAlignment=HAlign_Fill, verticalAlignment=VAlign_Fill, padding=<ltrb 24 24 24 24>}}
  }
  VerticalBox_0 {
   widgetClassPath=&VerticalBox
   parent=SizeBox_114
   widget=VerticalBox_0
   slot {class=SizeBoxSlot, properties {padding=<ltrb 0 10 0 0>}}
  }
  HorizontalBox_443 {widgetClassPath=&HorizontalBox, parent=VerticalBox_0, widget=HorizontalBox_443, slot {class=VerticalBoxSlot}}
  CommonTextBlock_105 {
   widgetClassPath=&CommonTextBlock
   parent=HorizontalBox_443
   widget=CommonTextBlock_105
   properties {
    text=LEVELS
    colorAndOpacity {colorUseRule=UseColor_Specified, specifiedColor=<rgba 1 1 1 0.8>}
    font {
     bForceMonospaced=false
     bMaterialIsStencil=false
     fontMaterial=null
     fontObject=&HeadingNow
     letterSpacing=0
     monospacedWidth=1
     size=32.25
     skewAmount=0
     typefaceFontName=86 Bold Italic
     outlineSettings {
      bApplyOutlineToDropShadows=false
      bMiteredCorners=false
      bSeparateFillAlpha=false
      outlineMaterial=null
      outlineSize=0
      outlineColor=<rgba 0 0 0 1>
     }
    }
    shadowOffset=<0 0>
   }
   slot {
    class=HorizontalBoxSlot
    properties {horizontalAlignment=HAlign_Left, verticalAlignment=VAlign_Center, padding=<ltrb 15 0 0 0>, size {sizeRule=Fill, value=0.6}}
   }
  }
  CommonTextBlock {
   widgetClassPath=&CommonTextBlock
   parent=HorizontalBox_443
   widget=CommonTextBlock
   properties {
    justification=Center
    text=TYPE
    visibility=Hidden
    colorAndOpacity {colorUseRule=UseColor_Specified, specifiedColor=<rgba 0.6 0.8667 1 1>}
    font {
     bForceMonospaced=false
     bMaterialIsStencil=false
     fontMaterial=null
     fontObject=&Burbank
     letterSpacing=0
     monospacedWidth=1
     size=28
     skewAmount=0
     typefaceFontName=Burbank-Black
     outlineSettings {
      bApplyOutlineToDropShadows=false
      bMiteredCorners=false
      bSeparateFillAlpha=false
      outlineMaterial=null
      outlineSize=0
      outlineColor=<rgba 0 0 0 1>
     }
    }
    shadowOffset=<0 0>
   }
   slot {class=HorizontalBoxSlot, properties {verticalAlignment=VAlign_Center, size {sizeRule=Fill, value=0.05}}}
  }
  CommonTextBlock_1 {
   widgetClassPath=&CommonTextBlock
   parent=HorizontalBox_443
   widget=CommonTextBlock_1
   properties {
    justification=Center
    text=MEMORY
    visibility=Collapsed
    colorAndOpacity {colorUseRule=UseColor_Specified, specifiedColor=<rgba 0.6 0.8667 1 1>}
    font {
     bForceMonospaced=false
     bMaterialIsStencil=false
     fontMaterial=null
     fontObject=&Burbank
     letterSpacing=0
     monospacedWidth=1
     size=28
     skewAmount=0
     typefaceFontName=Burbank-Black
     outlineSettings {
      bApplyOutlineToDropShadows=false
      bMiteredCorners=false
      bSeparateFillAlpha=false
      outlineMaterial=null
      outlineSize=0
      outlineColor=<rgba 0 0 0 1>
     }
    }
    shadowOffset=<0 0>
   }
   slot {class=HorizontalBoxSlot, properties {verticalAlignment=VAlign_Center, size {sizeRule=Fill, value=1}}}
  }
  CommonTextBlock_2 {
   widgetClassPath=&CommonTextBlock
   parent=HorizontalBox_443
   widget=CommonTextBlock_2
   properties {
    justification=Center
    text=PORTALS
    visibility=Collapsed
    colorAndOpacity {colorUseRule=UseColor_Specified, specifiedColor=<rgba 0.6 0.8667 1 1>}
    font {
     bForceMonospaced=false
     bMaterialIsStencil=false
     fontMaterial=null
     fontObject=&Burbank
     letterSpacing=0
     monospacedWidth=1
     size=28
     skewAmount=0
     typefaceFontName=Burbank-Black
     outlineSettings {
      bApplyOutlineToDropShadows=false
      bMiteredCorners=false
      bSeparateFillAlpha=false
      outlineMaterial=null
      outlineSize=0
      outlineColor=<rgba 0 0 0 1>
     }
    }
    shadowOffset=<0 0>
   }
   slot {class=HorizontalBoxSlot, properties {verticalAlignment=VAlign_Center, size {sizeRule=Fill, value=1}}}
  }
  CommonTextBlock_3 {
   widgetClassPath=&CommonTextBlock
   parent=HorizontalBox_443
   widget=CommonTextBlock_3
   properties {
    justification=Center
    text=REFERENCES
    visibility=Hidden
    colorAndOpacity {colorUseRule=UseColor_Specified, specifiedColor=<rgba 0.6 0.8667 1 1>}
    font {
     bForceMonospaced=false
     bMaterialIsStencil=false
     fontMaterial=null
     fontObject=&Burbank
     letterSpacing=0
     monospacedWidth=1
     size=28
     skewAmount=0
     typefaceFontName=Burbank-Black
     outlineSettings {
      bApplyOutlineToDropShadows=false
      bMiteredCorners=false
      bSeparateFillAlpha=false
      outlineMaterial=null
      outlineSize=0
      outlineColor=<rgba 0 0 0 1>
     }
    }
    shadowOffset=<0 0>
   }
   slot {class=HorizontalBoxSlot, properties {verticalAlignment=VAlign_Center, size {sizeRule=Fill, value=0.3}}}
  }
  Spacer_75 {
   widgetClassPath=&Spacer
   parent=HorizontalBox_443
   widget=Spacer_75
   slot {class=HorizontalBoxSlot, properties {padding=<ltrb 0 0 10 0>, size {sizeRule=Fill, value=0.05}}}
  }
  SizeBox_0 {
   widgetClassPath=&SizeBox
   parent=VerticalBox_0
   widget=SizeBox_0
   properties {bOverride_HeightOverride=true, heightOverride=660}
   slot {class=VerticalBoxSlot, properties {padding=<ltrb 0 25 0 0>}}
  }
  LevelManagementListView_Options {
   widgetClassPath=/Script/FortniteUI.LevelManagementListView
   bInherited=true
   bIsVariable=true
   parent=SizeBox_0
   widget=LevelManagementListView_Options
   properties {
    entryWidgetClass=&/WBP_Creative_Emporium_LevelManagementListEntry.WBP_Creative_Emporium_LevelManagementListEntry_C
    navigation=&self:WidgetTree.LevelManagementListView_Options.WidgetNavigation_0
    numDesignerPreviewEntries=10
   }
   slot {class=SizeBoxSlot, properties {padding=<ltrb 14 18 14 14>}}
  }
 ]
}
