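# Widget Blueprint for a sectioned item list entry in the Creative Emporium UI, showing a section label above a content area.
# A CommonBorder root stacks a padded label row over Section_ContentBox, an empty Overlay that always clips to bounds.
# The label and content box are inherited from native parent FortEmporiumSectionedItemListEntry.
# The event graph holds only unconnected event stubs, so the Blueprint supplies just layout and styling.
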WidgetBlueprint /CreativeEmporium/WBP_EmporiumSectionedItemListEntry
graphs [EventGraph {nodes [
   Event PreConstruct()
   Event Construct()
   Event Tick()
  ]}]
variables []
widgets {
 info {
  inheritedWidgetCount=2
  parentClass=/Script/CreativeEmporiumUI.FortEmporiumSectionedItemListEntry
  rootWidgetClass=&CommonBorder
  widgetCount=5
 }
 widgets [
  HoverBorder {widgetClassPath=&CommonBorder, widget=HoverBorder, properties {padding=<ltrb 0 0 0 0>}}
  VerticalBox_0 {
   widgetClassPath=&VerticalBox
   parent=HoverBorder
   widget=VerticalBox_0
   slot {class=BorderSlot, properties {padding=<ltrb 0 0 0 0>}}
  }
  HorizontalBox_42 {
   widgetClassPath=&HorizontalBox
   parent=VerticalBox_0
   widget=HorizontalBox_42
   slot {class=VerticalBoxSlot, properties {padding=<ltrb 12 0 12 0>}}
  }
  Text_SectionLabel {
   widgetClassPath=&CommonTextBlock
   bInherited=true
   bIsVariable=true
   parent=HorizontalBox_42
   widget=Text_SectionLabel
   properties {
    text="{Items}"
    colorAndOpacity {colorUseRule=UseColor_Specified, specifiedColor=<rgba 1 1 1 0.8>}
    font {
     bForceMonospaced=false
     bMaterialIsStencil=false
     fontMaterial=null
     fontObject=&HeadingNow
     letterSpacing=0
     monospacedWidth=1
     size=22.5
     skewAmount=0
     typefaceFontName=75 Medium Italic
     outlineSettings {
      bApplyOutlineToDropShadows=false
      bMiteredCorners=false
      bSeparateFillAlpha=false
      outlineMaterial=null
      outlineSize=0
      outlineColor=<rgba 0 0 0 1>
     }
    }
    shadowOffset=<0 0>
   }
   slot {class=HorizontalBoxSlot, properties {verticalAlignment=VAlign_Bottom, size {sizeRule=Fill, value=0.6}}}
  }
  Section_ContentBox {
   widgetClassPath=&Overlay
   bInherited=true
   bIsVariable=true
   parent=VerticalBox_0
   widget=Section_ContentBox
   properties {clipping=ClipToBoundsAlways}
   slot {class=VerticalBoxSlot}
  }
 ]
}
