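# WBP_EmporiumAssembliesSectionedTab is a Creative Emporium tab panel widget that, by its name, shows assemblies in sections, built on FortEmporiumSectionedItemListTabPanel.
# Its layout is a vertical box filled by the inherited Sections_ListView, which creates WBP_EmporiumSectionedItemListEntry rows with vertical spacing, clipping and custom navigation.
# The event graph holds only unconnected default events, so the Blueprint supplies layout while behavior is left to the parent class.
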WidgetBlueprint /CreativeEmporium/WBP_EmporiumAssembliesSectionedTab
graphs [EventGraph {nodes [
   Event PreConstruct()
   Event Construct()
   Event Tick()
  ]}]
variables []
widgets {
 info {
  inheritedWidgetCount=1
  parentClass=/Script/CreativeEmporiumUI.FortEmporiumSectionedItemListTabPanel
  rootWidgetClass=&VerticalBox
  widgetCount=2
 }
 widgets [
  VerticalBox_0 {widgetClassPath=&VerticalBox, widget=VerticalBox_0}
  Sections_ListView {
   widgetClassPath=/Script/CreativeEmporiumUI.FortEmporiumSectionedItemListView
   bInherited=true
   bIsVariable=true
   parent=VerticalBox_0
   widget=Sections_ListView
   properties {
    clipping=ClipToBoundsWithoutIntersecting
    entryWidgetClass=&/WBP_EmporiumSectionedItemListEntry.WBP_EmporiumSectionedItemListEntry_C
    navigation=&self:WidgetTree.Sections_ListView.WidgetNavigation_0
    numDesignerPreviewEntries=8
    verticalEntrySpacing=24
   }
   slot {class=VerticalBoxSlot, properties {padding=<ltrb 0 0 -32 0>, size {sizeRule=Fill, value=1}}}
  }
 ]
}
